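# Unlit, two-sided translucent material for Niagara sprites, ribbons and meshes, named for the FlossYawn character's music line effect.
# A panning color texture, desaturation-adjusted and tinted by particle color, feeds emissive adjustments.
# Opacity blends noise-modulated, gradient-warped streaks with sine-twinkling stars via Use_Star, then applies panning noise, UV edge fades, particle alpha and view-angle fade.
# A shading-path switch skips depth fade on forward and mobile.
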Material /BRCosmetics/Characters/Player/Female/Medium/Bodies/F_MED_FlossYawn/FX/M_Character_FlossYawn_MusicLine
graph {
 properties {
  bAutomaticallySetUsageInEditor=false
  bUsedWithNiagaraMeshParticles=true
  bUsedWithNiagaraRibbons=true
  bUsedWithNiagaraSprites=true
  blendMode=BLEND_Translucent
  shadingModel=MSM_Unlit
  twoSided=true
  usedShadingModels=Unlit
 }
 nodes [
  n1 = ScalarParameter() {defaultValue=-0.5, parameterName=BASE_TEX_SPEED_X}
  n2 = ScalarParameter() {parameterName=BASE_TEX_SPEED_Y}
  n3 = TextureSampleParameter2D(UVs=Panner(Coordinate=TextureCoordinate() {uTiling=1.5}, Speed=AppendVector(A=@n1, B=@n2)) {speedX=-0.5}) {parameterName=ColorTexture, texture=&/T_Character_FlossYawn_Color002}
  n4 = ScalarParameter() {defaultValue=-5, parameterName=Desaturation}
  n5 = ParticleColor()
  MaterialFunctionCall(Emissive=Multiply(A=Desaturation(None=@n3.RGB, Fraction=@n4), B=@n5.RGB)) {materialFunction=&MF_FX_EmissiveAdjustments}
  n6 = TextureCoordinate() {uTiling=5, vTiling=3}
  n7 = DynamicParameter() {defaultValue=<rgba 1 0.5 100 1>, paramNames [LerpTexturespeed_Offset, Bright_Random, star_bright, Param4]}
  n8 = AppendVector(
   A=Multiply(A=ScalarParameter() {defaultValue=-0.15, parameterName=StarTexture_SpeedX}, B=@n7.LerpTexturespeed_Offset)
   B=ScalarParameter() {parameterName=StarTexture_SpeedY}
  )
  n9 = Panner(Coordinate=Multiply(A=TextureCoordinate() {uTiling=0.3, vTiling=0.3}) {constB=2}) {speedX=-0.4}
  n10 = Add(
   A=Power(Base=TextureSample(UVs=@n9) {samplerType=SAMPLERTYPE_LinearColor, texture=/Game/TheOutpost/Textures/AnimatedOutpost_Base046_rgb_XVector_a_NormalizedHierPos_UV_2}.B)
   B=Multiply(A=Time())
  )
  n11 = TextureCoordinate() {uTiling=0.3, vTiling=0.3}
  n12 = Constant2Vector() {g=0.15, r=0.15}
  n13 = Panner(Coordinate=Multiply(A=Add(A=@n11, B=@n12)) {constB=2}) {speedX=-0.3}
  n14 = Add(
   A=Power(Base=TextureSample(UVs=@n13) {samplerType=SAMPLERTYPE_LinearColor, texture=&AnimatedOutpost_Base046_rgb_XVector_a_NormalizedHierPos_UV_2}.B)
   B=Multiply(A=Time()) {constB=0.3}
  )
  Multiply(
   A=TextureSample(UVs=Panner(Coordinate=@n6, Speed=@n8) {speedX=-0.8}) {samplerType=SAMPLERTYPE_Masks, texture=/BRCosmetics/Effects/Fort_Effects/Emotes/MajorSpeech/T_Emote_MajorSpeech_Star01}.R
   B=Multiply(A=Sine(None=@n10), B=Sine(None=@n14))
  )
  n15 = DynamicParameter() {defaultValue=<rgba 1 0.5 100 1>, paramNames [LerpTexturespeed_Offset, Bright_Random, star_bright, Param4]}
  n16 = Multiply(A=ScalarParameter() {defaultValue=-0.8, parameterName=LerpTexture_SpeedX}, B=@n15.LerpTexturespeed_Offset)
  n17 = ScalarParameter() {defaultValue=0.1, parameterName=LerpTexture_SpeedY}
  n18 = TextureSampleParameter2D(UVs=Panner(Coordinate=TextureCoordinate() {uTiling=2}, Speed=AppendVector(A=@n16, B=@n17)) {speedX=-0.8}) {parameterName=LerpTexture, samplerType=SAMPLERTYPE_Grayscale, texture=/Game/Global/Textures/Default/Noise/LowResBlurredNoiseNormalized}
  n19 = Multiply(
   A=TextureCoordinate()
   B=AppendVector(A=ScalarParameter() {defaultValue=1, parameterName=BaseTextureU}, B=ScalarParameter() {defaultValue=5, parameterName=BaseTextureV})
  )
  n20 = Panner(
   Coordinate=TextureCoordinate() {uTiling=3, vTiling=0}
   Speed=AppendVector(A=ScalarParameter() {defaultValue=-0.5, parameterName=BASE_TEX_SPEED_X}, B=ScalarParameter() {parameterName=BASE_TEX_SPEED_Y})
  ) {speedX=-0.5}
  n21 = Constant() {r=0.25}
  n22 = Multiply(A=TextureSample(UVs=MaterialFunctionCall(UVs=@n20, "Rotation Angle (0-1)"=@n21) {materialFunction=&CustomRotator}.Rotated Values) {texture=/Game/Global/Textures/Default/Masks_and_Gradients/T_VertGradient_02_sRGB}.R) {constB=0.3}
  n23 = Multiply(
   A=Power(Base=@n18.R, Exp=ScalarParameter() {defaultValue=2.5, parameterName=LerpTexture_Power}) {constExponent=1.5}
   B=TextureSampleParameter2D(UVs=Add(A=@n19, B=@n22)) {parameterName=BaseTexture, texture=/Game/Global/Textures/Default/Masks_and_Gradients/gradient_stasis}.R
  )
  n24 = TextureSample(UVs=Panner(Coordinate=TextureCoordinate() {uTiling=3, vTiling=0.2}) {speedX=-0.1}) {
   samplerType=SAMPLERTYPE_Masks
   texture=/Game/Environments/Helios/Props/Quest/Quest_GoldenFleeceStatuette_A/Textures/T_Quest_GoldenFleeceStatuette_ShinyStars
  }
  n25 = Multiply(A=TextureCoordinate() {uTiling=0.5, vTiling=0.5}) {constB=2}
  n26 = Power(Base=TextureSample(UVs=Panner(Coordinate=@n25) {speedX=-0.4}) {samplerType=SAMPLERTYPE_LinearColor, texture=&AnimatedOutpost_Base046_rgb_XVector_a_NormalizedHierPos_UV_2}.B)
  n27 = Multiply(A=Time())
  n28 = Multiply(A=Add(A=TextureCoordinate(), B=Constant2Vector() {g=0.15, r=0.15})) {constB=2}
  n29 = Power(Base=TextureSample(UVs=Panner(Coordinate=@n28) {speedX=-0.3}) {samplerType=SAMPLERTYPE_LinearColor, texture=&AnimatedOutpost_Base046_rgb_XVector_a_NormalizedHierPos_UV_2}.B)
  n30 = Multiply(A=Time()) {constB=0.3}
  n31 = Multiply(A=Sine(None=Add(A=@n26, B=@n27)), B=Sine(None=Add(A=@n29, B=@n30)))
  n32 = Multiply(A=@n5.A, B=Multiply(A=@n24.G, B=@n31))
  n33 = Multiply(A=@n5.A, B=Multiply(A=@n24.B, B=@n31))
  n34 = ScalarParameter() {defaultValue=-0.25, parameterName=MASK_PAN_SPEED_X}
  n35 = ScalarParameter() {parameterName=MASK_PAN_SPEED_Y}
  n36 = TextureSampleParameter2D(UVs=Panner(Coordinate=TextureCoordinate(), Speed=AppendVector(A=@n34, B=@n35)) {speedX=-0.25}) {parameterName=MaskTexture, samplerType=SAMPLERTYPE_Grayscale, texture=&LowResBlurredNoiseNormalized}
  n37 = Multiply(
   A=LinearInterpolate(A=Multiply(A=@n5.A, B=@n23), B=Add(A=@n32, B=@n33), Alpha=ScalarParameter() {parameterName=Use_Star})
   B=Power(Base=@n36.R)
  )
  n38 = TextureCoordinate()
  n39 = ComponentMask(None=@n38) {r=true}
  n40 = OneMinus(None=ComponentMask(None=@n38) {r=true})
  n41 = Power(Base=Multiply(A=Multiply(A=@n39, B=@n40)) {constB=10}) {constExponent=3}
  n42 = TextureCoordinate()
  n43 = ComponentMask(None=@n42) {g=true}
  n44 = OneMinus(None=ComponentMask(None=@n42) {g=true})
  n45 = Power(Base=Multiply(A=Multiply(A=@n43, B=@n44)) {constB=10}) {constExponent=3}
  n46 = Saturate(None=Multiply(A=Saturate(None=@n41), B=Saturate(None=@n45)))
  n47 = Multiply(
   A=Multiply(A=Multiply(A=@n37, B=@n46), B=ParticleColor().A)
   B=MaterialFunctionCall("FadeSensitivity (optional)"=ScalarParameter() {defaultValue=1, parameterName=MeshFade}) {materialFunction=&FadeBasedOnViewAngle}.Result
  )
  n48 = DepthFade(Opacity=@n47, FadeDistance=ScalarParameter() {defaultValue=5, parameterName=DepthFade})
  ShadingPathSwitch(Default=@n48, Deferred=@n48, Forward=@n47, Mobile=@n47)
 ]
}
