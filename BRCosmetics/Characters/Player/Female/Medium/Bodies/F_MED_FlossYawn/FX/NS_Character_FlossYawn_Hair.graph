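# Niagara hair effect for the FlossYawn character cosmetic, adding music-reactive glows, ribbons and sparks.
# Sprite glows sit on skeletal mesh bones, while FN_BoneFlow wraps ribbons along them.
# Line emitters spawn short-lived particles from music emitters through particle reads.
# Audio amplitude modulates brightness and ribbon alpha, while user flags handle off-physics variants, front-end preview and head-hidden visibility, alongside time-of-day, platform and distance scaling.
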NiagaraSystem /BRCosmetics/Characters/Player/Female/Medium/Bodies/F_MED_FlossYawn/FX/NS_Character_FlossYawn_Hair
dependencies {
 usedDataInterfaces [
  &NiagaraDataInterfaceCurve
  &NiagaraDataInterfaceArrayFloat
  &NiagaraDataInterfaceColorCurve
  &NiagaraDataInterfacePlatformSet
  &NiagaraDataInterfaceSpriteRendererInfo
  &NiagaraDataInterfaceArrayInt32
  &NiagaraDataInterfaceArrayFloat3
  &NiagaraDataInterfaceParticleRead
  &NiagaraDataInterfaceSkeletalMesh
 ]
 usedDynamicInputs [&MakeCustomLinearColorFromBool, &RandomBool]
 usedModules [
  &self:SystemSpawnScript.SystemScriptSource.SystemScriptGraph.NiagaraNodeAssignment_0.SetVariables_#1
  &self:SystemSpawnScript.SystemScriptSource.SystemScriptGraph.NiagaraNodeAssignment_4.SetVariables_#2
  &self:SystemSpawnScript.SystemScriptSource.SystemScriptGraph.NiagaraNodeAssignment_1.SetVariables_#3
  &self:SystemSpawnScript.SystemScriptSource.SystemScriptGraph.NiagaraNodeAssignment_2.SetVariables_#4
  &self:SystemSpawnScript.SystemScriptSource.SystemScriptGraph.NiagaraNodeAssignment_3.SetVariables_#5
  &InterpolateOverTime
  &SystemState
 ]
}
emitters {
 Dark_hairGlow {
  emitterUpdateScript [
   &EmitterState(
    Inactive Response=::NewEnumerator1
    Life Cycle Mode=::NewEnumerator1
    Loop Behavior=::NewEnumerator1
    Loop Duration Mode=::NewEnumerator1
    UseLoopDelay=false
   )
   &SpawnBurst_Instantaneous(
    Spawn Count=&MakeCustomIntFromBool(Bool=$User.Change_OffPhysics, True INT=0, False INT=12)
    Spawn Probability=&MakeCustomFloatFromBool(Bool=$User.bIsFrontEndPreview, True Float=1, False Float=0)
    Use Spawn Probability=true
    UseLoopCountLimit=false
   )
  ]
  particleSpawnScript [
   &PartitionParticles(Number Of Partitions=12)
   &InitializeParticle(
    Color=<rgba 0 0.0226 0.0833 1>
    Sprite Size Mode=::NewEnumerator3
    Uniform Sprite Size=&SelectFloatFromArray(
     Direct Array Index=$Particles.PartitionParticles.Partition
     Float Selection Array=DataInterface() {propertyValues {FloatData [50,80,80,80,120,100,50,80,80,80,120,100]}}
    )
    UsePositionOffset=false
    Write Material Random=true
   )
  ]
  particleUpdateScript [
   &ParticleState(Kill Particles When Lifetime Has Elapsed=false, Let Infinitely Lived Particles Die When Emitter Deactivates=true)
   &ScaleSpriteSize(
    Uniform Curve Scale=&MakeCustomFloatFromBool(Bool=&SetBoolByFloatComparison(A=$System.Music_Reactive_FX_Preview, B=0), True Float=1, False Float=1)
    Uniform Curve Sprite Scale=DataInterface() {propertyValues {Curve {keys [{tangentMode=RCTM_User, time=0, value=1}, {tangentMode=RCTM_User, time=1, value=1}]}}}
   )
   &SkeletalMeshLocation(
    Bone / Socket Index=&SelectIntFromArray(
     Direct Array Index=$Particles.PartitionParticles.Partition
     Int Selection Array=DataInterface() {propertyValues {IntData [0,0,1,1,2,2,3,3,4,4,5,5]}}
    )
    Bone Sampling Mode=::NewEnumerator4
    MaskSpawnGroup=false
    Sampled Position Offset=&SelectVectorFromArray(
     Direct Array Index=$Particles.PartitionParticles.Partition
     Vector Selection Array=DataInterface() {propertyValues {FloatData [<0 0 0>, <-13 6 0>, <0 0 0>, <-15 0 0>, <0 0 0>, <-15 6 0>, <0 0 0>, <13 -6 0>, <0 0 0>, <15 0 0>, <0 0 0>, <15 -6 0>]}}
    )
    Skeletal Mesh=$System.Skeletal Mesh
   )
   &SolveForcesAndVelocity(Clamp Velocity=false, Limit Acceleration=false)
   &DynamicMaterialParameters(
    Index 0 Param 1=3.5
    Index 0 Param 2=5
    Module.Index0Param0Write=true
    Module.Index0Param1Write=true
    Module.Index0Param2Write=true
    Module.Index0Param3Write=true
   )
   &ScaleColor(
    Scale RGB=&VectorFromFloat(Value=&Multiply_Float(
      A=&FN_FrontEndFloatModifier(Gameplay=1, Front End=1)
      B=&FNTimeOfDayScalar(Night Scale=0.7, Morning Scale=1, Evening Scale=1)
     ))
    ScaleA=true
    ScaleRGB=true
   )
   &ScaleColor(Scale Alpha=0.3, ScaleA=true, ScaleRGB=true) {module=ScaleColor001}
  ]
  rendererClasses [&NiagaraSpriteRendererProperties]
  renderers [{rendererClass=&NiagaraSpriteRendererProperties}]
 }
 Line {
  emitterUpdateScript [
   &EmitterState(Inactive Response=::NewEnumerator1, Life Cycle Mode=::NewEnumerator1, Loop Duration=1, UseLoopDelay=false)
   &SpawnParticlesFromOtherEmitter(
    Attribute Reader=DataInterface() {propertyValues {EmitterBinding {bindingMode=Other, emitterName=music}}}
    Spawn Rate=&Multiply_Float(
     A=15
     B=&MakeCustomFloatFromBool(Bool=$User.MusicReactive, True Float=1, False Float=$System.DistanceSpawn)
    )
    Use Maximum Age=false
    Use Minimum Age=false
   )
  ]
  particleSpawnScript [
   &InitializeParticle(
    Lifetime Max=0.35
    Lifetime Min=0.15
    Lifetime Mode=::NewEnumerator1
    Mesh Scale Max=<0.5 0.5 10>
    Mesh Scale Min=<0.1 0.1 5>
    Mesh Scale Mode=::NewEnumerator2
    Sprite Size Max=<1.5 16>
    Sprite Size Min=<0.5 10>
    Sprite Size Mode=::NewEnumerator2
    UsePositionOffset=false
    Write Material Random=true
   )
   &SampleParticlesFromOtherEmitter(
    Age Sampling:ENiagara_DirectReadSamplingMode=::NewEnumerator0
    Lifetime Application:ENiagara_LifetimeInheritanceOptions=::NewEnumerator0
    Lifetime Sampling:ENiagara_DirectReadSamplingMode=::NewEnumerator0
    Velocity Application:ENiagara_DirectReadApplicationMode=::NewEnumerator0
    Velocity Sampling=::NewEnumerator0
   )
   &ShapeLocation(MaskSpawnGroup=false, Override Seed=false, Shape Origin=$Particles.Position, Sphere Radius=4, UseOwnerScale=true) {module=ShapeLocation001}
  ]
  particleUpdateScript [
   &ParticleState()
   &ScaleMeshSize(Scale Factor=&Multiply_VectorByFloat(
     Vector=&MakeVector(
      X=1
      Y=1
      Z=&FloatFromCurve(FloatCurve=DataInterface() {propertyValues {Curve {keys [<at 0 0>,<at 0.2 0.95>,<at 1 1>]}}})
     )
     Float=&Multiply_Float(
      A=0.1
      B=&FloatFromCurve(
       FloatCurve=DataInterface() {propertyValues {Curve {keys [
           {arriveTangent=4.1785, time=0, value=0}
           <at 0.5 1.1>
           {arriveTangent=-0.0405, interpMode=RCIM_Cubic, leaveTangent=-0.0405, tangentMode=RCTM_User, time=1, value=1.5}
          ]}}}
       Scale Curve=&Multiply_Float(
        A=1
        B=&ClampFloat(
         Float=&FN_LOD_Distance_Remap(Start Distance=200, End Distance=400, Minimum Output=0.5, Maximum Output=1)
         Min=0.5
        )
       )
      )
     )
    ))
   &SolveForcesAndVelocity(Clamp Velocity=false, Limit Acceleration=false)
   &ScaleColor(
    Scale Alpha=&Multiply_Float(
     A=&FloatFromCurve(
      FloatCurve=DataInterface() {propertyValues {Curve {keys [<at 0 1>,<at 1 0>]}}}
      Scale Curve=&RandomRangeFloat(Minimum=0.4, Maximum=1)
     )
     B=3
    )
    Scale RGB=&Multiply_VectorByFloat(
     Vector=&MakeVectorFromLinearColorRGB(Color=&ColorFromCurve(ColorCurve=DataInterface() {propertyValues {
         AlphaCurve {keys []}
         BlueCurve {keys [<at 0.1306 0.8473 RCIM_Constant>,<at 0.1859 1 RCIM_Constant>,<at 0.3396 0.8473 RCIM_Constant>,<at 0.471 1 RCIM_Constant>,<at 0.5576 1 RCIM_Constant>,<at 0.68 1 RCIM_Constant>,<at 0.9084 0.8473>]}
         GreenCurve {keys [<at 0.1306 1 RCIM_Constant>,<at 0.1859 0.6163 RCIM_Constant>,<at 0.3396 1 RCIM_Constant>,<at 0.471 0.8644 RCIM_Constant>,<at 0.5576 0.4231 RCIM_Constant>,<at 0.68 0.5049 RCIM_Constant>,<at 0.9084 1>]}
         RedCurve {keys [<at 0.1306 0 RCIM_Constant>,<at 0.1859 0.0466 RCIM_Constant>,<at 0.3396 0 RCIM_Constant>,<at 0.471 0.4896 RCIM_Constant>,<at 0.5576 0 RCIM_Constant>,<at 0.68 0.1041 RCIM_Constant>,<at 0.9084 0>]}
        }}))
     Float=&FloatFromCurve(
      FloatCurve=DataInterface() {propertyValues {Curve {keys [
          {arriveTangent=-9.3569, interpMode=RCIM_Cubic, leaveTangent=-9.3569, tangentMode=RCTM_User, time=0, value=2}
          {arriveTangent=-2.1612, interpMode=RCIM_Cubic, leaveTangent=-2.1612, tangentMode=RCTM_User, time=0.1, value=1}
          <at 1 0.3>
         ]}}}
      CurveIndex=$Emitter.Age
      Scale Curve=1.5
     )
    )
    ScaleA=true
    ScaleRGB=true
   ) {enabled=false}
   /SetVariables(Particles.Color_glow:LinearColor=&MakeLinearColorFromVectorAndFloat(
     "Vector (RGB)"=&MakeVectorFromLinearColorRGB(Color=&ColorFromCurve(ColorCurve=DataInterface() {propertyValues {
         AlphaCurve {keys []}
         BlueCurve {keys [
           {interpMode=RCIM_Cubic, tangentMode=RCTM_User, time=0.1306, value=0.8473}
           {interpMode=RCIM_Cubic, tangentMode=RCTM_User, time=0.3325, value=1}
           {interpMode=RCIM_Cubic, tangentMode=RCTM_User, time=0.474, value=1.5}
           {interpMode=RCIM_Cubic, tangentMode=RCTM_User, time=0.5973, value=1}
           {interpMode=RCIM_Cubic, tangentMode=RCTM_User, time=0.7523, value=0.8473}
          ]}
         GreenCurve {keys [
           {interpMode=RCIM_Cubic, tangentMode=RCTM_User, time=0.1306, value=1}
           {interpMode=RCIM_Cubic, tangentMode=RCTM_User, time=0.3325, value=0.5388}
           {interpMode=RCIM_Cubic, tangentMode=RCTM_User, time=0.474, value=1.2267}
           {arriveTangent=3.4579, interpMode=RCIM_Cubic, leaveTangent=3.4579, tangentMode=RCTM_User, time=0.5973, value=0.4231}
           {interpMode=RCIM_Cubic, tangentMode=RCTM_User, time=0.7523, value=1}
          ]}
         RedCurve {keys [
           {interpMode=RCIM_Cubic, tangentMode=RCTM_User, time=0.1306, value=0}
           {interpMode=RCIM_Cubic, tangentMode=RCTM_User, time=0.3325, value=0.1031}
           {interpMode=RCIM_Cubic, tangentMode=RCTM_User, time=0.474, value=0.6135}
           {interpMode=RCIM_Cubic, tangentMode=RCTM_User, time=0.5973, value=0}
           {interpMode=RCIM_Cubic, tangentMode=RCTM_User, time=0.7523, value=0}
          ]}
        }}))
     "Float (Alpha)"=&FloatFromCurve(
      FloatCurve=DataInterface() {propertyValues {Curve {keys [<at 0.5 1>,<at 1 0>]}}}
      Scale Curve=&MakeCustomFloatFromBool(
       Bool=$User.MusicReactive
       True Float=1
       False Float=&MakeCustomFloatFromBool(
        Bool=&SetBoolByFloatComparison(A=$System.Music_Reactive_FX_Preview, B=0)
        True Float=&Multiply_Float(
         A=&RandomRangeFloat(Minimum=1, Maximum=2)
         B=&Lerp_Float(
          A=&FN_FrontEndFloatModifier(Gameplay=1, Front End=0.7)
          B=&FN_FrontEndFloatModifier(Gameplay=2, Front End=1.4)
          Alpha=$System.Music_Reactive_FX_Preview
         )
        )
        False Float=0
       )
      )
     )
    )) {isSetParameters=true, module=SetVariables_#6}
   /SetVariables(Particles.VisibilityTag=&MakeCustomIntFromBool(Bool=$User.bisHeadHidden, True INT=1, False INT=0)) {isSetParameters=true, module=SetVariables_#7}
  ]
  rendererClasses [&NiagaraSpriteRendererProperties]
  renderers [{rendererClass=&NiagaraSpriteRendererProperties}]
 }
 Line001 {
  emitterUpdateScript [
   &EmitterState(Inactive Response=::NewEnumerator1, Life Cycle Mode=::NewEnumerator1, Loop Duration=1, UseLoopDelay=false)
   &SpawnParticlesFromOtherEmitter(
    Attribute Reader=DataInterface() {propertyValues {EmitterBinding {bindingMode=Other, emitterName=music_OffPhysics}}}
    Spawn Rate=&Multiply_Float(A=15, B=$System.DistanceSpawn)
    Use Maximum Age=false
    Use Minimum Age=false
   )
  ]
  particleSpawnScript [
   &InitializeParticle(
    Lifetime Max=0.35
    Lifetime Min=0.15
    Lifetime Mode=::NewEnumerator1
    Mesh Scale Max=<0.5 0.5 10>
    Mesh Scale Min=<0.1 0.1 5>
    Mesh Scale Mode=::NewEnumerator2
    Sprite Size Max=<1.5 16>
    Sprite Size Min=<0.5 10>
    Sprite Size Mode=::NewEnumerator2
    UsePositionOffset=false
    Write Material Random=true
   )
   &SampleParticlesFromOtherEmitter(
    Age Sampling:ENiagara_DirectReadSamplingMode=::NewEnumerator0
    Lifetime Application:ENiagara_LifetimeInheritanceOptions=::NewEnumerator0
    Lifetime Sampling:ENiagara_DirectReadSamplingMode=::NewEnumerator0
    Velocity Application:ENiagara_DirectReadApplicationMode=::NewEnumerator0
    Velocity Sampling=::NewEnumerator0
   )
   &ShapeLocation(MaskSpawnGroup=false, Override Seed=false, Shape Origin=$Particles.Position, Sphere Radius=4, UseOwnerScale=true) {module=ShapeLocation001}
  ]
  particleUpdateScript [
   &ParticleState()
   &ScaleMeshSize(Scale Factor=&Multiply_VectorByFloat(
     Vector=&MakeVector(
      X=1
      Y=1
      Z=&FloatFromCurve(FloatCurve=DataInterface() {propertyValues {Curve {keys [<at 0 0>,<at 0.2 0.95>,<at 1 1>]}}})
     )
     Float=&Multiply_Float(
      A=0.1
      B=&FloatFromCurve(
       FloatCurve=DataInterface() {propertyValues {Curve {keys [
           {arriveTangent=4.1785, time=0, value=0}
           <at 0.5 1.1>
           {arriveTangent=-0.0405, interpMode=RCIM_Cubic, leaveTangent=-0.0405, tangentMode=RCTM_User, time=1, value=1.5}
          ]}}}
       Scale Curve=&Multiply_Float(
        A=1
        B=&ClampFloat(
         Float=&FN_LOD_Distance_Remap(Start Distance=200, End Distance=400, Minimum Output=0.5, Maximum Output=1)
         Min=0.5
        )
       )
      )
     )
    ))
   &SolveForcesAndVelocity(Clamp Velocity=false, Limit Acceleration=false)
   &ScaleColor(
    Scale Alpha=&Multiply_Float(
     A=&FloatFromCurve(
      FloatCurve=DataInterface() {propertyValues {Curve {keys [<at 0 1>,<at 1 0>]}}}
      Scale Curve=&RandomRangeFloat(Minimum=0.4, Maximum=1)
     )
     B=3
    )
    Scale RGB=&Multiply_VectorByFloat(
     Vector=&MakeVectorFromLinearColorRGB(Color=&ColorFromCurve(ColorCurve=DataInterface() {propertyValues {
         AlphaCurve {keys []}
         BlueCurve {keys [<at 0.1306 0.8473 RCIM_Constant>,<at 0.1859 1 RCIM_Constant>,<at 0.3396 0.8473 RCIM_Constant>,<at 0.471 1 RCIM_Constant>,<at 0.5576 1 RCIM_Constant>,<at 0.68 1 RCIM_Constant>,<at 0.9084 0.8473>]}
         GreenCurve {keys [<at 0.1306 1 RCIM_Constant>,<at 0.1859 0.6163 RCIM_Constant>,<at 0.3396 1 RCIM_Constant>,<at 0.471 0.8644 RCIM_Constant>,<at 0.5576 0.4231 RCIM_Constant>,<at 0.68 0.5049 RCIM_Constant>,<at 0.9084 1>]}
         RedCurve {keys [<at 0.1306 0 RCIM_Constant>,<at 0.1859 0.0466 RCIM_Constant>,<at 0.3396 0 RCIM_Constant>,<at 0.471 0.4896 RCIM_Constant>,<at 0.5576 0 RCIM_Constant>,<at 0.68 0.1041 RCIM_Constant>,<at 0.9084 0>]}
        }}))
     Float=&FloatFromCurve(
      FloatCurve=DataInterface() {propertyValues {Curve {keys [
          {arriveTangent=-9.3569, interpMode=RCIM_Cubic, leaveTangent=-9.3569, tangentMode=RCTM_User, time=0, value=2}
          {arriveTangent=-2.1612, interpMode=RCIM_Cubic, leaveTangent=-2.1612, tangentMode=RCTM_User, time=0.1, value=1}
          <at 1 0.3>
         ]}}}
      CurveIndex=$Emitter.Age
      Scale Curve=1.5
     )
    )
    ScaleA=true
    ScaleRGB=true
   ) {enabled=false}
   /SetVariables(Particles.Color_glow:LinearColor=&MakeLinearColorFromVectorAndFloat(
     "Vector (RGB)"=&MakeVectorFromLinearColorRGB(Color=&ColorFromCurve(ColorCurve=DataInterface() {propertyValues {
         AlphaCurve {keys []}
         BlueCurve {keys [
           {interpMode=RCIM_Cubic, tangentMode=RCTM_User, time=0.1306, value=0.8473}
           {interpMode=RCIM_Cubic, tangentMode=RCTM_User, time=0.3325, value=1}
           {interpMode=RCIM_Cubic, tangentMode=RCTM_User, time=0.474, value=1.5}
           {interpMode=RCIM_Cubic, tangentMode=RCTM_User, time=0.5973, value=1}
           {interpMode=RCIM_Cubic, tangentMode=RCTM_User, time=0.7523, value=0.8473}
          ]}
         GreenCurve {keys [
           {interpMode=RCIM_Cubic, tangentMode=RCTM_User, time=0.1306, value=1}
           {interpMode=RCIM_Cubic, tangentMode=RCTM_User, time=0.3325, value=0.5388}
           {interpMode=RCIM_Cubic, tangentMode=RCTM_User, time=0.474, value=1.2267}
           {arriveTangent=3.4579, interpMode=RCIM_Cubic, leaveTangent=3.4579, tangentMode=RCTM_User, time=0.5973, value=0.4231}
           {interpMode=RCIM_Cubic, tangentMode=RCTM_User, time=0.7523, value=1}
          ]}
         RedCurve {keys [
           {interpMode=RCIM_Cubic, tangentMode=RCTM_User, time=0.1306, value=0}
           {interpMode=RCIM_Cubic, tangentMode=RCTM_User, time=0.3325, value=0.1031}
           {interpMode=RCIM_Cubic, tangentMode=RCTM_User, time=0.474, value=0.6135}
           {interpMode=RCIM_Cubic, tangentMode=RCTM_User, time=0.5973, value=0}
           {interpMode=RCIM_Cubic, tangentMode=RCTM_User, time=0.7523, value=0}
          ]}
        }}))
     "Float (Alpha)"=&FloatFromCurve(
      FloatCurve=DataInterface() {propertyValues {Curve {keys [<at 0.5 1>,<at 1 0>]}}}
      Scale Curve=&MakeCustomFloatFromBool(
       Bool=$User.MusicReactive
       True Float=1
       False Float=&MakeCustomFloatFromBool(
        Bool=&SetBoolByFloatComparison(A=$System.Music_Reactive_FX_Preview, B=0)
        True Float=&Multiply_Float(
         A=&RandomRangeFloat(Minimum=1, Maximum=2)
         B=&Lerp_Float(
          A=&FN_FrontEndFloatModifier(Gameplay=1, Front End=0.7)
          B=&FN_FrontEndFloatModifier(Gameplay=2, Front End=1.4)
          Alpha=$System.Music_Reactive_FX_Preview
         )
        )
        False Float=0
       )
      )
     )
    )) {isSetParameters=true, module=SetVariables_#6}
   /SetVariables(Particles.VisibilityTag=&MakeCustomIntFromBool(Bool=$User.bisHeadHidden, True INT=1, False INT=0)) {isSetParameters=true, module=SetVariables_#7}
  ]
  rendererClasses [&NiagaraSpriteRendererProperties]
  renderers [{rendererClass=&NiagaraSpriteRendererProperties}]
 }
 RingGlow001 {
  emitterUpdateScript [
   &EmitterState(
    Inactive Response=::NewEnumerator1
    Life Cycle Mode=::NewEnumerator1
    Loop Behavior=::NewEnumerator1
    Loop Duration Mode=::NewEnumerator1
    UseLoopDelay=false
   )
   &SpawnBurst_Instantaneous(
    Spawn Count=40
    Spawn Probability=&MakeCustomFloatFromBool(Bool=$User.Change_OffPhysics, True Float=0, False Float=1)
    Use Spawn Probability=true
    UseLoopCountLimit=false
   )
  ]
  particleSpawnScript [
   &FN_SetSpawnGroup(Inclusivity=true, Normalize Spawn Group=true, Particles Per Group=20, Pattern=::NewEnumerator0, Set Ribbon ID and Ribbon Link Order=true)
   &InitializeParticle(
    Color=<rgba 0.7396 0.7396 0.7396 2>
    Ribbon Width=10
    Ribbon Width Mode=::NewEnumerator1
    UsePositionOffset=false
    Write Material Random=true
   )
  ]
  particleUpdateScript [
   &ParticleState(Kill Particles When Lifetime Has Elapsed=false, Let Infinitely Lived Particles Die When Emitter Deactivates=true)
   &FN_BoneFlow(
    Angle:NiagaraFloat=&FloatFromCurve(
     FloatCurve=DataInterface() {propertyValues {Curve {keys [
         {arriveTangent=0.7096, interpMode=RCIM_Cubic, leaveTangent=0.7096, tangentMode=RCTM_User, time=0, value=0}
         {arriveTangent=1.7067, interpMode=RCIM_Cubic, leaveTangent=1.7067, time=0.3426, value=0.3883}
         {arriveTangent=1.9954, interpMode=RCIM_Cubic, leaveTangent=1.9954, time=0.7022, value=1.1984}
         {arriveTangent=1.2828, interpMode=RCIM_Cubic, leaveTangent=1.2828, tangentMode=RCTM_User, time=1, value=1.7}
        ]}}}
     CurveIndex=$Particles.Normalized Spawn Group
    )
    BoneID:NiagaraInt32=&ModuloInt(A=$Particles.SpawnGroup, B=2)
    Height:NiagaraFloat=&Multiply_Float(
     A=$Particles.Normalized Spawn Group
     B=&SelectFloatFromArray(Direct Array Index=$Particles.SpawnGroup, Float Selection Array=DataInterface() {propertyValues {FloatData [1.3,1.2]}})
    )
    Radius:NiagaraFloat=&FloatFromCurve(
     FloatCurve=DataInterface() {propertyValues {Curve {keys [
         {arriveTangent=18.653, interpMode=RCIM_Cubic, leaveTangent=18.653, tangentMode=RCTM_User, time=0, value=0.5}
         {arriveTangent=6.6354, interpMode=RCIM_Cubic, leaveTangent=6.6354, tangentMode=RCTM_User, time=0.2082, value=3.3137}
         {arriveTangent=2.2303, interpMode=RCIM_Cubic, leaveTangent=2.2303, tangentMode=RCTM_User, time=0.9994, value=6.8915}
        ]}}}
     CurveIndex=$Particles.Normalized Spawn Group
    )
    SideVector1:Vector3f=<1 1 1>
    SideVector2:Vector3f=<1 1 1>
    SideVector3:Vector3f=<1 1 1>
    SkeletalMesh:NiagaraDataInterfaceSkeletalMesh=$System.Skeletal Mesh
    Stabalize Twisting=true
   )
   &ScaleColor(
    Scale Alpha=&FloatFromCurve(
     FloatCurve=DataInterface() {propertyValues {Curve {keys [
         {arriveTangent=0.6799, interpMode=RCIM_Cubic, leaveTangent=0.6799, tangentMode=RCTM_User, time=0, value=0}
         {arriveTangent=3.7861, interpMode=RCIM_Cubic, leaveTangent=-1.2051, tangentMode=RCTM_Break, time=0.4666, value=0.9922}
         {arriveTangent=-0.2647, interpMode=RCIM_Cubic, leaveTangent=-0.2647, tangentMode=RCTM_User, time=1, value=0}
        ]}}}
     CurveIndex=$Particles.RibbonLinkOrder
     Scale Curve=&MakeCustomFloatFromBool(
      Bool=$User.MusicReactive
      True Float=1
      False Float=&Multiply_Float(
       A=&FNTimeOfDayScalar(Night Scale=0.5, Morning Scale=1, Evening Scale=1)
       B=&Multiply_Float(
        A=&FN_FrontEndFloatModifier(Gameplay=1, Front End=0.6)
        B=&MakeCustomFloatFromBool(
         Bool=&SetBoolByFloatComparison(A=$System.Music_Reactive_FX_Preview, B=0)
         True Float=&Multiply_Float(
          A=0.5
          B=&Multiply_Float(A=3, B=&Multiply_Float(A=$NPC.FortniteNPC.EventAudio_Amplitude, B=$NPC.FortniteNPC.EventAudio_Amplitude))
         )
         False Float=0
        )
       )
      )
     )
    )
    Scale RGB=&MakeVectorFromLinearColorRGB(Color=&ColorFromCurve(
      ColorCurve=DataInterface() {propertyValues {
        AlphaCurve {keys []}
        BlueCurve {keys [<at 0.1537 0.5521 RCIM_Cubic>,<at 0.5447 0.5>,<at 0.7885 0.7>]}
        GreenCurve {keys [<at 0.1537 0.4885 RCIM_Cubic>,<at 0.5447 0.3879>,<at 0.7885 0.3162>]}
        RedCurve {keys [<at 0.1537 0.1872 RCIM_Cubic>,<at 0.5447 0.1647>,<at 0.7885 0.0164>]}
       }}
      CurveIndex=$StackContext.RibbonLinkOrder
     ))
    ScaleA=true
    ScaleRGB=true
   )
   &ScaleRibbonWidth(Ribbon Width Scale=&FloatFromCurve(
     FloatCurve=DataInterface() {propertyValues {Curve {keys [<at 0 0.5>,<at 0.3758 1.6403>,<at 1 1>]}}}
     CurveIndex=$Particles.RibbonLinkOrder
     Scale Curve=&MakeCustomFloatFromBool(
      Bool=$User.MusicReactive
      True Float=1
      False Float=&Lerp_Float(A=0, B=1, Alpha=$System.Music_Reactive_FX_Preview)
     )
    ))
   &ScaleRibbonWidth(Ribbon Width Scale=&Lerp_Float(A=1, B=1.2, Alpha=$System.InterpolateOverTime.InterpolateOverTime_Float.Moving Average)) {module=ScaleRibbonWidth001}
   &ScaleColor(
    Scale Alpha=&MakeCustomFloatFromBool(Bool=$User.bisHeadHidden, True Float=0, False Float=1)
    ScaleA=true
    ScaleRGB=false
   ) {module=ScaleColor001}
  ]
  rendererClasses [&NiagaraRibbonRendererProperties]
  renderers [{rendererClass=&NiagaraRibbonRendererProperties}]
 }
 RingGlow002 {
  bEnabled=false
  emitterUpdateScript [
   &EmitterState(
    Inactive Response=::NewEnumerator1
    Life Cycle Mode=::NewEnumerator1
    Loop Behavior=::NewEnumerator1
    Loop Duration Mode=::NewEnumerator1
    UseLoopDelay=false
   )
   &SpawnBurst_Instantaneous(
    Spawn Count=40
    Spawn Probability=&MakeCustomFloatFromBool(Bool=$User.Change_OffPhysics, True Float=1, False Float=0)
    Use Spawn Probability=true
    UseLoopCountLimit=false
   )
  ]
  particleSpawnScript [
   &FN_SetSpawnGroup(Inclusivity=true, Normalize Spawn Group=true, Particles Per Group=20, Pattern=::NewEnumerator0, Set Ribbon ID and Ribbon Link Order=true)
   &InitializeParticle(Color=<rgba 50 50 50 2.5>, Ribbon Width=4, Ribbon Width Mode=::NewEnumerator1, UsePositionOffset=false, Write Material Random=true)
  ]
  particleUpdateScript [
   &ParticleState(Kill Particles When Lifetime Has Elapsed=false, Let Infinitely Lived Particles Die When Emitter Deactivates=true)
   &FN_BoneFlow(
    Angle:NiagaraFloat=&FloatFromCurve(
     FloatCurve=DataInterface() {propertyValues {Curve {keys [{arriveTangent=0.7096, time=0, value=3.25}, {arriveTangent=1.2828, leaveTangent=1.2828, time=1.1, value=1}]}}}
     CurveIndex=$Particles.Normalized Spawn Group
     Scale Curve=&SelectFloatFromArray(Direct Array Index=$Particles.SpawnGroup, Float Selection Array=DataInterface() {propertyValues {FloatData [1,-1]}})
    )
    BoneID:NiagaraInt32=&ModuloInt(A=$Particles.SpawnGroup, B=1)
    Enable Position Offsets=true
    Height:NiagaraFloat=&Multiply_Float(
     A=$Particles.Normalized Spawn Group
     B=&SelectFloatFromArray(Direct Array Index=$Particles.SpawnGroup, Float Selection Array=DataInterface() {propertyValues {FloatData [1.2,1.15]}})
    )
    Offset End:Vector3f=&SelectVectorFromArray(Direct Array Index=$Particles.SpawnGroup, Vector Selection Array=DataInterface() {propertyValues {FloatData [<0 -30 25>, <0 -30 -25>]}})
    Offset Mid:Vector3f=&SelectVectorFromArray(Direct Array Index=$Particles.SpawnGroup, Vector Selection Array=DataInterface() {propertyValues {FloatData [<0 -20 20>, <0 -20 -20>]}})
    Offset Start:Vector3f=&SelectVectorFromArray(Direct Array Index=$Particles.SpawnGroup, Vector Selection Array=DataInterface() {propertyValues {FloatData [<10 -10 20>, <10 -10 -20>]}})
    Radius:NiagaraFloat=&FloatFromCurve(
     FloatCurve=DataInterface() {propertyValues {Curve {keys [
         {arriveTangent=5.133, interpMode=RCIM_Cubic, leaveTangent=5.133, tangentMode=RCTM_User, time=0, value=1}
         {arriveTangent=6.4737, interpMode=RCIM_Cubic, leaveTangent=6.4737, tangentMode=RCTM_User, time=0.2, value=2.5}
         {arriveTangent=1.3602, interpMode=RCIM_Cubic, leaveTangent=1.3602, tangentMode=RCTM_User, time=1, value=6.3889}
        ]}}}
     CurveIndex=$Particles.Normalized Spawn Group
     Scale Curve=1.1
    )
    SideVector1:Vector3f=<1 1 1>
    SideVector2:Vector3f=<1 1 1>
    SideVector3:Vector3f=<1 1 1>
    SkeletalMesh:NiagaraDataInterfaceSkeletalMesh=$System.Skeletal Mesh Reader
    Stabalize Twisting=true
   )
   &ScaleColor(
    Scale Alpha=&FloatFromCurve(
     FloatCurve=DataInterface() {propertyValues {Curve {keys [
         {arriveTangent=-0.0164, interpMode=RCIM_Cubic, leaveTangent=-0.0164, tangentMode=RCTM_User, time=0, value=0}
         {arriveTangent=2.3696, interpMode=RCIM_Cubic, leaveTangent=-1.2051, tangentMode=RCTM_Break, time=0.4, value=1}
         {arriveTangent=-1.5327, interpMode=RCIM_Cubic, leaveTangent=-1.5327, tangentMode=RCTM_User, time=1, value=0}
        ]}}}
     CurveIndex=$Particles.RibbonLinkOrder
     Scale Curve=&MakeCustomFloatFromBool(
      Bool=$User.bIsFrontEndPreview
      True Float=&Lerp_Float(A=0.85, B=1.2, Alpha=$NPC.FortniteNPC.EventAudio_Amplitude)
      False Float=&Lerp_Float(A=1, B=1.5, Alpha=$NPC.FortniteNPC.EventAudio_Amplitude)
     )
    )
    Scale RGB=&MakeVectorFromLinearColorRGB(Color=&ColorFromCurve(
      ColorCurve=DataInterface() {propertyValues {
        AlphaCurve {keys []}
        BlueCurve {keys [<at 0.1537 1 RCIM_Cubic>,<at 0.7885 0.7>]}
        GreenCurve {keys [<at 0.1537 0.8848 RCIM_Cubic>,<at 0.7885 0.3162>]}
        RedCurve {keys [<at 0.1537 0.339 RCIM_Cubic>,<at 0.7885 0.0164>]}
       }}
      CurveIndex=$StackContext.RibbonLinkOrder
     ))
    ScaleA=true
    ScaleRGB=true
   )
   /SetVariables(Particles.Color_glow:LinearColor=&Multiply_LinearColorByFloat(
     Linear Color=<rgba 0.0931 0.3739 1 1>
     Float=&MakeCustomFloatFromBool(
      Bool=$User.MusicReactive
      True Float=1
      False Float=&Multiply_Float(
       A=&FNTimeOfDayScalar(Night Scale=0.6, Morning Scale=1, Evening Scale=1)
       B=&Multiply_Float(
        A=&FN_FrontEndFloatModifier(Gameplay=1, Front End=0.6)
        B=&Lerp_Float(
         A=0.8
         B=&Multiply_Float(
          A=&FN_PickFloatDependOnPlatform(
           TargetPlatformActive=0.5
           TargetPlatformNoActive=&SelectFloatFromArray(
            Direct Array Index=$Particles.SpawnGroup
            Float Selection Array=DataInterface() {propertyValues {FloatData [0.5,0.5,0.5,0.5,0.5,0.5,0.3,0.3]}}
           )
          )
          B=&Multiply_Float(A=3, B=$NPC.FortniteNPC.EventAudio_Amplitude)
         )
         Alpha=$System.Music_Reactive_FX_Preview
        )
       )
      )
     )
    )) {isSetParameters=true, module=SetVariables_#8}
   /SetVariables(Particles.Color_star:LinearColor=&Multiply_LinearColorByFloat(
     Linear Color=<rgba 100 100 100 1>
     Float=&Lerp_Float(A=1, B=2, Alpha=$System.InterpolateOverTime.InterpolateOverTime_Float.Moving Average)
    )) {isSetParameters=true, module=SetVariables_#9}
   &ScaleRibbonWidth(Ribbon Width Scale=&FloatFromCurve(FloatCurve=DataInterface() {propertyValues {Curve {keys [<at 0 0.5>,<at 0.3 1.75>,<at 1 2.25>]}}}, CurveIndex=$Particles.RibbonLinkOrder))
   &ScaleRibbonWidth(Ribbon Width Scale=&Lerp_Float(A=0.8, B=1.4, Alpha=$System.InterpolateOverTime.InterpolateOverTime_Float.Moving Average)) {module=ScaleRibbonWidth001}
   &ScaleColor(
    Scale Alpha=&MakeCustomFloatFromBool(Bool=$User.bisHeadHidden, True Float=0, False Float=1)
    ScaleA=true
    ScaleRGB=false
   ) {module=ScaleColor001}
  ]
  rendererClasses [&NiagaraRibbonRendererProperties]
  renderers [
   {rendererClass=&NiagaraRibbonRendererProperties}
   {rendererClass=&NiagaraRibbonRendererProperties, rendererIndex=1}
   {rendererClass=&NiagaraRibbonRendererProperties, rendererIndex=2}
  ]
 }
 RingGlow004 {
  emitterUpdateScript [
   &EmitterState(
    Inactive Response=::NewEnumerator1
    Life Cycle Mode=::NewEnumerator1
    Loop Behavior=::NewEnumerator1
    Loop Duration Mode=::NewEnumerator1
    UseLoopDelay=false
   )
   &SpawnBurst_Instantaneous(
    Spawn Count=40
    Spawn Probability=&MakeCustomFloatFromBool(Bool=$User.Change_OffPhysics, True Float=1, False Float=1)
    Use Spawn Probability=true
    UseLoopCountLimit=false
   )
  ]
  particleSpawnScript [
   &FN_SetSpawnGroup(Inclusivity=true, Normalize Spawn Group=true, Particles Per Group=20, Pattern=::NewEnumerator0, Set Ribbon ID and Ribbon Link Order=true)
   &InitializeParticle(Color=<rgba 50 50 50 2.5>, Ribbon Width=4, Ribbon Width Mode=::NewEnumerator1, UsePositionOffset=false, Write Material Random=true)
  ]
  particleUpdateScript [
   &ParticleState(Kill Particles When Lifetime Has Elapsed=false, Let Infinitely Lived Particles Die When Emitter Deactivates=true)
   &FN_BoneFlow(
    Angle:NiagaraFloat=&FloatFromCurve(
     FloatCurve=DataInterface() {propertyValues {Curve {keys [{arriveTangent=0.7096, time=0, value=3.25}, {arriveTangent=1.2828, leaveTangent=1.2828, time=1.1, value=1}]}}}
     CurveIndex=$Particles.Normalized Spawn Group
     Scale Curve=&SelectFloatFromArray(Direct Array Index=$Particles.SpawnGroup, Float Selection Array=DataInterface() {propertyValues {FloatData [-1,-1]}})
    )
    BoneID:NiagaraInt32=&ModuloInt(A=$Particles.SpawnGroup, B=2)
    Height:NiagaraFloat=&Multiply_Float(
     A=$Particles.Normalized Spawn Group
     B=&SelectFloatFromArray(Direct Array Index=$Particles.SpawnGroup, Float Selection Array=DataInterface() {propertyValues {FloatData [1.25,1.15]}})
    )
    Radius:NiagaraFloat=&FloatFromCurve(
     FloatCurve=DataInterface() {propertyValues {Curve {keys [
         {arriveTangent=12.2683, interpMode=RCIM_Cubic, leaveTangent=12.2683, tangentMode=RCTM_User, time=0, value=2}
         {arriveTangent=6.4737, interpMode=RCIM_Cubic, leaveTangent=6.4737, tangentMode=RCTM_User, time=0.2028, value=4.0017}
         {arriveTangent=1.3602, interpMode=RCIM_Cubic, leaveTangent=1.3602, tangentMode=RCTM_User, time=1, value=6.3889}
        ]}}}
     CurveIndex=$Particles.Normalized Spawn Group
     Scale Curve=1.1
    )
    SideVector1:Vector3f=<1 1 1>
    SideVector2:Vector3f=<1 1 1>
    SideVector3:Vector3f=<1 1 1>
    SkeletalMesh:NiagaraDataInterfaceSkeletalMesh=$System.Skeletal Mesh
    Stabalize Twisting=true
   )
   &ScaleColor(
    Scale Alpha=&FloatFromCurve(
     FloatCurve=DataInterface() {propertyValues {Curve {keys [
         {arriveTangent=-0.0164, interpMode=RCIM_Cubic, leaveTangent=-0.0164, tangentMode=RCTM_User, time=0, value=0}
         {arriveTangent=2.3696, interpMode=RCIM_Cubic, leaveTangent=-1.2051, tangentMode=RCTM_Break, time=0.4, value=1}
         {arriveTangent=-1.5327, interpMode=RCIM_Cubic, leaveTangent=-1.5327, tangentMode=RCTM_User, time=1, value=0}
        ]}}}
     CurveIndex=$Particles.RibbonLinkOrder
     Scale Curve=&MakeCustomFloatFromBool(
      Bool=$User.bIsFrontEndPreview
      True Float=&Lerp_Float(A=0.85, B=1.2, Alpha=$NPC.FortniteNPC.EventAudio_Amplitude)
      False Float=&Lerp_Float(A=1, B=1.5, Alpha=$NPC.FortniteNPC.EventAudio_Amplitude)
     )
    )
    Scale RGB=&MakeVectorFromLinearColorRGB(Color=&ColorFromCurve(
      ColorCurve=DataInterface() {propertyValues {
        AlphaCurve {keys []}
        BlueCurve {keys [<at 0.1537 1 RCIM_Cubic>,<at 0.7885 0.7>]}
        GreenCurve {keys [<at 0.1537 0.8848 RCIM_Cubic>,<at 0.7885 0.3162>]}
        RedCurve {keys [<at 0.1537 0.339 RCIM_Cubic>,<at 0.7885 0.0164>]}
       }}
      CurveIndex=$StackContext.RibbonLinkOrder
     ))
    ScaleA=true
    ScaleRGB=true
   )
   /SetVariables(Particles.Color_glow:LinearColor=&Multiply_LinearColorByFloat(
     Linear Color=<rgba 0.0931 0.3739 1 1>
     Float=&MakeCustomFloatFromBool(
      Bool=$User.MusicReactive
      True Float=1
      False Float=&Multiply_Float(
       A=&FNTimeOfDayScalar(Night Scale=0.6, Morning Scale=1, Evening Scale=1)
       B=&Multiply_Float(
        A=&FN_FrontEndFloatModifier(Gameplay=1, Front End=0.6)
        B=&Lerp_Float(
         A=0.8
         B=&Multiply_Float(
          A=&FN_PickFloatDependOnPlatform(
           TargetPlatformActive=0.5
           TargetPlatformNoActive=&SelectFloatFromArray(
            Direct Array Index=$Particles.SpawnGroup
            Float Selection Array=DataInterface() {propertyValues {FloatData [0.5,0.5,0.5,0.5,0.5,0.5,0.3,0.3]}}
           )
          )
          B=&Multiply_Float(A=3, B=$NPC.FortniteNPC.EventAudio_Amplitude)
         )
         Alpha=$System.Music_Reactive_FX_Preview
        )
       )
      )
     )
    )) {isSetParameters=true, module=SetVariables_#8}
   /SetVariables(Particles.Color_star:LinearColor=&Multiply_LinearColorByFloat(
     Linear Color=<rgba 100 100 100 1>
     Float=&Lerp_Float(A=1, B=2, Alpha=$System.InterpolateOverTime.InterpolateOverTime_Float.Moving Average)
    )) {isSetParameters=true, module=SetVariables_#9}
   &ScaleRibbonWidth(Ribbon Width Scale=&FloatFromCurve(FloatCurve=DataInterface() {propertyValues {Curve {keys [<at 0 0.5>,<at 0.3 1.75>,<at 1 2.25>]}}}, CurveIndex=$Particles.RibbonLinkOrder))
   &ScaleRibbonWidth(Ribbon Width Scale=&Lerp_Float(A=0.8, B=1.4, Alpha=$System.InterpolateOverTime.InterpolateOverTime_Float.Moving Average)) {module=ScaleRibbonWidth001}
   &ScaleColor(
    Scale Alpha=&MakeCustomFloatFromBool(Bool=$User.bisHeadHidden, True Float=0, False Float=1)
    ScaleA=true
    ScaleRGB=false
   ) {module=ScaleColor001}
  ]
  rendererClasses [&NiagaraRibbonRendererProperties]
  renderers [
   {rendererClass=&NiagaraRibbonRendererProperties}
   {rendererClass=&NiagaraRibbonRendererProperties, rendererIndex=1}
   {rendererClass=&NiagaraRibbonRendererProperties, rendererIndex=2}
  ]
 }
 hairGlow001 {
  emitterUpdateScript [
   &EmitterState(
    Inactive Response=::NewEnumerator1
    Life Cycle Mode=::NewEnumerator1
    Loop Behavior=::NewEnumerator1
    Loop Duration Mode=::NewEnumerator1
    UseLoopDelay=false
   )
   &SpawnBurst_Instantaneous(
    Spawn Count=&MakeCustomIntFromBool(Bool=$User.Change_OffPhysics, True INT=0, False INT=14)
    Use Spawn Probability=true
    UseLoopCountLimit=false
   )
  ]
  particleSpawnScript [
   &PartitionParticles(Number Of Partitions=14)
   &InitializeParticle(
    Color=<rgba 0.0526 0.2913 0.5 3>
    Sprite Size Mode=::NewEnumerator3
    Uniform Sprite Size=&MakeCustomFloatFromBool(
     Bool=$User.Change_OffPhysics
     True Float=&SelectFloatFromArray(
      Direct Array Index=$Particles.PartitionParticles.Partition
      Float Selection Array=DataInterface() {propertyValues {FloatData [30,32,35,35,40,40,30,32,35,35,40,40,0,0]}}
     )
     False Float=&SelectFloatFromArray(
      Direct Array Index=$Particles.PartitionParticles.Partition
      Float Selection Array=DataInterface() {propertyValues {FloatData [30,33,35,38,42,45,45,30,33,35,38,42,45,45]}}
     )
    )
    UsePositionOffset=false
    Write Material Random=true
   )
  ]
  particleUpdateScript [
   &ParticleState(Kill Particles When Lifetime Has Elapsed=false, Let Infinitely Lived Particles Die When Emitter Deactivates=true)
   &ScaleSpriteSize(Uniform Curve Sprite Scale=DataInterface() {propertyValues {Curve {keys [{tangentMode=RCTM_User, time=0, value=1}, {tangentMode=RCTM_User, time=1, value=1}]}}})
   &SkeletalMeshLocation(
    Bone / Socket Index=&MakeCustomIntFromBool(
     Bool=$User.Change_OffPhysics
     True INT=&SelectIntFromArray(
      Direct Array Index=$Particles.PartitionParticles.Partition
      Int Selection Array=DataInterface() {propertyValues {IntData [1,1,11,11,2,2,4,4,12,12,5,5,8,8]}}
     )
     False INT=&SelectIntFromArray(
      Direct Array Index=$Particles.PartitionParticles.Partition
      Int Selection Array=DataInterface() {propertyValues {IntData [1,1,1,1,2,2,2,4,4,4,4,5,5,5]}}
     )
    )
    Bone Sampling Mode=::NewEnumerator4
    MaskSpawnGroup=false
    Sampled Position Offset=&MakeCustomVectorFromBool(
     Bool=$User.Change_OffPhysics
     True Vector=&SelectVectorFromArray(
      Direct Array Index=$Particles.PartitionParticles.Partition
      Vector Selection Array=DataInterface() {propertyValues {FloatData [<5 0 0>, <-5 0 0>, <5 -3 0>, <-5 0 0>, <5 0 0>, <-5 3 0>, <5 5 0>, <-5 0 0>, <5 0 0>, <-4 4 -3>, <5 0 0>, <-5 3 0>, <0 0 0>, <0 0 0>]}}
     )
     False Vector=&SelectVectorFromArray(
      Direct Array Index=$Particles.PartitionParticles.Partition
      Vector Selection Array=DataInterface() {propertyValues {FloatData [
         <10.9081 0.3129 1>
         <1.2535 0 0>
         <-8.3922 0 0>
         <-17.4839 0 -1.513>
         <0 0 0>
         <-11.3031 6 -2.032>
         <-25 5 -4.2863>
         <-10.9081 -0.31 -1>
         <-1.2535 0 0>
         <8.3922 0 0>
         <17.4839 0 1>
         <0 0 0>
         <11.3031 -6 2>
         <25 -5 2>
        ]}}
     )
    )
    Skeletal Mesh=$System.Skeletal Mesh
   )
   &SolveForcesAndVelocity(Clamp Velocity=false, Limit Acceleration=false)
   &DynamicMaterialParameters(
    Index 0 Param 1=3.5
    Index 0 Param 2=5
    Module.Index0Param0Write=true
    Module.Index0Param1Write=true
    Module.Index0Param2Write=true
    Module.Index0Param3Write=true
   )
   &ScaleColor(
    Scale Alpha=&Multiply_Float(
     A=&SelectFloatFromArray(
      Direct Array Index=$Particles.PartitionParticles.Partition
      Float Selection Array=DataInterface() {propertyValues {FloatData [0.2,0.4,0.7,1,0.8,0.5,0.4,0.1,0.3,0.6,0.9,0.8,0.6,0.5,0.2]}}
     )
     B=&FN_PickFloatDependOnPlatform(TargetPlatformActive=1, TargetPlatformNoActive=0.5)
    )
    Scale RGB=&VectorFromFloat(Value=&Multiply_Float(
      A=&FN_FrontEndFloatModifier(Gameplay=1, Front End=0.7)
      B=&FNTimeOfDayScalar(Night Scale=0.7, Morning Scale=1, Evening Scale=1)
     ))
    ScaleA=true
    ScaleRGB=true
   )
   &ScaleColor(
    Scale Alpha=&MakeCustomFloatFromBool(
     Bool=$User.MusicReactive
     True Float=0.4
     False Float=&Lerp_Float(
      A=0.15
      B=&Multiply_Float(A=0.6, B=&Multiply_Float(A=1, B=$NPC.FortniteNPC.EventAudio_Amplitude))
      Alpha=$System.Music_Reactive_FX_Preview
     )
    )
    ScaleA=true
    ScaleRGB=true
   ) {module=ScaleColor001}
   /SetVariables(Particles.VisibilityTag=&MakeCustomIntFromBool(Bool=$User.bisHeadHidden, True INT=1, False INT=0)) {isSetParameters=true, module=SetVariables_#7}
  ]
  rendererClasses [&NiagaraSpriteRendererProperties]
  renderers [{rendererClass=&NiagaraSpriteRendererProperties}]
 }
 music {
  emitterUpdateScript [
   &EmitterState(Inactive Response=::NewEnumerator1, Life Cycle Mode=::NewEnumerator1, Loop Duration=1, UseLoopDelay=false)
   &SpawnRate(
    Spawn Probability=&MakeCustomFloatFromBool(Bool=&SetBoolByFloatComparison(A=$System.Music_Reactive_FX_Preview, B=0), True Float=1, False Float=0)
    SpawnRate=&Multiply_Float(
     A=&MakeCustomFloatFromBool(Bool=$User.Change_OffPhysics, True Float=0, False Float=8)
     B=&MakeCustomFloatFromBool(Bool=$User.MusicReactive, True Float=1, False Float=$System.DistanceSpawn)
    )
    Use Spawn Probability=true
   )
  ]
  particleSpawnScript [
   &FN_SetSpawnGroup(Number of Spawn Groups=6)
   &InitializeParticle(
    Lifetime Max=1
    Lifetime Mode=::NewEnumerator1
    Sprite Rotation Angle=&RandomRangeFloat(Minimum=-0.15, Maximum=0.15)
    Sprite Rotation Mode=::NewEnumerator1
    Sprite Size Max=<11 14>
    Sprite Size Min=<10 12>
    Sprite Size Mode=::NewEnumerator2
    UsePositionOffset=false
    Write Material Random=true
   )
   &ShapeLocation(
    Handle Radius=5
    Large Radius=&SelectFloatFromArray(Direct Array Index=$Particles.SpawnGroup, Float Selection Array=DataInterface() {propertyValues {FloatData [10,25,30,10,25,30]}})
    MaskSpawnGroup=false
    Non Uniform Scale=<0.5 0.5 1>
    Override Seed=false
    Rotation Axis=<0 1 0>
    Rotation Mode=::NewEnumerator1
    Shape Primitive=::NewEnumerator3
    Surface Distribution=1
    UseOwnerScale=true
   )
   &CameraOffset(Camera Offset Amount=&MakeCustomFloatFromBool(Bool=$User.bIsFrontEndPreview, True Float=10, False Float=20))
  ]
  particleUpdateScript [
   &ParticleState()
   &SkeletalMeshLocation(
    Bone / Socket Index=$Particles.SpawnGroup
    Bone Sampling Mode=::NewEnumerator4
    MaskSpawnGroup=false
    Orientation Sampling=::NewEnumerator1
    Sampled Position Offset=$Particles.ShapeLocation.ShapeVector
    Skeletal Mesh=$System.Skeletal Mesh
   )
   &SpriteRotationRate(Rotation Rate=&Multiply_Float(
     A=&RandomRangeFloat(Minimum=-10, Maximum=10)
     B=&FloatFromCurve(FloatCurve=DataInterface() {propertyValues {Curve {keys [<at 0 1>,<at 1 0>]}}}, Scale Curve=10)
    ))
   &ScaleSpriteSize(
    Uniform Curve Scale=&Multiply_Float(
     A=&MakeCustomFloatFromBool(
      Bool=$User.bIsFrontEndPreview
      True Float=&RandomRangeFloat(Minimum=0.45, Maximum=0.75)
      False Float=&RandomRangeFloat(Minimum=0.5, Maximum=1)
     )
     B=&MakeCustomFloatFromBool(
      Bool=$User.MusicReactive
      True Float=1
      False Float=&Multiply_Float(
       A=&FN_FrontEndFloatModifier(Gameplay=1.2, Front End=1.5)
       B=&ClampFloat(Float=$NPC.FortniteNPC.EventAudio_Amplitude, Min=0.3, Max=0.8)
      )
     )
    )
    Uniform Curve Sprite Scale=DataInterface() {propertyValues {Curve {keys [{tangentMode=RCTM_User, time=0, value=0.3}, <at 0.1 1.2>, <at 0.2 1>, {tangentMode=RCTM_User, time=1, value=1}]}}}
   )
   &ScaleSpriteSize(
    Uniform Curve Index=&FN_LOD_Distance_Remap(Start Distance=500, End Distance=1500, Minimum Output=0, Maximum Output=1, Clamp Output=true)
    Uniform Curve Sprite Scale=DataInterface() {propertyValues {Curve {keys [{tangentMode=RCTM_User, time=0.4, value=1}, {tangentMode=RCTM_User, time=0.6, value=2}]}}}
   ) {module=ScaleSpriteSize001}
   /SetVariables(Particles.Size002:Vector2f=&Multiply_Vector2D(A=$Particles.SpriteSize, B=&Vector2DFromFloat(Value=1.2))) {isSetParameters=true, module=SetVariables_#10}
   &DynamicMaterialParameters(
    Index 0 Param 1=&FloatFromCurve(FloatCurve=DataInterface() {propertyValues {Curve {keys [<at 0 -1.5>,<at 0.5 -1.5>,<at 1 1>]}}})
    Index 0 Param 2=0.2
    Module.Index0Param0Write=true
    Module.Index0Param1Write=true
    Module.Index0Param2Write=true
    Module.Index0Param3Write=true
   )
   &ScaleColor(
    Scale Alpha=&MakeCustomFloatFromBool(
     Bool=$User.MusicReactive
     True Float=1
     False Float=&MakeCustomFloatFromBool(
      Bool=&SetBoolByFloatComparison(A=$System.Music_Reactive_FX_Preview, B=0)
      True Float=&Multiply_Float(
       A=&Lerp_Float(
        A=&FN_FrontEndFloatModifier(
         Gameplay=&FN_PickFloatDependOnPlatform(TargetPlatformActive=1, TargetPlatformNoActive=0.8)
         Front End=0.6
        )
        B=&FN_FrontEndFloatModifier(
         Gameplay=&FN_PickFloatDependOnPlatform(TargetPlatformActive=3, TargetPlatformNoActive=1.5)
         Front End=1
        )
        Alpha=$System.Music_Reactive_FX_Preview
       )
       B=&FNTimeOfDayScalar(Night Scale=0.3, Morning Scale=1, Evening Scale=1)
      )
      False Float=0
     )
    )
    Scale RGB=&MakeVectorFromLinearColorRGB(Color=&ColorFromCurve(ColorCurve=DataInterface() {propertyValues {
        AlphaCurve {keys []}
        BlueCurve {keys [<at 0.0474 5 RCIM_Cubic>,<at 0.198 1>,<at 0.446 1>,<at 0.8623 0.875 RCIM_Cubic>]}
        GreenCurve {keys [<at 0.0474 2.2571 RCIM_Cubic>,<at 0.198 0.8735>,<at 0.446 0.8369>,<at 0.8623 0.3426 RCIM_Cubic>]}
        RedCurve {keys [<at 0.0474 0.2566 RCIM_Cubic>,<at 0.198 0.6>,<at 0.446 0.4844>,<at 0.8623 0.1257 RCIM_Cubic>]}
       }}))
    ScaleA=true
    ScaleRGB=true
   )
   /SetVariables(Particles.Color_glow:LinearColor=&MakeLinearColorFromVectorAndFloat(
     "Vector (RGB)"=&MakeVectorFromLinearColorRGB(Color=<rgba 0.0858 0.5397 1 1>)
     "Float (Alpha)"=&MakeCustomFloatFromBool(
      Bool=&SetBoolByFloatComparison(A=$System.Music_Reactive_FX_Preview, B=0)
      True Float=&Multiply_Float(
       A=&FN_FrontEndFloatModifier(Gameplay=0.3, Front End=0.15)
       B=&FNTimeOfDayScalar(Night Scale=0.8, Morning Scale=1, Evening Scale=1)
      )
      False Float=0
     )
    )) {isSetParameters=true, module=SetVariables_#11}
   /SetVariables(Particles.Color_glow2:LinearColor=&MakeLinearColorFromVectorAndFloat(
     "Vector (RGB)"=&Multiply_Vector(
      A=&MakeVectorFromLinearColorRGB(Color=<rgba 1 0.0197 0.7901 1>)
      B=&VectorFromFloat(Value=&FN_PickFloatDependOnPlatform(TargetPlatformActive=1, TargetPlatformNoActive=0.5))
     )
     "Float (Alpha)"=&MakeCustomFloatFromBool(
      Bool=&SetBoolByFloatComparison(A=$System.Music_Reactive_FX_Preview, B=0)
      True Float=&Multiply_Float(
       A=&Lerp_Float(
        A=&FN_FrontEndFloatModifier(Gameplay=0.3, Front End=0.1)
        B=&FN_FrontEndFloatModifier(Gameplay=0.3, Front End=0.2)
        Alpha=$System.Music_Reactive_FX_Preview
       )
       B=&FNTimeOfDayScalar(Night Scale=0.8, Morning Scale=1, Evening Scale=1)
      )
      False Float=0
     )
    )) {isSetParameters=true, module=SetVariables_#12}
   &SolveForcesAndVelocity(Clamp Velocity=false, Limit Acceleration=false)
   /SetVariables(Particles.VisibilityTag=&MakeCustomIntFromBool(Bool=$User.bisHeadHidden, True INT=1, False INT=0)) {isSetParameters=true, module=SetVariables_#7}
  ]
  rendererClasses [&NiagaraSpriteRendererProperties]
  renderers [
   {rendererClass=&NiagaraSpriteRendererProperties}
   {rendererClass=&NiagaraSpriteRendererProperties, rendererIndex=1}
  ]
 }
 music_OffPhysics {
  emitterUpdateScript [
   &EmitterState(Inactive Response=::NewEnumerator1, Life Cycle Mode=::NewEnumerator1, Loop Duration=1, UseLoopDelay=false)
   &SpawnRate(
    Spawn Probability=&MakeCustomFloatFromBool(
     Bool=$User.MusicReactive
     True Float=1
     False Float=&MakeCustomFloatFromBool(Bool=&SetBoolByFloatComparison(A=$System.Music_Reactive_FX_Preview, B=0), True Float=1, False Float=0)
    )
    SpawnRate=&Multiply_Float(
     A=&MakeCustomFloatFromBool(Bool=$User.Change_OffPhysics, True Float=8, False Float=0)
     B=&MakeCustomFloatFromBool(Bool=$User.MusicReactive, True Float=1, False Float=$System.DistanceSpawn)
    )
    Use Spawn Probability=true
   )
  ]
  particleSpawnScript [
   &PartitionParticles(Number Of Partitions=3)
   &InitializeParticle(
    Lifetime Max=1
    Lifetime Mode=::NewEnumerator1
    Sprite Rotation Angle=&RandomRangeFloat(Minimum=-0.15, Maximum=0.15)
    Sprite Rotation Mode=::NewEnumerator1
    Sprite Size Max=<11 14>
    Sprite Size Min=<10 12>
    Sprite Size Mode=::NewEnumerator2
    UsePositionOffset=false
    Write Material Random=true
   )
   &ShapeLocation(
    Handle Radius=50
    Large Radius=30
    MaskSpawnGroup=false
    Non Uniform Scale=<0.5 0.5 1>
    Offset Mode=::NewEnumerator0
    Override Seed=false
    Rotation Axis=<0 1 0>
    Rotation Mode=::NewEnumerator1
    Shape Primitive=::NewEnumerator3
    Surface Distribution=1
    UseOwnerScale=true
   )
   &CameraOffset(Camera Offset Amount=&MakeCustomFloatFromBool(Bool=$User.bIsFrontEndPreview, True Float=10, False Float=20))
  ]
  particleUpdateScript [
   &ParticleState()
   &SkeletalMeshLocation(
    Bone / Socket Index=&SelectIntFromArray(Direct Array Index=$Particles.PartitionParticles.Partition, Int Selection Array=DataInterface() {propertyValues {IntData [0,1,2]}})
    Bone Sampling Mode=::NewEnumerator4
    MaskSpawnGroup=false
    Orientation Sampling=::NewEnumerator1
    Sampled Position Offset=&Add_Vector(A=&RandomRangeVector(Minimum=<10 10 25>, Maximum=<-10 -10 -25>), B=<0 -20 0>)
    Skeletal Mesh=$System.Skeletal Mesh Reader
   )
   &SpriteRotationRate(Rotation Rate=&Multiply_Float(
     A=&RandomRangeFloat(Minimum=-10, Maximum=10)
     B=&FloatFromCurve(FloatCurve=DataInterface() {propertyValues {Curve {keys [<at 0 1>,<at 1 0>]}}}, Scale Curve=10)
    ))
   &ScaleSpriteSize(
    Uniform Curve Scale=&MakeCustomFloatFromBool(
     Bool=$User.MusicReactive
     True Float=1
     False Float=&Multiply_Float(
      A=&MakeCustomFloatFromBool(
       Bool=$User.bIsFrontEndPreview
       True Float=&RandomRangeFloat(Minimum=0.45, Maximum=0.75)
       False Float=&RandomRangeFloat(Minimum=0.7, Maximum=1)
      )
      B=&Multiply_Float(
       A=&FN_FrontEndFloatModifier(Gameplay=1.2, Front End=1.5)
       B=&ClampFloat(Float=$NPC.FortniteNPC.EventAudio_Amplitude, Min=0.3, Max=0.8)
      )
     )
    )
    Uniform Curve Sprite Scale=DataInterface() {propertyValues {Curve {keys [{tangentMode=RCTM_User, time=0, value=0.3}, <at 0.1 1.2>, <at 0.2 1>, {tangentMode=RCTM_User, time=1, value=1}]}}}
   )
   &ScaleSpriteSize(
    Uniform Curve Index=&FN_LOD_Distance_Remap(Start Distance=500, End Distance=1500, Minimum Output=0, Maximum Output=1, Clamp Output=true)
    Uniform Curve Sprite Scale=DataInterface() {propertyValues {Curve {keys [{tangentMode=RCTM_User, time=0.4, value=1}, {tangentMode=RCTM_User, time=0.6, value=2}]}}}
   ) {module=ScaleSpriteSize001}
   /SetVariables(Particles.Size002:Vector2f=&Multiply_Vector2D(A=$Particles.SpriteSize, B=&Vector2DFromFloat(Value=1.2))) {isSetParameters=true, module=SetVariables_#10}
   &DynamicMaterialParameters(
    Index 0 Param 1=&FloatFromCurve(FloatCurve=DataInterface() {propertyValues {Curve {keys [<at 0 -1.5>,<at 0.5 -1.5>,<at 1 1>]}}})
    Index 0 Param 2=0.2
    Module.Index0Param0Write=true
    Module.Index0Param1Write=true
    Module.Index0Param2Write=true
    Module.Index0Param3Write=true
   )
   &ScaleColor(
    Scale Alpha=&MakeCustomFloatFromBool(
     Bool=$User.MusicReactive
     True Float=1
     False Float=&MakeCustomFloatFromBool(
      Bool=&SetBoolByFloatComparison(A=$System.Music_Reactive_FX_Preview, B=0)
      True Float=&Multiply_Float(
       A=&Lerp_Float(
        A=&FN_FrontEndFloatModifier(
         Gameplay=&FN_PickFloatDependOnPlatform(TargetPlatformActive=1, TargetPlatformNoActive=0.8)
         Front End=0.6
        )
        B=&FN_FrontEndFloatModifier(
         Gameplay=&FN_PickFloatDependOnPlatform(TargetPlatformActive=3, TargetPlatformNoActive=1.5)
         Front End=1
        )
        Alpha=$System.Music_Reactive_FX_Preview
       )
       B=&FNTimeOfDayScalar(Night Scale=0.3, Morning Scale=1, Evening Scale=1)
      )
      False Float=0
     )
    )
    Scale RGB=&MakeVectorFromLinearColorRGB(Color=&ColorFromCurve(ColorCurve=DataInterface() {propertyValues {
        AlphaCurve {keys []}
        BlueCurve {keys [<at 0.0474 5 RCIM_Cubic>,<at 0.198 1>,<at 0.446 1>,<at 0.8623 0.875 RCIM_Cubic>]}
        GreenCurve {keys [<at 0.0474 2.2571 RCIM_Cubic>,<at 0.198 0.8735>,<at 0.446 0.8369>,<at 0.8623 0.3426 RCIM_Cubic>]}
        RedCurve {keys [<at 0.0474 0.2566 RCIM_Cubic>,<at 0.198 0.6>,<at 0.446 0.4844>,<at 0.8623 0.1257 RCIM_Cubic>]}
       }}))
    ScaleA=true
    ScaleRGB=true
   )
   /SetVariables(Particles.Color_glow:LinearColor=&MakeLinearColorFromVectorAndFloat(
     "Vector (RGB)"=&MakeVectorFromLinearColorRGB(Color=<rgba 0.0858 0.5397 1 1>)
     "Float (Alpha)"=&MakeCustomFloatFromBool(
      Bool=$User.MusicReactive
      True Float=0.3
      False Float=&MakeCustomFloatFromBool(
       Bool=&SetBoolByFloatComparison(A=$System.Music_Reactive_FX_Preview, B=0)
       True Float=&Multiply_Float(
        A=&FN_FrontEndFloatModifier(Gameplay=0.3, Front End=0.15)
        B=&FNTimeOfDayScalar(Night Scale=0.8, Morning Scale=1, Evening Scale=1)
       )
       False Float=0
      )
     )
    )) {isSetParameters=true, module=SetVariables_#11}
   /SetVariables(Particles.Color_glow2:LinearColor=&MakeLinearColorFromVectorAndFloat(
     "Vector (RGB)"=&Multiply_Vector(
      A=&MakeVectorFromLinearColorRGB(Color=<rgba 1 0.0197 0.7901 1>)
      B=&VectorFromFloat(Value=&FN_PickFloatDependOnPlatform(TargetPlatformActive=1, TargetPlatformNoActive=0.5))
     )
     "Float (Alpha)"=&MakeCustomFloatFromBool(
      Bool=$User.MusicReactive
      True Float=0.3
      False Float=&MakeCustomFloatFromBool(
       Bool=&SetBoolByFloatComparison(A=$System.Music_Reactive_FX_Preview, B=0)
       True Float=&Multiply_Float(
        A=&Lerp_Float(
         A=&FN_FrontEndFloatModifier(Gameplay=0.3, Front End=0.1)
         B=&FN_FrontEndFloatModifier(Gameplay=0.3, Front End=0.2)
         Alpha=$System.Music_Reactive_FX_Preview
        )
        B=&FNTimeOfDayScalar(Night Scale=0.8, Morning Scale=1, Evening Scale=1)
       )
       False Float=0
      )
     )
    )) {isSetParameters=true, module=SetVariables_#12}
   &SolveForcesAndVelocity(Clamp Velocity=false, Limit Acceleration=false)
   /SetVariables(Particles.VisibilityTag=&MakeCustomIntFromBool(Bool=$User.bisHeadHidden, True INT=1, False INT=0)) {isSetParameters=true, module=SetVariables_#7}
  ]
  rendererClasses [&NiagaraSpriteRendererProperties]
  renderers [
   {rendererClass=&NiagaraSpriteRendererProperties}
   {rendererClass=&NiagaraSpriteRendererProperties, rendererIndex=1}
  ]
 }
 particle {
  emitterUpdateScript [
   &EmitterState(Inactive Response=::NewEnumerator1, Life Cycle Mode=::NewEnumerator1, Loop Duration=1, UseLoopDelay=false)
   &SpawnRate(
    Spawn Probability=&MakeCustomFloatFromBool(
     Bool=$User.MusicReactive
     True Float=1
     False Float=&MakeCustomFloatFromBool(Bool=&SetBoolByFloatComparison(A=$System.Music_Reactive_FX_Preview, B=0), True Float=1, False Float=0)
    )
    SpawnRate=&Multiply_Float(
     A=&MakeCustomFloatFromBool(Bool=$User.Change_OffPhysics, True Float=0, False Float=30)
     B=&MakeCustomFloatFromBool(Bool=$User.MusicReactive, True Float=1, False Float=$System.DistanceSpawn)
    )
    Use Spawn Probability=true
   )
  ]
  particleSpawnScript [
   &FN_SetSpawnGroup(Number of Spawn Groups=6)
   &InitializeParticle(
    Color=<rgba 0.4427 0.8727 1 1>
    Lifetime Max=0.8
    Lifetime Min=0.5
    Lifetime Mode=::NewEnumerator1
    Sprite Rotation Angle=&RandomRangeFloat(Minimum=-0.15, Maximum=0.15)
    Sprite Rotation Mode=::NewEnumerator1
    Sprite Size Mode=::NewEnumerator1
    Uniform Sprite Size Max=11
    Uniform Sprite Size Min=8
    UsePositionOffset=false
    Write Material Random=true
   )
   &ShapeLocation(
    Handle Radius=5
    Large Radius=&SelectFloatFromArray(Direct Array Index=$Particles.SpawnGroup, Float Selection Array=DataInterface() {propertyValues {FloatData [10,20,25,10,20,25]}})
    MaskSpawnGroup=false
    Non Uniform Scale=<0.5 0.5 3>
    Override Seed=false
    Rotation Axis=<0 1 0>
    Rotation Mode=::NewEnumerator1
    Shape Primitive=::NewEnumerator3
    Surface Distribution=1
    UseOwnerScale=true
   )
   /SetVariables(Particles.VisibilityTag=&RandomRangeInt(Minimum=0, Maximum=2)) {isSetParameters=true, module=SetVariables_#13}
  ]
  particleUpdateScript [
   &ParticleState()
   &SkeletalMeshLocation(
    Bone / Socket Index=$Particles.SpawnGroup
    Bone Sampling Mode=::NewEnumerator4
    MaskSpawnGroup=false
    Orientation Sampling=::NewEnumerator1
    Sampled Position Offset=$Particles.ShapeLocation.ShapeVector
    Skeletal Mesh=$System.Skeletal Mesh
   )
   &SpriteRotationRate(Rotation Rate=&Multiply_Float(
     A=&RandomRangeFloat(Minimum=-10, Maximum=10)
     B=&FloatFromCurve(FloatCurve=DataInterface() {propertyValues {Curve {keys [<at 0 1>,<at 1 0>]}}}, Scale Curve=10)
    ))
   &ScaleSpriteSize(
    Uniform Curve Scale=&RandomRangeFloat(Minimum=0.6, Maximum=1)
    Uniform Curve Sprite Scale=DataInterface() {propertyValues {Curve {keys [{tangentMode=RCTM_User, time=0, value=0.4}, <at 0.1 1.2>, <at 0.2 1>, {tangentMode=RCTM_User, time=1, value=1}]}}}
   )
   /SetVariables(Particles.Size002:Vector2f=&Multiply_Vector2D(
     A=$Particles.SpriteSize
     B=&Vector2DFromFloat(Value=&RandomRangeFloat(Minimum=0.5, Maximum=0.8))
    )) {isSetParameters=true, module=SetVariables_#14}
   &ScaleColor(
    Scale Alpha=&FloatFromCurve(
     FloatCurve=DataInterface() {propertyValues {Curve {keys [<at 0.5 1>,<at 1 0>]}}}
     Scale Curve=&MakeCustomFloatFromBool(
      Bool=$User.MusicReactive
      True Float=2
      False Float=&MakeCustomFloatFromBool(
       Bool=&SetBoolByFloatComparison(A=$System.Music_Reactive_FX_Preview, B=0)
       True Float=&Multiply_Float(
        A=&FN_FrontEndFloatModifier(Gameplay=2, Front End=0.4)
        B=&FNTimeOfDayScalar(Night Scale=0.8, Morning Scale=1, Evening Scale=1)
       )
       False Float=0
      )
     )
    )
    ScaleA=true
    ScaleRGB=true
   )
   /SetVariables(Particles.Color_glow:LinearColor=&MakeLinearColorFromVectorAndFloat(
     "Vector (RGB)"=&MakeVectorFromLinearColorRGB(Color=<rgba 1 1 1 1>)
     "Float (Alpha)"=&FloatFromCurve(
      FloatCurve=DataInterface() {propertyValues {Curve {keys [<at 0.5 1>,<at 1 0>]}}}
      Scale Curve=&MakeCustomFloatFromBool(
       Bool=$User.MusicReactive
       True Float=2
       False Float=&MakeCustomFloatFromBool(
        Bool=&SetBoolByFloatComparison(A=$System.Music_Reactive_FX_Preview, B=0)
        True Float=&Multiply_Float(
         A=&FN_FrontEndFloatModifier(Gameplay=2, Front End=0.3)
         B=&FNTimeOfDayScalar(Night Scale=0.1, Morning Scale=1, Evening Scale=1)
        )
        False Float=0
       )
      )
     )
    )) {isSetParameters=true, module=SetVariables_#15}
   /SetVariables(Particles.Color003:LinearColor=&MakeLinearColorFromVectorAndFloat(
     "Vector (RGB)"=&MakeVectorFromLinearColorRGB(Color=<rgba 0.0781 0.4051 1 1>)
     "Float (Alpha)"=&FloatFromCurve(FloatCurve=DataInterface() {propertyValues {Curve {keys [<at 0 0 RCIM_Cubic>,<at 0.4 1 RCIM_Cubic>,<at 1 0 RCIM_Cubic>]}}})
    )) {isSetParameters=true}
   /SetVariables(Particles.Size003:Vector2f=&Multiply_Vector2D(A=<10 10>, B=&Vector2DFromFloat(Value=1))) {isSetParameters=true}
   &SolveForcesAndVelocity(Clamp Velocity=false, Limit Acceleration=false)
   &SubUVAnimation(
    End Frame Range Override=3
    "Sprite Renderer "=DataInterface() {propertyValues {SpriteRenderer=&self:particle_2.NiagaraSpriteRendererProperties_0}}
    UseEndFrame=true
    UseStartFrame=true
   )
   /SetVariables(Particles.VisibilityTag=&MakeCustomIntFromBool(Bool=$User.bisHeadHidden, True INT=3, False INT=&RandomRangeInt(Minimum=0, Maximum=2))) {isSetParameters=true, module=SetVariables_#7}
  ]
  rendererClasses [&NiagaraSpriteRendererProperties]
  renderers [
   {rendererClass=&NiagaraSpriteRendererProperties}
   {rendererClass=&NiagaraSpriteRendererProperties, rendererIndex=1}
  ]
 }
 particle001 {
  emitterUpdateScript [
   &EmitterState(Inactive Response=::NewEnumerator1, Life Cycle Mode=::NewEnumerator1, Loop Duration=1, UseLoopDelay=false)
   &SpawnRate(
    Spawn Probability=&MakeCustomFloatFromBool(Bool=&SetBoolByFloatComparison(A=$System.Music_Reactive_FX_Preview, B=0), True Float=1, False Float=0)
    SpawnRate=&Multiply_Float(
     A=&MakeCustomFloatFromBool(Bool=$User.Change_OffPhysics, True Float=20, False Float=0)
     B=$System.DistanceSpawn
    )
    Use Spawn Probability=true
   )
  ]
  particleSpawnScript [
   &PartitionParticles(Number Of Partitions=3)
   &InitializeParticle(
    Color=<rgba 0.4427 0.8727 1 1>
    Lifetime Max=0.8
    Lifetime Min=0.5
    Lifetime Mode=::NewEnumerator1
    Sprite Rotation Angle=&RandomRangeFloat(Minimum=-0.15, Maximum=0.15)
    Sprite Rotation Mode=::NewEnumerator1
    Sprite Size Mode=::NewEnumerator1
    Uniform Sprite Size Max=11
    Uniform Sprite Size Min=8
    UsePositionOffset=false
    Write Material Random=true
   )
   /SetVariables(Particles.VisibilityTag=&RandomRangeInt(Minimum=0, Maximum=2)) {isSetParameters=true, module=SetVariables_#13}
  ]
  particleUpdateScript [
   &ParticleState()
   &SkeletalMeshLocation(
    Bone / Socket Index=&SelectIntFromArray(Direct Array Index=$Particles.PartitionParticles.Partition, Int Selection Array=DataInterface() {propertyValues {IntData [0,1,2]}})
    Bone Sampling Mode=::NewEnumerator4
    MaskSpawnGroup=false
    Orientation Sampling=::NewEnumerator1
    Sampled Position Offset=&Add_Vector(A=&RandomRangeVector(Minimum=<10 10 25>, Maximum=<-10 -10 -25>), B=<0 -20 0>)
    Skeletal Mesh=$System.Skeletal Mesh Reader
   )
   &SpriteRotationRate(Rotation Rate=&Multiply_Float(
     A=&RandomRangeFloat(Minimum=-10, Maximum=10)
     B=&FloatFromCurve(FloatCurve=DataInterface() {propertyValues {Curve {keys [<at 0 1>,<at 1 0>]}}}, Scale Curve=10)
    ))
   &ScaleSpriteSize(
    Uniform Curve Scale=&RandomRangeFloat(Minimum=0.6, Maximum=1)
    Uniform Curve Sprite Scale=DataInterface() {propertyValues {Curve {keys [{tangentMode=RCTM_User, time=0, value=0.4}, <at 0.1 1.2>, <at 0.2 1>, {tangentMode=RCTM_User, time=1, value=1}]}}}
   )
   /SetVariables(Particles.Size002:Vector2f=&Multiply_Vector2D(
     A=$Particles.SpriteSize
     B=&Vector2DFromFloat(Value=&RandomRangeFloat(Minimum=0.5, Maximum=0.8))
    )) {isSetParameters=true, module=SetVariables_#14}
   &ScaleColor(
    Scale Alpha=&FloatFromCurve(
     FloatCurve=DataInterface() {propertyValues {Curve {keys [<at 0.5 1>,<at 1 0>]}}}
     Scale Curve=&MakeCustomFloatFromBool(
      Bool=$User.MusicReactive
      True Float=1.5
      False Float=&MakeCustomFloatFromBool(
       Bool=&SetBoolByFloatComparison(A=$System.Music_Reactive_FX_Preview, B=0)
       True Float=&Multiply_Float(
        A=&FN_FrontEndFloatModifier(Gameplay=2, Front End=0.4)
        B=&FNTimeOfDayScalar(Night Scale=0.8, Morning Scale=1, Evening Scale=1)
       )
       False Float=0
      )
     )
    )
    ScaleA=true
    ScaleRGB=true
   )
   /SetVariables(Particles.Color_glow:LinearColor=&MakeLinearColorFromVectorAndFloat(
     "Vector (RGB)"=&MakeVectorFromLinearColorRGB(Color=<rgba 1 1 1 1>)
     "Float (Alpha)"=&FloatFromCurve(
      FloatCurve=DataInterface() {propertyValues {Curve {keys [<at 0.5 1>,<at 1 0>]}}}
      Scale Curve=&MakeCustomFloatFromBool(
       Bool=$User.MusicReactive
       True Float=2
       False Float=&MakeCustomFloatFromBool(
        Bool=&SetBoolByFloatComparison(A=$System.Music_Reactive_FX_Preview, B=0)
        True Float=&Multiply_Float(
         A=&FN_FrontEndFloatModifier(Gameplay=2, Front End=0.3)
         B=&FNTimeOfDayScalar(Night Scale=0.1, Morning Scale=1, Evening Scale=1)
        )
        False Float=0
       )
      )
     )
    )) {isSetParameters=true, module=SetVariables_#15}
   /SetVariables(Particles.Color003:LinearColor=&MakeLinearColorFromVectorAndFloat(
     "Vector (RGB)"=&MakeVectorFromLinearColorRGB(Color=<rgba 0.0781 0.4051 1 1>)
     "Float (Alpha)"=&FloatFromCurve(FloatCurve=DataInterface() {propertyValues {Curve {keys [<at 0 0 RCIM_Cubic>,<at 0.4 1 RCIM_Cubic>,<at 1 0 RCIM_Cubic>]}}})
    )) {isSetParameters=true}
   /SetVariables(Particles.Size003:Vector2f=&Multiply_Vector2D(A=<10 10>, B=&Vector2DFromFloat(Value=1))) {isSetParameters=true}
   &SolveForcesAndVelocity(Clamp Velocity=false, Limit Acceleration=false)
   /SetVariables(Particles.VisibilityTag=&MakeCustomIntFromBool(Bool=$User.bisHeadHidden, True INT=3, False INT=&RandomRangeInt(Minimum=0, Maximum=2))) {isSetParameters=true, module=SetVariables_#7}
  ]
  rendererClasses [&NiagaraSpriteRendererProperties]
  renderers [
   {rendererClass=&NiagaraSpriteRendererProperties}
   {rendererClass=&NiagaraSpriteRendererProperties, rendererIndex=1}
  ]
 }
}
summary {
 systemName=NS_Character_FlossYawn_Hair
 emitters [
  {emitterName=music, rendererClasses [&NiagaraSpriteRendererProperties]}
  {emitterName=particle, rendererClasses [&NiagaraSpriteRendererProperties]}
  {emitterName=hairGlow001, rendererClasses [&NiagaraSpriteRendererProperties]}
  {emitterName=Line, rendererClasses [&NiagaraSpriteRendererProperties]}
  {emitterName=Dark_hairGlow, rendererClasses [&NiagaraSpriteRendererProperties]}
  {emitterName=music_OffPhysics, rendererClasses [&NiagaraSpriteRendererProperties]}
  {emitterName=Line001, rendererClasses [&NiagaraSpriteRendererProperties]}
  {emitterName=particle001, rendererClasses [&NiagaraSpriteRendererProperties]}
  {emitterName=RingGlow001, rendererClasses [&NiagaraRibbonRendererProperties]}
  {bEnabled=false, emitterName=RingGlow002, rendererClasses [&NiagaraRibbonRendererProperties]}
  {emitterName=RingGlow004, rendererClasses [&NiagaraRibbonRendererProperties]}
 ]
 userVariables [
  User.bIsFrontEnd {type=NiagaraBool}
  User.bIsFrontEndPreview {type=NiagaraBool}
  User.bisHeadHidden {type=NiagaraBool}
  User.Change_OffPhysics {type=NiagaraBool}
  User.MusicReactive {type=NiagaraBool}
 ]
}
