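# Parent material for the FlossYawn female character cosmetic, adding audio-reactive snowflake glow to the shared 2023 anime shader.
# The anime Pre, Calculations and Post functions produce material attributes, and only emissive is overridden.
# Audio amplitude, with a debug blend, is power-shaped into reroutes driving snowflake mask and glow from FX channels.
# Emissive combines a face-accessory snowflake texture tinted by the shading gradient, an inverse-Fresnel channel glow and snowflake layers.
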Material /BRCosmetics/Characters/Player/Female/Medium/Bodies/F_MED_FlossYawn/Materials/M_FlossYawn_Parent
graph {
 properties {
  bAutomaticallySetUsageInEditor=false
  bUseMaterialAttributes=true
  bUsedWithClothing=true
  bUsedWithMeshDeformer=true
  bUsedWithMorphTargets=true
  bUsedWithNiagaraMeshParticles=true
  bUsedWithSkeletalMesh=true
  bUsedWithStaticLighting=true
  pixelDepthOffsetMode=PDOM_Legacy
  shadingModel=MSM_Cloth
  usedShadingModels=Cloth
 }
 nodes [
  Constant() {r=1}
  Multiply(A=Time(), B=ScalarParameter() {defaultValue=0.2, parameterName=Snowflake - Band Speed})
  n1 = Multiply(
   A=LinearInterpolate(
    A=MaterialFunctionCall() {materialFunction=/Game/Packages/Fortress_SharedMaterials/Base_Material_Functions/BaseCharacterShaderFunctions/Audio/MF_AudioReactivity}.Amplitude
    Alpha=ScalarParameter() {group=04_AudioReactivity, parameterName=Debug Audio Reactivity}
   )
   B=ScalarParameter() {defaultValue=1, parameterName=Is Reactive}
  )
  n2 = Saturate(None=Power(Base=@n1) {constExponent=3})
  n3 = Constant()
  n4 = Constant() {r=1}
  NamedRerouteDeclaration(None=Clamp(None=If(A=@n2, "A > B"=@n3, "A == B"=@n4, "A < B"=@n4)) {minDefault=0.5}) {name=Audio On, nodeColor=<rgba 0 1 1 1>}
  n5 = MaterialFunctionCall!(CustomUV) {materialFunction=/Game/Global/Materials/Character_Parent_2023/Anime_Parent_2023/MaterialFunctions/MFC_AnimeShader_2023_Pre}
  n6 = Multiply(
   A=ChannelMaskParameter(None=@n5.FX) {group=04_AudioReactivity, parameterName=Snowflake Glow Channel}.R
   B=ScalarParameter() {defaultValue=2, parameterName=Snowflake Glow - Intensity}
  )
  n7 = VectorParameter() {parameterName=Snowflake Glow - Color, defaultValue=<rgba 0 0.2624 1 1>}
  NamedRerouteDeclaration(None=Multiply(A=Multiply(A=@n6, B=@n7.RGB), B=NamedRerouteDeclaration(None=@n2) {name=Audio Reactivity, nodeColor=<rgba 0 1 0.0391 1>})) {name=SnowFlake_Glow, nodeColor=<rgba 0 0.0391 1 1>}
  n8 = ChannelMaskParameter(None=@n5.FX) {group=04_AudioReactivity, maskChannel=Green, parameterName=Snowflake Channel, defaultValue=<rgba 0 1 0 0>}
  n9 = Multiply(A=@n8.R, B=ScalarParameter() {defaultValue=3, parameterName=Snowflake Mask - Intensity})
  n10 = Power(Base=@n1, Exp=ScalarParameter() {defaultValue=7, parameterName=AE Exp})
  NamedRerouteDeclaration(None=Multiply(A=Saturate(None=@n9), B=Saturate(None=@n10))) {name=Snowflake_Mask, nodeColor=<rgba 0.7578 0 1 1>}
  n11 = VectorParameter() {parameterName=Snowflake Color, defaultValue=<rgba 0 0.277 1 1>}
  n12 = Multiply(A=@n8.R, B=ScalarParameter() {defaultValue=100, parameterName=Snowflake - Rim Intensity})
  NamedRerouteDeclaration(None=Multiply(A=Multiply(A=@n11.RGB, B=@n12), B=ScalarParameter() {defaultValue=0.7, parameterName=Snowflake - Band Brightness})) {name=Snowflake_Color, nodeColor=<rgba 0 0.9297 1 1>}
  n13 = MaterialFunctionCall!(CustomLightVector, OverrideNormal_WS, STT=@n5.STT, Normal_Map=@n5.STT, DFL=@n5.STT, FX=@n5.STT) {materialFunction=/Game/Global/Materials/Character_Parent_2023/Anime_Parent_2023/MaterialFunctions/MFC_AnimeShader_2023_Calculations}
  n14 = GetMaterialAttributes(None=MaterialFunctionCall!(
    Shading_Gradient=@n13.Shading_Gradient
    Threshold_Gradient=@n13.Shading_Gradient
    Specular_Gradient=@n13.Shading_Gradient
    Sheen_Mask=@n13.Shading_Gradient
    Thickness=@n13.Shading_Gradient
    DFL_Mask=@n13.Shading_Gradient
    FX=@n13.Shading_Gradient
    Metallic_Gradient=@n13.Shading_Gradient
    RimLight_Gradient=@n13.Shading_Gradient
    NormalBuffer=@n13.Shading_Gradient
    --------
    CINE_LightSources=@n13.Shading_Gradient
    CINE_LightColor=@n13.Shading_Gradient
    ---------
    Color_Lit=@n5.Color_Lit
    Color_Shaded=@n5.Color_Lit
    SRM=@n5.Color_Lit
    Emissive=@n5.Color_Lit
    DFL_Color=@n5.Color_Lit
    Specular_Color=@n5.Color_Lit
    RimLight_Color=@n5.Color_Lit
   ) {materialFunction=/Game/Global/Materials/Character_Parent_2023/Anime_Parent_2023/MaterialFunctions/MFC_AnimeShader_2023_Post}.Result) {attributeGetTypes [#1]}
  n15 = TextureSample() {texture=/BRCosmetics/Characters/Player/Female/Medium/Bodies/F_MED_FlossYawn/Textures/T_FlossYawn_FaceAcc_Snowflakes}
  n16 = Multiply(A=@n15.RGB, B=ScalarParameter() {defaultValue=2, parameterName=Snowflake - CL Boost}) {constB=1.1}
  n17 = Multiply(A=@n16) {constB=0.85}
  n18 = Multiply(A=ScalarParameter() {parameterName=Is FaceAcc?}, B=NamedRerouteUsage())
  n19 = ChannelMaskParameter(None=@n5.FX) {group=04_AudioReactivity, maskChannel=Green, parameterName=ReactivChannel, defaultValue=<rgba 0 1 0 0>}
  n20 = Power(Base=@n19.R, Exp=ScalarParameter() {defaultValue=2, group=04_AudioReactivity, parameterName=MaskExponent, sliderMax=8}) {constExponent=4}
  n21 = OneMinus(None=Fresnel() {baseReflectFraction=0, exponent=1})
  n22 = ScalarParameter() {defaultValue=2, group=04_AudioReactivity, parameterName=FresnelExponent, sliderMax=8}
  n23 = Multiply(A=Multiply(A=Saturate(None=@n20), B=Power(Base=@n21, Exp=@n22) {constExponent=4}), B=NamedRerouteUsage())
  n24 = ScalarParameter() {defaultValue=5, group=04_AudioReactivity, parameterName=EmissiveIntensity, sliderMax=10}
  n25 = Add(
   A=LinearInterpolate(A=@n14.Emissive Color, B=LinearInterpolate(A=@n17, B=@n16, Alpha=@n13.Shading_Gradient), Alpha=Multiply(A=@n15.B, B=@n18))
   B=Multiply(
    A=Multiply(A=@n23, B=@n24) {constB=5}
    B=VectorParameter() {group=04_AudioReactivity, parameterName=Emissive_Color, defaultValue=<rgba 0 1 0.7 1>}.RGB
   )
  )
  n26 = NamedRerouteUsage()
  n27 = NamedRerouteUsage()
  SetMaterialAttributes(MaterialAttributes=@n14.MaterialAttributes, Emissive Color=Add(A=LinearInterpolate(A=@n25, B=@n26, Alpha=@n27), B=NamedRerouteUsage())) {attributeSetTypes [#1]}
  n28 = ScalarParameter() {defaultValue=0.6774, parameterName=Snowflake - Size}
  SmoothStep(Min=@n28, Max=Add(A=@n28, B=ScalarParameter() {defaultValue=1.0304, parameterName=Snowflake - Smoothness}), Value=@n8.R)
 ]
}
