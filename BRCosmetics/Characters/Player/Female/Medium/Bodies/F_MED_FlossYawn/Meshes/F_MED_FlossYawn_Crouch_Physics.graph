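# Physics asset for the crouch variant of the F_MED_FlossYawn female cosmetic body, with capsule bodies for its scarves, tie, skirt and ponytails.
# Front scarf, back scarf and tie form swing-limited chains with restricted twist, anchored to upper-spine bodies through free base joints.
# Skirt bodies hang from the pelvis, and each ponytail has one capsule linked to spine_05.
# Spine, pelvis, thigh and right-arm capsules supply body volumes; the arm has no constraints.
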PhysicsAsset /BRCosmetics/Characters/Player/Female/Medium/Bodies/F_MED_FlossYawn/Meshes/F_MED_FlossYawn_Crouch_Physics
bodies {
 base_scarf [{shapeType=Capsule, length=0.505, radius=0.505, shapeName=base_scarf_capsule, rotation=<rot 0 -90 0>}]
 base_scarfBack [{shapeType=Capsule, length=0.505, radius=0.505, shapeName=base_scarfBack_capsule, rotation=<rot 0 -90 0>}]
 base_scarf_02 [{shapeType=Capsule, length=0, radius=3.5, shapeName=base_scarf_02_capsule, center=<1 0 0>, rotation=<rot 90 0 0>}]
 base_skirtAim [
  {
   shapeType=Capsule
   length=25.5
   radius=12.5
   shapeName=base_skirtAim_capsule
   center=<-36.2098 1.6119 -0.102>
   rotation=<rot 4.8761 0.0004 0>
  }
  {shapeType=Capsule, length=29.0215, radius=11.9102, shapeName=base_skirtAim_capsule, center=<-10.5489 -9.8275 0>, rotation=<rot 0 -110 0>}
 ]
 base_tie_01 [{shapeType=Capsule, length=1, radius=1, shapeName=base_tie_01_capsule, rotation=<rot 0 -90 0>}]
 dyn_ponytail_05_l [{
   shapeType=Capsule
   length=30.2984
   radius=5.7116
   shapeName=dyn_ponytail_05_l_capsule
   center=<-18.3878 4.179 -2.6727>
   rotation=<rot 80.7039 158.6026 171.5718>
  }]
 dyn_ponytail_05_r [{
   shapeType=Capsule
   length=26.5212
   radius=5.9888
   shapeName=dyn_ponytail_05_r_capsule
   center=<13.8268 -0.0191 3.7505>
   rotation=<rot 75.0993 158.0436 165.1806>
  }]
 dyn_scarfBack_01 [{shapeType=Capsule, length=3.4354, radius=2.7919, shapeName=dyn_scarfBack_01_capsule, center=<2.1135 -0.1347 0>, rotation=<rot 90 0 0>}]
 dyn_scarfBack_02 [{shapeType=Capsule, length=4.1226, radius=2.2816, shapeName=dyn_scarfBack_02_capsule, center=<5.4505 0 2>, rotation=<rot 90 0 0>}]
 dyn_scarfBack_03 [{shapeType=Capsule, length=2.6398, radius=2.7817, shapeName=dyn_scarfBack_03_capsule, center=<4.1628 0 1.5>, rotation=<rot 85.276 0 0>}]
 dyn_scarfBack_04 [{shapeType=Capsule, length=4.3919, radius=4.2391, shapeName=dyn_scarfBack_04_capsule, center=<5.3802 0 2.5>, rotation=<rot 90 0 0>}]
 dyn_scarfBack_05 [{shapeType=Capsule, length=7.483, radius=2.6488, shapeName=dyn_scarfBack_05_capsule, center=<3.8725 0 1.2>, rotation=<rot 90 0 0>}]
 dyn_scarfBack_06 [{shapeType=Capsule, length=4.1759, radius=3.8068, shapeName=dyn_scarfBack_06_capsule, center=<3.6344 0 2.3958>, rotation=<rot 90 0 0>}]
 dyn_scarfBack_07 [{shapeType=Capsule, length=7.3115, radius=5.2879, shapeName=dyn_scarfBack_07_capsule, center=<6.8072 0 3.9447>, rotation=<rot 90 0 0>}]
 dyn_scarf_01 [{shapeType=Capsule, length=1, radius=2, shapeName=dyn_scarf_01_capsule, center=<1 0 0>, rotation=<rot 90 0 0>}]
 dyn_scarf_02 [{shapeType=Capsule, length=0, radius=3.5, shapeName=dyn_scarf_02_capsule, center=<1 0 0>, rotation=<rot 90 0 0>}]
 dyn_scarf_03 [{shapeType=Capsule, length=0.5, radius=3.5, shapeName=dyn_scarf_03_capsule, center=<1 0 0>, rotation=<rot 90 0 0>}]
 dyn_scarf_04 [{shapeType=Capsule, length=3.2663, radius=3.2104, shapeName=dyn_scarf_04_capsule, center=<2.0881 0.2 0>, rotation=<rot 90 0 0>}]
 dyn_scarf_05 [{shapeType=Capsule, length=3.6059, radius=3.5791, shapeName=dyn_scarf_05_capsule, center=<2.3612 0.6 1.3792>, rotation=<rot 90 0 0>}]
 dyn_scarf_06 [{shapeType=Capsule, length=7.1433, radius=5.8071, shapeName=dyn_scarf_06_capsule, center=<7.0205 0.1271 3.6923>, rotation=<rot 90 0 0>}]
 dyn_skirt_fr_c_01 [{
   shapeType=Capsule
   length=4.9891
   radius=5.0572
   shapeName=dyn_skirt_fr_c_01_capsule
   center=<0.3405 -0.1375 0.5954>
   rotation=<rot 90 -92.9403 0>
  }]
 dyn_skirt_fr_l_01 [{
   shapeType=Capsule
   length=7.7963
   radius=5.0851
   shapeName=dyn_skirt_fr_l_01_capsule
   center=<-2.0933 1.8396 -2.9388>
   rotation=<rot 0 -4.3808 -77.9151>
  }]
 dyn_skirt_fr_l_03 [{
   shapeType=Capsule
   length=20.7361
   radius=8.8637
   shapeName=dyn_skirt_fr_l_03_capsule
   center=<0.6776 2.8238 -6.9814>
   rotation=<rot -5.1762 12.0536 -65.8362>
  }]
 dyn_skirt_fr_r_01 [{
   shapeType=Capsule
   length=8.4785
   radius=4.712
   shapeName=dyn_skirt_fr_r_01_capsule
   center=<1.0504 -1.2147 0.2731>
   rotation=<rot 0 -171.3968 -82.8269>
  }]
 dyn_skirt_fr_r_03 [{
   shapeType=Capsule
   length=8.4628
   radius=4.3913
   shapeName=dyn_skirt_fr_r_03_capsule
   center=<0.5734 -2.9598 -0.879>
   rotation=<rot 0 175.8795 -77.2684>
  }]
 dyn_tie_01 [{shapeType=Capsule, length=2, radius=2, shapeName=dyn_tie_01_capsule, rotation=<rot 90 0 0>}]
 dyn_tie_02 [{shapeType=Capsule, length=4, radius=2, shapeName=dyn_tie_02_capsule, center=<0.1346 0 -0.2716>, rotation=<rot 78.7004 180 180>}]
 dyn_tie_03 [{shapeType=Capsule, length=4, radius=2, shapeName=dyn_tie_03_capsule, center=<-0.1867 -0.0001 0.0556>, rotation=<rot 90 0 0>}]
 dyn_tie_04 [{shapeType=Capsule, length=4, radius=2.5, shapeName=dyn_tie_04_capsule, center=<0 0 0.0099>, rotation=<rot 90 0 0>}]
 dyn_tie_05 [{shapeType=Capsule, length=5.5, radius=3.5, shapeName=dyn_tie_05_capsule, center=<4.1059 0 1.429>, rotation=<rot 90 -0.0365 0>}]
 lowerarm_r [{shapeType=Capsule, length=22.3325, radius=6.3369, shapeName=lowerarm_r_capsule, center=<-12.7186 -0.8715 1.0835>, rotation=<rot 90 0 0>}]
 pelvis [
  {shapeType=Capsule, length=11, radius=14.384, shapeName=pelvis_capsule, center=<-2.9586 -0.54 0>, rotation=<rot 90 8.3588 0>}
  {
   shapeType=Capsule
   length=8.0468
   radius=12.2082
   shapeName=pelvis_capsule
   center=<-7.2071 -1.5053 6.6197>
   rotation=<rot 65.8618 -0.0001 -9.0859>
  }
 ]
 spine_01 [{shapeType=Capsule, length=7, radius=13, shapeName=spine_01_capsule, center=<-5.5304 0 0>}]
 spine_02 [{shapeType=Capsule, length=11, radius=10.5, shapeName=spine_02_capsule}]
 spine_03 [{shapeType=Capsule, length=12.7356, radius=8.7644, shapeName=spine_03_capsule, center=<0 0.4364 0>}]
 spine_04 [{shapeType=Capsule, length=10, radius=10, shapeName=spine_04_capsule, center=<0 1.9119 0>, rotation=<rot 0 0 -0.0154>}]
 spine_05 [{shapeType=Capsule, length=10, radius=10, shapeName=spine_05_capsule, center=<0 1.2856 0>, rotation=<rot 0 0 0.0365>}]
 thigh_l [{shapeType=Capsule, length=29.5408, radius=7.4592, shapeName=thigh_l_capsule, center=<-32.5183 1.2807 1.3313>, rotation=<rot 90 0 0>}]
 thigh_r [{shapeType=Capsule, length=29.5408, radius=7.4592, shapeName=thigh_l_capsule, center=<32.5183 -1.2807 -1.3313>, rotation=<rot 90 0 180>}]
 upperarm_r [{
   shapeType=Capsule
   length=25.7336
   radius=5.7667
   shapeName=upperarm_r_capsule
   center=<-13.473 0.3297 1.0566>
   rotation=<rot 88.1739 -179.999 -179.999>
  }]
}
constraints [
 {bone1Name=dyn_tie_02, bone2Name=dyn_tie_01, swing1LimitDegrees=15, swing2LimitDegrees=30, twistLimitDegrees=45, twistMotion=Locked}
 {bone1Name=dyn_tie_03, bone2Name=dyn_tie_02, swing1LimitDegrees=15, swing2LimitDegrees=50, twistLimitDegrees=45, twistMotion=Locked}
 {bone1Name=dyn_tie_04, bone2Name=dyn_tie_03, swing1LimitDegrees=25, swing2LimitDegrees=50, twistLimitDegrees=45, twistMotion=Locked}
 {bone1Name=dyn_tie_05, bone2Name=dyn_tie_04, swing1LimitDegrees=25, swing2LimitDegrees=50, twistLimitDegrees=45, twistMotion=Locked}
 {
  bone1Name=spine_02
  bone2Name=spine_01
  swing1LimitDegrees=45
  swing1Motion=Free
  swing2LimitDegrees=45
  swing2Motion=Free
  twistLimitDegrees=45
  twistMotion=Free
 }
 {
  bone1Name=spine_03
  bone2Name=spine_02
  swing1LimitDegrees=45
  swing1Motion=Free
  swing2LimitDegrees=45
  swing2Motion=Free
  twistLimitDegrees=45
  twistMotion=Free
 }
 {
  bone1Name=spine_04
  bone2Name=spine_03
  swing1LimitDegrees=45
  swing1Motion=Free
  swing2LimitDegrees=45
  swing2Motion=Free
  twistLimitDegrees=45
  twistMotion=Free
 }
 {
  bone1Name=spine_05
  bone2Name=spine_04
  swing1LimitDegrees=45
  swing1Motion=Free
  swing2LimitDegrees=45
  swing2Motion=Free
  twistLimitDegrees=45
  twistMotion=Free
 }
 {bone1Name=dyn_tie_01, bone2Name=base_tie_01, swing1LimitDegrees=20, swing2LimitDegrees=20, twistLimitDegrees=45, twistMotion=Locked}
 {
  bone1Name=base_tie_01
  bone2Name=spine_04
  swing1LimitDegrees=45
  swing1Motion=Free
  swing2LimitDegrees=45
  swing2Motion=Free
  twistLimitDegrees=45
  twistMotion=Free
 }
 {
  bone1Name=base_scarf
  bone2Name=spine_05
  swing1LimitDegrees=45
  swing1Motion=Free
  swing2LimitDegrees=45
  swing2Motion=Free
  twistLimitDegrees=45
  twistMotion=Free
 }
 {bone1Name=dyn_scarf_01, bone2Name=base_scarf, swing1LimitDegrees=5, swing2LimitDegrees=10, twistLimitDegrees=0}
 {bone1Name=base_scarf_02, bone2Name=dyn_scarf_01, swing1LimitDegrees=10, swing2LimitDegrees=15, twistLimitDegrees=0}
 {bone1Name=dyn_scarf_02, bone2Name=base_scarf_02, swing1LimitDegrees=10, swing2LimitDegrees=25, twistLimitDegrees=0}
 {bone1Name=dyn_scarf_03, bone2Name=dyn_scarf_02, swing1LimitDegrees=25, swing2LimitDegrees=25, twistLimitDegrees=0}
 {bone1Name=dyn_scarf_04, bone2Name=dyn_scarf_03, swing1LimitDegrees=25, swing2LimitDegrees=35, twistLimitDegrees=0}
 {bone1Name=dyn_scarf_05, bone2Name=dyn_scarf_04, swing1LimitDegrees=25, swing2LimitDegrees=40, twistLimitDegrees=0}
 {bone1Name=dyn_scarf_06, bone2Name=dyn_scarf_05, swing1LimitDegrees=25, swing2LimitDegrees=40, twistLimitDegrees=0}
 {
  bone1Name=spine_01
  bone2Name=pelvis
  swing1LimitDegrees=45
  swing1Motion=Free
  swing2LimitDegrees=45
  swing2Motion=Free
  twistLimitDegrees=45
  twistMotion=Free
 }
 {
  bone1Name=thigh_l
  bone2Name=pelvis
  swing1LimitDegrees=45
  swing1Motion=Free
  swing2LimitDegrees=45
  swing2Motion=Free
  twistLimitDegrees=45
  twistMotion=Free
 }
 {
  bone1Name=thigh_r
  bone2Name=pelvis
  swing1LimitDegrees=45
  swing1Motion=Free
  swing2LimitDegrees=45
  swing2Motion=Free
  twistLimitDegrees=45
  twistMotion=Free
 }
 {
  bone1Name=base_skirtAim
  bone2Name=pelvis
  swing1LimitDegrees=45
  swing1Motion=Free
  swing2LimitDegrees=45
  swing2Motion=Free
  twistLimitDegrees=45
  twistMotion=Free
 }
 {bone1Name=dyn_scarfBack_01, bone2Name=base_scarfBack, swing1LimitDegrees=15, swing2LimitDegrees=5, twistLimitDegrees=0}
 {bone1Name=dyn_scarfBack_02, bone2Name=dyn_scarfBack_01, swing1LimitDegrees=15, swing2LimitDegrees=5, twistLimitDegrees=0}
 {bone1Name=dyn_scarfBack_03, bone2Name=dyn_scarfBack_02, swing1LimitDegrees=15, swing2LimitDegrees=5, twistLimitDegrees=0}
 {bone1Name=dyn_scarfBack_04, bone2Name=dyn_scarfBack_03, swing1LimitDegrees=15, swing2LimitDegrees=15, twistLimitDegrees=0}
 {bone1Name=dyn_scarfBack_05, bone2Name=dyn_scarfBack_04, swing1LimitDegrees=15, swing2LimitDegrees=35, twistLimitDegrees=0}
 {bone1Name=dyn_scarfBack_06, bone2Name=dyn_scarfBack_05, swing1LimitDegrees=15, swing2LimitDegrees=35, twistLimitDegrees=0}
 {bone1Name=dyn_scarfBack_07, bone2Name=dyn_scarfBack_06, swing1LimitDegrees=15, swing2LimitDegrees=35, twistLimitDegrees=0}
 {
  bone1Name=base_scarfBack
  bone2Name=spine_05
  swing1LimitDegrees=45
  swing1Motion=Free
  swing2LimitDegrees=45
  swing2Motion=Free
  twistLimitDegrees=45
  twistMotion=Free
 }
 {
  bone1Name=dyn_ponytail_05_l
  bone2Name=spine_05
  swing1LimitDegrees=45
  swing1Motion=Free
  swing2LimitDegrees=45
  swing2Motion=Free
  twistLimitDegrees=45
  twistMotion=Free
 }
 {
  bone1Name=dyn_ponytail_05_r
  bone2Name=spine_05
  swing1LimitDegrees=45
  swing1Motion=Free
  swing2LimitDegrees=45
  swing2Motion=Free
  twistLimitDegrees=45
  twistMotion=Free
 }
 {bone1Name=dyn_skirt_fr_l_03, bone2Name=dyn_skirt_fr_l_01, swing1LimitDegrees=0, swing2LimitDegrees=0, twistLimitDegrees=0}
 {
  bone1Name=dyn_skirt_fr_c_01
  bone2Name=pelvis
  swing1LimitDegrees=45
  swing1Motion=Free
  swing2LimitDegrees=45
  swing2Motion=Free
  twistLimitDegrees=45
  twistMotion=Free
 }
 {
  bone1Name=dyn_skirt_fr_l_01
  bone2Name=pelvis
  swing1LimitDegrees=45
  swing1Motion=Free
  swing2LimitDegrees=45
  swing2Motion=Free
  twistLimitDegrees=45
  twistMotion=Free
 }
 {
  bone1Name=dyn_skirt_fr_r_01
  bone2Name=pelvis
  swing1LimitDegrees=45
  swing1Motion=Free
  swing2LimitDegrees=45
  swing2Motion=Free
  twistLimitDegrees=45
  twistMotion=Free
 }
 {
  bone1Name=dyn_skirt_fr_r_03
  bone2Name=dyn_skirt_fr_r_01
  swing1LimitDegrees=45
  swing1Motion=Free
  swing2LimitDegrees=45
  swing2Motion=Free
  twistLimitDegrees=45
  twistMotion=Free
 }
]
