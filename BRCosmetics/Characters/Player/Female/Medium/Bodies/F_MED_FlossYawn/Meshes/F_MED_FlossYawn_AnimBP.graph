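# Post-process AnimBP for the F_MED_FlossYawn female cosmetic body, adding secondary motion and fixes to clothing and accessory bones.
# Copies the parent mesh pose into a Control Rig driven by crouch, melee, emote, skydive, DBNO and idle values, then processes it through cached-pose stages.
# AnimDynamics simulates skirt, straps and sleeves, with bone constraints tying straps to skirt.
# Additive bone offsets correct scarf, skirt and strap, blended in during emotes.
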AnimBlueprint /BRCosmetics/Characters/Player/Female/Medium/Bodies/F_MED_FlossYawn/Meshes/F_MED_FlossYawn_AnimBP
target_skeleton=&/F_MED_FlossYawn_Skeleton
graphs [
 AnimGraph {nodes [
   n1 = NOT Boolean(A=Get bIsCrouching())
   n2 = "Transform (Modify) Bone - Bone: dyn_scarfBack_03"!(ComponentPose, bAlphaBoolEnabled:bool=@n1, Translation="0,0,0", Rotation="-3,0,0", Scale="1,1,1") -> (Pose) {
    alphaInputType=Bool
    rotationMode=BMM_Additive
    rotationSpace=BCS_BoneSpace
    alphaBoolBlend {blendInTime=0.2, blendOutTime=0.5}
    boneToModify {boneName=dyn_scarfBack_03}
    rotation {pitch=-3}
   }
   n3 = "Transform (Modify) Bone - Bone: dyn_scarfBack_04"!(ComponentPose=@n2, bAlphaBoolEnabled:bool=@n1, Translation="0,0,0", Rotation="-3,0,0", Scale="1,1,1") -> (Pose) {
    alphaInputType=Bool
    rotationMode=BMM_Additive
    rotationSpace=BCS_BoneSpace
    alphaBoolBlend {blendInTime=0.2, blendOutTime=0.5}
    boneToModify {boneName=dyn_scarfBack_04}
    rotation {pitch=-3}
   }
   "Transform (Modify) Bone - Bone: dyn_scarfBack_05"!(ComponentPose=@n3, bAlphaBoolEnabled:bool=@n1, Translation="0,0,0", Rotation="8,0,0", Scale="1,1,1") -> (Pose) {
    alphaInputType=Bool
    rotationMode=BMM_Additive
    rotationSpace=BCS_BoneSpace
    translationMode=BMM_Additive
    translationSpace=BCS_BoneSpace
    alphaBoolBlend {blendInTime=0.2, blendOutTime=0.5}
    boneToModify {boneName=dyn_scarfBack_05}
    rotation {pitch=8}
   }
   "Transform (Modify) Bone - Bone: base_tie_01"!(ComponentPose, bAlphaBoolEnabled:bool=True, Translation="0,0,0.5", Rotation="3.5,0,0", Scale="1,1,1") -> (Pose) {
    alphaInputType=Bool
    rotationMode=BMM_Additive
    rotationSpace=BCS_BoneSpace
    translationMode=BMM_Additive
    translationSpace=BCS_BoneSpace
    alphaBoolBlend {blendInTime=0.2, blendOption=Linear, blendOutTime=0.2}
    alphaScaleBiasClamp {bInterpResult=true, interpSpeedDecreasing=5, interpSpeedIncreasing=7}
    boneToModify {boneName=base_tie_01}
    rotation {pitch=3.5}
    translation {z=0.5}
   }
   n4 = "Transform (Modify) Bone - Bone: dyn_scarf_03"(ComponentPose=Local To Component(LocalPose=Use cached pose 'fixes_cache'()), Alpha=1, Translation="0,0,0", Rotation="2.5,0,0", Scale="1,1,1") {rotationMode=BMM_Additive, rotationSpace=BCS_BoneSpace, boneToModify {boneName=dyn_scarf_03}, rotation {pitch=2.5}}
   n5 = "Transform (Modify) Bone - Bone: base_skirt_hanging_fr_l"(
    ComponentPose="Transform (Modify) Bone - Bone: dyn_scarf_05"(
     ComponentPose="Transform (Modify) Bone - Bone: dyn_scarf_04"(ComponentPose=@n4, Alpha=1, Translation="0,0,0", Rotation="5,0,0", Scale="1,1,1") {rotationMode=BMM_Additive, rotationSpace=BCS_BoneSpace, boneToModify {boneName=dyn_scarf_04}, rotation {pitch=5}}
     Alpha=1
     Translation="0,0,0"
     Rotation="4,0,0"
     Scale="1,1,1"
    ) {rotationMode=BMM_Additive, rotationSpace=BCS_BoneSpace, boneToModify {boneName=dyn_scarf_05}, rotation {pitch=4}}
    Alpha=1
    Translation="0,0,2.814"
    Rotation="0,0,0"
    Scale="1,1,1"
   ) {
    rotationMode=BMM_Additive
    rotationSpace=BCS_BoneSpace
    translationMode=BMM_Additive
    translationSpace=BCS_BoneSpace
    boneToModify {boneName=base_skirt_hanging_fr_l}
    translation {z=2.814}
   }
   n6 = "Transform (Modify) Bone - Bone: base_skirt_hanging_fr_r"(ComponentPose=@n5, Alpha=1, Translation="0,0,3.58", Rotation="0,0,0", Scale="1,1,1") {
    rotationMode=BMM_Additive
    rotationSpace=BCS_BoneSpace
    translationMode=BMM_Additive
    translationSpace=BCS_BoneSpace
    boneToModify {boneName=base_skirt_hanging_fr_r}
    translation {z=3.58}
   }
   n7 = "Transform (Modify) Bone - Bone: dyn_scarfBack_06"(ComponentPose=@n6, Alpha=1, Translation="0,0,0", Rotation="4,0,0", Scale="1,1,1") {rotationMode=BMM_Additive, rotationSpace=BCS_BoneSpace, boneToModify {boneName=dyn_scarfBack_06}, rotation {pitch=4}}
   n8 = "Transform (Modify) Bone - Bone: base_strap_fr_l"(ComponentPose=@n7, Alpha=1, Translation="-0.000,-0.642,-0.000", Rotation="1.064,1.759,-5.513", Scale="1,1,1") {
    rotationMode=BMM_Additive
    rotationSpace=BCS_BoneSpace
    translationMode=BMM_Additive
    translationSpace=BCS_BoneSpace
    boneToModify {boneName=base_strap_fr_l}
    rotation=<rot 1.064 1.759 -5.5131>
    translation {y=-0.642}
   }
   n9 = "Transform (Modify) Bone - Bone: dyn_scarfBack_05"!(ComponentPose=@n8, Alpha:float=1, Translation="0,0,0", Rotation="6,0,0", Scale="1,1,1") -> (Pose) {rotationMode=BMM_Additive, rotationSpace=BCS_BoneSpace, boneToModify {boneName=dyn_scarfBack_05}, rotation {pitch=6}}
   n10 = "Transform (Modify) Bone - Bone: dyn_scarfBack_06"(ComponentPose=@n9, Alpha=1, Translation="0,0,0", Rotation="4,0,0", Scale="1,1,1") {rotationMode=BMM_Additive, rotationSpace=BCS_BoneSpace, boneToModify {boneName=dyn_scarfBack_06}, rotation {pitch=4}}
   Output Pose(Result=Blend Poses by bool(
     BlendPose_0=Component To Local(ComponentPose=@n10)
     BlendPose_1=Use cached pose 'fixes_cache'()
     BlendTime_0=0.8
     BlendTime_1=0.8
     bActiveValue=Get bIsPlayingEmote()
    )) {layerGroup=None}
   n11 = Control Rig!(
    Source=Copy Pose From Mesh() {bUseAttachedParent=true}
    Alpha
    PlayingEmote:double=Get Alpha_PlayingEmote() -> (Alpha_PlayingEmote:double)
    CrouchingMoving2D:double=Get Alpha_CrouchingMoving2D() -> (Alpha_CrouchingMoving2D:double)
    Crouching:double=Get Alpha_Crouching() -> (Alpha_Crouching:double)
    Melee_Bool:bool=Get bIsPlayingMeleeAnim()
    Crouching_Bool:bool=Get bIsCrouching()
    Skydiving:double=Get Alpha_NotSkydiving() -> (Alpha_NotSkydiving:double)
    DBNO:double=Get Alpha_DBNO() -> (Alpha_DBNO:double)
    FEIdle:double=Get FEIdle() -> (FEIdle:double)
   ) -> (Pose) {
    alphaBoolBlend {blendOption=Linear}
    controlRigAssetReference {blueprintRigClass=&/F_MED_FlossYawn_CtrlRig.F_MED_FlossYawn_CtrlRig_C}
   }
   Save cached pose 'CtrlRig Cache'(Pose=@n11) {cacheName=CtrlRig Cache}
   n12 = Anim Dynamics - Bone: dyn_skirt_fr_c_01(ComponentPose=Local To Component(LocalPose=Use cached pose 'RigidBodyCache'()), Alpha=1, ExternalForce="0,0,0") {
    angularDampingOverride=1
    angularSpringConstant=20000
    bAngularSpring=true
    bOverrideAngularDamping=true
    bUseGravityOverride=true
    gravityScale=0.35
    simulationSpace=RootRelative
    alphaBoolBlend {blendOption=Linear}
    alphaScaleBiasClamp {bInterpResult=true}
    boundBone {boneName=dyn_skirt_fr_c_01}
    componentAppliedLinearAccClamp=<3000 3000 3000>
    componentLinearAccScale=<0.65 0.65 0.65>
    componentLinearVelScale=<0.65 0.65 0.65>
    gravityOverride {z=-980}
    physicsBodyDefinitions [{
      boundBone {boneName=dyn_skirt_fr_c_01}
      boxExtents=<20 30 30>
      constraintSetup {angularLimitsMax=<5 0 10>, angularLimitsMin=<-5 -25 -10>, angularTarget=<1 0 0>}
      localJointOffset=<4 0 0>
     }]
    simSpaceSettings {simSpaceAngularAlpha=0.08}
   }
   n13 = Anim Dynamics - Bone: dyn_skirt_fr_c_02(ComponentPose=@n12, Alpha=1, ExternalForce="0,0,0") {
    angularDampingOverride=1
    angularSpringConstant=20000
    bAngularSpring=true
    bOverrideAngularDamping=true
    bUseGravityOverride=true
    gravityScale=0.35
    simulationSpace=RootRelative
    alphaBoolBlend {blendOption=Linear}
    alphaScaleBiasClamp {bInterpResult=true}
    boundBone {boneName=dyn_skirt_fr_c_02}
    componentAppliedLinearAccClamp=<3000 3000 3000>
    componentLinearAccScale=<0.65 0.65 0.65>
    componentLinearVelScale=<0.65 0.65 0.65>
    gravityOverride {z=-980}
    physicsBodyDefinitions [{
      boundBone {boneName=dyn_skirt_fr_c_02}
      boxExtents=<20 30 30>
      constraintSetup {angularLimitsMax=<5 0 10>, angularLimitsMin=<-5 -25 -10>, angularTarget=<1 0 0>}
      localJointOffset=<4 0 0>
     }]
    simSpaceSettings {simSpaceAngularAlpha=0.08}
   }
   n14 = Anim Dynamics - Bone: dyn_skirt_fr_c_03(ComponentPose=@n13, Alpha=1, ExternalForce="0,0,0") {
    angularDampingOverride=1
    angularSpringConstant=20000
    bAngularSpring=true
    bOverrideAngularDamping=true
    bUseGravityOverride=true
    gravityScale=0.35
    simulationSpace=RootRelative
    alphaBoolBlend {blendOption=Linear}
    alphaScaleBiasClamp {bInterpResult=true}
    boundBone {boneName=dyn_skirt_fr_c_03}
    componentAppliedLinearAccClamp=<3000 3000 3000>
    componentLinearAccScale=<0.65 0.65 0.65>
    componentLinearVelScale=<0.65 0.65 0.65>
    gravityOverride {z=-980}
    physicsBodyDefinitions [{
      boundBone {boneName=dyn_skirt_fr_c_03}
      boxExtents=<20 30 30>
      constraintSetup {angularLimitsMax=<5 0 10>, angularLimitsMin=<-5 -25 -10>, angularTarget=<1 0 0>}
      localJointOffset=<4 0 0>
     }]
    simSpaceSettings {simSpaceAngularAlpha=0.08}
   }
   n15 = Anim Dynamics - Bone: dyn_skirt_fr_l_01(ComponentPose=@n14, Alpha=1, ExternalForce="0,0,0") {
    angularDampingOverride=1
    angularSpringConstant=20000
    bAngularSpring=true
    bOverrideAngularDamping=true
    bUseGravityOverride=true
    gravityScale=0.35
    simulationSpace=RootRelative
    alphaBoolBlend {blendOption=Linear}
    alphaScaleBiasClamp {bInterpResult=true}
    boundBone {boneName=dyn_skirt_fr_l_01}
    componentAppliedLinearAccClamp=<3000 3000 3000>
    componentLinearAccScale=<0.65 0.65 0.65>
    componentLinearVelScale=<0.65 0.65 0.65>
    gravityOverride {z=-980}
    physicsBodyDefinitions [{
      boundBone {boneName=dyn_skirt_fr_l_01}
      boxExtents=<20 30 30>
      constraintSetup {angularLimitsMax=<5 0 5>, angularLimitsMin=<-5 -15 -5>, angularTarget=<1 0 0>}
      localJointOffset=<2 0 0>
     }]
    simSpaceSettings {simSpaceAngularAlpha=0.08}
   }
   n16 = Anim Dynamics - Bone: dyn_skirt_fr_l_02(ComponentPose=@n15, Alpha=1, ExternalForce="0,0,0") {
    angularDampingOverride=1
    angularSpringConstant=20000
    bAngularSpring=true
    bOverrideAngularDamping=true
    bUseGravityOverride=true
    gravityScale=0.35
    simulationSpace=RootRelative
    alphaBoolBlend {blendOption=Linear}
    alphaScaleBiasClamp {bInterpResult=true}
    boundBone {boneName=dyn_skirt_fr_l_02}
    componentAppliedLinearAccClamp=<3000 3000 3000>
    componentLinearAccScale=<0.65 0.65 0.65>
    componentLinearVelScale=<0.65 0.65 0.65>
    gravityOverride {z=-980}
    physicsBodyDefinitions [{
      boundBone {boneName=dyn_skirt_fr_l_02}
      boxExtents=<20 30 30>
      constraintSetup {angularLimitsMax=<5 0 5>, angularLimitsMin=<-5 -5 -5>, angularTarget=<1 0 0>}
      localJointOffset=<2 0 0>
     }]
    simSpaceSettings {simSpaceAngularAlpha=0.08}
   }
   n17 = Anim Dynamics - Bone: dyn_skirt_fr_l_03(ComponentPose=@n16, Alpha=1, ExternalForce="0,0,0") {
    angularDampingOverride=1
    angularSpringConstant=20000
    bAngularSpring=true
    bOverrideAngularDamping=true
    bUseGravityOverride=true
    gravityScale=0.35
    simulationSpace=RootRelative
    alphaBoolBlend {blendOption=Linear}
    alphaScaleBiasClamp {bInterpResult=true}
    boundBone {boneName=dyn_skirt_fr_l_03}
    componentAppliedLinearAccClamp=<3000 3000 3000>
    componentLinearAccScale=<0.65 0.65 0.65>
    componentLinearVelScale=<0.65 0.65 0.65>
    gravityOverride {z=-980}
    physicsBodyDefinitions [{
      boundBone {boneName=dyn_skirt_fr_l_03}
      boxExtents=<20 30 30>
      constraintSetup {angularLimitsMax=<5 0 5>, angularLimitsMin=<-5 -5 -15>, angularTarget=<1 0 0>}
      localJointOffset=<2 0 0>
     }]
    simSpaceSettings {simSpaceAngularAlpha=0.08}
   }
   n18 = Anim Dynamics - Bone: dyn_skirt_bk_c_01(ComponentPose=@n17, Alpha=1, ExternalForce="0,0,0") {
    angularDampingOverride=1
    angularSpringConstant=20000
    bAngularSpring=true
    bOverrideAngularDamping=true
    bUseGravityOverride=true
    gravityScale=0.35
    simulationSpace=RootRelative
    alphaBoolBlend {blendOption=Linear}
    alphaScaleBiasClamp {bInterpResult=true}
    boundBone {boneName=dyn_skirt_bk_c_01}
    componentAppliedLinearAccClamp=<3000 3000 3000>
    componentLinearAccScale=<0.65 0.65 0.65>
    componentLinearVelScale=<0.65 0.65 0.65>
    gravityOverride {z=-980}
    physicsBodyDefinitions [{
      boundBone {boneName=dyn_skirt_bk_c_01}
      boxExtents=<20 30 30>
      constraintSetup {angularLimitsMax=<5 0 10>, angularLimitsMin=<-5 -25 -10>, angularTarget=<1 0 0>}
      localJointOffset=<4 0 0>
     }]
    simSpaceSettings {simSpaceAngularAlpha=0.08}
   }
   n19 = Anim Dynamics - Bone: dyn_skirt_bk_c_02(ComponentPose=@n18, Alpha=1, ExternalForce="0,0,0") {
    angularDampingOverride=1
    angularSpringConstant=20000
    bAngularSpring=true
    bOverrideAngularDamping=true
    bUseGravityOverride=true
    gravityScale=0.35
    simulationSpace=RootRelative
    alphaBoolBlend {blendOption=Linear}
    alphaScaleBiasClamp {bInterpResult=true}
    boundBone {boneName=dyn_skirt_bk_c_02}
    componentAppliedLinearAccClamp=<3000 3000 3000>
    componentLinearAccScale=<0.65 0.65 0.65>
    componentLinearVelScale=<0.65 0.65 0.65>
    gravityOverride {z=-980}
    physicsBodyDefinitions [{
      boundBone {boneName=dyn_skirt_bk_c_02}
      boxExtents=<20 30 30>
      constraintSetup {angularLimitsMax=<5 0 10>, angularLimitsMin=<-5 -25 -10>, angularTarget=<1 0 0>}
      localJointOffset=<4 0 0>
     }]
    simSpaceSettings {simSpaceAngularAlpha=0.08}
   }
   n20 = Anim Dynamics - Bone: dyn_skirt_bk_c_03(ComponentPose=@n19, Alpha=1, ExternalForce="0,0,0") {
    angularDampingOverride=1
    angularSpringConstant=20000
    bAngularSpring=true
    bOverrideAngularDamping=true
    bUseGravityOverride=true
    gravityScale=0.35
    simulationSpace=RootRelative
    alphaBoolBlend {blendOption=Linear}
    alphaScaleBiasClamp {bInterpResult=true}
    boundBone {boneName=dyn_skirt_bk_c_03}
    componentAppliedLinearAccClamp=<3000 3000 3000>
    componentLinearAccScale=<0.65 0.65 0.65>
    componentLinearVelScale=<0.65 0.65 0.65>
    gravityOverride {z=-980}
    physicsBodyDefinitions [{
      boundBone {boneName=dyn_skirt_bk_c_03}
      boxExtents=<20 30 30>
      constraintSetup {angularLimitsMax=<5 0 10>, angularLimitsMin=<-5 -25 -10>, angularTarget=<1 0 0>}
      localJointOffset=<4 0 0>
     }]
    simSpaceSettings {simSpaceAngularAlpha=0.08}
   }
   n21 = Anim Dynamics - Bone: dyn_skirt_bk_r_01(ComponentPose=@n20, Alpha=1, ExternalForce="0,0,0") {
    angularDampingOverride=1
    angularSpringConstant=20000
    bAngularSpring=true
    bOverrideAngularDamping=true
    bUseGravityOverride=true
    gravityScale=0.35
    simulationSpace=RootRelative
    alphaBoolBlend {blendOption=Linear}
    alphaScaleBiasClamp {bInterpResult=true}
    boundBone {boneName=dyn_skirt_bk_r_01}
    componentAppliedLinearAccClamp=<3000 3000 3000>
    componentLinearAccScale=<0.65 0.65 0.65>
    componentLinearVelScale=<0.65 0.65 0.65>
    gravityOverride {z=-980}
    physicsBodyDefinitions [{
      boundBone {boneName=dyn_skirt_bk_r_01}
      boxExtents=<20 30 30>
      constraintSetup {angularLimitsMax=<5 0 10>, angularLimitsMin=<-5 -25 -10>, angularTarget=<1 0 0>}
      localJointOffset=<4 0 0>
     }]
    simSpaceSettings {simSpaceAngularAlpha=0.08}
   }
   n22 = Anim Dynamics - Bone: dyn_skirt_bk_r_02(ComponentPose=@n21, Alpha=1, ExternalForce="0,0,0") {
    angularDampingOverride=1
    angularSpringConstant=20000
    bAngularSpring=true
    bOverrideAngularDamping=true
    bUseGravityOverride=true
    gravityScale=0.35
    simulationSpace=RootRelative
    alphaBoolBlend {blendOption=Linear}
    alphaScaleBiasClamp {bInterpResult=true}
    boundBone {boneName=dyn_skirt_bk_r_02}
    componentAppliedLinearAccClamp=<3000 3000 3000>
    componentLinearAccScale=<0.65 0.65 0.65>
    componentLinearVelScale=<0.65 0.65 0.65>
    gravityOverride {z=-980}
    physicsBodyDefinitions [{
      boundBone {boneName=dyn_skirt_bk_r_02}
      boxExtents=<20 30 30>
      constraintSetup {angularLimitsMax=<5 0 10>, angularLimitsMin=<-5 -25 -10>, angularTarget=<1 0 0>}
      localJointOffset=<4 0 0>
     }]
    simSpaceSettings {simSpaceAngularAlpha=0.08}
   }
   n23 = Anim Dynamics - Bone: dyn_skirt_bk_r_03(ComponentPose=@n22, Alpha=1, ExternalForce="0,0,0") {
    angularDampingOverride=1
    angularSpringConstant=20000
    bAngularSpring=true
    bOverrideAngularDamping=true
    bUseGravityOverride=true
    gravityScale=0.35
    simulationSpace=RootRelative
    alphaBoolBlend {blendOption=Linear}
    alphaScaleBiasClamp {bInterpResult=true}
    boundBone {boneName=dyn_skirt_bk_r_03}
    componentAppliedLinearAccClamp=<3000 3000 3000>
    componentLinearAccScale=<0.65 0.65 0.65>
    componentLinearVelScale=<0.65 0.65 0.65>
    gravityOverride {z=-980}
    physicsBodyDefinitions [{
      boundBone {boneName=dyn_skirt_bk_r_03}
      boxExtents=<20 30 30>
      constraintSetup {angularLimitsMax=<5 0 10>, angularLimitsMin=<-5 -25 -10>, angularTarget=<1 0 0>}
      localJointOffset=<4 0 0>
     }]
    simSpaceSettings {simSpaceAngularAlpha=0.08}
   }
   n24 = Anim Dynamics - Bone: dyn_skirt_sd_l_01(ComponentPose=@n23, Alpha=1, ExternalForce="0,0,0") {
    angularDampingOverride=1
    angularSpringConstant=20000
    bAngularSpring=true
    bOverrideAngularDamping=true
    bUseGravityOverride=true
    gravityScale=0.35
    simulationSpace=RootRelative
    alphaBoolBlend {blendOption=Linear}
    alphaScaleBiasClamp {bInterpResult=true}
    boundBone {boneName=dyn_skirt_sd_l_01}
    componentAppliedLinearAccClamp=<3000 3000 3000>
    componentLinearAccScale=<0.65 0.65 0.65>
    componentLinearVelScale=<0.65 0.65 0.65>
    gravityOverride {z=-980}
    physicsBodyDefinitions [{
      boundBone {boneName=dyn_skirt_sd_l_01}
      boxExtents=<20 30 30>
      constraintSetup {angularLimitsMax=<5 0 10>, angularLimitsMin=<-5 -25 -10>, angularTarget=<1 0 0>}
      localJointOffset=<4 0 0>
     }]
    simSpaceSettings {simSpaceAngularAlpha=0.08}
   }
   n25 = Anim Dynamics - Bone: dyn_skirt_sd_l_02(ComponentPose=@n24, Alpha=1, ExternalForce="0,0,0") {
    angularDampingOverride=1
    angularSpringConstant=20000
    bAngularSpring=true
    bOverrideAngularDamping=true
    bUseGravityOverride=true
    gravityScale=0.35
    simulationSpace=RootRelative
    alphaBoolBlend {blendOption=Linear}
    alphaScaleBiasClamp {bInterpResult=true}
    boundBone {boneName=dyn_skirt_sd_l_02}
    componentAppliedLinearAccClamp=<3000 3000 3000>
    componentLinearAccScale=<0.65 0.65 0.65>
    componentLinearVelScale=<0.65 0.65 0.65>
    gravityOverride {z=-980}
    physicsBodyDefinitions [{
      boundBone {boneName=dyn_skirt_sd_l_02}
      boxExtents=<20 30 30>
      constraintSetup {angularLimitsMax=<5 0 10>, angularLimitsMin=<-5 -25 -10>, angularTarget=<1 0 0>}
      localJointOffset=<4 0 0>
     }]
    simSpaceSettings {simSpaceAngularAlpha=0.08}
   }
   n26 = Anim Dynamics - Bone: dyn_skirt_sd_l_03(ComponentPose=@n25, Alpha=1, ExternalForce="0,0,0") {
    angularDampingOverride=1
    angularSpringConstant=20000
    bAngularSpring=true
    bOverrideAngularDamping=true
    bUseGravityOverride=true
    gravityScale=0.35
    simulationSpace=RootRelative
    alphaBoolBlend {blendOption=Linear}
    alphaScaleBiasClamp {bInterpResult=true}
    boundBone {boneName=dyn_skirt_sd_l_03}
    componentAppliedLinearAccClamp=<3000 3000 3000>
    componentLinearAccScale=<0.65 0.65 0.65>
    componentLinearVelScale=<0.65 0.65 0.65>
    gravityOverride {z=-980}
    physicsBodyDefinitions [{
      boundBone {boneName=dyn_skirt_sd_l_03}
      boxExtents=<20 30 30>
      constraintSetup {angularLimitsMax=<5 0 10>, angularLimitsMin=<-5 -25 -10>, angularTarget=<1 0 0>}
      localJointOffset=<4 0 0>
     }]
    simSpaceSettings {simSpaceAngularAlpha=0.08}
   }
   n27 = Anim Dynamics - Bone: dyn_skirt_bk_l_01(ComponentPose=@n26, Alpha=1, ExternalForce="0,0,0") {
    angularDampingOverride=1
    angularSpringConstant=20000
    bAngularSpring=true
    bOverrideAngularDamping=true
    bUseGravityOverride=true
    gravityScale=0.35
    simulationSpace=RootRelative
    alphaBoolBlend {blendOption=Linear}
    alphaScaleBiasClamp {bInterpResult=true}
    boundBone {boneName=dyn_skirt_bk_l_01}
    componentAppliedLinearAccClamp=<3000 3000 3000>
    componentLinearAccScale=<0.65 0.65 0.65>
    componentLinearVelScale=<0.65 0.65 0.65>
    gravityOverride {z=-980}
    physicsBodyDefinitions [{
      boundBone {boneName=dyn_skirt_bk_l_01}
      boxExtents=<20 30 30>
      constraintSetup {angularLimitsMax=<5 0 10>, angularLimitsMin=<-5 -25 -10>, angularTarget=<1 0 0>}
      localJointOffset=<4 0 0>
     }]
    simSpaceSettings {simSpaceAngularAlpha=0.08}
   }
   n28 = Anim Dynamics - Bone: dyn_skirt_bk_l_02(ComponentPose=@n27, Alpha=1, ExternalForce="0,0,0") {
    angularDampingOverride=1
    angularSpringConstant=20000
    bAngularSpring=true
    bOverrideAngularDamping=true
    bUseGravityOverride=true
    gravityScale=0.35
    simulationSpace=RootRelative
    alphaBoolBlend {blendOption=Linear}
    alphaScaleBiasClamp {bInterpResult=true}
    boundBone {boneName=dyn_skirt_bk_l_02}
    componentAppliedLinearAccClamp=<3000 3000 3000>
    componentLinearAccScale=<0.65 0.65 0.65>
    componentLinearVelScale=<0.65 0.65 0.65>
    gravityOverride {z=-980}
    physicsBodyDefinitions [{
      boundBone {boneName=dyn_skirt_bk_l_02}
      boxExtents=<20 30 30>
      constraintSetup {angularLimitsMax=<5 0 10>, angularLimitsMin=<-5 -25 -10>, angularTarget=<1 0 0>}
      localJointOffset=<4 0 0>
     }]
    simSpaceSettings {simSpaceAngularAlpha=0.08}
   }
   n29 = Anim Dynamics - Bone: dyn_skirt_bk_l_03(ComponentPose=@n28, Alpha=1, ExternalForce="0,0,0") {
    angularDampingOverride=1
    angularSpringConstant=20000
    bAngularSpring=true
    bOverrideAngularDamping=true
    bUseGravityOverride=true
    gravityScale=0.35
    simulationSpace=RootRelative
    alphaBoolBlend {blendOption=Linear}
    alphaScaleBiasClamp {bInterpResult=true}
    boundBone {boneName=dyn_skirt_bk_l_03}
    componentAppliedLinearAccClamp=<3000 3000 3000>
    componentLinearAccScale=<0.65 0.65 0.65>
    componentLinearVelScale=<0.65 0.65 0.65>
    gravityOverride {z=-980}
    physicsBodyDefinitions [{
      boundBone {boneName=dyn_skirt_bk_l_03}
      boxExtents=<20 30 30>
      constraintSetup {angularLimitsMax=<5 0 10>, angularLimitsMin=<-5 -25 -10>, angularTarget=<1 0 0>}
      localJointOffset=<4 0 0>
     }]
    simSpaceSettings {simSpaceAngularAlpha=0.08}
   }
   n30 = Anim Dynamics - Bone: dyn_skirt_sd_r_01(ComponentPose=@n29, Alpha=1, ExternalForce="0,0,0") {
    angularDampingOverride=1
    angularSpringConstant=20000
    bAngularSpring=true
    bOverrideAngularDamping=true
    bUseGravityOverride=true
    gravityScale=0.35
    simulationSpace=RootRelative
    alphaBoolBlend {blendOption=Linear}
    alphaScaleBiasClamp {bInterpResult=true}
    boundBone {boneName=dyn_skirt_sd_r_01}
    componentAppliedLinearAccClamp=<3000 3000 3000>
    componentLinearAccScale=<0.65 0.65 0.65>
    componentLinearVelScale=<0.65 0.65 0.65>
    gravityOverride {z=-980}
    physicsBodyDefinitions [{
      boundBone {boneName=dyn_skirt_sd_r_01}
      boxExtents=<20 30 30>
      constraintSetup {angularLimitsMax=<5 0 10>, angularLimitsMin=<-5 -25 -10>, angularTarget=<1 0 0>}
      localJointOffset=<4 0 0>
     }]
    simSpaceSettings {simSpaceAngularAlpha=0.08}
   }
   n31 = Anim Dynamics - Bone: dyn_skirt_sd_r_02(ComponentPose=@n30, Alpha=1, ExternalForce="0,0,0") {
    angularDampingOverride=1
    angularSpringConstant=20000
    bAngularSpring=true
    bOverrideAngularDamping=true
    bUseGravityOverride=true
    gravityScale=0.35
    simulationSpace=RootRelative
    alphaBoolBlend {blendOption=Linear}
    alphaScaleBiasClamp {bInterpResult=true}
    boundBone {boneName=dyn_skirt_sd_r_02}
    componentAppliedLinearAccClamp=<3000 3000 3000>
    componentLinearAccScale=<0.65 0.65 0.65>
    componentLinearVelScale=<0.65 0.65 0.65>
    gravityOverride {z=-980}
    physicsBodyDefinitions [{
      boundBone {boneName=dyn_skirt_sd_r_02}
      boxExtents=<20 30 30>
      constraintSetup {angularLimitsMax=<5 0 10>, angularLimitsMin=<-5 -25 -10>, angularTarget=<1 0 0>}
      localJointOffset=<4 0 0>
     }]
    simSpaceSettings {simSpaceAngularAlpha=0.08}
   }
   n32 = Anim Dynamics - Bone: dyn_skirt_sd_r_03(ComponentPose=@n31, Alpha=1, ExternalForce="0,0,0") {
    angularDampingOverride=1
    angularSpringConstant=20000
    bAngularSpring=true
    bOverrideAngularDamping=true
    bUseGravityOverride=true
    gravityScale=0.35
    simulationSpace=RootRelative
    alphaBoolBlend {blendOption=Linear}
    alphaScaleBiasClamp {bInterpResult=true}
    boundBone {boneName=dyn_skirt_sd_r_03}
    componentAppliedLinearAccClamp=<3000 3000 3000>
    componentLinearAccScale=<0.65 0.65 0.65>
    componentLinearVelScale=<0.65 0.65 0.65>
    gravityOverride {z=-980}
    physicsBodyDefinitions [{
      boundBone {boneName=dyn_skirt_sd_r_03}
      boxExtents=<20 30 30>
      constraintSetup {angularLimitsMax=<5 0 10>, angularLimitsMin=<-5 -25 -10>, angularTarget=<1 0 0>}
      localJointOffset=<4 0 0>
     }]
    simSpaceSettings {simSpaceAngularAlpha=0.08}
   }
   n33 = Anim Dynamics - Bone: dyn_skirt_fr_r_01(ComponentPose=@n32, Alpha=1, ExternalForce="0,0,0") {
    angularDampingOverride=1
    angularSpringConstant=20000
    bAngularSpring=true
    bOverrideAngularDamping=true
    bUseGravityOverride=true
    gravityScale=0.35
    simulationSpace=RootRelative
    alphaBoolBlend {blendOption=Linear}
    alphaScaleBiasClamp {bInterpResult=true}
    boundBone {boneName=dyn_skirt_fr_r_01}
    componentAppliedLinearAccClamp=<3000 3000 3000>
    componentLinearAccScale=<0.65 0.65 0.65>
    componentLinearVelScale=<0.65 0.65 0.65>
    gravityOverride {z=-980}
    physicsBodyDefinitions [{
      boundBone {boneName=dyn_skirt_fr_r_01}
      boxExtents=<20 30 30>
      constraintSetup {angularLimitsMax=<5 0 5>, angularLimitsMin=<-5 -15 -5>, angularTarget=<1 0 0>}
      localJointOffset=<2 0 0>
     }]
    simSpaceSettings {simSpaceAngularAlpha=0.08}
   }
   n34 = Anim Dynamics - Bone: dyn_skirt_fr_r_02(ComponentPose=@n33, Alpha=1, ExternalForce="0,0,0") {
    angularDampingOverride=1
    angularSpringConstant=20000
    bAngularSpring=true
    bOverrideAngularDamping=true
    bUseGravityOverride=true
    gravityScale=0.35
    simulationSpace=RootRelative
    alphaBoolBlend {blendOption=Linear}
    alphaScaleBiasClamp {bInterpResult=true}
    boundBone {boneName=dyn_skirt_fr_r_02}
    componentAppliedLinearAccClamp=<3000 3000 3000>
    componentLinearAccScale=<0.65 0.65 0.65>
    componentLinearVelScale=<0.65 0.65 0.65>
    gravityOverride {z=-980}
    physicsBodyDefinitions [{
      boundBone {boneName=dyn_skirt_fr_r_02}
      boxExtents=<20 30 30>
      constraintSetup {angularLimitsMax=<5 0 5>, angularLimitsMin=<-5 -5 -5>, angularTarget=<1 0 0>}
      localJointOffset=<2 0 0>
     }]
    simSpaceSettings {simSpaceAngularAlpha=0.08}
   }
   n35 = Anim Dynamics - Bone: dyn_skirt_fr_r_03(ComponentPose=@n34, Alpha=1, ExternalForce="0,0,0") {
    angularDampingOverride=1
    angularSpringConstant=20000
    bAngularSpring=true
    bOverrideAngularDamping=true
    bUseGravityOverride=true
    gravityScale=0.35
    simulationSpace=RootRelative
    alphaBoolBlend {blendOption=Linear}
    alphaScaleBiasClamp {bInterpResult=true}
    boundBone {boneName=dyn_skirt_fr_r_03}
    componentAppliedLinearAccClamp=<3000 3000 3000>
    componentLinearAccScale=<0.65 0.65 0.65>
    componentLinearVelScale=<0.65 0.65 0.65>
    gravityOverride {z=-980}
    physicsBodyDefinitions [{
      boundBone {boneName=dyn_skirt_fr_r_03}
      boxExtents=<20 30 30>
      constraintSetup {angularLimitsMax=<5 -5 5>, angularLimitsMin=<-5 -15 -5>, angularTarget=<1 0 0>}
      localJointOffset=<2 0 0>
     }]
    simSpaceSettings {simSpaceAngularAlpha=0.08}
   }
   n36 = Anim Dynamics - Bone: dyn_strap_fr_l_01(ComponentPose=@n35, Alpha=1, ExternalForce="0,0,0") {
    angularDampingOverride=1
    angularSpringConstant=20000
    bAngularSpring=true
    bOverrideAngularDamping=true
    bUseGravityOverride=true
    gravityScale=0.35
    simulationSpace=RootRelative
    alphaBoolBlend {blendOption=Linear}
    boundBone {boneName=dyn_strap_fr_l_01}
    componentAppliedLinearAccClamp=<3000 3000 3000>
    componentLinearAccScale=<0.7 0.7 0.7>
    componentLinearVelScale=<1 1 1>
    gravityOverride {z=-980}
    physicsBodyDefinitions [{
      boundBone {boneName=dyn_strap_fr_l_01}
      boxExtents=<20 30 30>
      constraintSetup {angularLimitsMax=<5 0 0>, angularLimitsMin=<-5 -5 -5>, angularTarget=<1 0 0>}
      localJointOffset=<4 0 0>
     }]
   }
   n37 = Anim Dynamics - Bone: dyn_strap_fr_l_02(ComponentPose=@n36, Alpha=1, ExternalForce="0,0,0") {
    angularDampingOverride=1
    angularSpringConstant=20000
    bAngularSpring=true
    bOverrideAngularDamping=true
    bUseGravityOverride=true
    gravityScale=0.35
    simulationSpace=RootRelative
    alphaBoolBlend {blendOption=Linear}
    boundBone {boneName=dyn_strap_fr_l_02}
    componentAppliedLinearAccClamp=<3000 3000 3000>
    componentLinearAccScale=<0.7 0.7 0.7>
    componentLinearVelScale=<1 1 1>
    gravityOverride {z=-980}
    physicsBodyDefinitions [{
      boundBone {boneName=dyn_strap_fr_l_02}
      boxExtents=<20 30 30>
      constraintSetup {angularLimitsMax=<5 0 0>, angularLimitsMin=<-25 -5 -25>, angularTarget=<1 0 0>}
      localJointOffset=<4 0 0>
     }]
   }
   n38 = Anim Dynamics - Bone: dyn_strap_fr_l_03(ComponentPose=@n37, Alpha=1, ExternalForce="0,0,0") {
    angularDampingOverride=1
    angularSpringConstant=20000
    bAngularSpring=true
    bOverrideAngularDamping=true
    bUseGravityOverride=true
    gravityScale=0.35
    simulationSpace=RootRelative
    alphaBoolBlend {blendOption=Linear}
    boundBone {boneName=dyn_strap_fr_l_03}
    componentAppliedLinearAccClamp=<3000 3000 3000>
    componentLinearAccScale=<0.7 0.7 0.7>
    componentLinearVelScale=<1 1 1>
    gravityOverride {z=-980}
    physicsBodyDefinitions [{
      boundBone {boneName=dyn_strap_fr_l_03}
      boxExtents=<20 30 30>
      constraintSetup {angularLimitsMax=<0 0 0>, angularLimitsMin=<0 -5 -25>, angularTarget=<1 0 0>}
      localJointOffset=<4 0 0>
     }]
   }
   n39 = Anim Dynamics - Bone: dyn_strap_fr_l_04(ComponentPose=@n38, Alpha=1, ExternalForce="0,0,0") {
    angularDampingOverride=1
    angularSpringConstant=20000
    bAngularSpring=true
    bOverrideAngularDamping=true
    bUseGravityOverride=true
    gravityScale=0.35
    simulationSpace=RootRelative
    alphaBoolBlend {blendOption=Linear}
    boundBone {boneName=dyn_strap_fr_l_04}
    componentAppliedLinearAccClamp=<3000 3000 3000>
    componentLinearAccScale=<0.7 0.7 0.7>
    componentLinearVelScale=<1 1 1>
    gravityOverride {z=-980}
    physicsBodyDefinitions [{
      boundBone {boneName=dyn_strap_fr_l_04}
      boxExtents=<20 30 30>
      constraintSetup {angularLimitsMax=<0 0 5>, angularLimitsMin=<0 0 -10>, angularTarget=<1 0 0>}
      localJointOffset=<4 0 0>
     }]
   }
   n40 = Anim Dynamics - Bone: dyn_strap_bk_l_01(ComponentPose=@n39, Alpha=1, ExternalForce="0,0,0") {
    angularDampingOverride=1
    angularSpringConstant=20000
    bAngularSpring=true
    bOverrideAngularDamping=true
    bUseGravityOverride=true
    gravityScale=0.35
    simulationSpace=RootRelative
    alphaBoolBlend {blendOption=Linear}
    boundBone {boneName=dyn_strap_bk_l_01}
    componentAppliedLinearAccClamp=<3000 3000 3000>
    componentLinearAccScale=<0.7 0.7 0.7>
    componentLinearVelScale=<1 1 1>
    gravityOverride {z=-980}
    physicsBodyDefinitions [{
      boundBone {boneName=dyn_strap_bk_l_01}
      boxExtents=<20 30 30>
      constraintSetup {angularLimitsMax=<0 0 2>, angularLimitsMin=<-2 0 0>, angularTarget=<1 0 0>}
      localJointOffset=<4 0 0>
     }]
   }
   n41 = Anim Dynamics - Bone: dyn_strap_bk_l_02(ComponentPose=@n40, Alpha=1, ExternalForce="0,0,0") {
    angularDampingOverride=1
    angularSpringConstant=20000
    bAngularSpring=true
    bOverrideAngularDamping=true
    bUseGravityOverride=true
    gravityScale=0.35
    simulationSpace=RootRelative
    alphaBoolBlend {blendOption=Linear}
    boundBone {boneName=dyn_strap_bk_l_02}
    componentAppliedLinearAccClamp=<3000 3000 3000>
    componentLinearAccScale=<0.7 0.7 0.7>
    componentLinearVelScale=<1 1 1>
    gravityOverride {z=-980}
    physicsBodyDefinitions [{
      boundBone {boneName=dyn_strap_bk_l_02}
      boxExtents=<20 30 30>
      constraintSetup {angularLimitsMax=<2 0 20>, angularLimitsMin=<-20 0 0>, angularTarget=<1 0 0>}
      localJointOffset=<4 0 0>
     }]
   }
   n42 = Anim Dynamics - Bone: dyn_strap_bk_l_03(ComponentPose=@n41, Alpha=1, ExternalForce="0,0,0") {
    angularDampingOverride=1
    angularSpringConstant=20000
    bAngularSpring=true
    bOverrideAngularDamping=true
    bUseGravityOverride=true
    gravityScale=0.35
    simulationSpace=RootRelative
    alphaBoolBlend {blendOption=Linear}
    boundBone {boneName=dyn_strap_bk_l_03}
    componentAppliedLinearAccClamp=<3000 3000 3000>
    componentLinearAccScale=<0.7 0.7 0.7>
    componentLinearVelScale=<1 1 1>
    gravityOverride {z=-980}
    physicsBodyDefinitions [{
      boundBone {boneName=dyn_strap_bk_l_03}
      boxExtents=<20 30 30>
      constraintSetup {angularLimitsMax=<2 0 15>, angularLimitsMin=<-10 0 0>, angularTarget=<1 0 0>}
      localJointOffset=<4 0 0>
     }]
   }
   n43 = Anim Dynamics - Bone: dyn_strap_bk_l_04(ComponentPose=@n42, Alpha=1, ExternalForce="0,0,0") {
    angularDampingOverride=1
    angularSpringConstant=20000
    bAngularSpring=true
    bOverrideAngularDamping=true
    bUseGravityOverride=true
    gravityScale=0.35
    simulationSpace=RootRelative
    alphaBoolBlend {blendOption=Linear}
    boundBone {boneName=dyn_strap_bk_l_04}
    componentAppliedLinearAccClamp=<3000 3000 3000>
    componentLinearAccScale=<0.7 0.7 0.7>
    componentLinearVelScale=<1 1 1>
    gravityOverride {z=-980}
    physicsBodyDefinitions [{
      boundBone {boneName=dyn_strap_bk_l_04}
      boxExtents=<20 30 30>
      constraintSetup {angularLimitsMax=<0 0 0>, angularLimitsMin=<-20 0 -10>, angularTarget=<1 0 0>}
      localJointOffset=<4 0 0>
     }]
   }
   n44 = "Constraint - dyn_strap_bk_l_02 (R)"(ComponentPose=@n43, Alpha=0.2, ConstraintWeights_0=1) {
    alphaBoolBlend {blendOption=Linear}
    boneToModify {boneName=dyn_strap_bk_l_02}
    constraintSetup [{transformType=Rotation, perAxis {bY=false, bZ=false}, targetBone {boneName=dyn_skirt_bk_l_03}}]
    constraintWeights [1]
   }
   n45 = Anim Dynamics - Bone: dyn_sleeves_fr_l(ComponentPose=@n44, Alpha=1, ExternalForce="0,0,0") {
    angularDampingOverride=1
    angularSpringConstant=20000
    bAngularSpring=true
    bOverrideAngularDamping=true
    bUseGravityOverride=true
    gravityScale=0.35
    simulationSpace=RootRelative
    alphaBoolBlend {blendOption=Linear}
    boundBone {boneName=dyn_sleeves_fr_l}
    componentAppliedLinearAccClamp=<3000 3000 3000>
    componentLinearAccScale=<1 1 1>
    componentLinearVelScale=<1 1 1>
    gravityOverride {z=-980}
    physicsBodyDefinitions [{
      boundBone {boneName=dyn_sleeves_fr_l}
      boxExtents=<20 30 30>
      constraintSetup {angularLimitsMax=<15 15 15>, angularLimitsMin=<-15 -15 0>, angularTarget=<1 0 0>}
      localJointOffset=<2 0 0>
     }]
   }
   n46 = Anim Dynamics - Bone: dyn_sleeves_bk_l(ComponentPose=@n45, Alpha=1, ExternalForce="0,0,0") {
    angularDampingOverride=1
    angularSpringConstant=20000
    bAngularSpring=true
    bOverrideAngularDamping=true
    bUseGravityOverride=true
    gravityScale=0.35
    simulationSpace=RootRelative
    alphaBoolBlend {blendOption=Linear}
    boundBone {boneName=dyn_sleeves_bk_l}
    componentAppliedLinearAccClamp=<3000 3000 3000>
    componentLinearAccScale=<1 1 1>
    componentLinearVelScale=<1 1 1>
    gravityOverride {z=-980}
    physicsBodyDefinitions [{
      boundBone {boneName=dyn_sleeves_bk_l}
      boxExtents=<20 30 30>
      constraintSetup {angularLimitsMax=<15 15 15>, angularLimitsMin=<-15 -15 -15>, angularTarget=<1 0 0>}
      localJointOffset=<2 0 0>
     }]
   }
   n47 = Anim Dynamics - Bone: dyn_sleeves_bk_r(ComponentPose=@n46, Alpha=1, ExternalForce="0,0,0") {
    angularDampingOverride=1
    angularSpringConstant=20000
    bAngularSpring=true
    bOverrideAngularDamping=true
    bUseGravityOverride=true
    gravityScale=0.35
    simulationSpace=RootRelative
    alphaBoolBlend {blendOption=Linear}
    boundBone {boneName=dyn_sleeves_bk_r}
    componentAppliedLinearAccClamp=<3000 3000 3000>
    componentLinearAccScale=<1 1 1>
    componentLinearVelScale=<1 1 1>
    gravityOverride {z=-980}
    physicsBodyDefinitions [{
      boundBone {boneName=dyn_sleeves_bk_r}
      boxExtents=<20 30 30>
      constraintSetup {angularLimitsMax=<15 15 15>, angularLimitsMin=<-15 -15 -15>, angularTarget=<1 0 0>}
      localJointOffset=<2 0 0>
     }]
   }
   n48 = Anim Dynamics - Bone: dyn_sleeves_fr_r(ComponentPose=@n47, Alpha=1, ExternalForce="0,0,0") {
    angularDampingOverride=1
    angularSpringConstant=20000
    bAngularSpring=true
    bOverrideAngularDamping=true
    bUseGravityOverride=true
    gravityScale=0.35
    simulationSpace=RootRelative
    alphaBoolBlend {blendOption=Linear}
    boundBone {boneName=dyn_sleeves_fr_r}
    componentAppliedLinearAccClamp=<3000 3000 3000>
    componentLinearAccScale=<1 1 1>
    componentLinearVelScale=<1 1 1>
    gravityOverride {z=-980}
    physicsBodyDefinitions [{
      boundBone {boneName=dyn_sleeves_fr_r}
      boxExtents=<20 30 30>
      constraintSetup {angularLimitsMax=<15 15 15>, angularLimitsMin=<-15 -15 -15>, angularTarget=<1 0 0>}
      localJointOffset=<2 0 0>
     }]
   }
   n49 = Anim Dynamics - Bone: MSTR_sleeve_l(ComponentPose=@n48, Alpha=Get Alpha_NotSkydiving(), ExternalForce="0,0,0") {
    angularDampingOverride=1
    angularSpringConstant=20000
    bAngularSpring=true
    bOverrideAngularDamping=true
    bUseGravityOverride=true
    gravityScale=0.35
    simulationSpace=RootRelative
    alphaBoolBlend {blendOption=Linear}
    alphaScaleBiasClamp {bInterpResult=true, interpSpeedIncreasing=0.3}
    boundBone {boneName=MSTR_sleeve_l}
    componentAppliedLinearAccClamp=<3000 3000 3000>
    componentLinearAccScale=<1 1 1>
    componentLinearVelScale=<1 1 1>
    gravityOverride {z=-980}
    physicsBodyDefinitions [{
      boundBone {boneName=MSTR_sleeve_l}
      boxExtents=<20 30 30>
      constraintSetup {angularLimitsMax=<0 0 0>, angularLimitsMin=<0 0 -7>, angularTarget=<1 0 0>}
      localJointOffset=<3.5 0 0>
     }]
   }
   n50 = "Constraint - dyn_strap_fr_l_04 (P)"(ComponentPose=@n49, Alpha=1, ConstraintWeights_0=1) {boneToModify {boneName=dyn_strap_fr_l_04}, constraintSetup [{targetBone {boneName=dyn_skirt_sd_l_03}}], constraintWeights [1]}
   n51 = "Constraint - dyn_strap_bk_l_04 (P)"(ComponentPose=@n50, Alpha=1, ConstraintWeights_0=1) {boneToModify {boneName=dyn_strap_bk_l_04}, constraintSetup [{targetBone {boneName=dyn_skirt_sd_l_03}}], constraintWeights [1]}
   n52 = "Constraint - dyn_strap_bk_l_02 (P)"!(ComponentPose=@n51, bAlphaBoolEnabled:bool=NOT Boolean(A=Get bIsCrouching()), ConstraintWeights_0=1) -> (Pose) {
    alphaInputType=Bool
    boneToModify {boneName=dyn_strap_bk_l_02}
    constraintSetup [{targetBone {boneName=dyn_skirt_bk_l_02}}]
    constraintWeights [1]
   }
   n53 = "Constraint - dyn_strap_bk_l_02 (P)"(ComponentPose=@n52, bAlphaBoolEnabled=Get bIsCrouching(), ConstraintWeights_0=1) {
    alphaInputType=Bool
    boneToModify {boneName=dyn_strap_bk_l_02}
    constraintSetup [{targetBone {boneName=dyn_skirt_sd_l_02}}]
    constraintWeights [1]
   }
   Save cached pose 'Dynamics Cache'(Pose=Component To Local(ComponentPose=@n53)) {cacheName=Dynamics Cache}
   n54 = "Transform (Modify) Bone - Bone: dyn_ponytail_01_l"(ComponentPose=Local To Component(LocalPose=Use cached pose 'CtrlRig Cache'()), Alpha=1, Translation="0,0,0", Rotation="0,0,0", Scale="1,1,1") {
    rotationMode=BMM_Additive
    rotationSpace=BCS_BoneSpace
    translationMode=BMM_Additive
    translationSpace=BCS_BoneSpace
    boneToModify {boneName=dyn_ponytail_01_l}
   }
   n55 = "Transform (Modify) Bone - Bone: dyn_ponytail_02_l"(ComponentPose=@n54, Alpha=1, Translation="0,0,0", Rotation="0,0,0", Scale="1,1,1") {
    rotationMode=BMM_Additive
    rotationSpace=BCS_BoneSpace
    translationMode=BMM_Additive
    translationSpace=BCS_BoneSpace
    boneToModify {boneName=dyn_ponytail_02_l}
   }
   n56 = "Transform (Modify) Bone - Bone: dyn_ponytail_03_l"(ComponentPose=@n55, Alpha=1, Translation="0,0,0", Rotation="0,0,0", Scale="1,1,1") {
    rotationMode=BMM_Additive
    rotationSpace=BCS_BoneSpace
    translationMode=BMM_Additive
    translationSpace=BCS_BoneSpace
    boneToModify {boneName=dyn_ponytail_03_l}
   }
   n57 = "Transform (Modify) Bone - Bone: dyn_ponytail_04_l"(ComponentPose=@n56, Alpha=1, Translation="0,0,0", Rotation="0,0,0", Scale="1,1,1") {
    rotationMode=BMM_Additive
    rotationSpace=BCS_BoneSpace
    translationMode=BMM_Additive
    translationSpace=BCS_BoneSpace
    boneToModify {boneName=dyn_ponytail_04_l}
   }
   n58 = "Transform (Modify) Bone - Bone: dyn_ponytail_01_r"(ComponentPose=@n57, Alpha=1, Translation="0,0,0", Rotation="0,0,0", Scale="1,1,1") {
    rotationMode=BMM_Additive
    rotationSpace=BCS_BoneSpace
    translationMode=BMM_Additive
    translationSpace=BCS_BoneSpace
    boneToModify {boneName=dyn_ponytail_01_r}
   }
   n59 = "Transform (Modify) Bone - Bone: dyn_ponytail_02_r"(ComponentPose=@n58, Alpha=1, Translation="0,0,0", Rotation="0,0,0", Scale="1,1,1") {
    rotationMode=BMM_Additive
    rotationSpace=BCS_BoneSpace
    translationMode=BMM_Additive
    translationSpace=BCS_BoneSpace
    boneToModify {boneName=dyn_ponytail_02_r}
   }
   n60 = "Transform (Modify) Bone - Bone: dyn_ponytail_03_l"(ComponentPose=@n59, Alpha=1, Translation="0,0,0", Rotation="0,0,0", Scale="1,1,1") {
    rotationMode=BMM_Additive
    rotationSpace=BCS_BoneSpace
    translationMode=BMM_Additive
    translationSpace=BCS_BoneSpace
    boneToModify {boneName=dyn_ponytail_03_l}
   }
   n61 = "Transform (Modify) Bone - Bone: dyn_ponytail_04_l"(ComponentPose=@n60, Alpha=1, Translation="0,0,0", Rotation="0,0,0", Scale="1,1,1") {
    rotationMode=BMM_Additive
    rotationSpace=BCS_BoneSpace
    translationMode=BMM_Additive
    translationSpace=BCS_BoneSpace
    boneToModify {boneName=dyn_ponytail_04_l}
   }
   Save cached pose 'SavePonytailPose_L'(Pose=Component To Local(ComponentPose=@n61)) {cacheName=SavePonytailPose_L}
   n62 = "Transform (Modify) Bone - Bone: base_tie_01"!(
    ComponentPose=Local To Component(LocalPose=Use cached pose 'SavePonytailPose_L'())
    Alpha:float=1
    Translation="0,0,0"
    Rotation="10,0,0"
    Scale="1,1,1"
   ) -> (Pose) {
    rotationMode=BMM_Additive
    rotationSpace=BCS_BoneSpace
    translationMode=BMM_Additive
    translationSpace=BCS_BoneSpace
    alphaBoolBlend {blendInTime=0.2, blendOption=Linear, blendOutTime=0.2}
    alphaScaleBiasClamp {bInterpResult=true, interpSpeedDecreasing=8, interpSpeedIncreasing=12}
    boneToModify {boneName=base_tie_01}
    rotation {pitch=10}
   }
   n63 = "Transform (Modify) Bone - Bone: base_tie_01"(ComponentPose=@n62, Alpha=1, Translation="0,0,0", Rotation="20,0,0", Scale="1,1,1") {
    rotationMode=BMM_Additive
    rotationSpace=BCS_BoneSpace
    translationMode=BMM_Additive
    translationSpace=BCS_BoneSpace
    alphaBoolBlend {blendInTime=0.2, blendOption=Linear, blendOutTime=0.2}
    alphaScaleBiasClamp {bInterpResult=true, interpSpeedDecreasing=8, interpSpeedIncreasing=8}
    boneToModify {boneName=base_tie_01}
    rotation {pitch=20}
   }
   n64 = RigidBody!(ComponentPose=@n63, Alpha, OverrideWorldGravity:Vector=Get Override World Gravity() -> (Override World Gravity:Vector), ExternalForce) -> (Pose) {
    bOverrideWorldGravity=true
    overridePhysicsAsset=&/F_MED_FlossYawn_BP1
    simulationSpace=BaseBoneSpace
    alphaBoolBlend {blendOption=Linear}
    alphaScaleBiasClamp {bInterpResult=true, interpSpeedIncreasing=12}
    baseBoneRef {boneName=spine_05}
    componentAppliedLinearAccClamp=<3000 3000 3000>
    overrideWorldGravity {z=-980}
    simSpaceSettings {
     dampingAlpha=0
     maxAngularAcceleration=32
     maxAngularVelocity=8
     maxLinearAcceleration=1600
     maxLinearVelocity=400
     velocityScaleZ=0.25
     worldAlpha=1
     externalLinearDragV=<0.3 0.3 0.3>
    }
   }
   n65 = "Transform (Modify) Bone - Bone: dyn_scarfBack_05"!(ComponentPose=@n64, Alpha=1, Translation="0,0,0", Rotation="30,-0.011,0.031") -> (Pose) {
    rotationMode=BMM_Additive
    rotationSpace=BCS_BoneSpace
    translationMode=BMM_Additive
    translationSpace=BCS_BoneSpace
    alphaBoolBlend {blendInTime=0.2, blendOption=Linear, blendOutTime=0.2}
    alphaScaleBiasClamp {bInterpResult=true, interpSpeedDecreasing=8, interpSpeedIncreasing=8}
    boneToModify {boneName=dyn_scarfBack_05}
    rotation=<rot 30 -0.011 0.031>
   }
   n66 = "Transform (Modify) Bone - Bone: dyn_scarf_02"!(ComponentPose=@n65, Alpha=1, Translation="0,0,0", Rotation="-2,0,0") -> (Pose) {
    rotationMode=BMM_Additive
    rotationSpace=BCS_BoneSpace
    translationMode=BMM_Additive
    translationSpace=BCS_BoneSpace
    alphaBoolBlend {blendInTime=0.2, blendOption=Linear, blendOutTime=0.2}
    alphaScaleBiasClamp {bInterpResult=true, interpSpeedDecreasing=8, interpSpeedIncreasing=8}
    boneToModify {boneName=dyn_scarf_02}
    rotation {pitch=-2}
   }
   n67 = "Transform (Modify) Bone - Bone: dyn_tie_02"!(ComponentPose=@n66, Alpha=1, Translation="0,0,0", Rotation="-5,0,0") -> (Pose) {
    rotationMode=BMM_Additive
    rotationSpace=BCS_BoneSpace
    translationMode=BMM_Additive
    translationSpace=BCS_BoneSpace
    alphaBoolBlend {blendInTime=0.2, blendOption=Linear, blendOutTime=0.2}
    alphaScaleBiasClamp {bInterpResult=true, interpSpeedDecreasing=8, interpSpeedIncreasing=8}
    boneToModify {boneName=dyn_tie_02}
    rotation {pitch=-5}
   }
   n68 = "Transform (Modify) Bone - Bone: dyn_scarf_01"!(ComponentPose=@n67, Alpha=1, Translation="0,0,0", Rotation="-0.452,-1.714,0.015") -> (Pose) {
    rotationMode=BMM_Additive
    rotationSpace=BCS_BoneSpace
    translationMode=BMM_Additive
    translationSpace=BCS_BoneSpace
    alphaBoolBlend {blendInTime=0.2, blendOption=Linear, blendOutTime=0.2}
    alphaScaleBiasClamp {bInterpResult=true, interpSpeedDecreasing=8, interpSpeedIncreasing=8}
    boneToModify {boneName=dyn_scarf_01}
    rotation=<rot -0.452 -1.714 0.015>
   }
   n69 = Component To Local(ComponentPose=@n68)
   n70 = "Transform (Modify) Bone - Bone: base_tie_01"(
    ComponentPose=Local To Component(LocalPose=Use cached pose 'SavePonytailPose_L'())
    Alpha=1
    Translation="0,0,0"
    Rotation="10,0,0"
    Scale="1,1,1"
   ) {
    rotationMode=BMM_Additive
    rotationSpace=BCS_BoneSpace
    translationMode=BMM_Additive
    translationSpace=BCS_BoneSpace
    alphaBoolBlend {blendInTime=0.2, blendOption=Linear, blendOutTime=0.2}
    alphaScaleBiasClamp {bInterpResult=true, interpSpeedDecreasing=8, interpSpeedIncreasing=8}
    boneToModify {boneName=base_tie_01}
    rotation {pitch=10}
   }
   n71 = "Transform (Modify) Bone - Bone: base_tie_01"(ComponentPose=@n70, Alpha=1, Translation="0,0,0", Rotation="20,0,0", Scale="1,1,1") {
    rotationMode=BMM_Additive
    rotationSpace=BCS_BoneSpace
    translationMode=BMM_Additive
    translationSpace=BCS_BoneSpace
    alphaBoolBlend {blendInTime=0.2, blendOption=Linear, blendOutTime=0.2}
    alphaScaleBiasClamp {bInterpResult=true, interpSpeedDecreasing=8, interpSpeedIncreasing=8}
    boneToModify {boneName=base_tie_01}
    rotation {pitch=20}
   }
   n72 = RigidBody(ComponentPose=@n71, OverrideWorldGravity=Get Override World Gravity()) {
    bOverrideWorldGravity=true
    overridePhysicsAsset=&/F_MED_FlossYawn_BackPack_Crouch
    simulationSpace=BaseBoneSpace
    alphaBoolBlend {blendOption=Linear}
    alphaScaleBiasClamp {bInterpResult=true, interpSpeedIncreasing=12}
    baseBoneRef {boneName=spine_05}
    componentAppliedLinearAccClamp=<3000 3000 3000>
    overrideWorldGravity {z=-980}
    simSpaceSettings {
     dampingAlpha=0
     maxAngularAcceleration=32
     maxAngularVelocity=8
     maxLinearAcceleration=1600
     maxLinearVelocity=400
     velocityScaleZ=0.25
     worldAlpha=1
     externalLinearDragV=<0.3 0.3 0.3>
    }
   }
   n73 = "Transform (Modify) Bone - Bone: dyn_scarfBack_04"!(ComponentPose=@n72, Alpha:float=1, Translation="0,0,0", Rotation="-6.756,0.011,-0.008", Scale="1,1,1") -> (Pose) {
    rotationMode=BMM_Additive
    rotationSpace=BCS_BoneSpace
    translationMode=BMM_Additive
    translationSpace=BCS_BoneSpace
    alphaBoolBlend {blendInTime=0.2, blendOption=Linear, blendOutTime=0.2}
    alphaScaleBiasClamp {bInterpResult=true, interpSpeedDecreasing=8, interpSpeedIncreasing=12}
    boneToModify {boneName=dyn_scarfBack_04}
    rotation=<rot -6.756 0.011 -0.008>
   }
   n74 = "Transform (Modify) Bone - Bone: dyn_scarfBack_05"!(ComponentPose=@n73, bAlphaBoolEnabled:bool=True, Translation="0,0,0", Rotation="7.687,0,0", Scale:Vector="1,1,1") -> (Pose) {
    alphaInputType=Bool
    rotationMode=BMM_Additive
    rotationSpace=BCS_BoneSpace
    translationMode=BMM_Additive
    translationSpace=BCS_BoneSpace
    alphaBoolBlend {blendInTime=0.2, blendOption=Linear, blendOutTime=0.2}
    alphaScaleBiasClamp {bInterpResult=true, interpSpeedDecreasing=8, interpSpeedIncreasing=12}
    boneToModify {boneName=dyn_scarfBack_05}
    rotation {pitch=7.687}
   }
   n75 = Component To Local(ComponentPose=@n74)
   n76 = Get bIsCrouching()
   n77 = "Transform (Modify) Bone - Bone: dyn_scarfBack_01"(
    ComponentPose=Local To Component(LocalPose=Blend Poses by bool(BlendPose_0=@n75, BlendPose_1=@n69, bActiveValue=@n76))
    Alpha=1
    Translation="0,0,0"
    Rotation="-5,0,0"
    Scale="1,1,1"
   ) {
    rotationMode=BMM_Additive
    rotationSpace=BCS_BoneSpace
    translationMode=BMM_Additive
    translationSpace=BCS_BoneSpace
    alphaScaleBiasClamp {bInterpResult=true, interpSpeedDecreasing=5, interpSpeedIncreasing=5}
    boneToModify {boneName=dyn_scarfBack_01}
    rotation {pitch=-5}
   }
   n78 = "Transform (Modify) Bone - Bone: dyn_scarfBack_03"!(ComponentPose=@n77, Alpha:float=1, Translation="0,0,0", Rotation="-5,0,0", Scale="1,1,1") -> (Pose) {
    rotationMode=BMM_Additive
    rotationSpace=BCS_BoneSpace
    translationMode=BMM_Additive
    translationSpace=BCS_BoneSpace
    alphaScaleBiasClamp {bInterpResult=true, interpSpeedDecreasing=5, interpSpeedIncreasing=5}
    boneToModify {boneName=dyn_scarfBack_03}
    rotation {pitch=-5}
   }
   n79 = "Transform (Modify) Bone - Bone: dyn_scarfBack_05"!(ComponentPose=@n78, Alpha:float=1, Translation="0,0,0", Rotation="-7,0,0", Scale="1,1,1") -> (Pose) {
    rotationMode=BMM_Additive
    rotationSpace=BCS_BoneSpace
    translationMode=BMM_Additive
    translationSpace=BCS_BoneSpace
    alphaScaleBiasClamp {bInterpResult=true, interpSpeedDecreasing=5, interpSpeedIncreasing=5}
    boneToModify {boneName=dyn_scarfBack_05}
    rotation {pitch=-7}
   }
   n80 = "Transform (Modify) Bone - Bone: dyn_scarfBack_06"(ComponentPose=@n79, Alpha=1, Translation="0,0,0", Rotation="0,0,0", Scale="1,1,1") {
    rotationMode=BMM_Additive
    rotationSpace=BCS_BoneSpace
    translationMode=BMM_Additive
    translationSpace=BCS_BoneSpace
    alphaScaleBiasClamp {bInterpResult=true, interpSpeedDecreasing=5, interpSpeedIncreasing=5}
    boneToModify {boneName=dyn_scarfBack_06}
   }
   n81 = "Transform (Modify) Bone - Bone: base_scarfBack"(ComponentPose=@n80, Alpha=0, Translation="0,0,0", Rotation="12,0,0", Scale="1,1,1") {
    rotationMode=BMM_Additive
    rotationSpace=BCS_BoneSpace
    translationMode=BMM_Additive
    translationSpace=BCS_BoneSpace
    alphaScaleBiasClamp {bInterpResult=true, interpSpeedDecreasing=8, interpSpeedIncreasing=12}
    boneToModify {boneName=base_scarfBack}
    rotation {pitch=12}
   }
   n82 = "Transform (Modify) Bone - Bone: base_scarf_02"(ComponentPose=@n81, Alpha=0, Translation="0,0,0", Rotation="13,0,0", Scale="1,1,1") {
    rotationMode=BMM_Additive
    rotationSpace=BCS_BoneSpace
    translationMode=BMM_Additive
    translationSpace=BCS_BoneSpace
    alphaScaleBiasClamp {bInterpResult=true, interpSpeedDecreasing=8, interpSpeedIncreasing=12}
    boneToModify {boneName=base_scarf_02}
    rotation {pitch=13}
   }
   n83 = "Constraint - base_scarf_02 (P)"!(ComponentPose=@n82, bAlphaBoolEnabled:bool=True, ConstraintWeights_0=1) -> (Pose) {
    alphaInputType=Bool
    alphaBoolBlend {blendInTime=0.5, blendOutTime=0.6}
    boneToModify {boneName=base_scarf_02}
    constraintSetup [{targetBone {boneName=spine_05}}]
    constraintWeights [1]
   }
   n84 = "Constraint - dyn_scarf_02 (P)"!(ComponentPose=@n83, bAlphaBoolEnabled:bool=True, ConstraintWeights_0=1) -> (Pose) {
    alphaInputType=Bool
    alphaBoolBlend {blendInTime=0.5, blendOutTime=0.6}
    boneToModify {boneName=dyn_scarf_02}
    constraintSetup [{targetBone {boneName=spine_05}}]
    constraintWeights [1]
   }
   n85 = "Constraint - dyn_tie_01 (P)"!(ComponentPose=@n84, bAlphaBoolEnabled:bool=True, ConstraintWeights_0=1) -> (Pose) {
    alphaInputType=Bool
    alphaBoolBlend {blendInTime=0.5, blendOutTime=0.6}
    boneToModify {boneName=dyn_tie_01}
    constraintSetup [{targetBone {boneName=spine_05}}]
    constraintWeights [1]
   }
   n86 = "Constraint - dyn_tie_02 (P)"!(ComponentPose=@n85, bAlphaBoolEnabled:bool=True, ConstraintWeights_0=1) -> (Pose) {
    alphaInputType=Bool
    alphaBoolBlend {blendInTime=0.5, blendOutTime=0.6}
    boneToModify {boneName=dyn_tie_02}
    constraintSetup [{targetBone {boneName=spine_05}}]
    constraintWeights [1]
   }
   n87 = "Constraint - dyn_tie_03 (P)"!(ComponentPose=@n86, bAlphaBoolEnabled:bool=True, ConstraintWeights_0=1) -> (Pose) {
    alphaInputType=Bool
    alphaBoolBlend {blendInTime=0.5, blendOutTime=0.6}
    boneToModify {boneName=dyn_tie_03}
    constraintSetup [{targetBone {boneName=spine_05}}]
    constraintWeights [1]
   }
   n88 = "Constraint - dyn_tie_04 (P)"!(ComponentPose=@n87, bAlphaBoolEnabled:bool=True, ConstraintWeights_0=1) -> (Pose) {
    alphaInputType=Bool
    alphaBoolBlend {blendInTime=0.5, blendOutTime=0.6}
    boneToModify {boneName=dyn_tie_04}
    constraintSetup [{targetBone {boneName=spine_05}}]
    constraintWeights [1]
   }
   n89 = "Constraint - dyn_tie_05 (R)"!(ComponentPose=@n88, bAlphaBoolEnabled:bool=True, ConstraintWeights_0=1) -> (Pose) {
    alphaInputType=Bool
    alphaBoolBlend {blendInTime=0.5, blendOutTime=0.6}
    boneToModify {boneName=dyn_tie_05}
    constraintSetup [{transformType=Rotation, targetBone {boneName=dyn_skirt_fr_c_01}}]
    constraintWeights [1]
   }
   n90 = "Constraint - dyn_scarfBack_02 (P)"(ComponentPose=@n89, Alpha=0.6, ConstraintWeights_0=1) {
    boneToModify {boneName=dyn_scarfBack_02}
    constraintSetup [{perAxis {bX=false, bY=false}, targetBone {boneName=spine_05}}]
    constraintWeights [1]
   }
   n91 = "Constraint - dyn_scarfBack_03 (P)"(ComponentPose=@n90, Alpha=0.6, ConstraintWeights_0=1) {
    boneToModify {boneName=dyn_scarfBack_03}
    constraintSetup [{perAxis {bX=false, bY=false}, targetBone {boneName=spine_04}}]
    constraintWeights [1]
   }
   n92 = "Constraint - dyn_scarfBack_04 (P)"(ComponentPose=@n91, Alpha=0.6, ConstraintWeights_0=1) {
    boneToModify {boneName=dyn_scarfBack_04}
    constraintSetup [{perAxis {bX=false, bY=false}, targetBone {boneName=spine_03}}]
    constraintWeights [1]
   }
   n93 = "Constraint - dyn_scarfBack_05 (P)"(ComponentPose=@n92, Alpha=0.6, ConstraintWeights_0=1) {
    boneToModify {boneName=dyn_scarfBack_05}
    constraintSetup [{perAxis {bX=false, bY=false}, targetBone {boneName=spine_02}}]
    constraintWeights [1]
   }
   n94 = "Transform (Modify) Bone - Bone: dyn_scarfBack_03"(ComponentPose=@n93, Alpha=1, Translation="0,0,0", Rotation="-8.5,0,0", Scale="1,1,1") {rotationMode=BMM_Additive, rotationSpace=BCS_BoneSpace, boneToModify {boneName=dyn_scarfBack_03}, rotation {pitch=-8.5}}
   n95 = "Transform (Modify) Bone - Bone: dyn_scarfBack_04"(ComponentPose=@n94, Alpha=1, Translation="0,0,0", Rotation="-1.644,0,0", Scale="1,1,1") {rotationMode=BMM_Additive, rotationSpace=BCS_BoneSpace, boneToModify {boneName=dyn_scarfBack_04}, rotation {pitch=-1.644}}
   n96 = "Transform (Modify) Bone - Bone: dyn_scarfBack_05"(ComponentPose=@n95, Alpha=1, Translation="0,0,0", Rotation="12.395,-0.032,-0.145", Scale="1,1,1") {rotationMode=BMM_Additive, rotationSpace=BCS_BoneSpace, boneToModify {boneName=dyn_scarfBack_05}, rotation=<rot 12.395 -0.032 -0.145>}
   n97 = "Transform (Modify) Bone - Bone: base_scarf_02"!(ComponentPose=@n96, bAlphaBoolEnabled:bool=True, Translation="0.524,0.000,2.353", Rotation="0,0,0", Scale="1,1,1") -> (Pose) {
    alphaInputType=Bool
    rotationMode=BMM_Additive
    rotationSpace=BCS_BoneSpace
    translationMode=BMM_Additive
    translationSpace=BCS_BoneSpace
    alphaBoolBlend {blendInTime=0.5, blendOutTime=0.5}
    boneToModify {boneName=base_scarf_02}
    translation {x=0.524, z=2.353}
   }
   n98 = RigidBody(ComponentPose=Local To Component(LocalPose=Use cached pose 'SavePonytailPose_L'()), OverrideWorldGravity=Get Override World Gravity()) {
    bOverrideWorldGravity=true
    overridePhysicsAsset=&F_MED_FlossYawn_BP1
    simulationSpace=BaseBoneSpace
    alphaBoolBlend {blendOption=Linear}
    alphaScaleBiasClamp {bInterpResult=true, interpSpeedIncreasing=12}
    baseBoneRef {boneName=spine_05}
    componentAppliedLinearAccClamp=<3000 3000 3000>
    overrideWorldGravity {z=-980}
    simSpaceSettings {
     dampingAlpha=0
     maxAngularAcceleration=32
     maxAngularVelocity=8
     maxLinearAcceleration=1600
     maxLinearVelocity=400
     velocityScaleZ=0.25
     worldAlpha=1
     externalLinearDragV=<0.3 0.3 0.3>
    }
   }
   n99 = "Transform (Modify) Bone - Bone: dyn_scarfBack_03"!(ComponentPose=@n98, Alpha=1, Translation="0,0,0", Rotation="-0.000,-6.223,0.000") -> (Pose) {
    rotationMode=BMM_Additive
    rotationSpace=BCS_BoneSpace
    translationMode=BMM_Additive
    translationSpace=BCS_BoneSpace
    alphaBoolBlend {blendInTime=0.2, blendOption=Linear, blendOutTime=0.2}
    alphaScaleBiasClamp {bInterpResult=true, interpSpeedDecreasing=8, interpSpeedIncreasing=8}
    boneToModify {boneName=dyn_scarfBack_03}
    rotation=<rot 0 -6.223 0>
   }
   n100 = "Transform (Modify) Bone - Bone: dyn_scarf_02"(ComponentPose=@n99, Alpha=1, Translation="0,0,0", Rotation="0.146,-0.000,-0.011") {
    rotationMode=BMM_Additive
    rotationSpace=BCS_BoneSpace
    translationMode=BMM_Additive
    translationSpace=BCS_BoneSpace
    alphaBoolBlend {blendInTime=0.2, blendOption=Linear, blendOutTime=0.2}
    alphaScaleBiasClamp {bInterpResult=true, interpSpeedDecreasing=8, interpSpeedIncreasing=8}
    boneToModify {boneName=dyn_scarf_02}
    rotation=<rot 0.146 0 -0.011>
   }
   n101 = "Transform (Modify) Bone - Bone: dyn_tie_02"(ComponentPose=@n100, Alpha=1, Translation="0,0,0", Rotation="-5,0,0") {
    rotationMode=BMM_Additive
    rotationSpace=BCS_BoneSpace
    translationMode=BMM_Additive
    translationSpace=BCS_BoneSpace
    alphaBoolBlend {blendInTime=0.2, blendOption=Linear, blendOutTime=0.2}
    alphaScaleBiasClamp {bInterpResult=true, interpSpeedDecreasing=8, interpSpeedIncreasing=8}
    boneToModify {boneName=dyn_tie_02}
    rotation {pitch=-5}
   }
   n102 = RigidBody(ComponentPose=Local To Component(LocalPose=Use cached pose 'SavePonytailPose_L'()), OverrideWorldGravity=Get Override World Gravity()) {
    bOverrideWorldGravity=true
    overridePhysicsAsset=&/F_MED_FlossYawn_main
    simulationSpace=BaseBoneSpace
    alphaBoolBlend {blendOption=Linear}
    alphaScaleBiasClamp {bInterpResult=true, interpSpeedIncreasing=12}
    baseBoneRef {boneName=spine_05}
    componentAppliedLinearAccClamp=<3000 3000 3000>
    overrideWorldGravity {z=-980}
    simSpaceSettings {
     dampingAlpha=0
     maxAngularAcceleration=32
     maxAngularVelocity=8
     maxLinearAcceleration=1600
     maxLinearVelocity=400
     velocityScaleZ=0.25
     worldAlpha=1
     externalLinearDragV=<0.3 0.3 0.3>
    }
   }
   n103 = "Transform (Modify) Bone - Bone: base_tie_02"!(ComponentPose=@n102, bAlphaBoolEnabled:bool=True, Rotation="3,0,0") -> (Pose) {
    alphaInputType=Bool
    rotationMode=BMM_Additive
    rotationSpace=BCS_BoneSpace
    translationMode=BMM_Additive
    translationSpace=BCS_BoneSpace
    alphaBoolBlend {blendInTime=0.1, blendOutTime=0.1}
    boneToModify {boneName=base_tie_02}
    rotation {pitch=3}
    translation {z=-0.5}
   }
   n104 = "Transform (Modify) Bone - Bone: dyn_scarf_01"!(ComponentPose=@n103, bAlphaBoolEnabled:bool=True, Rotation="1,0,0") -> (Pose) {
    alphaInputType=Bool
    rotationMode=BMM_Additive
    rotationSpace=BCS_BoneSpace
    translationMode=BMM_Additive
    translationSpace=BCS_BoneSpace
    alphaBoolBlend {blendInTime=0.2, blendOutTime=0.2}
    boneToModify {boneName=dyn_scarf_01}
    rotation {pitch=1}
    translation=<0 0 0.8>
   }
   n105 = Blend Poses by bool(
    BlendPose_0=Component To Local(ComponentPose=@n101)
    BlendPose_1=Component To Local(ComponentPose=@n104)
    bActiveValue=Get Bool_FEIdle() -> (Bool_FEIdle:bool)
   )
   n106 = "Transform (Modify) Bone - Bone: base_tie_01"(
    ComponentPose=Local To Component(LocalPose=Use cached pose 'SavePonytailPose_L'())
    Alpha=1
    Translation="0,0,0"
    Rotation="10,0,0"
    Scale="1,1,1"
   ) {
    rotationMode=BMM_Additive
    rotationSpace=BCS_BoneSpace
    translationMode=BMM_Additive
    translationSpace=BCS_BoneSpace
    alphaBoolBlend {blendInTime=0.2, blendOption=Linear, blendOutTime=0.2}
    alphaScaleBiasClamp {bInterpResult=true, interpSpeedDecreasing=8, interpSpeedIncreasing=8}
    boneToModify {boneName=base_tie_01}
    rotation {pitch=10}
   }
   n107 = RigidBody(ComponentPose=@n106, OverrideWorldGravity=Get Override World Gravity()) {
    bOverrideWorldGravity=true
    overridePhysicsAsset=&/F_MED_FlossYawn_Crouch_Physics
    simulationSpace=BaseBoneSpace
    alphaBoolBlend {blendOption=Linear}
    alphaScaleBiasClamp {bInterpResult=true, interpSpeedIncreasing=12}
    baseBoneRef {boneName=spine_05}
    componentAppliedLinearAccClamp=<3000 3000 3000>
    overrideWorldGravity {z=-980}
    simSpaceSettings {
     dampingAlpha=0
     maxAngularAcceleration=32
     maxAngularVelocity=8
     maxLinearAcceleration=1600
     maxLinearVelocity=400
     velocityScaleZ=0.25
     worldAlpha=1
     externalLinearDragV=<0.3 0.3 0.3>
    }
   }
   n108 = Component To Local(ComponentPose=@n107)
   n109 = Get bIsCrouching()
   n110 = "Transform (Modify) Bone - Bone: dyn_scarf_04"!(
    ComponentPose=Local To Component(LocalPose=Blend Poses by bool(BlendPose_0=@n108, BlendPose_1=@n105, bActiveValue=@n109))
    bAlphaBoolEnabled:bool=True
    Translation="0,0,0"
    Rotation="10,0,0"
    Scale="1,1,1"
   ) -> (Pose) {
    alphaInputType=Bool
    rotationMode=BMM_Additive
    rotationSpace=BCS_BoneSpace
    translationMode=BMM_Additive
    translationSpace=BCS_BoneSpace
    boneToModify {boneName=dyn_scarf_04}
    rotation {pitch=10}
   }
   n111 = "Transform (Modify) Bone - Bone: dyn_scarf_05"!(ComponentPose=@n110, bAlphaBoolEnabled:bool=True, Translation="0,0,1", Rotation="31.198,-2.474,0.984", Scale="1,1,1") -> (Pose) {
    alphaInputType=Bool
    rotationMode=BMM_Additive
    rotationSpace=BCS_BoneSpace
    translationMode=BMM_Additive
    translationSpace=BCS_BoneSpace
    boneToModify {boneName=dyn_scarf_05}
    rotation=<rot 31.198 -2.474 0.984>
    translation {z=1}
   }
   n112 = "Transform (Modify) Bone - Bone: base_scarfBack"(ComponentPose=@n111, Alpha=1, Translation="0,0,0.5", Rotation="0,0,0", Scale="1,1,1") {
    rotationSpace=BCS_BoneSpace
    translationMode=BMM_Additive
    translationSpace=BCS_BoneSpace
    boneToModify {boneName=base_scarfBack}
    translation {z=0.5}
   }
   n113 = Anim Dynamics - Bone: dyn_scarf_07(ComponentPose=@n112, Alpha=1, ExternalForce="0,0,0") {
    angularDampingOverride=1
    angularSpringConstant=5000
    bAngularSpring=true
    bOverrideAngularDamping=true
    bUseGravityOverride=true
    gravityScale=0.35
    simulationSpace=RootRelative
    boundBone {boneName=dyn_scarf_07}
    componentLinearAccScale=<1 1 1>
    componentLinearVelScale=<1 1 1>
    gravityOverride {z=-980}
    physicsBodyDefinitions [{
      boundBone {boneName=dyn_scarf_07}
      boxExtents=<20 30 30>
      constraintSetup {angularLimitsMax=<0 15 0>, angularLimitsMin=<0 -15 0>, angularTarget=<1 0 0>}
      localJointOffset=<4 0 0>
     }]
    simSpaceSettings {maxAngularAcceleration=100, maxAngularVelocity=10, simSpaceAngularAlpha=0.3}
   }
   n114 = Anim Dynamics - Bone: dyn_scarfBack_08(ComponentPose=@n113, Alpha=1, ExternalForce="0,0,0") {
    angularDampingOverride=1
    angularSpringConstant=5000
    bAngularSpring=true
    bOverrideAngularDamping=true
    bUseGravityOverride=true
    gravityScale=0.35
    simulationSpace=RootRelative
    boundBone {boneName=dyn_scarfBack_08}
    componentLinearAccScale=<1 1 1>
    componentLinearVelScale=<1 1 1>
    gravityOverride {z=-980}
    physicsBodyDefinitions [{
      boundBone {boneName=dyn_scarfBack_08}
      boxExtents=<20 30 30>
      constraintSetup {angularLimitsMax=<0 15 0>, angularLimitsMin=<0 -15 0>, angularTarget=<1 0 0>}
      localJointOffset=<4 0 0>
     }]
    simSpaceSettings {maxAngularAcceleration=100, maxAngularVelocity=10, simSpaceAngularAlpha=0.3}
   }
   n115 = "Transform (Modify) Bone - Bone: base_scarfBack"(ComponentPose=@n114, Alpha=1, Translation="0,0,0", Rotation="12,0,0", Scale="1,1,1") {
    rotationMode=BMM_Additive
    rotationSpace=BCS_BoneSpace
    translationMode=BMM_Additive
    translationSpace=BCS_BoneSpace
    alphaScaleBiasClamp {bInterpResult=true, interpSpeedDecreasing=8, interpSpeedIncreasing=12}
    boneToModify {boneName=base_scarfBack}
    rotation {pitch=12}
   }
   n116 = "Constraint - base_scarf_02 (P)"(ComponentPose=@n115, bAlphaBoolEnabled=True, ConstraintWeights_0=1) {
    alphaInputType=Bool
    alphaBoolBlend {blendInTime=0.5, blendOutTime=0.6}
    boneToModify {boneName=base_scarf_02}
    constraintSetup [{targetBone {boneName=spine_05}}]
    constraintWeights [1]
   }
   n117 = "Constraint - dyn_scarf_02 (P)"(ComponentPose=@n116, bAlphaBoolEnabled=True, ConstraintWeights_0=1) {
    alphaInputType=Bool
    alphaBoolBlend {blendInTime=0.5, blendOutTime=0.6}
    boneToModify {boneName=dyn_scarf_02}
    constraintSetup [{targetBone {boneName=spine_05}}]
    constraintWeights [1]
   }
   n118 = "Constraint - dyn_tie_01 (P)"(ComponentPose=@n117, bAlphaBoolEnabled=True, ConstraintWeights_0=1) {
    alphaInputType=Bool
    alphaBoolBlend {blendInTime=0.5, blendOutTime=0.6}
    boneToModify {boneName=dyn_tie_01}
    constraintSetup [{targetBone {boneName=spine_05}}]
    constraintWeights [1]
   }
   n119 = "Constraint - dyn_tie_02 (P)"(ComponentPose=@n118, bAlphaBoolEnabled=True, ConstraintWeights_0=1) {
    alphaInputType=Bool
    alphaBoolBlend {blendInTime=0.5, blendOutTime=0.6}
    boneToModify {boneName=dyn_tie_02}
    constraintSetup [{targetBone {boneName=spine_05}}]
    constraintWeights [1]
   }
   n120 = "Constraint - dyn_tie_03 (R)"!(ComponentPose=@n119, bAlphaBoolEnabled:bool=True, ConstraintWeights_0=1) -> (Pose) {
    alphaInputType=Bool
    alphaBoolBlend {blendInTime=0.5, blendOutTime=0.6}
    boneToModify {boneName=dyn_tie_03}
    constraintSetup [{transformType=Rotation, targetBone {boneName=dyn_skirt_fr_c_01}}]
    constraintWeights [1]
   }
   n121 = "Constraint - dyn_tie_04 (R)"!(ComponentPose=@n120, bAlphaBoolEnabled:bool=True, ConstraintWeights_0=1) -> (Pose) {
    alphaInputType=Bool
    alphaBoolBlend {blendInTime=0.5, blendOutTime=0.6}
    boneToModify {boneName=dyn_tie_04}
    constraintSetup [{transformType=Rotation, targetBone {boneName=dyn_skirt_fr_c_02}}]
    constraintWeights [1]
   }
   n122 = "Constraint - dyn_tie_05 (R)"(ComponentPose=@n121, bAlphaBoolEnabled=True, ConstraintWeights_0=1) {
    alphaInputType=Bool
    alphaBoolBlend {blendInTime=0.5, blendOutTime=0.6}
    boneToModify {boneName=dyn_tie_05}
    constraintSetup [{transformType=Rotation, targetBone {boneName=dyn_skirt_fr_c_01}}]
    constraintWeights [1]
   }
   n123 = "Transform (Modify) Bone - Bone: base_scarf_02"(ComponentPose=@n122, bAlphaBoolEnabled=True, Translation="0.524,0.000,2.353", Rotation="0,0,0", Scale="1,1,1") {
    alphaInputType=Bool
    rotationMode=BMM_Additive
    rotationSpace=BCS_BoneSpace
    translationMode=BMM_Additive
    translationSpace=BCS_BoneSpace
    alphaBoolBlend {blendInTime=0.5, blendOutTime=0.5}
    boneToModify {boneName=base_scarf_02}
    translation {x=0.524, z=2.353}
   }
   Save cached pose 'RigidBodyCache'(Pose=Blend Poses by bool(
     BlendPose_0=Component To Local(ComponentPose=@n97)
     BlendPose_1=Component To Local(ComponentPose=@n123)
     bActiveValue=Get isWearingPack() -> (isWearingPack:bool)
    )) {cacheName=RigidBodyCache}
   n124 = "Transform (Modify) Bone - Bone: dyn_tie_03"!(
    ComponentPose=Local To Component(LocalPose=Use cached pose 'Dynamics Cache'())
    bAlphaBoolEnabled:bool=True
    Translation="0,0,0"
    Rotation="-1.283,0.073,-0.119"
    Scale="1,1,1"
   ) -> (Pose) {
    alphaInputType=Bool
    rotationMode=BMM_Additive
    rotationSpace=BCS_BoneSpace
    translationMode=BMM_Additive
    translationSpace=BCS_BoneSpace
    alphaBoolBlend {blendInTime=0.2, blendOption=Linear, blendOutTime=0.5}
    alphaScaleBiasClamp {bInterpResult=true, interpSpeedDecreasing=5, interpSpeedIncreasing=5}
    boneToModify {boneName=dyn_tie_03}
    rotation=<rot -1.283 0.073 -0.119>
   }
   n125 = "Transform (Modify) Bone - Bone: base_strap_fr_l"(ComponentPose=@n124, Alpha=1, Translation="0,0,0", Rotation="-11.086,-0.000,-0.000", Scale="1,1,1") {
    rotationMode=BMM_Additive
    rotationSpace=BCS_BoneSpace
    translationMode=BMM_Additive
    translationSpace=BCS_BoneSpace
    alphaBoolBlend {blendInTime=0.2, blendOption=Linear, blendOutTime=0.5}
    alphaScaleBiasClamp {bInterpResult=true, interpSpeedDecreasing=5, interpSpeedIncreasing=5}
    boneToModify {boneName=base_strap_fr_l}
    rotation {pitch=-11.086, roll=0}
   }
   n126 = "Transform (Modify) Bone - Bone: sleeves_low_l"!(ComponentPose=@n125, bAlphaBoolEnabled:bool=True, Translation="-2,0,0", Rotation="0,0,0", Scale="1,1,1") -> (Pose) {
    alphaInputType=Bool
    rotationMode=BMM_Additive
    rotationSpace=BCS_BoneSpace
    translationMode=BMM_Additive
    translationSpace=BCS_BoneSpace
    alphaBoolBlend {blendInTime=0.2, blendOption=Linear, blendOutTime=0.2}
    alphaScaleBiasClamp {bInterpResult=true, interpSpeedDecreasing=5, interpSpeedIncreasing=5}
    boneToModify {boneName=sleeves_low_l}
    translation {x=-2}
   }
   n127 = "Transform (Modify) Bone - Bone: dyn_wrist_sleeve_l"!(ComponentPose=@n126, bAlphaBoolEnabled:bool=True, Translation="-2,0,0", Rotation="0,0,0", Scale="1,1,1") -> (Pose) {
    alphaInputType=Bool
    rotationMode=BMM_Additive
    rotationSpace=BCS_BoneSpace
    translationMode=BMM_Additive
    translationSpace=BCS_BoneSpace
    alphaBoolBlend {blendInTime=0.2, blendOption=Linear, blendOutTime=0.2}
    alphaScaleBiasClamp {bInterpResult=true, interpSpeedDecreasing=5, interpSpeedIncreasing=5}
    boneToModify {boneName=dyn_wrist_sleeve_l}
    translation {x=-2}
   }
   n128 = "Transform (Modify) Bone - Bone: sleeves_low_r"!(ComponentPose=@n127, bAlphaBoolEnabled:bool=True, Translation="3.439,0.000,0.000", Rotation="0,0,0", Scale="1,1,1") -> (Pose) {
    alphaInputType=Bool
    rotationMode=BMM_Additive
    rotationSpace=BCS_BoneSpace
    translationMode=BMM_Additive
    translationSpace=BCS_BoneSpace
    alphaBoolBlend {blendInTime=0.2, blendOption=Linear, blendOutTime=0.2}
    alphaScaleBiasClamp {bInterpResult=true, interpSpeedDecreasing=5, interpSpeedIncreasing=5}
    boneToModify {boneName=sleeves_low_r}
    translation {x=3.439}
   }
   n129 = "Transform (Modify) Bone - Bone: sleeves_low_r"(ComponentPose=@n128, bAlphaBoolEnabled=True, Translation="2,0,0", Rotation="0,0,0", Scale="1,1,1") {
    alphaInputType=Bool
    rotationMode=BMM_Additive
    rotationSpace=BCS_BoneSpace
    translationMode=BMM_Additive
    translationSpace=BCS_BoneSpace
    alphaBoolBlend {blendInTime=0.2, blendOption=Linear, blendOutTime=0.2}
    alphaScaleBiasClamp {bInterpResult=true, interpSpeedDecreasing=5, interpSpeedIncreasing=5}
    boneToModify {boneName=sleeves_low_r}
    translation {x=2}
   }
   n130 = "Transform (Modify) Bone - Bone: base_strap_bk_l"!(ComponentPose=@n129, bAlphaBoolEnabled:bool=True, Translation="0,0,0", Rotation="38,0.07,0.067", Scale="1,1,1") -> (Pose) {
    alphaInputType=Bool
    rotationMode=BMM_Additive
    rotationSpace=BCS_BoneSpace
    translationMode=BMM_Additive
    translationSpace=BCS_BoneSpace
    alphaBoolBlend {blendInTime=0.2, blendOption=Linear, blendOutTime=0.2}
    alphaScaleBiasClamp {bInterpResult=true, interpSpeedDecreasing=1, interpSpeedIncreasing=5}
    boneToModify {boneName=base_strap_bk_l}
    rotation=<rot 38 0.07 0.067>
   }
   n131 = "Transform (Modify) Bone - Bone: sleeves_low_r"(ComponentPose=@n130, bAlphaBoolEnabled=True, Translation="2.85,0,0", Rotation="0,0,0", Scale="1,1,1") {
    alphaInputType=Bool
    rotationMode=BMM_Additive
    rotationSpace=BCS_BoneSpace
    translationMode=BMM_Additive
    translationSpace=BCS_BoneSpace
    alphaBoolBlend {blendInTime=0.2, blendOption=Linear, blendOutTime=0.2}
    alphaScaleBiasClamp {bInterpResult=true, interpSpeedDecreasing=5, interpSpeedIncreasing=7}
    boneToModify {boneName=sleeves_low_r}
    translation {x=2.85}
   }
   n132 = Get Alpha_NotSkydiving()
   n133 = "Transform (Modify) Bone - Bone: MSTR_sleeve_r"(ComponentPose=@n131, Alpha=@n132, Translation="1.258,-0.174,0.000", Rotation="0.008,-4.703,0.004", Scale="1,1,1") {
    rotationMode=BMM_Additive
    rotationSpace=BCS_BoneSpace
    translationMode=BMM_Additive
    translationSpace=BCS_BoneSpace
    alphaBoolBlend {blendInTime=0.2, blendOption=Linear, blendOutTime=0.2}
    alphaScaleBiasClamp {bInterpResult=true, interpSpeedDecreasing=5, interpSpeedIncreasing=7}
    boneToModify {boneName=MSTR_sleeve_r}
    rotation=<rot 0.008 -4.703 0.004>
    translation=<1.258 -0.174>
   }
   n134 = "Transform (Modify) Bone - Bone: sleeves_low_l"!(ComponentPose=@n133, Alpha:float=@n132, Translation="-4.467,-0.000,0.000", Rotation="0,0,0", Scale="1,1,1") -> (Pose) {
    rotationMode=BMM_Additive
    rotationSpace=BCS_BoneSpace
    translationMode=BMM_Additive
    translationSpace=BCS_BoneSpace
    alphaBoolBlend {blendInTime=0.2, blendOption=Linear, blendOutTime=0.2}
    alphaScaleBiasClamp {bInterpResult=true, interpSpeedDecreasing=5, interpSpeedIncreasing=7}
    boneToModify {boneName=sleeves_low_l}
    translation {x=-4.467}
   }
   n135 = Look At - Bone: dyn_tie_05(ComponentPose=@n134, Alpha=Get Alpha_RB Disable() -> (Alpha_RB Disable:double)) {
    alphaBoolBlend {blendOption=Linear}
    boneToModify {boneName=dyn_tie_05}
    lookAtLocation=<-0.7256 0.876 -1.0449>
    lookAtTarget {boneReference {boneName=dyn_skirt_fr_c_03}}
    lookAt_Axis {axis=<1 0>}
   }
   n136 = "Transform (Modify) Bone - Bone: dyn_wrist_sleeve_l"(ComponentPose=@n135, bAlphaBoolEnabled=True, Translation="-2,0,0", Rotation="0,0,0", Scale="1,1,1") {
    alphaInputType=Bool
    rotationMode=BMM_Additive
    rotationSpace=BCS_BoneSpace
    translationMode=BMM_Additive
    translationSpace=BCS_BoneSpace
    boneToModify {boneName=dyn_wrist_sleeve_l}
    translation {x=-2}
   }
   n137 = "Transform (Modify) Bone - Bone: dyn_wrist_sleeve_l"(ComponentPose=@n136, bAlphaBoolEnabled=True, Translation="-3,0,0", Rotation="0,0,0", Scale="1,1,1") {
    alphaInputType=Bool
    rotationMode=BMM_Additive
    rotationSpace=BCS_BoneSpace
    translationMode=BMM_Additive
    translationSpace=BCS_BoneSpace
    boneToModify {boneName=dyn_wrist_sleeve_l}
    translation {x=-3}
   }
   n138 = "Transform (Modify) Bone - Bone: dyn_scarfBack_05"!(ComponentPose=@n137, bAlphaBoolEnabled:bool=True, Translation="0,0,0", Rotation="4,0,0", Scale="1,1,1") -> (Pose) {alphaInputType=Bool, rotationMode=BMM_Additive, rotationSpace=BCS_BoneSpace, boneToModify {boneName=dyn_scarfBack_05}, rotation {pitch=4}}
   n139 = "Transform (Modify) Bone - Bone: dyn_scarfBack_06"!(ComponentPose=@n138, bAlphaBoolEnabled:bool=True, Translation="0,0,0", Rotation="4,0,0", Scale="1,1,1") -> (Pose) {alphaInputType=Bool, rotationMode=BMM_Additive, rotationSpace=BCS_BoneSpace, boneToModify {boneName=dyn_scarfBack_06}, rotation {pitch=4}}
   n140 = "Transform (Modify) Bone - Bone: base_strap_fr_l"!(ComponentPose=@n139, bAlphaBoolEnabled:bool=True, Translation="0,0,0", Rotation="-3,0,-4", Scale="1,1,1") -> (Pose) {
    alphaInputType=Bool
    rotationMode=BMM_Additive
    rotationSpace=BCS_BoneSpace
    translationMode=BMM_Additive
    translationSpace=BCS_BoneSpace
    alphaBoolBlend {blendInTime=0.2, blendOption=Linear, blendOutTime=0.5}
    alphaScaleBiasClamp {bInterpResult=true, interpSpeedDecreasing=5, interpSpeedIncreasing=5}
    boneToModify {boneName=base_strap_fr_l}
    rotation {pitch=-3, roll=-4}
   }
   n141 = "Transform (Modify) Bone - Bone: dyn_strap_fr_l_04"!(ComponentPose=@n140, bAlphaBoolEnabled:bool=True, Translation="0,3,0", Rotation="0,0,15", Scale="1,1,1") -> (Pose) {
    alphaInputType=Bool
    rotationMode=BMM_Additive
    rotationSpace=BCS_BoneSpace
    translationMode=BMM_Additive
    translationSpace=BCS_BoneSpace
    alphaBoolBlend {blendInTime=0.2, blendOption=Linear, blendOutTime=0.5}
    alphaScaleBiasClamp {bInterpResult=true, interpSpeedDecreasing=5, interpSpeedIncreasing=5}
    boneToModify {boneName=dyn_strap_fr_l_04}
    rotation {roll=15}
    translation {y=3}
   }
   n142 = "Transform (Modify) Bone - Bone: base_strap_bk_l"(ComponentPose=@n141, bAlphaBoolEnabled=True, Translation="0,0,0", Rotation="-9.5,0,6", Scale="1,1,1") {
    alphaInputType=Bool
    rotationMode=BMM_Additive
    rotationSpace=BCS_BoneSpace
    translationMode=BMM_Additive
    translationSpace=BCS_BoneSpace
    alphaBoolBlend {blendInTime=0.2, blendOption=Linear, blendOutTime=0.5}
    alphaScaleBiasClamp {bInterpResult=true, interpSpeedDecreasing=5, interpSpeedIncreasing=5}
    boneToModify {boneName=base_strap_bk_l}
    rotation {pitch=-9.5, roll=6}
   }
   Save cached pose 'fixes_cache'(Pose=Component To Local(ComponentPose=@n142)) {cacheName=fixes_cache}
  ]}
 EventGraph {nodes [
   n1 = Event BlueprintUpdateAnimation()
   > Set Alpha_DeltaTimeX(Alpha_DeltaTimeX:double=@n1.DeltaTimeX) -> (Output_Get:double)
   > n2 = Sequence!(execute) -> (
    then_0
    then_1
    then_2:Exec
    then_3:Exec
    then_4:Exec
    then_5:Exec
    then_6:Exec
    then_7:Exec
    then_8:Exec
    then_9:Exec
    then_10:Exec
    then_11:Exec
    then_12:Exec
    then_13:Exec
    then_14:Exec
   )
   n3 = Get FAceACC_Mesh() -> (FAceACC_Mesh:Skeletal Mesh Component Object Reference)
   Is Valid(exec=@n2.then_11, InputObject=@n3)
   n4 = GetSocketBoneName!(self:Skinned Mesh Component Object Reference=@n3, InSocketName:Name=pelvis) -> (ReturnValue:Name=None)
   > n5 = Set isFaceAttached(isFaceAttached:bool="Not Equal Exactly (String)"(A="To String (Name)"(InName=@n4), B=None)) -> (Output_Get:bool)
   > Branch(Condition=@n5)
   n6 = Get FAceACC_Mesh()
   > Set dyn_ponytail_l_rot_01(dyn_ponytail_l_rot_01:Rotator=BreakTransform(InTransform=GetSocketTransform(self=@n6, InSocketName=dyn_ponytail_01_l, TransformSpace=RTS_ParentBoneSpace)).Rotation) -> (Output_Get:Rotator)
   n7 = Get FAceACC_Mesh()
   > Set dyn_ponytail_l_rot_02(dyn_ponytail_l_rot_02:Rotator=BreakTransform(InTransform=GetSocketTransform(self=@n7, InSocketName=dyn_ponytail_02_l, TransformSpace=RTS_ParentBoneSpace)).Rotation) -> (Output_Get:Rotator)
   n8 = Get FAceACC_Mesh()
   > Set dyn_ponytail_l_rot_03(dyn_ponytail_l_rot_03:Rotator=BreakTransform(InTransform=GetSocketTransform(self=@n8, InSocketName=dyn_ponytail_03_l, TransformSpace=RTS_ParentBoneSpace)).Rotation) -> (Output_Get:Rotator)
   n9 = Get FAceACC_Mesh()
   > Set dyn_ponytail_l_rot_04(dyn_ponytail_l_rot_04:Rotator=BreakTransform(InTransform=GetSocketTransform(self=@n9, InSocketName=dyn_ponytail_04_l, TransformSpace=RTS_ParentBoneSpace)).Rotation) -> (Output_Get:Rotator)
   n10 = Get FAceACC_Mesh()
   > Set dyn_ponytail_r_rot_01(dyn_ponytail_r_rot_01:Rotator=BreakTransform(InTransform=GetSocketTransform(self=@n10, InSocketName=dyn_ponytail_01_r, TransformSpace=RTS_ParentBoneSpace)).Rotation) -> (Output_Get:Rotator)
   n11 = Get FAceACC_Mesh()
   > Set dyn_ponytail_r_rot_02(dyn_ponytail_r_rot_02:Rotator=BreakTransform(InTransform=GetSocketTransform(self=@n11, InSocketName=dyn_ponytail_02_r, TransformSpace=RTS_ParentBoneSpace)).Rotation) -> (Output_Get:Rotator)
   n12 = Get FAceACC_Mesh()
   > Set dyn_ponytail_r_rot_03(dyn_ponytail_r_rot_03:Rotator=BreakTransform(InTransform=GetSocketTransform(self=@n12, InSocketName=dyn_ponytail_03_r, TransformSpace=RTS_ParentBoneSpace)).Rotation) -> (Output_Get:Rotator)
   n13 = Get FAceACC_Mesh()
   > Set dyn_ponytail_r_rot_04(dyn_ponytail_r_rot_04:Rotator=BreakTransform(InTransform=GetSocketTransform(self=@n13, InSocketName=dyn_ponytail_04_r, TransformSpace=RTS_ParentBoneSpace)).Rotation) -> (Output_Get:Rotator)
   n14 = Branch(execute=@n2.then_14, Condition=Get bIsSkydiveDiveMode())
   > Delay(Duration=0.3)
   > Set Pa_alpha(Pa_alpha:double=0) -> (Output_Get:double)
   > Set SkidiveVar(SkidiveVar:double=1) -> (Output_Get:double)
   > Set wrist_r_rot(wrist_r_rot:double=GetAlphaFromReferenceTransformDelta(SkeletalMeshComponent=GetOwnerSkeletalMeshComponent(), BoneName=hand_r, BaseName=spine_05, InputRange=(X=5,Y=40), TwistAxis="0,0,1")) -> (Output_Get:double)
   n15 = GetFortPlayerControllerFromActor!(Actor:Actor Object Reference=TryGetPawnOwner()) -> (ReturnValue:Fort Player Controller Object Reference)
   Set Pa_alpha(
    execute=@n14.else
    Pa_alpha=Select(
     Option 0:double=1
     Option 1:double=0
     Index=IsUsingFirstPersonCamera!(self:Fort Player Controller Object Reference=@n15) -> (ReturnValue:bool)
    ) -> (ReturnValue:double)
   )
   > Set SkidiveVar(SkidiveVar=0)
   n16 = Branch(execute=@n2.then_2, Condition=Get bIsPlayingMeleeAnim())
   > Set Alpha_Melee(Alpha_Melee:double=1) -> (Output_Get:double)
   Set Alpha_Melee(execute=@n16.else, Alpha_Melee=0)
   Set Alpha_Crouching(execute=@n2.then_1, Alpha_Crouching:double="To Float (Boolean)"(InBool=Get bIsCrouching())) -> (Output_Get:double)
   n17 = GetOwnerSkeletalMeshComponent()
   Set Test_DetectedRotAvocadoEmote(
    execute=@n2.then_8
    Test_DetectedRotAvocadoEmote:bool=Montage_IsPlaying(self=GetAnimInstance(self=@n17), Montage=/BRCosmetics/Animation/Game/MainPlayer/Emotes/AvocadoSeal/CMF/Emote_AvocadoSeal_CMF_M)
   ) -> (Output_Get:bool)
   n18 = GetOwnerSkeletalMeshComponent()
   n19 = BreakRotator(InRot=BreakTransform(InTransform=GetDeltaTransformFromRefPose(self=@n18, BoneName=neck_01, BaseName=pelvis)).Rotation)
   > n20 = Branch(Condition=AND Boolean!(A="float > float"(A=@n19, B=14), B=Get Test_DetectedRotAvocadoEmote() -> (Test_DetectedRotAvocadoEmote:bool), C:bool=Get bRigidBodyEnabled()) -> (ReturnValue))
   > Set Alpha_DetectedAvocadoEmote(Alpha_DetectedAvocadoEmote:double=1) -> (Output_Get:double)
   Set Alpha_DetectedAvocadoEmote(execute=@n20.else, Alpha_DetectedAvocadoEmote=0)
   n21 = AND Boolean!(A=NOT Boolean(A=Get bIsMoving2D()), B=Get bIsInFrontEnd(), C=NOT Boolean(A=Get bIsPlayingEmote()), D:bool=NOT Boolean(A=Get bIsSkyDiving())) -> (ReturnValue)
   Set FEIdle(execute=@n2.then_10, FEIdle:double="To Float (Boolean)"(InBool=@n21)) -> (Output_Get:double)
   > Set Bool_FEIdle(Bool_FEIdle:bool=@n21) -> (Output_Get:bool)
   > Set BPEmoting(BPEmoting:bool=AND Boolean!(A=Get isWearingPack(), B=Get bIsPlayingEmote()) -> (ReturnValue)) -> (Output_Get:bool)
   n22 = Get bIsPlayingEmote()
   > Set FENoEmote(FENoEmote:bool=AND Boolean(A=Get bIsInFrontEnd(), B=NOT Boolean(A=@n22))) -> (Output_Get:bool)
   n23 = NOT Boolean(A=Get bIsMoving2D())
   n24 = NOT Boolean(A=Get bIsCrouching())
   n25 = Get bIsOnGround() -> (bIsOnGround:bool)
   Set Alpha_StandIdle(execute=@n2.then_3, Alpha_StandIdle:double="To Float (Boolean)"(InBool=AND Boolean!(A=@n23, B=@n24, C:bool=@n25) -> (ReturnValue))) -> (Output_Get:double)
   n26 = Branch(execute=@n2.then_6, Condition=Get bRigidBodyEnabled())
   > Set Alpha_RB Disable(Alpha_RB Disable:double=0) -> (Output_Get:double)
   Set Alpha_RB Disable(execute=@n26.else, Alpha_RB Disable=1)
   Set Alpha_DBNO(execute=@n2.then_7, Alpha_DBNO:double="To Float (Boolean)"(InBool=Get bIsDBNO())) -> (Output_Get:double)
   n27 = AND Boolean!(A=Get bIsMoving2D(), B=Get bIsCrouching()) -> (ReturnValue)
   Set Alpha_CrouchingMoving2D(
    execute=@n2.then_0
    Alpha_CrouchingMoving2D:double=FInterpTo(
     Current=Get Alpha_CrouchingMoving2D()
     Target="To Float (Boolean)"(InBool=@n27)
     DeltaTime=Get Alpha_DeltaTimeX() -> (Alpha_DeltaTimeX:double)
     InterpSpeed=10
    )
   ) -> (Output_Get:double)
   n28 = NOT Boolean(A=Get bIsMoving2D())
   n29 = Get bIsCrouching()
   > Set Alpha_CrouchingIdle(Alpha_CrouchingIdle:double="To Float (Boolean)"(InBool=AND Boolean(A=@n29, B=@n28))) -> (Output_Get:double)
   n30 = Get bIsSliding()
   n31 = Get bIsKneeSliding()
   n32 = Branch(execute=@n2.then_4, Condition=AND Boolean!(A=NOT Boolean(A=@n31), B=NOT Boolean(A=@n30), C:bool=Get bIsCrouching()) -> (ReturnValue))
   > Set Crouching Bool(Crouching Bool:bool=true) -> (Output_Get:bool)
   Delay(execute=@n32.else, Duration=0.2)
   > Set Crouching Bool(Crouching Bool=false)
   n33 = Get bIsSkyDiving()
   n34 = Get bIsMoving2D()
   n35 = Get bIsPlayingEmote()
   n36 = GetOwnerSkeletalMeshComponent()
   n37 = BreakRotator(InRot=BreakTransform(InTransform=GetDeltaTransformFromRefPose(self=@n36, BoneName=thigh_l, BaseName=pelvis)).Rotation)
   n38 = Branch(
    execute=@n2.then_9
    Condition=AND Boolean!(A="float < float"(A=@n37.Yaw, B=-30), B=NOT Boolean(A=@n34), C=NOT Boolean(A=@n35), D:bool=NOT Boolean(A=@n33), E:bool=Get bIsInFrontEnd()) -> (ReturnValue)
   )
   > Set DetectedLegBackIdleAnim(DetectedLegBackIdleAnim:double=1) -> (Output_Get:double)
   Set DetectedLegBackIdleAnim(execute=@n38.else, DetectedLegBackIdleAnim=0)
   n39 = Branch(execute=@n2.then_5, Condition=Get bIsSkyDiving())
   > Set Alpha_NotSkydiving(Alpha_NotSkydiving:double=0) -> (Output_Get:double)
   Set Alpha_NotSkydiving(execute=@n39.else, Alpha_NotSkydiving=1)
   n40 = CheckForBackbling()
   Set isWearingPack(execute=@n2.then_12, isWearingPack:bool=OR Boolean!(A=@n40, B=@n40.bIsBackblingACape, C:bool=@n40.bIsBackblingLong) -> (ReturnValue)) -> (Output_Get:bool)
   > Set isWearingCape(isWearingCape=@n40.bIsBackblingLong)
   n41 = "Gravity Override State Function (Parameter Asset) (Nativized)"(
    execute=@n2.then_13
    CharacterPartAnimInstance=Self-Reference()
    GameTimeInSeconds=GetGameTimeInSeconds()
    Parameters=/BRCosmetics/Characters/Player/Female/Medium/Bodies/F_MED_FlossYawn/Meshes/Parts/M_MED_FlossYawn_Parameters
   )
   > Set Override World Gravity(Override World Gravity:Vector=@n41) -> (Output_Get:Vector)
   Event OnBeginPlayingEmote() {is=title K2Node_Event}
   > Set Alpha_PlayingEmote(Alpha_PlayingEmote:double=1) -> (Output_Get:double)
   Event OnEndPlayingEmote() {is=title K2Node_Event}
   > Set Alpha_PlayingEmote(Alpha_PlayingEmote=0)
   Event OnWeaponChanged()
   > n42 = Cast To FortPlayerAnimInstance(Object=GetAnimInstance(self=GetOwnerSkeletalMeshComponent()))
   > Set Selected_Pickaxe(Selected_Pickaxe:bool=false) -> (Output_Get:bool)
   > Set Selected_pistol(Selected_pistol:bool=false) -> (Output_Get:bool)
   > Set Selected_Shotgun(Selected_Shotgun:bool=false) -> (Output_Get:bool)
   > n43 = Switch on EFortWeaponCoreAnimation!(execute, Selection:EFortWeaponCoreAnimation Enum=Get WeaponCoreAnim(self=@n42)) -> (
    Melee:Exec
    Pistol:Exec
    Shotgun:Exec
    PaperBlueprint:Exec
    Rifle:Exec
    MeleeOneHand:Exec
    MachinePistol:Exec
    RocketLauncher:Exec
    GrenadeLauncher:Exec
    GoingCommando:Exec
    AssaultRifle:Exec
    TacticalShotgun:Exec
    SniperRifle:Exec
    TrapPlacement:Exec
    ShoulderLauncher:Exec
    AbilityDecoTool:Exec
    Crossbow:Exec
    C4:Exec
    RemoteControl:Exec
    DualWield:Exec
    AR_BullPup:Exec
    AR_ForwardGrip:Exec
    MedPackPaddles:Exec
    SMG_P90:Exec
    AR_DrumGun:Exec
    Consumable_Small:Exec
    Consumable_Large:Exec
    Balloon:Exec
    MountedTurret:Exec
    CreativeTool:Exec
    ExplosiveBow:Exec
    AshtonIndigo:Exec
    AshtonChicago:Exec
    MeleeDualWield:Exec
    Unarmed:Exec
   )
   > Set Selected_Pickaxe(Selected_Pickaxe=1)
   Set Selected_pistol(execute=@n43.Pistol, Selected_pistol=1)
   Set Selected_Shotgun(execute=@n43.Shotgun, Selected_Shotgun=true)
   n44 = Event OnCharacterCustomizationLoaded()
   > Set FAceACC_Mesh(FAceACC_Mesh:Skeletal Mesh Component Object Reference=Get Array Item with EFort Custom Part Type(Array=@n44.CustomMeshParts, PartType=Face)) -> (Output_Get:Skeletal Mesh Component Object Reference)
  ]}
 NewEventDispatcher {nodes [NewEventDispatcher() {is=title K2Node_FunctionEntry}]}
 NewEventDispatcher_0 {nodes [NewEventDispatcher_0() {is=title K2Node_FunctionEntry}]}
]
variables [
 Alpha_Crouching {type=number}
 Alpha_CrouchingIdle {type=number}
 Alpha_CrouchingMoving2D {type=number}
 Alpha_DBNO {type=number}
 Alpha_DeltaTimeX {type=number}
 Alpha_DetectedAvocadoEmote {type=number}
 Alpha_Jumping Not Moving2d {type=number}
 Alpha_Melee {type=number}
 Alpha_NotSkydiving {type=number}
 Alpha_PickaxeIdle {type=number}
 Alpha_PlayingEmote {type=number}
 Alpha_RB Disable {type=number}
 Alpha_sleeveL_FixWeapons {type=number}
 Alpha_StandIdle {type=number}
 Alpha_UnderwaterJump {type=number}
 base_ponytail_L_rot {type=Rotator}
 base_ponytail_L_tras {type=Vector}
 Bool_FEIdle {type=boolean}
 BPEmoting {type=boolean}
 BPorCape {type=boolean}
 crouch {type=Rotator}
 Crouching Bool {type=boolean}
 crouchingScarf {type=boolean}
 crouchWalking {type=number}
 DetectedLegBackIdleAnim {type=number}
 dyn_PadArm_2_Translation {type=Vector}
 dyn_ponytail_l_rot_01 {type=Rotator}
 dyn_ponytail_l_rot_02 {type=Rotator}
 dyn_ponytail_l_rot_03 {type=Rotator}
 dyn_ponytail_l_rot_04 {type=Rotator}
 dyn_ponytail_L_tras_01 {type=Vector}
 dyn_ponytail_L_tras_02 {type=Vector}
 dyn_ponytail_L_tras_03 {type=Vector}
 dyn_ponytail_r_rot_01 {type=Rotator}
 dyn_ponytail_r_rot_02 {type=Rotator}
 dyn_ponytail_r_rot_03 {type=Rotator}
 dyn_ponytail_r_rot_04 {type=Rotator}
 FAceACC_Mesh {type=SkeletalMeshComponent}
 FEIdle {type=number}
 FENoEmote {type=boolean}
 Gravity Override Params {
  type=GravityOverrideParamsStruct
  default {
   onGround_Moving {gravityOverride=<0 -100 -980>}
   skydive_FloatingDown {gravityOverride=<0 -1000 0>, windAmplitude=<100 -100 100>}
   skydive_GoingUp {gravityOverride=<0 -1000 -500>, windAmplitude=<100 -100 100>, windOffset=<0 -600 400>}
   swimming {jointName=root, gravityOverride=<0 -700 -100>}
  }
 }
 isFaceAttached {type=boolean}
 isWearingCape {type=boolean}
 isWearingPack {type=boolean}
 NewEventDispatcher {default=()}
 NewEventDispatcher_0 {default=()}
 Override World Gravity {type=Vector, default=<0 0 -980>}
 Pa_alpha {type=number}
 rotator {type=Rotator}
 Selected_Pickaxe {type=boolean}
 Selected_pistol {type=boolean}
 Selected_Shotgun {type=boolean}
 SkidiveVar {type=number}
 skydiveMode {type=number}
 Test_DetectedRotAvocadoEmote {type=boolean}
 wrist_r_rot {type=number}
]
