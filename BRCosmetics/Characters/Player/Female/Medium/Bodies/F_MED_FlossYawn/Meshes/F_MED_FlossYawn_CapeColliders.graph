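# Physics asset of collision capsules for the FlossYawn female medium player body, named as the colliders for its cape.
# Capsules cover the pelvis, spine, arms, hands and legs, plus non-body bones: an eight-bone back scarf chain, three rear skirt bones and a tie bone.
# Constraints link them into a tree rooted at the pelvis, with angular motion explicitly free on the spine, hips, ankles, scarf and skirt.
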PhysicsAsset /BRCosmetics/Characters/Player/Female/Medium/Bodies/F_MED_FlossYawn/Meshes/F_MED_FlossYawn_CapeColliders
bodies {
 calf_l [{shapeType=Capsule, length=28.6339, radius=11.064, center=<-27.6471 -0.0714 -0.9885>, rotation=<rot 0 79.183 -88.6776>}]
 calf_r [{shapeType=Capsule, length=28.6339, radius=11.064, center=<27.6471 0.0714 0.9885>, rotation=<rot 0 79.183 -88.6776>}]
 dyn_scarfBack_01 [{shapeType=Capsule, length=9.9354, radius=3.7919, shapeName=dyn_scarfBack_01_capsule, center=<2.1135 -0.1347 -0.0074>, rotation=<rot 90 0 0>}]
 dyn_scarfBack_02 [{shapeType=Capsule, length=14.1939, radius=5.2102, shapeName=dyn_scarfBack_02_capsule, center=<5.4505 -0.0248 0.1682>, rotation=<rot 90 0 0>}]
 dyn_scarfBack_03 [{shapeType=Capsule, length=13.4455, radius=5.4759, shapeName=dyn_scarfBack_03_capsule, center=<4.1628 -0.1686 -0.3159>, rotation=<rot 90 0 0>}]
 dyn_scarfBack_04 [{shapeType=Capsule, length=14.8919, radius=5.7391, shapeName=dyn_scarfBack_04_capsule, center=<7.0221 0.0916 -0.0594>, rotation=<rot 90 0 0>}]
 dyn_scarfBack_05 [{shapeType=Capsule, length=7.62, radius=6.0118, shapeName=dyn_scarfBack_05_capsule, center=<5.3571 -0.0776 0.0022>, rotation=<rot 90 90 0>}]
 dyn_scarfBack_06 [{
   shapeType=Capsule
   length=12.1759
   radius=5.8068
   shapeName=dyn_scarfBack_06_capsule
   center=<3.5033 -0.1901 -0.0059>
   rotation=<rot 90 90 0>
  }]
 dyn_scarfBack_07 [{shapeType=Capsule, length=12.1234, radius=6.476, shapeName=dyn_scarfBack_07_capsule, center=<2.7749 0.5995 -0.3916>, rotation=<rot 90 90 0>}]
 dyn_scarfBack_08 [{
   shapeType=Capsule
   length=11.9003
   radius=7.0287
   shapeName=dyn_scarfBack_08_capsule
   center=<5.2876 -2.3117 0.0986>
   rotation=<rot 90 -90 0>
  }]
 dyn_skirt_bk_c_03 [{shapeType=Capsule, length=9.8476, radius=4.5, shapeName=dyn_skirt_bk_c_03_capsule, center=<2.3777 -0.3266 0.5849>, rotation=<rot 0 0 90>}]
 dyn_skirt_bk_l_03 [{
   shapeType=Capsule
   length=8.276
   radius=5
   shapeName=dyn_skirt_bk_l_03_capsule
   center=<1.0731 -1.7666 -1.7705>
   rotation=<rot 0 171.1392 -61.9249>
  }]
 dyn_skirt_bk_r_03 [{
   shapeType=Capsule
   length=8.276
   radius=5
   shapeName=dyn_skirt_bk_l_03_capsule
   center=<1.6682 0.0226 -1.1544>
   rotation=<rot 3.9469 -177.0863 71.7262>
  }]
 dyn_tie_01 [{shapeType=Capsule, length=19.7787, radius=3.8069, center=<19.7986 -0.3025 -7.989>, rotation=<rot 0 -88.7144 -58.7146>}]
 foot_l [{shapeType=Capsule, length=10.5, radius=7, shapeName=foot_l_capsule, center=<-3.92 5.601 0>, rotation=<rot 0 6.7594 90>}]
 foot_r [{shapeType=Capsule, length=10.5, radius=7, shapeName=foot_l_capsule, center=<3.92 -5.601 0>, rotation=<rot 0 6.7594 -90>}]
 hand_l [{shapeType=Capsule, length=8.7464, radius=5.6478, center=<8.2519 -2.6605 -0.069>, rotation=<rot 0 81.071 -71.8704>}]
 hand_r [{shapeType=Capsule, length=8.7464, radius=5.6478, center=<-8.2519 2.6605 0.069>, rotation=<rot 0 81.071 -71.8704>}]
 lowerarm_l [{shapeType=Capsule, length=14.2095, radius=7.8957, center=<12.2213 0.1704 -1.4491>, rotation=<rot 0 -91.2928 -87.4337>}]
 lowerarm_r [{shapeType=Capsule, length=14.2095, radius=7.8957, center=<-12.2213 -0.1704 1.4491>, rotation=<rot 0 -91.2928 -87.4337>}]
 pelvis [{shapeType=Capsule, length=13.7459, radius=21.933, center=<-13.3639 0.1721 0.2841>, rotation=<rot 10.9044 -90 -1.326>}]
 spine_01 [{shapeType=Capsule, length=10.8777, radius=15.6616, shapeName=spine_01_capsule, center=<-2.2546 -3.2091 0>}]
 spine_02 [{shapeType=Capsule, length=14.9192, radius=12.5443, shapeName=spine_02_capsule, center=<13.119 0 0>, rotation=<rot 90 8.1479 0>}]
 thigh_l [{shapeType=Capsule, length=20.3977, radius=14.7105, center=<-27.582 -5.5485 1.5122>, rotation=<rot 0 -90.1171 93.1475>}]
 thigh_r [{shapeType=Capsule, length=20.3977, radius=14.7105, center=<27.582 5.5485 -1.5122>, rotation=<rot 0 -90.1171 -86.8525>}]
 upperarm_l [{shapeType=Capsule, length=13.6971, radius=5.8414, center=<13.0072 0.2824 -0.9398>, rotation=<rot 0 88.3156 -84.273>}]
 upperarm_r [{shapeType=Capsule, length=13.6967, radius=5.8411, center=<-13.0035 -0.2709 0.94>, rotation=<rot 0 88.5178 -84.275>}]
}
constraints [
 {bone1Name=upperarm_l, bone2Name=spine_02, swing1LimitDegrees=45, swing2LimitDegrees=45, twistLimitDegrees=45}
 {bone1Name=lowerarm_l, bone2Name=upperarm_l, swing1LimitDegrees=45, swing2LimitDegrees=45, twistLimitDegrees=45}
 {bone1Name=hand_l, bone2Name=lowerarm_l, swing1LimitDegrees=45, swing2LimitDegrees=45, twistLimitDegrees=45}
 {bone1Name=upperarm_r, bone2Name=spine_02, swing1LimitDegrees=45, swing2LimitDegrees=45, twistLimitDegrees=45}
 {bone1Name=lowerarm_r, bone2Name=upperarm_r, swing1LimitDegrees=45, swing2LimitDegrees=45, twistLimitDegrees=45}
 {bone1Name=hand_r, bone2Name=lowerarm_r, swing1LimitDegrees=45, swing2LimitDegrees=45, twistLimitDegrees=45}
 {bone1Name=dyn_tie_01, bone2Name=spine_02, swing1LimitDegrees=45, swing2LimitDegrees=45, twistLimitDegrees=45}
 {bone1Name=calf_l, bone2Name=thigh_l, swing1LimitDegrees=45, swing2LimitDegrees=45, twistLimitDegrees=45}
 {bone1Name=calf_r, bone2Name=thigh_r, swing1LimitDegrees=45, swing2LimitDegrees=45, twistLimitDegrees=45}
 {
  bone1Name=foot_l
  bone2Name=calf_l
  swing1LimitDegrees=45
  swing1Motion=Free
  swing2LimitDegrees=45
  swing2Motion=Free
  twistLimitDegrees=45
  twistMotion=Free
 }
 {
  bone1Name=foot_r
  bone2Name=calf_r
  swing1LimitDegrees=45
  swing1Motion=Free
  swing2LimitDegrees=45
  swing2Motion=Free
  twistLimitDegrees=45
  twistMotion=Free
 }
 {
  bone1Name=spine_02
  bone2Name=spine_01
  swing1LimitDegrees=45
  swing1Motion=Free
  swing2LimitDegrees=45
  swing2Motion=Free
  twistLimitDegrees=45
  twistMotion=Free
 }
 {
  bone1Name=spine_01
  bone2Name=pelvis
  swing1LimitDegrees=45
  swing1Motion=Free
  swing2LimitDegrees=45
  swing2Motion=Free
  twistLimitDegrees=45
  twistMotion=Free
 }
 {
  bone1Name=dyn_skirt_bk_c_03
  bone2Name=pelvis
  swing1LimitDegrees=45
  swing1Motion=Free
  swing2LimitDegrees=45
  swing2Motion=Free
  twistLimitDegrees=45
  twistMotion=Free
 }
 {
  bone1Name=dyn_skirt_bk_l_03
  bone2Name=pelvis
  swing1LimitDegrees=45
  swing1Motion=Free
  swing2LimitDegrees=45
  swing2Motion=Free
  twistLimitDegrees=45
  twistMotion=Free
 }
 {
  bone1Name=dyn_skirt_bk_r_03
  bone2Name=pelvis
  swing1LimitDegrees=45
  swing1Motion=Free
  swing2LimitDegrees=45
  swing2Motion=Free
  twistLimitDegrees=45
  twistMotion=Free
 }
 {
  bone1Name=thigh_l
  bone2Name=pelvis
  swing1LimitDegrees=45
  swing1Motion=Free
  swing2LimitDegrees=45
  swing2Motion=Free
  twistLimitDegrees=45
  twistMotion=Free
 }
 {
  bone1Name=thigh_r
  bone2Name=pelvis
  swing1LimitDegrees=45
  swing1Motion=Free
  swing2LimitDegrees=45
  swing2Motion=Free
  twistLimitDegrees=45
  twistMotion=Free
 }
 {
  bone1Name=dyn_scarfBack_05
  bone2Name=dyn_scarfBack_04
  swing1LimitDegrees=45
  swing1Motion=Free
  swing2LimitDegrees=45
  swing2Motion=Free
  twistLimitDegrees=45
  twistMotion=Free
 }
 {
  bone1Name=dyn_scarfBack_06
  bone2Name=dyn_scarfBack_05
  swing1LimitDegrees=45
  swing1Motion=Free
  swing2LimitDegrees=45
  swing2Motion=Free
  twistLimitDegrees=45
  twistMotion=Free
 }
 {
  bone1Name=dyn_scarfBack_07
  bone2Name=dyn_scarfBack_06
  swing1LimitDegrees=45
  swing1Motion=Free
  swing2LimitDegrees=45
  swing2Motion=Free
  twistLimitDegrees=45
  twistMotion=Free
 }
 {
  bone1Name=dyn_scarfBack_04
  bone2Name=dyn_scarfBack_03
  swing1LimitDegrees=45
  swing1Motion=Free
  swing2LimitDegrees=45
  swing2Motion=Free
  twistLimitDegrees=45
  twistMotion=Free
 }
 {
  bone1Name=dyn_scarfBack_03
  bone2Name=dyn_scarfBack_02
  swing1LimitDegrees=45
  swing1Motion=Free
  swing2LimitDegrees=45
  swing2Motion=Free
  twistLimitDegrees=45
  twistMotion=Free
 }
 {
  bone1Name=dyn_scarfBack_02
  bone2Name=dyn_scarfBack_01
  swing1LimitDegrees=45
  swing1Motion=Free
  swing2LimitDegrees=45
  swing2Motion=Free
  twistLimitDegrees=45
  twistMotion=Free
 }
 {
  bone1Name=dyn_scarfBack_01
  bone2Name=spine_02
  swing1LimitDegrees=45
  swing1Motion=Free
  swing2LimitDegrees=45
  swing2Motion=Free
  twistLimitDegrees=45
  twistMotion=Free
 }
 {
  bone1Name=dyn_scarfBack_08
  bone2Name=dyn_scarfBack_07
  swing1LimitDegrees=45
  swing1Motion=Free
  swing2LimitDegrees=45
  swing2Motion=Free
  twistLimitDegrees=45
  twistMotion=Free
 }
]
