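# Physics asset for the face accessory of the F_MED_JadeTowelGloss female character, named Cape but built on left and right ponytail bone chains.
# Each chain runs from a base on the head through six capsule segments, ending in two strands.
# Chain joints lock twist and have small swing limits that mostly widen down the chain, with differences between sides.
# Capsules on head, spine, pelvis, arms and calves form a body proxy with free angular constraints.
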PhysicsAsset /BRCosmetics/Characters/Player/Female/Medium/Bodies/F_MED_JadeTowelGloss/Meshes/Parts/F_MED_JadeTowelGloss_FaceAcc_Physics_Cape
bodies {
 base_ponytail_l [{shapeType=Capsule, length=1, radius=1, shapeName=base_ponytail_l_capsule, rotation=<rot 0 -90 0>}]
 base_ponytail_r [{shapeType=Capsule, length=1, radius=1, shapeName=base_ponytail_r_capsule, rotation=<rot 0 -90 0>}]
 calf_l [{shapeType=Capsule, length=26, radius=17, shapeName=calf_l_capsule, center=<-14.7061 -0.5088 -0.1658>, rotation=<rot 90 80.7734 84.0177>}]
 calf_r [{shapeType=Capsule, length=26, radius=17, shapeName=calf_l_capsule, center=<14.7061 0.5088 -0.1658>, rotation=<rot 90 80.7734 -95.9823>}]
 dyn_ponytail_01_l [{
   shapeType=Capsule
   length=8
   radius=4
   shapeName=dyn_ponytail_01_l_capsule
   center=<4.3551 -1.2084 1.3172>
   rotation=<rot 73.7525 164.4921 180>
  }]
 dyn_ponytail_01_r [{
   shapeType=Capsule
   length=8
   radius=4
   shapeName=dyn_ponytail_01_l_capsule
   center=<-4.3551 1.2084 -1.3172>
   rotation=<rot 73.7525 164.4921 0>
  }]
 dyn_ponytail_02_l [{shapeType=Capsule, length=7, radius=4, shapeName=dyn_ponytail_02_l_capsule, center=<-0.1177 -0.87 0>, rotation=<rot 90 0 7.6414>}]
 dyn_ponytail_02_r [{shapeType=Capsule, length=7, radius=4, shapeName=dyn_ponytail_02_l_capsule, center=<0.1177 0.87 0>, rotation=<rot 90 0 -172.3586>}]
 dyn_ponytail_03_l [{shapeType=Capsule, length=7, radius=7.5, shapeName=dyn_ponytail_03_l_capsule, center=<0.8746 0 -2.3>, rotation=<rot 90 0 0>}]
 dyn_ponytail_03_r [{shapeType=Capsule, length=7, radius=6, shapeName=dyn_ponytail_03_l_capsule, center=<-0.8746 0 2.3>, rotation=<rot 90 0 180>}]
 dyn_ponytail_04_l [{shapeType=Capsule, length=15, radius=9, shapeName=dyn_ponytail_04_l_capsule, center=<6.877 -0.87 -0.3>, rotation=<rot 90 0 8.6727>}]
 dyn_ponytail_04_r [{shapeType=Capsule, length=15, radius=8.5, shapeName=dyn_ponytail_04_l_capsule, center=<-6.877 0.87 0.3>, rotation=<rot 90 0 -171.3273>}]
 dyn_ponytail_05_l [{shapeType=Capsule, length=7.5, radius=7, shapeName=dyn_ponytail_05_l_capsule, center=<6.1224 -3 0.6>, rotation=<rot 90 0 0>}]
 dyn_ponytail_05_r [{shapeType=Capsule, length=7.5, radius=7, shapeName=dyn_ponytail_05_l_capsule, center=<-6.1224 3 -0.6>, rotation=<rot 90 0 180>}]
 dyn_ponytail_06_l [{shapeType=Capsule, length=13, radius=15.5, shapeName=dyn_ponytail_06_l_capsule, center=<10.9331 2.4288 2>, rotation=<rot 90 0 0>}]
 dyn_ponytail_06_r [{shapeType=Capsule, length=10, radius=13, shapeName=dyn_ponytail_06_l_capsule, center=<-10.028 -0.3219 -0.7648>, rotation=<rot 90 0 180>}]
 dyn_ponytail_strand_a_l [{shapeType=Capsule, length=4.7255, radius=2.5717, shapeName=dyn_ponytail_strand_a_l_capsule, center=<3.867 0 0>, rotation=<rot 90 0 0>}]
 dyn_ponytail_strand_a_r [{
   shapeType=Capsule
   length=8.2255
   radius=2.5717
   shapeName=dyn_ponytail_strand_a_l_capsule
   center=<-4.2199 -0.2 6.8>
   rotation=<rot 45.3639 0.0107 174.3447>
  }]
 dyn_ponytail_strand_b_l [{shapeType=Capsule, length=8.0737, radius=5.4978, shapeName=dyn_ponytail_strand_b_l_capsule, center=<4.3739 0 -1.9>, rotation=<rot 90 0 0>}]
 dyn_ponytail_strand_b_r [{
   shapeType=Capsule
   length=16.0737
   radius=5.4978
   shapeName=dyn_ponytail_strand_b_l_capsule
   center=<-7.6426 0 11.6>
   rotation=<rot 67.6583 0.0339 -179.9634>
  }]
 head [{shapeType=Capsule, length=2, radius=2, shapeName=head_capsule}]
 lowerarm_l [{shapeType=Capsule, length=35.51, radius=6, shapeName=lowerarm_l_capsule, center=<18.2404 0 0>, rotation=<rot 90 0 0>}]
 lowerarm_r [{shapeType=Capsule, length=35.51, radius=6, shapeName=lowerarm_l_capsule, center=<-18.2404 0 0>, rotation=<rot 90 0 180>}]
 pelvis [{shapeType=Capsule, length=30.5, radius=19, shapeName=pelvis_capsule, center=<0 -1.1903 -0.3019>}]
 spine_04 [{shapeType=Capsule, length=25, radius=18.5, shapeName=spine_04_capsule, center=<0 8.7048 0>}]
 spine_05 [
  {shapeType=Capsule, length=56.5, radius=17.5, shapeName=spine_05_capsule, center=<-32.7556 -5.9917 0>, rotation=<rot 90 180 175.1861>}
  {shapeType=Capsule, length=8, radius=17.5, shapeName=spine_05_capsule, center=<-32.8104 -8.2034 0>}
 ]
 upperarm_l [{shapeType=Capsule, length=19.005, radius=6, shapeName=upperarm_l_capsule, center=<12.4116 0 0>, rotation=<rot 90 0 0>}]
 upperarm_r [{shapeType=Capsule, length=19.005, radius=6, shapeName=upperarm_l_capsule, center=<-12.4116 0 0>, rotation=<rot 90 0 180>}]
}
constraints [
 {
  bone1Name=dyn_ponytail_06_l
  bone2Name=dyn_ponytail_05_l
  swing1LimitDegrees=25
  swing2LimitDegrees=25
  twistLimitDegrees=45
  twistMotion=Locked
 }
 {
  bone1Name=dyn_ponytail_05_l
  bone2Name=dyn_ponytail_04_l
  swing1LimitDegrees=25
  swing2LimitDegrees=25
  twistLimitDegrees=45
  twistMotion=Locked
 }
 {
  bone1Name=dyn_ponytail_04_l
  bone2Name=dyn_ponytail_03_l
  swing1LimitDegrees=25
  swing2LimitDegrees=25
  twistLimitDegrees=45
  twistMotion=Locked
 }
 {
  bone1Name=dyn_ponytail_03_l
  bone2Name=dyn_ponytail_02_l
  swing1LimitDegrees=15
  swing2LimitDegrees=10
  twistLimitDegrees=45
  twistMotion=Locked
 }
 {
  bone1Name=dyn_ponytail_02_l
  bone2Name=dyn_ponytail_01_l
  swing1LimitDegrees=12
  swing2LimitDegrees=8
  twistLimitDegrees=45
  twistMotion=Locked
 }
 {
  bone1Name=dyn_ponytail_strand_a_l
  bone2Name=dyn_ponytail_06_l
  swing1LimitDegrees=10
  swing2LimitDegrees=10
  twistLimitDegrees=45
  twistMotion=Locked
 }
 {
  bone1Name=dyn_ponytail_strand_b_l
  bone2Name=dyn_ponytail_06_l
  swing1LimitDegrees=10
  swing2LimitDegrees=10
  twistLimitDegrees=45
  twistMotion=Locked
 }
 {bone1Name=dyn_ponytail_01_l, bone2Name=base_ponytail_l, swing1LimitDegrees=7, swing2LimitDegrees=7, twistLimitDegrees=45, twistMotion=Locked}
 {
  bone1Name=dyn_ponytail_02_r
  bone2Name=dyn_ponytail_01_r
  swing1LimitDegrees=25
  swing2LimitDegrees=5
  twistLimitDegrees=45
  twistMotion=Locked
 }
 {
  bone1Name=dyn_ponytail_03_r
  bone2Name=dyn_ponytail_02_r
  swing1LimitDegrees=15
  swing2LimitDegrees=5
  twistLimitDegrees=45
  twistMotion=Locked
 }
 {
  bone1Name=dyn_ponytail_05_r
  bone2Name=dyn_ponytail_04_r
  swing1LimitDegrees=25
  swing2LimitDegrees=25
  twistLimitDegrees=45
  twistMotion=Locked
 }
 {
  bone1Name=dyn_ponytail_strand_b_r
  bone2Name=dyn_ponytail_06_r
  swing1LimitDegrees=10
  swing2LimitDegrees=10
  twistLimitDegrees=45
  twistMotion=Locked
 }
 {
  bone1Name=dyn_ponytail_strand_a_r
  bone2Name=dyn_ponytail_06_r
  swing1LimitDegrees=10
  swing2LimitDegrees=10
  twistLimitDegrees=45
  twistMotion=Locked
 }
 {
  bone1Name=dyn_ponytail_06_r
  bone2Name=dyn_ponytail_05_r
  swing1LimitDegrees=25
  swing2LimitDegrees=25
  twistLimitDegrees=45
  twistMotion=Locked
 }
 {
  bone1Name=head
  bone2Name=spine_05
  swing1LimitDegrees=45
  swing1Motion=Free
  swing2LimitDegrees=45
  swing2Motion=Free
  twistLimitDegrees=45
  twistMotion=Free
 }
 {
  bone1Name=spine_04
  bone2Name=pelvis
  swing1LimitDegrees=45
  swing1Motion=Free
  swing2LimitDegrees=45
  swing2Motion=Free
  twistLimitDegrees=45
  twistMotion=Free
 }
 {
  bone1Name=upperarm_l
  bone2Name=spine_05
  swing1LimitDegrees=45
  swing1Motion=Free
  swing2LimitDegrees=45
  swing2Motion=Free
  twistLimitDegrees=45
  twistMotion=Free
 }
 {
  bone1Name=dyn_ponytail_04_r
  bone2Name=dyn_ponytail_03_r
  swing1LimitDegrees=25
  swing2LimitDegrees=25
  twistLimitDegrees=45
  twistMotion=Locked
 }
 {
  bone1Name=upperarm_r
  bone2Name=spine_05
  swing1LimitDegrees=45
  swing1Motion=Free
  swing2LimitDegrees=45
  swing2Motion=Free
  twistLimitDegrees=45
  twistMotion=Free
 }
 {
  bone1Name=lowerarm_r
  bone2Name=upperarm_r
  swing1LimitDegrees=45
  swing1Motion=Free
  swing2LimitDegrees=45
  swing2Motion=Free
  twistLimitDegrees=45
  twistMotion=Free
 }
 {
  bone1Name=base_ponytail_l
  bone2Name=head
  swing1LimitDegrees=45
  swing1Motion=Free
  swing2LimitDegrees=45
  swing2Motion=Free
  twistLimitDegrees=45
  twistMotion=Free
 }
 {
  bone1Name=base_ponytail_r
  bone2Name=head
  swing1LimitDegrees=45
  swing1Motion=Free
  swing2LimitDegrees=45
  swing2Motion=Free
  twistLimitDegrees=45
  twistMotion=Free
 }
 {
  bone1Name=dyn_ponytail_01_r
  bone2Name=base_ponytail_r
  swing1LimitDegrees=15
  swing2LimitDegrees=7
  swing2Motion=Locked
  twistLimitDegrees=45
  twistMotion=Locked
 }
 {
  bone1Name=lowerarm_l
  bone2Name=upperarm_l
  swing1LimitDegrees=45
  swing1Motion=Free
  swing2LimitDegrees=45
  swing2Motion=Free
  twistLimitDegrees=45
  twistMotion=Free
 }
 {
  bone1Name=calf_l
  bone2Name=pelvis
  swing1LimitDegrees=45
  swing1Motion=Free
  swing2LimitDegrees=45
  swing2Motion=Free
  twistLimitDegrees=45
  twistMotion=Free
 }
 {
  bone1Name=calf_r
  bone2Name=pelvis
  swing1LimitDegrees=45
  swing1Motion=Free
  swing2LimitDegrees=45
  swing2Motion=Free
  twistLimitDegrees=45
  twistMotion=Free
 }
 {
  bone1Name=spine_05
  bone2Name=spine_04
  swing1LimitDegrees=45
  swing1Motion=Free
  swing2LimitDegrees=45
  swing2Motion=Free
  twistLimitDegrees=45
  twistMotion=Free
 }
]
