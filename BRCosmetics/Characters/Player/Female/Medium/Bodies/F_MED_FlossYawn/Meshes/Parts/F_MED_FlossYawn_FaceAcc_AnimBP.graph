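# Animation Blueprint for the F_MED_FlossYawn female cosmetic's face-accessory skeleton, driving secondary motion on hair and ponytails.
# Bool blends (cape, backpack, DBNO, Alpha_RBEnable) pick between RigidBody simulations with separate physics assets and constrained fallback poses.
# Additive bone offsets and constraints to head, neck and spine correct ponytail placement per state.
# A chain of spring-based AnimDynamics nodes follows for front and side hair.
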AnimBlueprint /BRCosmetics/Characters/Player/Female/Medium/Bodies/F_MED_FlossYawn/Meshes/Parts/F_MED_FlossYawn_FaceAcc_AnimBP
target_skeleton=&/F_MED_FlossYawn_FaceAcc_Skeleton
graphs [
 AnimGraph {nodes [
   n1 = "Transform (Modify) Bone - Bone: dyn_ponytail_02_r"(Alpha=1, Translation="0,0,0", Rotation="0,-60,0", Scale="1,1,1") {rotationMode=BMM_Additive, rotationSpace=BCS_BoneSpace, boneToModify {boneName=dyn_ponytail_02_r}, rotation {yaw=-60}}
   "Transform (Modify) Bone - Bone: dyn_ponytail_02_l"(ComponentPose=@n1, Alpha=1, Translation="0,0,0", Rotation="0,-60,0", Scale="1,1,1") {rotationMode=BMM_Additive, rotationSpace=BCS_BoneSpace, boneToModify {boneName=dyn_ponytail_02_l}, rotation {yaw=-60}}
   n2 = "Transform (Modify) Bone - Bone: dyn_ponytail_01_l"(Alpha=1, Translation="0,0,0", Rotation="6,-8,0", Scale="1,1,1") {rotationMode=BMM_Additive, rotationSpace=BCS_BoneSpace, boneToModify {boneName=dyn_ponytail_01_l}, rotation {pitch=6, yaw=-8}}
   "Transform (Modify) Bone - Bone: dyn_ponytail_01_r"(ComponentPose=@n2, Alpha=1, Translation="0,0,0", Rotation="4,-8,0", Scale="1,1,1") {rotationMode=BMM_Additive, rotationSpace=BCS_BoneSpace, boneToModify {boneName=dyn_ponytail_01_r}, rotation {pitch=4, yaw=-8}}
   RigidBody!(ComponentPose, Alpha, OverrideWorldGravity:Vector="0,0,-980", ExternalForce) -> (Pose) {
    bOverrideWorldGravity=true
    overridePhysicsAsset=&/F_MED_FLossYawn_FaceAcc_PhysicsHairFront
    simulationSpace=BaseBoneSpace
    alphaBoolBlend {blendOption=Linear}
    baseBoneRef {boneName=head}
    componentAppliedLinearAccClamp=<3000 3000 3000>
    componentLinearAccScale=<0.3 0.3 0.3>
    componentLinearVelScale=<0.3 0.3 0.3>
    overrideWorldGravity {z=-980}
    simSpaceSettings {worldAlpha=0.5}
   }
   n3 = "Transform (Modify) Bone - Bone: base_ponytail_l"(
    ComponentPose=Local To Component(LocalPose=Use cached pose 'Fix Ponytails Cache'())
    Alpha=1
    Translation="0,0,0"
    Rotation="0,-20,0"
    Scale="1,1,1"
   ) {
    rotationMode=BMM_Additive
    rotationSpace=BCS_BoneSpace
    translationMode=BMM_Additive
    translationSpace=BCS_BoneSpace
    alphaBoolBlend {blendInTime=0.2, blendOption=Linear, blendOutTime=0.2}
    alphaScaleBiasClamp {bInterpResult=true, interpSpeedDecreasing=7, interpSpeedIncreasing=7}
    boneToModify {boneName=base_ponytail_l}
    rotation {yaw=-20}
   }
   n4 = "Transform (Modify) Bone - Bone: base_ponytail_r"(ComponentPose=@n3, Alpha=1, Translation="0,0,0", Rotation="10,-20,0", Scale="1,1,1") {
    rotationMode=BMM_Additive
    rotationSpace=BCS_BoneSpace
    translationMode=BMM_Additive
    translationSpace=BCS_BoneSpace
    alphaBoolBlend {blendInTime=0.2, blendOption=Linear, blendOutTime=0.2}
    alphaScaleBiasClamp {bInterpResult=true, interpSpeedDecreasing=7, interpSpeedIncreasing=7}
    boneToModify {boneName=base_ponytail_r}
    rotation {pitch=10, yaw=-20}
   }
   n5 = RigidBody(ComponentPose=@n4, OverrideWorldGravity="0,0,-980") {
    bOverrideWorldGravity=true
    overridePhysicsAsset=/BRCosmetics/Characters/Player/Female/Medium/Bodies/F_MED_JadeTowelGloss/Meshes/Parts/F_MED_JadeTowelGloss_FaceAcc_Physics_Cape
    alphaBoolBlend {blendInTime=0.2, blendOption=Linear, blendOutTime=0.2}
    alphaScaleBiasClamp {bInterpResult=true, interpSpeedIncreasing=6}
    baseBoneRef {boneName=head}
    componentAppliedLinearAccClamp=<3000 3000 3000>
    componentLinearAccScale=<0.3 0.3 0.3>
    componentLinearVelScale=<0.3 0.3 0.3>
    overrideWorldGravity {z=-980}
    simSpaceSettings {velocityScaleZ=0.5, worldAlpha=0.3}
   }
   n6 = Blend Poses by bool(
    BlendPose_0=Component To Local(ComponentPose=@n5)
    BlendPose_1=Use cached pose 'Constrains RB disble Cache'()
    bActiveValue=Get Alpha_RBEnable() -> (Alpha_RBEnable:bool)
   ) {blendType=Linear}
   n7 = Get Alpha_NotEmoteBackpack() -> (Alpha_NotEmoteBackpack:double)
   n8 = "Constraint - dyn_ponytail_01_l (P)"(ComponentPose=Local To Component(LocalPose=Use cached pose 'Fix Ponytails Cache'()), Alpha=@n7, ConstraintWeights_0=1) {
    alphaBoolBlend {blendOption=Linear}
    boneToModify {boneName=dyn_ponytail_01_l}
    constraintSetup [{targetBone {boneName=head}}]
    constraintWeights [1]
   }
   n9 = "Constraint - dyn_ponytail_01_r (P)"(ComponentPose=@n8, Alpha=@n7, ConstraintWeights_0=1) {
    alphaBoolBlend {blendOption=Linear}
    boneToModify {boneName=dyn_ponytail_01_r}
    constraintSetup [{targetBone {boneName=head}}]
    constraintWeights [1]
   }
   n10 = "Constraint - dyn_ponytail_02_l (P)"(ComponentPose=@n9, Alpha=1, ConstraintWeights_0=1) {
    alphaBoolBlend {blendOption=Linear}
    alphaScaleBiasClamp {bInterpResult=true, interpSpeedDecreasing=5, interpSpeedIncreasing=5}
    boneToModify {boneName=dyn_ponytail_02_l}
    constraintSetup [{targetBone {boneName=neck_01}}]
    constraintWeights [1]
   }
   n11 = "Constraint - dyn_ponytail_02_r (P)"(ComponentPose=@n10, Alpha=1, ConstraintWeights_0=1) {
    alphaBoolBlend {blendOption=Linear}
    alphaScaleBiasClamp {bInterpResult=true, interpSpeedDecreasing=5, interpSpeedIncreasing=5}
    boneToModify {boneName=dyn_ponytail_02_r}
    constraintSetup [{targetBone {boneName=neck_01}}]
    constraintWeights [1]
   }
   n12 = Get Alpha_ConstrainBackpack() -> (Alpha_ConstrainBackpack:double)
   n13 = "Constraint - dyn_ponytail_03_l (P)"(ComponentPose=@n11, Alpha=@n12, ConstraintWeights_0=1) {
    alphaBoolBlend {blendOption=Linear}
    boneToModify {boneName=dyn_ponytail_03_l}
    constraintSetup [{targetBone {boneName=spine_05}}]
    constraintWeights [1]
   }
   n14 = "Constraint - dyn_ponytail_03_r (P)"(ComponentPose=@n13, Alpha=@n12, ConstraintWeights_0=1) {
    alphaBoolBlend {blendOption=Linear}
    boneToModify {boneName=dyn_ponytail_03_r}
    constraintSetup [{targetBone {boneName=spine_05}}]
    constraintWeights [1]
   }
   n15 = "Constraint - dyn_ponytail_04_l (P)"(ComponentPose=@n14, Alpha=1, ConstraintWeights_0=1) {
    alphaBoolBlend {blendOption=Linear}
    boneToModify {boneName=dyn_ponytail_04_l}
    constraintSetup [{targetBone {boneName=spine_04}}]
    constraintWeights [1]
   }
   n16 = "Constraint - dyn_ponytail_04_r (P)"(ComponentPose=@n15, Alpha=1, ConstraintWeights_0=1) {
    alphaBoolBlend {blendOption=Linear}
    boneToModify {boneName=dyn_ponytail_04_r}
    constraintSetup [{targetBone {boneName=spine_04}}]
    constraintWeights [1]
   }
   n17 = "Constraint - dyn_ponytail_05_l (P)"(ComponentPose=@n16, Alpha=1, ConstraintWeights_0=1) {
    alphaBoolBlend {blendOption=Linear}
    boneToModify {boneName=dyn_ponytail_05_l}
    constraintSetup [{targetBone {boneName=spine_03}}]
    constraintWeights [1]
   }
   n18 = "Constraint - dyn_ponytail_05_l (P)"(ComponentPose=@n17, Alpha=1, ConstraintWeights_0=1) {
    alphaBoolBlend {blendOption=Linear}
    boneToModify {boneName=dyn_ponytail_05_l}
    constraintSetup [{targetBone {boneName=spine_03}}]
    constraintWeights [1]
   }
   n19 = "Constraint - dyn_ponytail_06_l (P)"(ComponentPose=@n18, Alpha=1, ConstraintWeights_0=1) {
    alphaBoolBlend {blendOption=Linear}
    boneToModify {boneName=dyn_ponytail_06_l}
    constraintSetup [{targetBone {boneName=spine_01}}]
    constraintWeights [1]
   }
   n20 = "Constraint - dyn_ponytail_06_r (P)"(ComponentPose=@n19, Alpha=1, ConstraintWeights_0=1) {
    alphaBoolBlend {blendOption=Linear}
    boneToModify {boneName=dyn_ponytail_06_r}
    constraintSetup [{targetBone {boneName=spine_01}}]
    constraintWeights [1]
   }
   n21 = "Transform (Modify) Bone - Bone: base_ponytail_l"(ComponentPose=@n20, Alpha=1, Translation="0,0,0", Rotation="5,0,0", Scale="1,1,1") {
    rotationMode=BMM_Additive
    rotationSpace=BCS_BoneSpace
    translationMode=BMM_Additive
    translationSpace=BCS_BoneSpace
    alphaBoolBlend {blendInTime=0.2, blendOption=Linear, blendOutTime=0.2}
    alphaScaleBiasClamp {bInterpResult=true, interpSpeedDecreasing=7, interpSpeedIncreasing=7}
    boneToModify {boneName=base_ponytail_l}
    rotation {pitch=5}
   }
   n22 = "Transform (Modify) Bone - Bone: base_ponytail_r"(ComponentPose=@n21, Alpha=1, Translation="0,0,0", Rotation="5,-7,0", Scale="1,1,1") {
    rotationMode=BMM_Additive
    rotationSpace=BCS_BoneSpace
    translationMode=BMM_Additive
    translationSpace=BCS_BoneSpace
    alphaBoolBlend {blendInTime=0.2, blendOption=Linear, blendOutTime=0.2}
    alphaScaleBiasClamp {bInterpResult=true, interpSpeedDecreasing=7, interpSpeedIncreasing=7}
    boneToModify {boneName=base_ponytail_r}
    rotation {pitch=5, yaw=-7}
   }
   n23 = "Transform (Modify) Bone - Bone: base_ponytail_l"!(ComponentPose=@n22, bAlphaBoolEnabled:bool=True, Translation="0,0,0", Rotation="5,-15,0", Scale="1,1,1") -> (Pose) {
    alphaInputType=Bool
    rotationMode=BMM_Additive
    rotationSpace=BCS_BoneSpace
    translationMode=BMM_Additive
    translationSpace=BCS_BoneSpace
    alphaBoolBlend {blendInTime=0.2, blendOption=Linear, blendOutTime=0.2}
    alphaScaleBiasClamp {bInterpResult=true, interpSpeedDecreasing=7, interpSpeedIncreasing=7}
    boneToModify {boneName=base_ponytail_l}
    rotation {pitch=5, yaw=-15}
   }
   n24 = "Transform (Modify) Bone - Bone: base_ponytail_r"!(ComponentPose=@n23, bAlphaBoolEnabled:bool=True, Translation="0,0,0", Rotation="5,-10,0", Scale="1,1,1") -> (Pose) {
    alphaInputType=Bool
    rotationMode=BMM_Additive
    rotationSpace=BCS_BoneSpace
    translationMode=BMM_Additive
    translationSpace=BCS_BoneSpace
    alphaBoolBlend {blendInTime=0.2, blendOption=Linear, blendOutTime=0.2}
    alphaScaleBiasClamp {bInterpResult=true, interpSpeedDecreasing=7, interpSpeedIncreasing=7}
    boneToModify {boneName=base_ponytail_r}
    rotation {pitch=5, yaw=-10}
   }
   n25 = "Transform (Modify) Bone - Bone: base_ponytail_l"!(ComponentPose=@n24, Alpha:float=1, Translation="0,0,0", Rotation="5,-36.255,0", Scale="1,1,1") -> (Pose) {
    rotationMode=BMM_Additive
    rotationSpace=BCS_BoneSpace
    translationMode=BMM_Additive
    translationSpace=BCS_BoneSpace
    alphaBoolBlend {blendInTime=0.2, blendOption=Linear, blendOutTime=0.2}
    alphaScaleBiasClamp {bInterpResult=true, interpSpeedDecreasing=2, interpSpeedIncreasing=7}
    boneToModify {boneName=base_ponytail_l}
    rotation {pitch=5, yaw=-36.255}
   }
   n26 = "Transform (Modify) Bone - Bone: base_ponytail_r"!(ComponentPose=@n25, Alpha:float=1, Translation="0,0,0", Rotation="-0.000,-37.158,-0.000", Scale="1,1,1") -> (Pose) {
    rotationMode=BMM_Additive
    rotationSpace=BCS_BoneSpace
    translationMode=BMM_Additive
    translationSpace=BCS_BoneSpace
    alphaBoolBlend {blendInTime=0.2, blendOption=Linear, blendOutTime=0.2}
    alphaScaleBiasClamp {bInterpResult=true, interpSpeedDecreasing=2, interpSpeedIncreasing=7}
    boneToModify {boneName=base_ponytail_r}
    rotation=<rot 0 -37.158 0>
   }
   n27 = "Transform (Modify) Bone - Bone: dyn_ponytail_01_l"!(ComponentPose=@n26, bAlphaBoolEnabled:bool=True, Translation="0,0,0", Rotation="0,-25,0", Scale="1,1,1") -> (Pose) {
    alphaInputType=Bool
    rotationMode=BMM_Additive
    rotationSpace=BCS_BoneSpace
    translationMode=BMM_Additive
    translationSpace=BCS_BoneSpace
    alphaBoolBlend {blendInTime=0.2, blendOption=Linear, blendOutTime=0.2}
    alphaScaleBiasClamp {bInterpResult=true, interpSpeedDecreasing=2, interpSpeedIncreasing=7}
    boneToModify {boneName=dyn_ponytail_01_l}
    rotation {yaw=-25}
   }
   n28 = "Transform (Modify) Bone - Bone: base_ponytail_r"!(ComponentPose=@n27, bAlphaBoolEnabled:bool=True, Translation="0,0,0", Rotation="0,-25,0", Scale="1,1,1") -> (Pose) {
    alphaInputType=Bool
    rotationMode=BMM_Additive
    rotationSpace=BCS_BoneSpace
    translationMode=BMM_Additive
    translationSpace=BCS_BoneSpace
    alphaBoolBlend {blendInTime=0.2, blendOption=Linear, blendOutTime=0.2}
    alphaScaleBiasClamp {bInterpResult=true, interpSpeedDecreasing=2, interpSpeedIncreasing=7}
    boneToModify {boneName=base_ponytail_r}
    rotation {yaw=-25}
   }
   n29 = "Transform (Modify) Bone - Bone: dyn_ponytail_02_l"(ComponentPose=@n28, Alpha=1, Translation="0,0,0", Rotation="0,-30,0", Scale="1,1,1") {
    rotationMode=BMM_Additive
    rotationSpace=BCS_BoneSpace
    translationMode=BMM_Additive
    translationSpace=BCS_BoneSpace
    alphaBoolBlend {blendInTime=0.2, blendOption=Linear, blendOutTime=0.2}
    alphaScaleBiasClamp {bInterpResult=true, interpSpeedDecreasing=2, interpSpeedIncreasing=7}
    boneToModify {boneName=dyn_ponytail_02_l}
    rotation {yaw=-30}
   }
   n30 = "Transform (Modify) Bone - Bone: dyn_ponytail_01_l"!(ComponentPose=@n29, Alpha:float=1, Translation="-1.761,2.038,-4.159", Rotation="-4.087,-12.266,-25.129", Scale="1,1,1") -> (Pose) {
    rotationMode=BMM_Additive
    rotationSpace=BCS_BoneSpace
    translationMode=BMM_Additive
    translationSpace=BCS_BoneSpace
    alphaBoolBlend {blendInTime=0.2, blendOption=Linear, blendOutTime=0.2}
    alphaScaleBiasClamp {bInterpResult=true, interpSpeedDecreasing=7, interpSpeedIncreasing=7}
    boneToModify {boneName=dyn_ponytail_01_l}
    rotation=<rot -4.087 -12.266 -25.129>
    translation=<-1.761 2.038 -4.159>
   }
   n31 = "Transform (Modify) Bone - Bone: dyn_ponytail_02_l"(ComponentPose=@n30, Alpha=1, Translation="-5.693,-10.000,0.000", Rotation="0,0,0", Scale="1,1,1") {
    rotationMode=BMM_Additive
    rotationSpace=BCS_BoneSpace
    translationMode=BMM_Additive
    translationSpace=BCS_BoneSpace
    alphaBoolBlend {blendInTime=0.2, blendOption=Linear, blendOutTime=0.2}
    alphaScaleBiasClamp {bInterpResult=true, interpSpeedDecreasing=2, interpSpeedIncreasing=7}
    boneToModify {boneName=dyn_ponytail_02_l}
    translation=<-5.693 -10>
   }
   n32 = "Transform (Modify) Bone - Bone: base_ponytail_l"(ComponentPose=@n31, Alpha=1, Translation="0,0,0", Rotation="0,-10,0", Scale="1,1,1") {
    rotationMode=BMM_Additive
    rotationSpace=BCS_BoneSpace
    translationMode=BMM_Additive
    translationSpace=BCS_BoneSpace
    alphaBoolBlend {blendInTime=0.2, blendOption=Linear, blendOutTime=0.2}
    alphaScaleBiasClamp {bInterpResult=true, interpSpeedDecreasing=2, interpSpeedIncreasing=7}
    boneToModify {boneName=base_ponytail_l}
    rotation {yaw=-10}
   }
   n33 = "Transform (Modify) Bone - Bone: base_ponytail_l"(ComponentPose=@n32, Alpha=1, Translation="0,0,0", Rotation="12.905,-23.447,-1.947", Scale="1,1,1") {
    rotationMode=BMM_Additive
    rotationSpace=BCS_BoneSpace
    translationMode=BMM_Additive
    translationSpace=BCS_BoneSpace
    alphaBoolBlend {blendInTime=0.2, blendOption=Linear, blendOutTime=0.2}
    alphaScaleBiasClamp {bInterpResult=true, interpSpeedDecreasing=2, interpSpeedIncreasing=7}
    boneToModify {boneName=base_ponytail_l}
    rotation=<rot 12.905 -23.447 -1.947>
   }
   n34 = RigidBody(ComponentPose=@n33, OverrideWorldGravity="0,0,-980") {
    bOverrideWorldGravity=true
    overridePhysicsAsset=&/F_MED_FlossYawn_Physics_Backpack
    alphaBoolBlend {blendInTime=0.2, blendOption=Linear, blendOutTime=0.2}
    alphaScaleBiasClamp {bInterpResult=true, interpSpeedDecreasing=7, interpSpeedIncreasing=7}
    baseBoneRef {boneName=head}
    componentAppliedLinearAccClamp=<3000 3000 3000>
    componentLinearAccScale=<0.3 0.3 0.3>
    componentLinearVelScale=<0.3 0.3 0.3>
    overrideWorldGravity {z=-980}
    simSpaceSettings {velocityScaleZ=0, worldAlpha=0.2}
   }
   n35 = RigidBody(ComponentPose=@n34, OverrideWorldGravity="0,0,-980") {
    overridePhysicsAsset=&/F_MED_FlossYawn_FaceAcc_Physics_MeleeBP
    simulationSpace=BaseBoneSpace
    alphaBoolBlend {blendInTime=0.2, blendOption=Linear, blendOutTime=0.2}
    alphaScaleBiasClamp {bInterpResult=true, interpSpeedDecreasing=7, interpSpeedIncreasing=7}
    baseBoneRef {boneName=spine_03}
    componentAppliedLinearAccClamp=<3000 3000 3000>
    componentLinearAccScale=<0.3 0.3 0.3>
    componentLinearVelScale=<0.3 0.3 0.3>
    overrideWorldGravity {z=-980}
    simSpaceSettings {velocityScaleZ=0.5, worldAlpha=1}
   }
   n36 = RigidBody(ComponentPose=Local To Component(LocalPose=Use cached pose 'Fix Ponytails Cache'()), OverrideWorldGravity="0,0,-980") {
    bAllowPoseScalePropagation=false
    bOverrideWorldGravity=true
    overridePhysicsAsset=&/F_MED_flossYAwnFaceAcc_Physics
    simulationSpace=BaseBoneSpace
    alphaBoolBlend {blendOption=Linear}
    baseBoneRef {boneName=spine_03}
    componentAppliedLinearAccClamp=<3000 3000 3000>
    componentLinearAccScale=<0.3 0.3 0.3>
    componentLinearVelScale=<0.3 0.3 0.3>
    overrideWorldGravity {z=-980}
    simSpaceSettings {
     maxAngularAcceleration=1000
     maxAngularVelocity=3
     maxLinearAcceleration=700
     maxLinearVelocity=500
     velocityScaleZ=0
     worldAlpha=0.4
     externalLinearDragV=<0.2 0.2>
    }
   }
   n37 = "Constraint - dyn_ponytail_02_l (P)"(ComponentPose=Local To Component(LocalPose=Use cached pose 'Fix Ponytails Cache'()), Alpha=1, ConstraintWeights_0=1) {
    alphaBoolBlend {blendOption=Linear}
    boneToModify {boneName=dyn_ponytail_02_l}
    constraintSetup [{targetBone {boneName=neck_01}}]
    constraintWeights [1]
   }
   n38 = "Constraint - dyn_ponytail_02_r (P)"(ComponentPose=@n37, Alpha=1, ConstraintWeights_0=1) {
    alphaBoolBlend {blendOption=Linear}
    boneToModify {boneName=dyn_ponytail_02_r}
    constraintSetup [{targetBone {boneName=neck_01}}]
    constraintWeights [1]
   }
   n39 = "Constraint - dyn_ponytail_03_l (P)"(ComponentPose=@n38, Alpha=1, ConstraintWeights_0=1) {
    alphaBoolBlend {blendOption=Linear}
    boneToModify {boneName=dyn_ponytail_03_l}
    constraintSetup [{targetBone {boneName=spine_05}}]
    constraintWeights [1]
   }
   n40 = "Constraint - dyn_ponytail_03_r (P)"(ComponentPose=@n39, Alpha=1, ConstraintWeights_0=1) {
    alphaBoolBlend {blendOption=Linear}
    boneToModify {boneName=dyn_ponytail_03_r}
    constraintSetup [{targetBone {boneName=spine_05}}]
    constraintWeights [1]
   }
   n41 = "Transform (Modify) Bone - Bone: base_ponytail_r"(ComponentPose=@n40, bAlphaBoolEnabled=True, Translation="0,0,0", Rotation="0,-5,0", Scale="1,1,1") {
    alphaInputType=Bool
    rotationMode=BMM_Additive
    rotationSpace=BCS_BoneSpace
    translationMode=BMM_Additive
    translationSpace=BCS_BoneSpace
    alphaBoolBlend {blendInTime=0.2, blendOption=Linear, blendOutTime=0.2}
    alphaScaleBiasClamp {bInterpResult=true, interpSpeedDecreasing=7, interpSpeedIncreasing=7}
    boneToModify {boneName=base_ponytail_r}
    rotation {yaw=-5}
   }
   n42 = "Transform (Modify) Bone - Bone: base_ponytail_r"(ComponentPose=@n41, bAlphaBoolEnabled=True, Translation="0,0,0", Rotation="0,10,0", Scale="1,1,1") {
    alphaInputType=Bool
    rotationMode=BMM_Additive
    rotationSpace=BCS_BoneSpace
    translationMode=BMM_Additive
    translationSpace=BCS_BoneSpace
    alphaBoolBlend {blendInTime=0.2, blendOption=Linear, blendOutTime=0.2}
    alphaScaleBiasClamp {bInterpResult=true, interpSpeedDecreasing=7, interpSpeedIncreasing=7}
    boneToModify {boneName=base_ponytail_r}
    rotation {yaw=10}
   }
   n43 = "Transform (Modify) Bone - Bone: base_ponytail_l"!(ComponentPose=@n42, bAlphaBoolEnabled:bool=True, Translation="0,0,0", Rotation="-5.879,-0.000,0.000", Scale="1,1,1") -> (Pose) {
    alphaInputType=Bool
    rotationMode=BMM_Additive
    rotationSpace=BCS_BoneSpace
    translationMode=BMM_Additive
    translationSpace=BCS_BoneSpace
    alphaBoolBlend {blendInTime=0.2, blendOption=Linear, blendOutTime=0.2}
    alphaScaleBiasClamp {bInterpResult=true, interpSpeedDecreasing=7, interpSpeedIncreasing=7}
    boneToModify {boneName=base_ponytail_l}
    rotation=<rot -5.879 0 0>
   }
   n44 = RigidBody(ComponentPose=@n43, OverrideWorldGravity="0,0,-980") {
    bOverrideWorldGravity=true
    overridePhysicsAsset=&F_MED_FlossYawn_Physics_Backpack
    alphaBoolBlend {blendInTime=0.2, blendOption=Linear, blendOutTime=0.2}
    alphaScaleBiasClamp {bInterpResult=true, interpSpeedDecreasing=7, interpSpeedIncreasing=7}
    baseBoneRef {boneName=head}
    componentAppliedLinearAccClamp=<3000 3000 3000>
    componentLinearAccScale=<0.3 0.3 0.3>
    componentLinearVelScale=<0.3 0.3 0.3>
    overrideWorldGravity {z=-980}
    simSpaceSettings {velocityScaleZ=0, worldAlpha=0.2}
   }
   n45 = "Transform (Modify) Bone - Bone: base_ponytail_r"!(
    ComponentPose=Local To Component(LocalPose=Use cached pose 'Constrains RB disble Cache'())
    Alpha:float=1
    Translation="0,0,0"
    Rotation="0,-25,0"
    Scale="1,1,1"
   ) -> (Pose) {
    rotationMode=BMM_Additive
    rotationSpace=BCS_BoneSpace
    translationMode=BMM_Additive
    translationSpace=BCS_BoneSpace
    alphaBoolBlend {blendInTime=0.2, blendOption=Linear, blendOutTime=0.2}
    alphaScaleBiasClamp {bInterpResult=true, interpSpeedDecreasing=7, interpSpeedIncreasing=7}
    boneToModify {boneName=base_ponytail_r}
    rotation {yaw=-25}
   }
   n46 = "Transform (Modify) Bone - Bone: dyn_ponytail_01_l"(ComponentPose=@n45, Alpha=1, Translation="0,0,0", Rotation="0,-25,0", Scale="1,1,1") {
    rotationMode=BMM_Additive
    rotationSpace=BCS_BoneSpace
    translationMode=BMM_Additive
    translationSpace=BCS_BoneSpace
    alphaBoolBlend {blendInTime=0.2, blendOption=Linear, blendOutTime=0.2}
    alphaScaleBiasClamp {bInterpResult=true, interpSpeedDecreasing=7, interpSpeedIncreasing=7}
    boneToModify {boneName=dyn_ponytail_01_l}
    rotation {yaw=-25}
   }
   n47 = "Transform (Modify) Bone - Bone: ponytail_ASW_02_r"!(ComponentPose=@n46, bAlphaBoolEnabled:bool=True, Translation="5,5,0", Rotation="0,0,0", Scale="1,1,1") -> (Pose) {
    alphaInputType=Bool
    rotationMode=BMM_Additive
    rotationSpace=BCS_BoneSpace
    translationMode=BMM_Additive
    translationSpace=BCS_BoneSpace
    alphaBoolBlend {blendInTime=0.2, blendOption=Linear, blendOutTime=0.2}
    alphaScaleBiasClamp {bInterpResult=true, interpSpeedDecreasing=7, interpSpeedIncreasing=7}
    boneToModify {boneName=ponytail_ASW_02_r}
    translation=<5 5>
   }
   n48 = "Transform (Modify) Bone - Bone: ponytail_ASW_02_l"!(ComponentPose=@n47, bAlphaBoolEnabled:bool=True, Translation="-5,-10,0", Rotation="0,0,0", Scale="1,1,1") -> (Pose) {
    alphaInputType=Bool
    rotationMode=BMM_Additive
    rotationSpace=BCS_BoneSpace
    translationMode=BMM_Additive
    translationSpace=BCS_BoneSpace
    alphaBoolBlend {blendInTime=0.2, blendOption=Linear, blendOutTime=0.2}
    alphaScaleBiasClamp {bInterpResult=true, interpSpeedDecreasing=7, interpSpeedIncreasing=7}
    boneToModify {boneName=ponytail_ASW_02_l}
    translation=<-5 -10>
   }
   n49 = "Transform (Modify) Bone - Bone: ponytail_ASW_01_l"(ComponentPose=@n48, Alpha=1, Translation="-10,-10,0", Rotation="0,0,0", Scale="1,1,1") {
    rotationMode=BMM_Additive
    rotationSpace=BCS_BoneSpace
    translationMode=BMM_Additive
    translationSpace=BCS_BoneSpace
    alphaBoolBlend {blendInTime=0.2, blendOption=Linear, blendOutTime=0.2}
    alphaScaleBiasClamp {bInterpResult=true}
    boneToModify {boneName=ponytail_ASW_01_l}
    translation=<-10 -10>
   }
   n50 = "Transform (Modify) Bone - Bone: ponytail_ASW_02_l"!(ComponentPose=@n49, Alpha:float=1, Translation="-20,0,0", Rotation="0,0,0", Scale="1,1,1") -> (Pose) {
    rotationMode=BMM_Additive
    rotationSpace=BCS_BoneSpace
    translationMode=BMM_Additive
    translationSpace=BCS_BoneSpace
    alphaBoolBlend {blendInTime=0.2, blendOption=Linear, blendOutTime=0.2}
    alphaScaleBiasClamp {bInterpResult=true}
    boneToModify {boneName=ponytail_ASW_02_l}
    translation {x=-20}
   }
   n51 = "Transform (Modify) Bone - Bone: ponytail_ASW_01_r"(ComponentPose=@n50, Alpha=1, Translation="10,10,0", Rotation="0,0,0", Scale="1,1,1") {
    rotationMode=BMM_Additive
    rotationSpace=BCS_BoneSpace
    translationMode=BMM_Additive
    translationSpace=BCS_BoneSpace
    alphaBoolBlend {blendInTime=0.2, blendOption=Linear, blendOutTime=0.2}
    alphaScaleBiasClamp {bInterpResult=true}
    boneToModify {boneName=ponytail_ASW_01_r}
    translation=<10 10>
   }
   n52 = "Transform (Modify) Bone - Bone: ponytail_ASW_02_r"!(ComponentPose=@n51, Alpha:float=1, Translation="20,0,0", Rotation="0,0,0", Scale="1,1,1") -> (Pose) {
    rotationMode=BMM_Additive
    rotationSpace=BCS_BoneSpace
    translationMode=BMM_Additive
    translationSpace=BCS_BoneSpace
    alphaBoolBlend {blendInTime=0.2, blendOption=Linear, blendOutTime=0.2}
    alphaScaleBiasClamp {bInterpResult=true}
    boneToModify {boneName=ponytail_ASW_02_r}
    translation {x=20}
   }
   n53 = "Transform (Modify) Bone - Bone: ponytail_ASW_01_r"(ComponentPose=@n52, Alpha=1, Translation="10,0,0", Rotation="0,0,0", Scale="1,1,1") {
    rotationMode=BMM_Additive
    rotationSpace=BCS_BoneSpace
    translationMode=BMM_Additive
    translationSpace=BCS_BoneSpace
    alphaBoolBlend {blendInTime=0.2, blendOption=Linear, blendOutTime=0.2}
    alphaScaleBiasClamp {bInterpResult=true}
    boneToModify {boneName=ponytail_ASW_01_r}
    translation {x=10}
   }
   n54 = "Transform (Modify) Bone - Bone: ponytail_ASW_02_r"(ComponentPose=@n53, Alpha=1, Translation="10,7,0", Rotation="0,0,0", Scale="1,1,1") {
    rotationMode=BMM_Additive
    rotationSpace=BCS_BoneSpace
    translationMode=BMM_Additive
    translationSpace=BCS_BoneSpace
    alphaBoolBlend {blendInTime=0.2, blendOption=Linear, blendOutTime=0.2}
    alphaScaleBiasClamp {bInterpResult=true}
    boneToModify {boneName=ponytail_ASW_02_r}
    translation=<10 7>
   }
   n55 = "Transform (Modify) Bone - Bone: dyn_ponytail_03_r"(ComponentPose=@n54, Alpha=1, Translation="0,10,0", Rotation="0,0,0", Scale="1,1,1") {
    rotationMode=BMM_Additive
    rotationSpace=BCS_BoneSpace
    translationMode=BMM_Additive
    translationSpace=BCS_BoneSpace
    alphaBoolBlend {blendInTime=0.2, blendOption=Linear, blendOutTime=0.2}
    alphaScaleBiasClamp {bInterpResult=true}
    boneToModify {boneName=dyn_ponytail_03_r}
    translation {y=10}
   }
   n56 = Blend Poses by bool(
    BlendPose_0=Blend Poses by bool(BlendPose_0=Component To Local(ComponentPose=@n44), BlendPose_1=Component To Local(ComponentPose=@n36), bActiveValue=Get bIsDBNO()) {blendType=Linear}
    BlendPose_1=Component To Local(ComponentPose=@n55)
    bActiveValue=Get Alpha_RBEnable()
   ) {blendType=Linear}
   n57 = Blend Poses by bool(
    BlendPose_0=Component To Local(ComponentPose=@n35)
    BlendPose_1=Component To Local(ComponentPose=Local To Component(LocalPose=@n56))
    bActiveValue=Get Alpha_backPack() -> (Alpha_backPack:bool)
   ) {blendType=Linear}
   n58 = Get Is Wearing Cape() -> (Is Wearing Cape:bool)
   n59 = Anim Dynamics - Bone: dyn_hair_fr_b_l(
    ComponentPose=Local To Component(LocalPose=Blend Poses by bool(BlendPose_0=@n6, BlendPose_1=@n57, bActiveValue=@n58) {blendType=Linear})
    Alpha=1
    ExternalForce="0,0,0"
   ) {
    angularDampingOverride=1
    angularSpringConstant=10000
    bAngularSpring=true
    bOverrideAngularDamping=true
    bUseGravityOverride=true
    gravityScale=0.35
    simulationSpace=RootRelative
    boundBone {boneName=dyn_hair_fr_b_l}
    componentLinearAccScale=<1 1 1>
    componentLinearVelScale=<1 1 1>
    gravityOverride {z=-980}
    physicsBodyDefinitions [{
      boundBone {boneName=dyn_hair_fr_b_l}
      boxExtents=<20 30 30>
      constraintSetup {angularLimitsMax=<0 60 60>, angularLimitsMin=<0 -2 -5>, angularTarget=<1 0 0>}
      localJointOffset=<4.5 0 0>
     }]
    simSpaceSettings {maxAngularAcceleration=100, maxAngularVelocity=5, simSpaceAngularAlpha=0.3}
   }
   n60 = Anim Dynamics - Bone: dyn_hair_sd_a_r(ComponentPose=@n59, Alpha=1, ExternalForce="0,0,0") {
    angularDampingOverride=1
    angularSpringConstant=10000
    bAngularSpring=true
    bOverrideAngularDamping=true
    bUseGravityOverride=true
    gravityScale=0.35
    simulationSpace=RootRelative
    boundBone {boneName=dyn_hair_sd_a_r}
    componentLinearAccScale=<1 1 1>
    componentLinearVelScale=<1 1 1>
    gravityOverride {z=-980}
    physicsBodyDefinitions [{
      boundBone {boneName=dyn_hair_sd_a_r}
      boxExtents=<20 30 30>
      constraintSetup {angularLimitsMax=<0 60 60>, angularLimitsMin=<0 -2 -5>, angularTarget=<1 0 0>}
      localJointOffset=<4.5 0 0>
     }]
    simSpaceSettings {maxAngularAcceleration=100, maxAngularVelocity=5, simSpaceAngularAlpha=0.3}
   }
   n61 = Anim Dynamics - Bone: dyn_hair_sd_a_01_r(ComponentPose=@n60, Alpha=1, ExternalForce="0,0,0") {
    angularDampingOverride=1
    angularSpringConstant=10000
    bAngularSpring=true
    bOverrideAngularDamping=true
    bUseGravityOverride=true
    gravityScale=0.35
    simulationSpace=RootRelative
    boundBone {boneName=dyn_hair_sd_a_01_r}
    componentLinearAccScale=<1 1 1>
    componentLinearVelScale=<1 1 1>
    gravityOverride {z=-980}
    physicsBodyDefinitions [{
      boundBone {boneName=dyn_hair_sd_a_01_r}
      boxExtents=<20 30 30>
      constraintSetup {angularLimitsMax=<0 60 60>, angularLimitsMin=<0 -2 -5>, angularTarget=<1 0 0>}
      localJointOffset=<4.5 0 0>
     }]
    simSpaceSettings {maxAngularAcceleration=100, maxAngularVelocity=5, simSpaceAngularAlpha=0.3}
   }
   n62 = Anim Dynamics - Bone: dyn_hair_sd_a_02_r(ComponentPose=@n61, Alpha=1, ExternalForce="0,0,0") {
    angularDampingOverride=1
    angularSpringConstant=10000
    bAngularSpring=true
    bOverrideAngularDamping=true
    bUseGravityOverride=true
    gravityScale=0.35
    simulationSpace=RootRelative
    boundBone {boneName=dyn_hair_sd_a_02_r}
    componentLinearAccScale=<1 1 1>
    componentLinearVelScale=<1 1 1>
    gravityOverride {z=-980}
    physicsBodyDefinitions [{
      boundBone {boneName=dyn_hair_sd_a_02_r}
      boxExtents=<20 30 30>
      constraintSetup {angularLimitsMax=<0 60 60>, angularLimitsMin=<0 -2 -5>, angularTarget=<1 0 0>}
      localJointOffset=<4.5 0 0>
     }]
    simSpaceSettings {maxAngularAcceleration=100, maxAngularVelocity=5, simSpaceAngularAlpha=0.3}
   }
   n63 = Anim Dynamics - Bone: dyn_hair_sd_b_01_r(ComponentPose=@n62, Alpha=1, ExternalForce="0,0,0") {
    angularDampingOverride=1
    angularSpringConstant=10000
    bAngularSpring=true
    bOverrideAngularDamping=true
    bUseGravityOverride=true
    gravityScale=0.35
    simulationSpace=RootRelative
    boundBone {boneName=dyn_hair_sd_b_01_r}
    componentLinearAccScale=<0.5 0.5 0.5>
    componentLinearVelScale=<0.5 0.5 0.5>
    gravityOverride {z=-980}
    physicsBodyDefinitions [{
      boundBone {boneName=dyn_hair_sd_b_01_r}
      boxExtents=<30 30 30>
      constraintSetup {angularLimitsMax=<0 25 25>, angularLimitsMin=<0 0 -5>, angularTarget=<1 0 0>}
      localJointOffset=<4 0 0>
     }]
    simSpaceSettings {maxAngularAcceleration=50, maxAngularVelocity=5, simSpaceAngularAlpha=0.3}
   }
   n64 = Anim Dynamics - Bone: dyn_hair_sd_b_02_r(ComponentPose=@n63, Alpha=1, ExternalForce="0,0,0") {
    angularDampingOverride=1
    angularSpringConstant=10000
    bAngularSpring=true
    bOverrideAngularDamping=true
    bUseGravityOverride=true
    gravityScale=0.35
    simulationSpace=RootRelative
    boundBone {boneName=dyn_hair_sd_b_02_r}
    componentLinearAccScale=<0.5 0.5 0.5>
    componentLinearVelScale=<0.5 0.5 0.5>
    gravityOverride {z=-980}
    physicsBodyDefinitions [{
      boundBone {boneName=dyn_hair_sd_b_02_r}
      boxExtents=<25 30 30>
      constraintSetup {angularLimitsMax=<0 30 30>, angularLimitsMin=<0 0 -5>, angularTarget=<1 0 0>}
      localJointOffset=<12 0 0>
     }]
    simSpaceSettings {maxAngularAcceleration=50, maxAngularVelocity=5, simSpaceAngularAlpha=0.3}
   }
   n65 = Anim Dynamics - Bone: dyn_hair_sd_b_03_r(ComponentPose=@n64, Alpha=1, ExternalForce="0,0,0") {
    angularDampingOverride=1
    angularSpringConstant=10000
    bAngularSpring=true
    bOverrideAngularDamping=true
    bUseGravityOverride=true
    gravityScale=0.35
    simulationSpace=RootRelative
    boundBone {boneName=dyn_hair_sd_b_03_r}
    componentLinearAccScale=<0.5 0.5 0.5>
    componentLinearVelScale=<0.5 0.5 0.5>
    gravityOverride {z=-980}
    physicsBodyDefinitions [{
      boundBone {boneName=dyn_hair_sd_b_03_r}
      boxExtents=<25 30 30>
      constraintSetup {angularLimitsMax=<0 30 30>, angularLimitsMin=<0 0 -5>, angularTarget=<1 0 0>}
      localJointOffset=<12 0 0>
     }]
    simSpaceSettings {maxAngularAcceleration=50, maxAngularVelocity=5, simSpaceAngularAlpha=0.3}
   }
   n66 = Anim Dynamics - Bone: dyn_hair_sd_b_01_l(ComponentPose=@n65, Alpha=1, ExternalForce="0,0,0") {
    angularDampingOverride=1
    angularSpringConstant=10000
    bAngularSpring=true
    bOverrideAngularDamping=true
    bUseGravityOverride=true
    gravityScale=0.35
    simulationSpace=RootRelative
    boundBone {boneName=dyn_hair_sd_b_01_l}
    componentLinearAccScale=<1 1 1>
    componentLinearVelScale=<1 1 1>
    gravityOverride {z=-980}
    physicsBodyDefinitions [{
      boundBone {boneName=dyn_hair_sd_b_01_l}
      boxExtents=<20 30 30>
      constraintSetup {angularLimitsMax=<0 5 60>, angularLimitsMin=<0 -60 -5>, angularTarget=<1 0 0>}
      localJointOffset=<4.5 0 0>
     }]
    simSpaceSettings {maxAngularAcceleration=100, maxAngularVelocity=10, simSpaceAngularAlpha=0.5}
   }
   n67 = Anim Dynamics - Bone: dyn_hair_sd_b_02_l(ComponentPose=@n66, Alpha=1, ExternalForce="0,0,0") {
    angularDampingOverride=1
    angularSpringConstant=10000
    bAngularSpring=true
    bOverrideAngularDamping=true
    bUseGravityOverride=true
    gravityScale=0.35
    simulationSpace=RootRelative
    boundBone {boneName=dyn_hair_sd_b_02_l}
    componentLinearAccScale=<1 1 1>
    componentLinearVelScale=<1 1 1>
    gravityOverride {z=-980}
    physicsBodyDefinitions [{
      boundBone {boneName=dyn_hair_sd_b_02_l}
      boxExtents=<20 30 30>
      constraintSetup {angularLimitsMax=<0 5 60>, angularLimitsMin=<0 -60 -5>, angularTarget=<1 0 0>}
      localJointOffset=<4.5 0 0>
     }]
    simSpaceSettings {maxAngularAcceleration=100, maxAngularVelocity=10, simSpaceAngularAlpha=0.5}
   }
   n68 = Anim Dynamics - Bone: dyn_hair_sd_b_03_l(ComponentPose=@n67, Alpha=1, ExternalForce="0,0,0") {
    angularDampingOverride=1
    angularSpringConstant=10000
    bAngularSpring=true
    bOverrideAngularDamping=true
    bUseGravityOverride=true
    gravityScale=0.35
    simulationSpace=RootRelative
    boundBone {boneName=dyn_hair_sd_b_03_l}
    componentLinearAccScale=<1 1 1>
    componentLinearVelScale=<1 1 1>
    gravityOverride {z=-980}
    physicsBodyDefinitions [{
      boundBone {boneName=dyn_hair_sd_b_03_l}
      boxExtents=<20 30 30>
      constraintSetup {angularLimitsMax=<0 5 60>, angularLimitsMin=<0 -60 -5>, angularTarget=<1 0 0>}
      localJointOffset=<4.5 0 0>
     }]
    simSpaceSettings {maxAngularAcceleration=100, maxAngularVelocity=10, simSpaceAngularAlpha=0.5}
   }
   n69 = Anim Dynamics - Bone: dyn_hair_fr_a_01_c(ComponentPose=@n68, Alpha=1, ExternalForce="0,0,0") {
    angularDampingOverride=1
    angularSpringConstant=10000
    bAngularSpring=true
    bOverrideAngularDamping=true
    bUseGravityOverride=true
    gravityScale=0.35
    simulationSpace=RootRelative
    boundBone {boneName=dyn_hair_fr_a_01_c}
    componentLinearAccScale=<1 1 1>
    componentLinearVelScale=<1 1 1>
    gravityOverride {z=-980}
    physicsBodyDefinitions [{
      boundBone {boneName=dyn_hair_fr_a_01_c}
      boxExtents=<20 30 30>
      constraintSetup {angularLimitsMax=<0 0 30>, angularLimitsMin=<0 0 0>, angularTarget=<1 0 0>}
      localJointOffset=<4.5 0 0>
     }]
    simSpaceSettings {maxAngularAcceleration=50, maxAngularVelocity=5, simSpaceAngularAlpha=0.5}
   }
   n70 = Anim Dynamics - Bone: dyn_hair_fr_a_02_c(ComponentPose=@n69, Alpha=1, ExternalForce="0,0,0") {
    angularDampingOverride=1
    angularSpringConstant=10000
    bAngularSpring=true
    bOverrideAngularDamping=true
    bUseGravityOverride=true
    gravityScale=0.35
    simulationSpace=RootRelative
    boundBone {boneName=dyn_hair_fr_a_02_c}
    componentLinearAccScale=<1 1 1>
    componentLinearVelScale=<1 1 1>
    gravityOverride {z=-980}
    physicsBodyDefinitions [{
      boundBone {boneName=dyn_hair_fr_a_02_c}
      boxExtents=<20 30 30>
      constraintSetup {angularLimitsMax=<0 0 60>, angularLimitsMin=<0 0 0>, angularTarget=<1 0 0>}
      localJointOffset=<4.5 0 0>
     }]
    simSpaceSettings {maxAngularAcceleration=100, maxAngularVelocity=10, simSpaceAngularAlpha=0.5}
   }
   n71 = Anim Dynamics - Bone: dyn_hair_fr_a_03_c(ComponentPose=@n70, Alpha=1, ExternalForce="0,0,0") {
    angularDampingOverride=1
    angularSpringConstant=10000
    bAngularSpring=true
    bOverrideAngularDamping=true
    bUseGravityOverride=true
    gravityScale=0.35
    simulationSpace=RootRelative
    boundBone {boneName=dyn_hair_fr_a_03_c}
    componentLinearAccScale=<1 1 1>
    componentLinearVelScale=<1 1 1>
    gravityOverride {z=-980}
    physicsBodyDefinitions [{
      boundBone {boneName=dyn_hair_fr_a_03_c}
      boxExtents=<20 30 30>
      constraintSetup {angularLimitsMax=<0 0 30>, angularLimitsMin=<0 0 0>, angularTarget=<1 0 0>}
      localJointOffset=<4.5 0 0>
     }]
    simSpaceSettings {maxAngularAcceleration=100, maxAngularVelocity=10, simSpaceAngularAlpha=0.5}
   }
   n72 = Anim Dynamics - Bone: dyn_hair_sd_c_01_l(ComponentPose=@n71, Alpha=1, ExternalForce="0,0,0") {
    angularDampingOverride=1
    angularSpringConstant=10000
    bAngularSpring=true
    bOverrideAngularDamping=true
    bUseGravityOverride=true
    gravityScale=0.35
    simulationSpace=RootRelative
    boundBone {boneName=dyn_hair_sd_c_01_l}
    componentLinearAccScale=<1 1 1>
    componentLinearVelScale=<1 1 1>
    gravityOverride {z=-980}
    physicsBodyDefinitions [{
      boundBone {boneName=dyn_hair_sd_c_01_l}
      boxExtents=<20 30 30>
      constraintSetup {angularLimitsMax=<0 5 60>, angularLimitsMin=<0 -60 -5>, angularTarget=<1 0 0>}
      localJointOffset=<3.5 0 0>
     }]
    simSpaceSettings {maxAngularAcceleration=100, maxAngularVelocity=10, simSpaceAngularAlpha=0.7}
   }
   n73 = Anim Dynamics - Bone: dyn_hair_sd_a_01_l(ComponentPose=@n72, Alpha=1, ExternalForce="0,0,0") {
    angularDampingOverride=1
    angularSpringConstant=10000
    bAngularSpring=true
    bOverrideAngularDamping=true
    bUseGravityOverride=true
    gravityScale=0.35
    simulationSpace=RootRelative
    boundBone {boneName=dyn_hair_sd_a_01_l}
    componentLinearAccScale=<0.5 0.5 0.5>
    componentLinearVelScale=<1 1 1>
    gravityOverride {z=-980}
    physicsBodyDefinitions [{
      boundBone {boneName=dyn_hair_sd_a_01_l}
      boxExtents=<20 30 30>
      constraintSetup {angularLimitsMax=<0 5 30>, angularLimitsMin=<0 -60 0>, angularTarget=<1 0 0>}
      localJointOffset=<4.5 0 0>
     }]
    simSpaceSettings {maxAngularAcceleration=100, maxAngularVelocity=10, simSpaceAngularAlpha=0.5}
   }
   n74 = Anim Dynamics - Bone: dyn_hair_sd_a_02_l(ComponentPose=@n73, Alpha=1, ExternalForce="0,0,0") {
    angularDampingOverride=1
    angularSpringConstant=10000
    bAngularSpring=true
    bOverrideAngularDamping=true
    bUseGravityOverride=true
    gravityScale=0.35
    simulationSpace=RootRelative
    boundBone {boneName=dyn_hair_sd_a_02_l}
    componentLinearAccScale=<0.5 0.5 0.5>
    componentLinearVelScale=<1 1 1>
    gravityOverride {z=-980}
    physicsBodyDefinitions [{
      boundBone {boneName=dyn_hair_sd_a_02_l}
      boxExtents=<20 30 30>
      constraintSetup {angularLimitsMax=<0 5 30>, angularLimitsMin=<0 -60 0>, angularTarget=<1 0 0>}
      localJointOffset=<4.5 0 0>
     }]
    simSpaceSettings {maxAngularAcceleration=100, maxAngularVelocity=10, simSpaceAngularAlpha=0.5}
   }
   n75 = "Transform (Modify) Bone - Bone: base_backHair_l"(ComponentPose=@n74, Alpha=1, Translation="0,0,0", Rotation="68,0,0", Scale="1,1,1") {
    rotationMode=BMM_Additive
    rotationSpace=BCS_BoneSpace
    alphaBoolBlend {blendInTime=0.3, blendOutTime=0.7}
    boneToModify {boneName=base_backHair_l}
    rotation {pitch=68}
   }
   n76 = "Transform (Modify) Bone - Bone: base_backHair_r"(ComponentPose=@n75, Alpha=1, Translation="0,0,0", Rotation="-68,0,0", Scale="1,1,1") {
    rotationMode=BMM_Additive
    rotationSpace=BCS_BoneSpace
    alphaBoolBlend {blendInTime=0.3, blendOutTime=0.7}
    boneToModify {boneName=base_backHair_r}
    rotation {pitch=-68}
   }
   Output Pose(Result=Component To Local(ComponentPose=@n76)) {layerGroup=None}
   n77 = Layered blend per bone(
    BasePose=Copy Pose From Mesh() {bUseAttachedParent=true}
    BlendPoses_0=Copy Pose From Mesh(SourceMeshComponent=Get Body_Mesh()) {bUseAttachedParent=true}
   ) {layerSetup [{branchFilters [
       {boneName=clavicle_l}
       {boneName=upperarm_l}
       {boneName=clavicle_r}
       {boneName=upperarm_r}
       {boneName=calf_l}
       {boneName=calf_r}
       {boneName=thigh_r}
       {boneName=thigh_l}
      ]}]}
   Save cached pose 'Start Cache'(Pose=@n77) {cacheName=Start Cache}
   n78 = "Transform (Modify) Bone - Bone: base_ponytail_l"!(
    ComponentPose=Local To Component(LocalPose=Use cached pose 'Start Cache'())
    Alpha:float=1
    Translation="0,0,0"
    Rotation="-50,-50,0"
    Scale="1,1,1"
   ) -> (Pose) {
    rotationMode=BMM_Additive
    rotationSpace=BCS_BoneSpace
    scaleMode=BMM_Additive
    scaleSpace=BCS_BoneSpace
    translationMode=BMM_Additive
    translationSpace=BCS_BoneSpace
    alphaBoolBlend {blendOption=Linear}
    alphaScaleBiasClamp {bInterpResult=true}
    boneToModify {boneName=base_ponytail_l}
    rotation {pitch=-50, yaw=-50}
   }
   n79 = "Transform (Modify) Bone - Bone: base_ponytail_r"(ComponentPose=@n78, Alpha=1, Translation="0,0,0", Rotation="-20,-60,0", Scale="1,1,1") {
    rotationMode=BMM_Additive
    rotationSpace=BCS_BoneSpace
    scaleMode=BMM_Additive
    scaleSpace=BCS_BoneSpace
    translationMode=BMM_Additive
    translationSpace=BCS_BoneSpace
    alphaBoolBlend {blendOption=Linear}
    alphaScaleBiasClamp {bInterpResult=true, interpSpeedDecreasing=30}
    boneToModify {boneName=base_ponytail_r}
    rotation {pitch=-20, yaw=-60}
   }
   n80 = "Transform (Modify) Bone - Bone: base_ponytail_r"(ComponentPose=@n79, Alpha=1, Translation="5.680,0.401,0.000", Rotation="0,-25,0", Scale="1,1,1") {
    rotationMode=BMM_Additive
    rotationSpace=BCS_BoneSpace
    scaleMode=BMM_Additive
    scaleSpace=BCS_BoneSpace
    translationMode=BMM_Additive
    translationSpace=BCS_BoneSpace
    alphaBoolBlend {blendOption=Linear}
    alphaScaleBiasClamp {bInterpResult=true}
    boneToModify {boneName=base_ponytail_r}
    rotation {yaw=-25}
    translation=<5.68 0.401>
   }
   n81 = "Transform (Modify) Bone - Bone: base_ponytail_l"(ComponentPose=@n80, Alpha=1, Translation="0,0,0", Rotation="0,-15,0", Scale="1,1,1") {
    rotationMode=BMM_Additive
    rotationSpace=BCS_BoneSpace
    scaleMode=BMM_Additive
    scaleSpace=BCS_BoneSpace
    translationMode=BMM_Additive
    translationSpace=BCS_BoneSpace
    alphaBoolBlend {blendOption=Linear}
    alphaScaleBiasClamp {bInterpResult=true}
    boneToModify {boneName=base_ponytail_l}
    rotation {yaw=-15}
   }
   n82 = "Transform (Modify) Bone - Bone: base_ponytail_l"!(ComponentPose=@n81, bAlphaBoolEnabled:bool=True, Translation="0,0,0", Rotation="0,-30,0", Scale="1,1,1") -> (Pose) {
    alphaInputType=Bool
    rotationMode=BMM_Additive
    rotationSpace=BCS_BoneSpace
    scaleMode=BMM_Additive
    scaleSpace=BCS_BoneSpace
    translationMode=BMM_Additive
    translationSpace=BCS_BoneSpace
    alphaBoolBlend {blendInTime=0.1, blendOption=Linear, blendOutTime=0.2}
    alphaScaleBiasClamp {bInterpResult=true}
    boneToModify {boneName=base_ponytail_l}
    rotation {yaw=-30}
   }
   n83 = "Transform (Modify) Bone - Bone: base_ponytail_r"!(ComponentPose=@n82, bAlphaBoolEnabled:bool=True, Translation="0,0,0", Rotation="0,-30,0", Scale="1,1,1") -> (Pose) {
    alphaInputType=Bool
    rotationMode=BMM_Additive
    rotationSpace=BCS_BoneSpace
    scaleMode=BMM_Additive
    scaleSpace=BCS_BoneSpace
    translationMode=BMM_Additive
    translationSpace=BCS_BoneSpace
    alphaBoolBlend {blendInTime=0.1, blendOption=Linear, blendOutTime=0.2}
    alphaScaleBiasClamp {bInterpResult=true}
    boneToModify {boneName=base_ponytail_r}
    rotation {yaw=-30}
   }
   n84 = "Transform (Modify) Bone - Bone: base_ponytail_l"(ComponentPose=@n83, bAlphaBoolEnabled=True, Translation="0,0,0", Rotation="0,-50,0", Scale="1,1,1") {
    alphaInputType=Bool
    rotationMode=BMM_Additive
    rotationSpace=BCS_BoneSpace
    scaleMode=BMM_Additive
    scaleSpace=BCS_BoneSpace
    translationMode=BMM_Additive
    translationSpace=BCS_BoneSpace
    alphaBoolBlend {blendInTime=0.2, blendOption=Linear, blendOutTime=0.2}
    alphaScaleBiasClamp {bInterpResult=true}
    boneToModify {boneName=base_ponytail_l}
    rotation {yaw=-50}
   }
   n85 = "Transform (Modify) Bone - Bone: base_ponytail_r"(ComponentPose=@n84, bAlphaBoolEnabled=True, Translation="0,0,0", Rotation="-0.008,-60,0.003", Scale="1,1,1") {
    alphaInputType=Bool
    rotationMode=BMM_Additive
    rotationSpace=BCS_BoneSpace
    scaleMode=BMM_Additive
    scaleSpace=BCS_BoneSpace
    translationMode=BMM_Additive
    translationSpace=BCS_BoneSpace
    alphaBoolBlend {blendInTime=0.2, blendOption=Linear, blendOutTime=0.2}
    alphaScaleBiasClamp {bInterpResult=true}
    boneToModify {boneName=base_ponytail_r}
    rotation=<rot -0.008 -60 0.003>
   }
   n86 = "Transform (Modify) Bone - Bone: base_ponytail_r"(ComponentPose=@n85, bAlphaBoolEnabled=True, Translation="0,0,0", Rotation="-15.277,-10.829,-19.090", Scale="1,1,1") {
    alphaInputType=Bool
    rotationMode=BMM_Additive
    rotationSpace=BCS_BoneSpace
    scaleMode=BMM_Additive
    scaleSpace=BCS_BoneSpace
    translationMode=BMM_Additive
    translationSpace=BCS_BoneSpace
    alphaBoolBlend {blendInTime=0.2, blendOption=Linear, blendOutTime=0.2}
    alphaScaleBiasClamp {bInterpResult=true}
    boneToModify {boneName=base_ponytail_r}
    rotation=<rot -15.277 -10.8289 -19.09>
   }
   n87 = "Transform (Modify) Bone - Bone: base_ponytail_r"!(ComponentPose=@n86, Alpha:float=1, Translation="0,0,0", Rotation="0,-20,0", Scale="1,1,1") -> (Pose) {
    rotationMode=BMM_Additive
    rotationSpace=BCS_BoneSpace
    scaleMode=BMM_Additive
    scaleSpace=BCS_BoneSpace
    translationMode=BMM_Additive
    translationSpace=BCS_BoneSpace
    alphaBoolBlend {blendInTime=0.2, blendOption=Linear, blendOutTime=0.2}
    alphaScaleBiasClamp {bInterpResult=true}
    boneToModify {boneName=base_ponytail_r}
    rotation {yaw=-20}
   }
   n88 = "Transform (Modify) Bone - Bone: base_ponytail_r"(ComponentPose=@n87, Alpha=1, Translation="0,0,0", Rotation="0,-15.877,0", Scale="1,1,1") {
    rotationMode=BMM_Additive
    rotationSpace=BCS_BoneSpace
    scaleMode=BMM_Additive
    scaleSpace=BCS_BoneSpace
    translationMode=BMM_Additive
    translationSpace=BCS_BoneSpace
    alphaBoolBlend {blendInTime=0.2, blendOption=Linear, blendOutTime=0.2}
    alphaScaleBiasClamp {bInterpResult=true}
    boneToModify {boneName=base_ponytail_r}
    rotation {yaw=-15.877}
   }
   n89 = "Transform (Modify) Bone - Bone: base_ponytail_l"(ComponentPose=@n88, bAlphaBoolEnabled=True, Translation="0,0,0", Rotation="-0.000,-21.524,-0.000", Scale="1,1,1") {
    alphaInputType=Bool
    rotationMode=BMM_Additive
    rotationSpace=BCS_BoneSpace
    scaleMode=BMM_Additive
    scaleSpace=BCS_BoneSpace
    translationMode=BMM_Additive
    translationSpace=BCS_BoneSpace
    alphaBoolBlend {blendInTime=0.2, blendOption=Linear, blendOutTime=0.2}
    alphaScaleBiasClamp {bInterpResult=true}
    boneToModify {boneName=base_ponytail_l}
    rotation=<rot 0 -21.5239 0>
   }
   n90 = "Constraint - base_middle_collision (RR)"!(ComponentPose=@n89, Alpha=1, ConstraintWeights_0=0.5, ConstraintWeights_1:float=0.5) -> (Pose) {
    alphaBoolBlend {blendOption=Linear}
    boneToModify {boneName=base_middle_collision}
    constraintSetup [{transformType=Rotation, targetBone {boneName=pelvis}}, {transformType=Rotation, targetBone {boneName=spine_05}}]
    constraintWeights [1,1]
   }
   n91 = "Transform (Modify) Bone - Bone: base_ponytail_l"!(ComponentPose=@n90, Alpha:float=1, Translation="-3.313,-3.625,0", Rotation="-20,-50,0", Scale="1,1,1") -> (Pose) {
    rotationMode=BMM_Additive
    rotationSpace=BCS_BoneSpace
    translationMode=BMM_Additive
    translationSpace=BCS_BoneSpace
    alphaBoolBlend {blendInTime=0.2, blendOption=Linear, blendOutTime=0.2}
    alphaScaleBiasClamp {bInterpResult=true, interpSpeedDecreasing=6}
    boneToModify {boneName=base_ponytail_l}
    rotation {pitch=-20, yaw=-50}
    translation=<-3.313 -3.625>
   }
   n92 = "Transform (Modify) Bone - Bone: dyn_ponytail_01_l"!(ComponentPose=@n91, bAlphaBoolEnabled:bool=True, Translation="0,-2,0", Rotation="-0.094,-8.419,-0.731", Scale="1,1,1") -> (Pose) {
    alphaInputType=Bool
    rotationMode=BMM_Additive
    rotationSpace=BCS_BoneSpace
    translationMode=BMM_Additive
    translationSpace=BCS_BoneSpace
    alphaBoolBlend {blendInTime=0.2, blendOption=Linear, blendOutTime=0.2}
    alphaScaleBiasClamp {bInterpResult=true, interpSpeedDecreasing=2}
    boneToModify {boneName=dyn_ponytail_01_l}
    rotation=<rot -0.094 -8.419 -0.731>
    translation {y=-2}
   }
   Save cached pose 'Fix Ponytails Cache'(Pose=Component To Local(ComponentPose=@n92)) {cacheName=Fix Ponytails Cache}
   n93 = Local To Component(LocalPose=Use cached pose 'Fix Ponytails Cache'())
   n94 = Anim Dynamics - Bone: dyn_ponytail_01_r!(
    ComponentPose="Transform (Modify) Bone - Bone: base_ponytail_l"(
     ComponentPose="Transform (Modify) Bone - Bone: base_ponytail_r"(ComponentPose=@n93, Alpha=1, Translation="0,0,0", Rotation="0,-80,0", Scale="1,1,1") {rotationMode=BMM_Additive, rotationSpace=BCS_BoneSpace, boneToModify {boneName=base_ponytail_r}, rotation {yaw=-80}}
     Alpha=1
     Translation="0,0,0"
     Rotation="0,-80,0"
     Scale="1,1,1"
    ) {rotationMode=BMM_Additive, rotationSpace=BCS_BoneSpace, boneToModify {boneName=base_ponytail_l}, rotation {yaw=-80}}
    bAlphaBoolEnabled:bool=True
    ExternalForce="0,0,0"
   ) -> (Pose) {
    alphaInputType=Bool
    angularDampingOverride=1
    angularSpringConstant=10000
    bAngularSpring=true
    bOverrideAngularDamping=true
    bUseGravityOverride=true
    gravityScale=0.35
    simulationSpace=RootRelative
    alphaBoolBlend {blendInTime=0.2, blendOption=Linear, blendOutTime=0.2}
    boundBone {boneName=dyn_ponytail_01_r}
    componentLinearAccScale=<1 1 1>
    componentLinearVelScale=<1 1 1>
    gravityOverride {z=-980}
    physicsBodyDefinitions [{
      boundBone {boneName=dyn_ponytail_01_r}
      boxExtents=<35 35 35>
      constraintSetup {angularLimitsMax=<15 5 30>, angularLimitsMin=<-15 -5 0>, angularTarget=<1 0 0>}
      localJointOffset=<-3 0 0>
     }]
   }
   n95 = Anim Dynamics - Bone: dyn_ponytail_02_r!(ComponentPose=@n94, bAlphaBoolEnabled:bool=True, ExternalForce="0,0,0") -> (Pose) {
    alphaInputType=Bool
    angularDampingOverride=1
    angularSpringConstant=10000
    bAngularSpring=true
    bOverrideAngularDamping=true
    bUseGravityOverride=true
    gravityScale=0.35
    simulationSpace=RootRelative
    alphaBoolBlend {blendInTime=0.2, blendOption=Linear, blendOutTime=0.2}
    boundBone {boneName=dyn_ponytail_02_r}
    componentLinearAccScale=<1 1 1>
    componentLinearVelScale=<1 1 1>
    gravityOverride {z=-980}
    physicsBodyDefinitions [{
      boundBone {boneName=dyn_ponytail_02_r}
      boxExtents=<35 35 35>
      constraintSetup {angularLimitsMax=<15 5 30>, angularLimitsMin=<-15 -5 0>, angularTarget=<1 0 0>}
      localJointOffset=<-3 0 0>
     }]
   }
   n96 = Anim Dynamics - Bone: dyn_ponytail_03_r!(ComponentPose=@n95, bAlphaBoolEnabled:bool=True, ExternalForce="0,0,0") -> (Pose) {
    alphaInputType=Bool
    angularDampingOverride=1
    angularSpringConstant=10000
    bAngularSpring=true
    bOverrideAngularDamping=true
    bUseGravityOverride=true
    gravityScale=0.35
    simulationSpace=RootRelative
    alphaBoolBlend {blendInTime=0.2, blendOption=Linear, blendOutTime=0.2}
    boundBone {boneName=dyn_ponytail_03_r}
    componentLinearAccScale=<1 1 1>
    componentLinearVelScale=<1 1 1>
    gravityOverride {z=-980}
    physicsBodyDefinitions [{
      boundBone {boneName=dyn_ponytail_03_r}
      boxExtents=<35 35 35>
      constraintSetup {angularLimitsMax=<15 5 30>, angularLimitsMin=<-15 -5 0>, angularTarget=<1 0 0>}
      localJointOffset=<-3 0 0>
     }]
   }
   n97 = Anim Dynamics - Bone: dyn_ponytail_04_r!(ComponentPose=@n96, bAlphaBoolEnabled:bool=True, ExternalForce="0,0,0") -> (Pose) {
    alphaInputType=Bool
    angularDampingOverride=1
    angularSpringConstant=10000
    bAngularSpring=true
    bOverrideAngularDamping=true
    bUseGravityOverride=true
    gravityScale=0.35
    simulationSpace=RootRelative
    alphaBoolBlend {blendInTime=0.2, blendOption=Linear, blendOutTime=0.2}
    boundBone {boneName=dyn_ponytail_04_r}
    componentLinearAccScale=<1 1 1>
    componentLinearVelScale=<1 1 1>
    gravityOverride {z=-980}
    physicsBodyDefinitions [{
      boundBone {boneName=dyn_ponytail_04_r}
      boxExtents=<35 35 35>
      constraintSetup {angularLimitsMax=<15 5 30>, angularLimitsMin=<-15 -5 0>, angularTarget=<1 0 0>}
      localJointOffset=<-3 0 0>
     }]
   }
   n98 = Anim Dynamics - Bone: dyn_ponytail_05_r!(ComponentPose=@n97, bAlphaBoolEnabled:bool=True, ExternalForce="0,0,0") -> (Pose) {
    alphaInputType=Bool
    angularDampingOverride=1
    angularSpringConstant=10000
    bAngularSpring=true
    bOverrideAngularDamping=true
    bUseGravityOverride=true
    gravityScale=0.35
    simulationSpace=RootRelative
    alphaBoolBlend {blendInTime=0.2, blendOption=Linear, blendOutTime=0.2}
    boundBone {boneName=dyn_ponytail_05_r}
    componentLinearAccScale=<1 1 1>
    componentLinearVelScale=<1 1 1>
    gravityOverride {z=-980}
    physicsBodyDefinitions [{
      boundBone {boneName=dyn_ponytail_05_r}
      boxExtents=<35 35 35>
      constraintSetup {angularLimitsMax=<15 5 30>, angularLimitsMin=<-15 -5 0>, angularTarget=<1 0 0>}
      localJointOffset=<-3 0 0>
     }]
   }
   n99 = Anim Dynamics - Bone: dyn_ponytail_06_r!(ComponentPose=@n98, bAlphaBoolEnabled:bool=True, ExternalForce="0,0,0") -> (Pose) {
    alphaInputType=Bool
    angularDampingOverride=1
    angularSpringConstant=10000
    bAngularSpring=true
    bOverrideAngularDamping=true
    bUseGravityOverride=true
    gravityScale=0.35
    simulationSpace=RootRelative
    alphaBoolBlend {blendInTime=0.2, blendOption=Linear, blendOutTime=0.2}
    boundBone {boneName=dyn_ponytail_06_r}
    componentLinearAccScale=<1 1 1>
    componentLinearVelScale=<1 1 1>
    gravityOverride {z=-980}
    physicsBodyDefinitions [{
      boundBone {boneName=dyn_ponytail_06_r}
      boxExtents=<35 35 35>
      constraintSetup {angularLimitsMax=<15 5 30>, angularLimitsMin=<-15 -5 0>, angularTarget=<1 0 0>}
      localJointOffset=<-3 0 0>
     }]
   }
   n100 = Anim Dynamics - Bone: dyn_ponytail_strand_b_r!(ComponentPose=@n99, bAlphaBoolEnabled:bool=True, ExternalForce="0,0,0") -> (Pose) {
    alphaInputType=Bool
    angularDampingOverride=1
    angularSpringConstant=10000
    bAngularSpring=true
    bOverrideAngularDamping=true
    bUseGravityOverride=true
    gravityScale=0.35
    simulationSpace=RootRelative
    alphaBoolBlend {blendInTime=0.2, blendOption=Linear, blendOutTime=0.2}
    boundBone {boneName=dyn_ponytail_strand_b_r}
    componentLinearAccScale=<1 1 1>
    componentLinearVelScale=<1 1 1>
    gravityOverride {z=-980}
    physicsBodyDefinitions [{
      boundBone {boneName=dyn_ponytail_strand_b_r}
      boxExtents=<35 35 35>
      constraintSetup {angularLimitsMax=<15 5 30>, angularLimitsMin=<-15 -5 0>, angularTarget=<1 0 0>}
      localJointOffset=<-3 0 0>
     }]
   }
   n101 = Anim Dynamics - Bone: dyn_ponytail_strand_a_r!(ComponentPose=@n100, bAlphaBoolEnabled:bool=True, ExternalForce="0,0,0") -> (Pose) {
    alphaInputType=Bool
    angularDampingOverride=1
    angularSpringConstant=10000
    bAngularSpring=true
    bOverrideAngularDamping=true
    bUseGravityOverride=true
    gravityScale=0.35
    simulationSpace=RootRelative
    alphaBoolBlend {blendInTime=0.2, blendOption=Linear, blendOutTime=0.2}
    boundBone {boneName=dyn_ponytail_strand_a_r}
    componentLinearAccScale=<1 1 1>
    componentLinearVelScale=<1 1 1>
    gravityOverride {z=-980}
    physicsBodyDefinitions [{
      boundBone {boneName=dyn_ponytail_strand_a_r}
      boxExtents=<35 35 35>
      constraintSetup {angularLimitsMax=<15 5 30>, angularLimitsMin=<-15 -5 0>, angularTarget=<1 0 0>}
      localJointOffset=<-3 0 0>
     }]
   }
   n102 = Anim Dynamics - Bone: dyn_ponytail_01_l!(ComponentPose=@n101, bAlphaBoolEnabled:bool=True, ExternalForce="0,0,0") -> (Pose) {
    alphaInputType=Bool
    angularDampingOverride=1
    angularSpringConstant=10000
    bAngularSpring=true
    bOverrideAngularDamping=true
    bUseGravityOverride=true
    gravityScale=0.35
    simulationSpace=RootRelative
    alphaBoolBlend {blendInTime=0.2, blendOption=Linear, blendOutTime=0.2}
    boundBone {boneName=dyn_ponytail_01_l}
    componentLinearAccScale=<1 1 1>
    componentLinearVelScale=<1 1 1>
    gravityOverride {z=-980}
    physicsBodyDefinitions [{
      boundBone {boneName=dyn_ponytail_01_l}
      boxExtents=<35 35 35>
      constraintSetup {angularLimitsMax=<15 5 30>, angularLimitsMin=<-15 -5 0>, angularTarget=<1 0 0>}
      localJointOffset=<3 0 0>
     }]
   }
   n103 = Anim Dynamics - Bone: dyn_ponytail_02_l!(ComponentPose=@n102, bAlphaBoolEnabled:bool=True, ExternalForce="0,0,0") -> (Pose) {
    alphaInputType=Bool
    angularDampingOverride=1
    angularSpringConstant=10000
    bAngularSpring=true
    bOverrideAngularDamping=true
    bUseGravityOverride=true
    gravityScale=0.35
    simulationSpace=RootRelative
    alphaBoolBlend {blendInTime=0.2, blendOption=Linear, blendOutTime=0.2}
    boundBone {boneName=dyn_ponytail_02_l}
    componentLinearAccScale=<1 1 1>
    componentLinearVelScale=<1 1 1>
    gravityOverride {z=-980}
    physicsBodyDefinitions [{
      boundBone {boneName=dyn_ponytail_02_l}
      boxExtents=<35 35 35>
      constraintSetup {angularLimitsMax=<15 5 30>, angularLimitsMin=<-15 -5 0>, angularTarget=<1 0 0>}
      localJointOffset=<3 0 0>
     }]
   }
   n104 = Anim Dynamics - Bone: dyn_ponytail_03_l!(ComponentPose=@n103, bAlphaBoolEnabled:bool=True, ExternalForce="0,0,0") -> (Pose) {
    alphaInputType=Bool
    angularDampingOverride=1
    angularSpringConstant=10000
    bAngularSpring=true
    bOverrideAngularDamping=true
    bUseGravityOverride=true
    gravityScale=0.35
    simulationSpace=RootRelative
    alphaBoolBlend {blendInTime=0.2, blendOption=Linear, blendOutTime=0.2}
    boundBone {boneName=dyn_ponytail_03_l}
    componentLinearAccScale=<1 1 1>
    componentLinearVelScale=<1 1 1>
    gravityOverride {z=-980}
    physicsBodyDefinitions [{
      boundBone {boneName=dyn_ponytail_03_l}
      boxExtents=<35 35 35>
      constraintSetup {angularLimitsMax=<15 5 30>, angularLimitsMin=<-15 -5 0>, angularTarget=<1 0 0>}
      localJointOffset=<3 0 0>
     }]
   }
   n105 = Anim Dynamics - Bone: dyn_ponytail_04_l!(ComponentPose=@n104, bAlphaBoolEnabled:bool=True, ExternalForce="0,0,0") -> (Pose) {
    alphaInputType=Bool
    angularDampingOverride=1
    angularSpringConstant=10000
    bAngularSpring=true
    bOverrideAngularDamping=true
    bUseGravityOverride=true
    gravityScale=0.35
    simulationSpace=RootRelative
    alphaBoolBlend {blendInTime=0.2, blendOption=Linear, blendOutTime=0.2}
    boundBone {boneName=dyn_ponytail_04_l}
    componentLinearAccScale=<1 1 1>
    componentLinearVelScale=<1 1 1>
    gravityOverride {z=-980}
    physicsBodyDefinitions [{
      boundBone {boneName=dyn_ponytail_04_l}
      boxExtents=<35 35 35>
      constraintSetup {angularLimitsMax=<15 5 30>, angularLimitsMin=<-15 -5 0>, angularTarget=<1 0 0>}
      localJointOffset=<3 0 0>
     }]
   }
   n106 = Anim Dynamics - Bone: dyn_ponytail_05_l!(ComponentPose=@n105, bAlphaBoolEnabled:bool=True, ExternalForce="0,0,0") -> (Pose) {
    alphaInputType=Bool
    angularDampingOverride=1
    angularSpringConstant=10000
    bAngularSpring=true
    bOverrideAngularDamping=true
    bUseGravityOverride=true
    gravityScale=0.35
    simulationSpace=RootRelative
    alphaBoolBlend {blendInTime=0.2, blendOption=Linear, blendOutTime=0.2}
    boundBone {boneName=dyn_ponytail_05_l}
    componentLinearAccScale=<1 1 1>
    componentLinearVelScale=<1 1 1>
    gravityOverride {z=-980}
    physicsBodyDefinitions [{
      boundBone {boneName=dyn_ponytail_05_l}
      boxExtents=<35 35 35>
      constraintSetup {angularLimitsMax=<15 5 30>, angularLimitsMin=<-15 -5 0>, angularTarget=<1 0 0>}
      localJointOffset=<3 0 0>
     }]
   }
   n107 = Anim Dynamics - Bone: dyn_ponytail_06_l!(ComponentPose=@n106, bAlphaBoolEnabled:bool=True, ExternalForce="0,0,0") -> (Pose) {
    alphaInputType=Bool
    angularDampingOverride=1
    angularSpringConstant=10000
    bAngularSpring=true
    bOverrideAngularDamping=true
    bUseGravityOverride=true
    gravityScale=0.35
    simulationSpace=RootRelative
    alphaBoolBlend {blendInTime=0.2, blendOption=Linear, blendOutTime=0.2}
    boundBone {boneName=dyn_ponytail_06_l}
    componentLinearAccScale=<1 1 1>
    componentLinearVelScale=<1 1 1>
    gravityOverride {z=-980}
    physicsBodyDefinitions [{
      boundBone {boneName=dyn_ponytail_06_l}
      boxExtents=<35 35 35>
      constraintSetup {angularLimitsMax=<15 5 30>, angularLimitsMin=<-15 -5 0>, angularTarget=<1 0 0>}
      localJointOffset=<3 0 0>
     }]
   }
   n108 = Anim Dynamics - Bone: dyn_ponytail_strand_b_l!(ComponentPose=@n107, bAlphaBoolEnabled:bool=True, ExternalForce="0,0,0") -> (Pose) {
    alphaInputType=Bool
    angularDampingOverride=1
    angularSpringConstant=10000
    bAngularSpring=true
    bOverrideAngularDamping=true
    bUseGravityOverride=true
    gravityScale=0.35
    simulationSpace=RootRelative
    alphaBoolBlend {blendInTime=0.2, blendOption=Linear, blendOutTime=0.2}
    boundBone {boneName=dyn_ponytail_strand_b_l}
    componentLinearAccScale=<1 1 1>
    componentLinearVelScale=<1 1 1>
    gravityOverride {z=-980}
    physicsBodyDefinitions [{
      boundBone {boneName=dyn_ponytail_strand_b_l}
      boxExtents=<35 35 35>
      constraintSetup {angularLimitsMax=<15 5 30>, angularLimitsMin=<-15 -5 0>, angularTarget=<1 0 0>}
      localJointOffset=<3 0 0>
     }]
   }
   n109 = Anim Dynamics - Bone: dyn_ponytail_strand_a_l!(ComponentPose=@n108, bAlphaBoolEnabled:bool=True, ExternalForce="0,0,0") -> (Pose) {
    alphaInputType=Bool
    angularDampingOverride=1
    angularSpringConstant=10000
    bAngularSpring=true
    bOverrideAngularDamping=true
    bUseGravityOverride=true
    gravityScale=0.35
    simulationSpace=RootRelative
    alphaBoolBlend {blendInTime=0.2, blendOption=Linear, blendOutTime=0.2}
    boundBone {boneName=dyn_ponytail_strand_a_l}
    componentLinearAccScale=<1 1 1>
    componentLinearVelScale=<1 1 1>
    gravityOverride {z=-980}
    physicsBodyDefinitions [{
      boundBone {boneName=dyn_ponytail_strand_a_l}
      boxExtents=<35 35 35>
      constraintSetup {angularLimitsMax=<15 5 30>, angularLimitsMin=<-15 -5 0>, angularTarget=<1 0 0>}
      localJointOffset=<3 0 0>
     }]
   }
   n110 = "Constraint - dyn_ponytail_01_l (P)"(ComponentPose=@n109, Alpha=1, ConstraintWeights_0=1) {
    alphaBoolBlend {blendOption=Linear}
    alphaScaleBiasClamp {bInterpResult=true, interpSpeedDecreasing=15, interpSpeedIncreasing=15}
    boneToModify {boneName=dyn_ponytail_01_l}
    constraintSetup [{targetBone {boneName=neck_01}}]
    constraintWeights [1]
   }
   n111 = "Constraint - dyn_ponytail_01_r (P)"(ComponentPose=@n110, Alpha=1, ConstraintWeights_0=1) {
    alphaBoolBlend {blendOption=Linear}
    alphaScaleBiasClamp {bInterpResult=true, interpSpeedDecreasing=15, interpSpeedIncreasing=15}
    boneToModify {boneName=dyn_ponytail_01_r}
    constraintSetup [{targetBone {boneName=neck_01}}]
    constraintWeights [1]
   }
   n112 = "Constraint - dyn_ponytail_02_l (P)"(ComponentPose=@n111, Alpha=1, ConstraintWeights_0=1) {
    alphaBoolBlend {blendOption=Linear}
    alphaScaleBiasClamp {bInterpResult=true, interpSpeedDecreasing=15, interpSpeedIncreasing=15}
    boneToModify {boneName=dyn_ponytail_02_l}
    constraintSetup [{targetBone {boneName=neck_01}}]
    constraintWeights [1]
   }
   n113 = "Constraint - dyn_ponytail_02_r (P)"(ComponentPose=@n112, Alpha=1, ConstraintWeights_0=1) {
    alphaBoolBlend {blendOption=Linear}
    alphaScaleBiasClamp {bInterpResult=true, interpSpeedDecreasing=15, interpSpeedIncreasing=15}
    boneToModify {boneName=dyn_ponytail_02_r}
    constraintSetup [{targetBone {boneName=neck_01}}]
    constraintWeights [1]
   }
   n114 = "Constraint - dyn_ponytail_03_l (P)"(ComponentPose=@n113, Alpha=1, ConstraintWeights_0=1) {
    alphaBoolBlend {blendOption=Linear}
    alphaScaleBiasClamp {bInterpResult=true, interpSpeedDecreasing=15, interpSpeedIncreasing=15}
    boneToModify {boneName=dyn_ponytail_03_l}
    constraintSetup [{targetBone {boneName=spine_05}}]
    constraintWeights [1]
   }
   n115 = "Constraint - dyn_ponytail_03_r (P)"(ComponentPose=@n114, Alpha=1, ConstraintWeights_0=1) {
    alphaBoolBlend {blendOption=Linear}
    alphaScaleBiasClamp {bInterpResult=true, interpSpeedDecreasing=15, interpSpeedIncreasing=15}
    boneToModify {boneName=dyn_ponytail_03_r}
    constraintSetup [{targetBone {boneName=spine_05}}]
    constraintWeights [1]
   }
   n116 = "Constraint - dyn_ponytail_04_l (P)"(ComponentPose=@n115, Alpha=1, ConstraintWeights_0=1) {
    alphaBoolBlend {blendOption=Linear}
    alphaScaleBiasClamp {bInterpResult=true, interpSpeedDecreasing=15, interpSpeedIncreasing=15}
    boneToModify {boneName=dyn_ponytail_04_l}
    constraintSetup [{targetBone {boneName=spine_04}}]
    constraintWeights [1]
   }
   n117 = "Constraint - dyn_ponytail_04_r (P)"(ComponentPose=@n116, Alpha=1, ConstraintWeights_0=1) {
    alphaBoolBlend {blendOption=Linear}
    alphaScaleBiasClamp {bInterpResult=true, interpSpeedDecreasing=15, interpSpeedIncreasing=15}
    boneToModify {boneName=dyn_ponytail_04_r}
    constraintSetup [{targetBone {boneName=spine_04}}]
    constraintWeights [1]
   }
   n118 = "Constraint - dyn_ponytail_05_l (P)"(ComponentPose=@n117, Alpha=1, ConstraintWeights_0=1) {
    alphaBoolBlend {blendOption=Linear}
    alphaScaleBiasClamp {bInterpResult=true, interpSpeedDecreasing=15, interpSpeedIncreasing=15}
    boneToModify {boneName=dyn_ponytail_05_l}
    constraintSetup [{targetBone {boneName=spine_03}}]
    constraintWeights [1]
   }
   n119 = "Constraint - dyn_ponytail_05_l (P)"(ComponentPose=@n118, Alpha=1, ConstraintWeights_0=1) {
    alphaBoolBlend {blendOption=Linear}
    alphaScaleBiasClamp {bInterpResult=true, interpSpeedDecreasing=15, interpSpeedIncreasing=15}
    boneToModify {boneName=dyn_ponytail_05_l}
    constraintSetup [{targetBone {boneName=spine_03}}]
    constraintWeights [1]
   }
   n120 = "Constraint - dyn_ponytail_06_l (P)"(ComponentPose=@n119, Alpha=1, ConstraintWeights_0=1) {
    alphaBoolBlend {blendOption=Linear}
    alphaScaleBiasClamp {bInterpResult=true, interpSpeedDecreasing=15, interpSpeedIncreasing=15}
    boneToModify {boneName=dyn_ponytail_06_l}
    constraintSetup [{targetBone {boneName=spine_01}}]
    constraintWeights [1]
   }
   n121 = "Constraint - dyn_ponytail_06_r (P)"(ComponentPose=@n120, Alpha=1, ConstraintWeights_0=1) {
    alphaBoolBlend {blendOption=Linear}
    alphaScaleBiasClamp {bInterpResult=true, interpSpeedDecreasing=15, interpSpeedIncreasing=15}
    boneToModify {boneName=dyn_ponytail_06_r}
    constraintSetup [{targetBone {boneName=spine_01}}]
    constraintWeights [1]
   }
   Save cached pose 'Constrains RB disble Cache'(Pose=Component To Local(ComponentPose=@n121)) {cacheName=Constrains RB disble Cache}
  ]}
 EventGraph {nodes [
   Event BlueprintUpdateAnimation()
   > n1 = Sequence!(execute) -> (
    then_0
    then_1
    then_2:Exec
    then_3:Exec
    then_4:Exec
    then_5:Exec
    then_6:Exec
    then_7:Exec
    then_8:Exec
    then_9:Exec
    then_10:Exec
    then_11:Exec
    then_12:Exec
    then_13:Exec
   )
   n2 = Cast To PlayerPawn_Athena(execute=@n1.then_2, Object=TryGetPawnOwner())
   > Set Body_Mesh(Body_Mesh=GetSkeletalMeshForPartType(self=@n2))
   n3 = Get SkelMesh()
   > Set headUpRot(headUpRot:double="Clamp (Float)"(
     Value=GetAlphaFromReferenceTransformDelta(SkeletalMeshComponent=@n3, BoneName=spine_05, BaseName=head, InputRange=(X=0,Y=90), TwistAxis="0,0,1")
     Min=0.1
     Max=1
    )) -> (Output_Get:double)
   > Set pawnSpeed2D_divider(pawnSpeed2D_divider:double=float / float(A=Get PawnSpeed2D(), B=500)) -> (Output_Get:double)
   n4 = "Gravity Override State Function (Parameter Asset) (Nativized)"(
    execute=@n1.then_1
    CharacterPartAnimInstance=Self-Reference()
    GameTimeInSeconds=GetGameTimeInSeconds()
    Parameters=&/M_MED_FlossYawn_Parameters
   )
   > Set GravityOverride(GravityOverride=@n4)
   > Set RBAN Settings Out(RBAN Settings Out=@n4.OutRigidBodyAnimNodeParameters)
   n5 = Get bIsTacticalSprinting()
   n6 = Get bIsJumping()
   Set Alpha_Jumping in Tactical Sprinting(execute=@n1.then_4, Alpha_Jumping in Tactical Sprinting:double="To Float (Boolean)"(InBool=AND Boolean(A=@n5, B=@n6))) -> (Output_Get:double)
   n7 = NOT Boolean(A=Get bIsJumping())
   n8 = OR Boolean(A=Get bIsSprinting(), B=Get bIsTacticalSprinting())
   Set Alpha_JustSprinting(execute=@n1.then_8, Alpha_JustSprinting:double="To Float (Boolean)"(InBool=AND Boolean(A=@n8, B=@n7))) -> (Output_Get:double)
   n9 = Sequence!(execute=@n1.then_0) -> (then_0, then_1, then_2, then_3, then_4, then_5, then_6, then_7, then_8, then_9)
   n10 = NOT Boolean(A=Get Is Wearing Cape())
   n11 = GetOwnerSkeletalMeshComponent()
   n12 = BreakRotator(InRot=BreakTransform(InTransform=GetDeltaTransformFromRefPose(self=@n11, BoneName=spine_05, BaseName=pelvis)).Rotation)
   n13 = "float < float"(A=@n12, B=-20)
   n14 = Get Selected_Pickaxe() -> (Selected_Pickaxe:bool)
   n15 = Get bIsPlayingMeleeAnim()
   Set Alpha_CollisionHairAndArm R(
    execute=@n9.then_2
    Alpha_CollisionHairAndArm R:double="To Float (Boolean)"(InBool=AND Boolean!(A=@n13, B=@n15, C:bool=@n10, D:bool=@n14) -> (ReturnValue))
   ) -> (Output_Get:double)
   n16 = Get bIsParachuteOpened()
   n17 = OR Boolean!(A=Get bIsPlayingEmote(), B=Get bIsSwimSprinting() -> (bIsSwimSprinting:bool), C:bool=Get bIsSkyDiving(), D:bool=Get bIsMoving2D()) -> (ReturnValue)
   n18 = GetDeltaTransformFromRefPose(self=GetOwnerSkeletalMeshComponent(), BoneName=upperarm_r, BaseName=spine_05)
   n19 = "float < float"(A=BreakRotator(InRot=BreakTransform(InTransform=@n18).Rotation).Yaw, B=-10)
   n20 = Branch(execute=@n9.then_8, Condition=AND Boolean!(A=AND Boolean(A=@n17, B=@n19), B=NOT Boolean(A=@n16)) -> (ReturnValue))
   > Set Alpha_DetectedArmBack_R(Alpha_DetectedArmBack_R:double=1) -> (Output_Get:double)
   Set Alpha_DetectedArmBack_R(execute=@n20.else, Alpha_DetectedArmBack_R=0)
   n21 = GetOwnerSkeletalMeshComponent()
   n22 = BreakRotator(InRot=BreakTransform(InTransform=GetDeltaTransformFromRefPose(self=@n21, BoneName=neck_02, BaseName=spine_05)).Rotation)
   n23 = Branch(execute=@n9.then_4, Condition=AND Boolean(A="float < float"(A=@n22.Yaw, B=-14), B=Get Alpha_backPack()))
   > Set Alpha_HeadRotateUpBackpack(Alpha_HeadRotateUpBackpack:double=1) -> (Output_Get:double)
   Set Alpha_HeadRotateUpBackpack(execute=@n23.else, Alpha_HeadRotateUpBackpack=0)
   n24 = Get bRigidBodyEnabled()
   n25 = Get bIsSkyDiving()
   n26 = GetOwnerSkeletalMeshComponent()
   n27 = BreakRotator(InRot=BreakTransform(InTransform=GetDeltaTransformFromRefPose(self=@n26, BoneName=clavicle_r, BaseName=spine_05)).Rotation)
   n28 = Branch(
    execute=@n9.then_9
    Condition=AND Boolean!(A="float < float"(A=@n27.Yaw, B=-10), B=Get bIsInFrontEnd(), C:bool=NOT Boolean(A=@n25), D:bool=NOT Boolean(A=@n24)) -> (ReturnValue)
   )
   > Set Alpha_DetectedClavicleBack_R(Alpha_DetectedClavicleBack_R:double=1) -> (Output_Get:double)
   Set Alpha_DetectedClavicleBack_R(execute=@n28.else, Alpha_DetectedClavicleBack_R=0)
   n29 = CheckForBackbling()
   n30 = Set isWearingBackbling(execute=@n9.then_0, isWearingBackbling=@n29)
   > n31 = Set Is Wearing Cape(Is Wearing Cape=OR Boolean!(A=@n29.bIsBackblingACape, B=@n29.bIsBackblingLong) -> (ReturnValue))
   > Set Alpha_backPack(Alpha_backPack:bool=AND Boolean!(A=@n30, B=NOT Boolean(A=@n31)) -> (ReturnValue)) -> (Output_Get:bool)
   n32 = Get bIsParachuteOpened()
   n33 = GetOwnerSkeletalMeshComponent()
   n34 = BreakRotator(InRot=BreakTransform(InTransform=GetDeltaTransformFromRefPose(self=@n33, BoneName=upperarm_l, BaseName=spine_05)).Rotation)
   n35 = AND Boolean(A="float < float"(A=@n34.Yaw, B=-10), B=@n17)
   n36 = AND Boolean(A=Get bIsSkydiveFloating(), B="float > float"(A=@n34.Yaw, B=100))
   n37 = Branch(execute=@n9.then_7, Condition=AND Boolean(A=NOT Boolean(A=@n32), B=OR Boolean(A=@n36, B=@n35)))
   > Set Alpha_DetectedArmBack_L(Alpha_DetectedArmBack_L:double=1) -> (Output_Get:double)
   Set Alpha_DetectedArmBack_L(execute=@n37.else, Alpha_DetectedArmBack_L=0)
   n38 = NOT Boolean(A=Get bIsSurfaceSwimming())
   n39 = NOT Boolean(A=Get bIsPlayingEmote())
   n40 = NOT Boolean(A=Get bIsSkyDiving())
   n41 = NOT Boolean(A=Get bIsMoving2D())
   n42 = GetDeltaTransformFromRefPose(self=GetOwnerSkeletalMeshComponent(), BoneName=upperarm_l, BaseName=spine_05)
   n43 = "float > float"(A=BreakRotator(InRot=BreakTransform(InTransform=@n42).Rotation), B=40)
   n44 = Get Alpha_backPack()
   Set Alpha_FrontEndIdle(
    execute=@n9.then_5
    Alpha_FrontEndIdle:double="To Float (Boolean)"(InBool=AND Boolean!(A=@n43, B=@n39, C:bool=@n40, D:bool=@n41, E:bool=@n38, F:bool=@n44) -> (ReturnValue))
   ) -> (Output_Get:double)
   n45 = GetOwnerSkeletalMeshComponent()
   Set Test_DetectedRotAvocadoEmote(
    execute=@n9.then_6
    Test_DetectedRotAvocadoEmote:bool=Montage_IsPlaying(self=GetAnimInstance(self=@n45), Montage=/BRCosmetics/Animation/Game/MainPlayer/Emotes/AvocadoSeal/CMF/Emote_AvocadoSeal_CMF_M)
   ) -> (Output_Get:bool)
   n46 = GetOwnerSkeletalMeshComponent()
   n47 = BreakRotator(InRot=BreakTransform(InTransform=GetDeltaTransformFromRefPose(self=@n46, BoneName=neck_01, BaseName=pelvis)).Rotation)
   > n48 = Branch(Condition=AND Boolean!(A="float > float"(A=@n47, B=14), B=Get Test_DetectedRotAvocadoEmote() -> (Test_DetectedRotAvocadoEmote:bool), C=Get bRigidBodyEnabled()) -> (ReturnValue))
   > Set Alpha_DetectedAvocadoEmote(Alpha_DetectedAvocadoEmote:double=1) -> (Output_Get:double)
   Set Alpha_DetectedAvocadoEmote(execute=@n48.else, Alpha_DetectedAvocadoEmote=0)
   n49 = NOT Boolean(A=Get Is Wearing Cape())
   n50 = "float > float"(A=@n12, B=12)
   n51 = Get bIsPlayingMeleeAnim()
   n52 = Get Selected_Pickaxe()
   Set Alpha_CollisionHairAndArm L(
    execute=@n9.then_1
    Alpha_CollisionHairAndArm L:double="To Float (Boolean)"(InBool=AND Boolean!(A=@n50, B=@n51, C=@n49, D:bool=@n52) -> (ReturnValue))
   ) -> (Output_Get:double)
   n53 = Branch(execute=@n9.then_3, Condition=AND Boolean!(A="float < float"(A=@n22.Yaw, B=-30), B=Get bIsCrouching()) -> (ReturnValue))
   > Set Aplha_HeadRotateUpExtrem(Aplha_HeadRotateUpExtrem:double=1) -> (Output_Get:double)
   > Set Alpha_ConstrainHeadRotateUpExtrem(Alpha_ConstrainHeadRotateUpExtrem:double=0.7) -> (Output_Get:double)
   Set Aplha_HeadRotateUpExtrem(execute=@n53.else, Aplha_HeadRotateUpExtrem=0)
   > Set Alpha_ConstrainHeadRotateUpExtrem(Alpha_ConstrainHeadRotateUpExtrem=1)
   n54 = Get bIsSkydiveDiveMode()
   n55 = Get bIsSkydiveFloating()
   Set Alpha_FloatingAndSkydiveMode(execute=@n1.then_6, Alpha_FloatingAndSkydiveMode:double="To Float (Boolean)"(InBool=OR Boolean(A=@n54, B=@n55))) -> (Output_Get:double)
   n56 = Branch(execute=@n1.then_10, Condition=NOT Boolean(A=Get bRigidBodyEnabled()))
   > Set Alpha_RBDisable(Alpha_RBDisable:double=1) -> (Output_Get:double)
   > Set Alpha_RBEnable(Alpha_RBEnable:bool=false) -> (Output_Get:bool)
   Set Alpha_RBDisable(execute=@n56.else, Alpha_RBDisable=0)
   > Set Alpha_RBEnable(Alpha_RBEnable=true)
   Set neck_rot(
    execute=@n1.then_12
    neck_rot:double=GetAlphaFromReferenceTransformDelta(SkeletalMeshComponent=GetOwnerSkeletalMeshComponent(), BoneName=head, BaseName=spine_05, InputRange=(X=-10,Y=-40), TwistAxis="0,0,1")
   ) -> (Output_Get:double)
   n57 = Get bIsSkyDiving()
   n58 = Get bIsInFrontEnd()
   n59 = Get Alpha_backPack()
   Set Alpha_SkydivingFrontEnd(execute=@n1.then_11, Alpha_SkydivingFrontEnd:double="To Float (Boolean)"(InBool=AND Boolean!(A=@n57, B=@n58, C:bool=@n59) -> (ReturnValue))) -> (Output_Get:double)
   n60 = Get bIsSwimSprinting()
   n61 = Get Alpha_backPack()
   Set Alpha_SwimSprinting with backpack(execute=@n1.then_9, Alpha_SwimSprinting with backpack:double="To Float (Boolean)"(InBool=AND Boolean!(A=@n60, B=@n61) -> (ReturnValue))) -> (Output_Get:double)
   n62 = Branch(execute=@n1.then_7, Condition=AND Boolean(A=Get isWearingBackbling(), B=Get bIsPlayingEmote()))
   > Set Alpha_EmoteBackpack(Alpha_EmoteBackpack:double=1) -> (Output_Get:double)
   > Set Alpha_NotEmoteBackpack(Alpha_NotEmoteBackpack:double=0) -> (Output_Get:double)
   > Set Alpha_ConstrainBackpack(Alpha_ConstrainBackpack:double=0) -> (Output_Get:double)
   Set Alpha_EmoteBackpack(execute=@n62.else, Alpha_EmoteBackpack=0)
   > Set Alpha_NotEmoteBackpack(Alpha_NotEmoteBackpack=1)
   > Set Alpha_ConstrainBackpack(Alpha_ConstrainBackpack=0.5)
   Event OnWeaponChanged()
   > n63 = Cast To FortPlayerAnimInstance(Object=GetAnimInstance(self=GetOwnerSkeletalMeshComponent()))
   > Set Selected_Pickaxe(Selected_Pickaxe:bool=false) -> (Output_Get:bool)
   > Switch on EFortWeaponCoreAnimation!(execute, Selection:EFortWeaponCoreAnimation Enum=Get WeaponCoreAnim(self=@n63)) -> (
    Melee:Exec
    Pistol:Exec
    Shotgun:Exec
    PaperBlueprint:Exec
    Rifle:Exec
    MeleeOneHand:Exec
    MachinePistol:Exec
    RocketLauncher:Exec
    GrenadeLauncher:Exec
    GoingCommando:Exec
    AssaultRifle:Exec
    TacticalShotgun:Exec
    SniperRifle:Exec
    TrapPlacement:Exec
    ShoulderLauncher:Exec
    AbilityDecoTool:Exec
    Crossbow:Exec
    C4:Exec
    RemoteControl:Exec
    DualWield:Exec
    AR_BullPup:Exec
    AR_ForwardGrip:Exec
    MedPackPaddles:Exec
    SMG_P90:Exec
    AR_DrumGun:Exec
    Consumable_Small:Exec
    Consumable_Large:Exec
    Balloon:Exec
    MountedTurret:Exec
    CreativeTool:Exec
    ExplosiveBow:Exec
    AshtonIndigo:Exec
    AshtonChicago:Exec
    MeleeDualWield:Exec
    Unarmed:Exec
   )
   > Set Selected_Pickaxe(Selected_Pickaxe=1)
   Event BlueprintInitializeAnimation()
   > Set SkelMesh(SkelMesh=GetOwnerSkeletalMeshComponent())
   n64 = Get SkelMesh()
   Set upperarm_L_rot(upperarm_L_rot:double=GetAlphaFromReferenceTransformDelta(SkeletalMeshComponent=@n64, BoneName=spine_02, BaseName=upperarm_l, InputRange=(X=0,Y=-60), TwistAxis="0,0,1")) -> (Output_Get:double)
   > Set upperarm_R_rot(upperarm_R_rot:double=GetAlphaFromReferenceTransformDelta(SkeletalMeshComponent=@n64, BoneName=spine_02, BaseName=upperarm_r, InputRange=(X=0,Y=-60), TwistAxis="0,0,1")) -> (Output_Get:double)
   Get RBAN Params()
   Get Gravity Override Params()
  ]}
]
variables [
 Alpha_backPack {type=boolean}
 Alpha_CollisionHairAndArm L {type=number}
 Alpha_CollisionHairAndArm R {type=number}
 Alpha_ConstrainBackpack {type=number}
 Alpha_ConstrainHeadRotateUpExtrem {type=number}
 Alpha_DetectedArmBack_L {type=number}
 Alpha_DetectedArmBack_R {type=number}
 Alpha_DetectedAvocadoEmote {type=number}
 Alpha_DetectedClavicleBack_R {type=number}
 Alpha_EmoteBackpack {type=number}
 Alpha_FloatingAndSkydiveMode {type=number}
 Alpha_FrontEndIdle {type=number}
 Alpha_HeadRotateUpBackpack {type=number}
 Alpha_Jumping in Tactical Sprinting {type=number}
 Alpha_JustSprinting {type=number}
 Alpha_NotEmoteBackpack {type=number}
 Alpha_PlayingMelee {type=number}
 Alpha_RBDisable {type=number}
 Alpha_RBEnable {type=boolean}
 Alpha_SkydivingFrontEnd {type=number}
 Alpha_SwimSprinting with backpack {type=number}
 Aplha_HeadRotateUpExtrem {type=number}
 Body_Mesh {type=SkeletalMeshComponent}
 Constant {type=Vector}
 Gravity Override Params {type=GravityOverrideParamsStruct}
 GravityOverride {type=Vector}
 headUpRot {type=number}
 Is Wearing Cape {type=boolean}
 isWearingBackbling {type=boolean}
 neck_rot {type=number}
 pawnSpeed2D_divider {type=number}
 RBAN Params {
  type=RBANParamsStruct
  default {
   dBNO {simSpaceSettings {
     maxAngularAcceleration=1000
     maxAngularVelocity=3
     maxLinearAcceleration=300
     maxLinearVelocity=700
     velocityScaleZ=0
     externalLinearDragV=<0 0.5 0>
    }}
   dBNO_Carried {simSpaceSettings {
     maxAngularAcceleration=1000
     maxAngularVelocity=3
     maxLinearAcceleration=300
     maxLinearVelocity=700
     velocityScaleZ=0
     externalLinearDragV=<0 0.5 0>
    }}
   emoteOrMelee {simSpaceSettings {
     maxAngularAcceleration=1000
     maxAngularVelocity=3
     maxLinearAcceleration=300
     maxLinearVelocity=700
     velocityScaleZ=0
     externalLinearDragV=<0 0.5 0>
    }}
   front_End {simSpaceSettings {
     maxAngularAcceleration=1000
     maxAngularVelocity=3
     maxLinearAcceleration=300
     maxLinearVelocity=700
     velocityScaleZ=0
     externalLinearDragV=<0 0.5 0>
    }}
   onGround_Crouch_Moving {simSpaceSettings {
     maxAngularAcceleration=1000
     maxAngularVelocity=3
     maxLinearAcceleration=300
     maxLinearVelocity=700
     velocityScaleZ=0
     externalLinearDragV=<0 0.5 0>
    }}
   onGround_Falling {simSpaceSettings {
     maxAngularAcceleration=1000
     maxAngularVelocity=3
     maxLinearAcceleration=300
     maxLinearVelocity=700
     velocityScaleZ=0
     externalLinearDragV=<0 0.5 0>
    }}
   onGround_Moving {simSpaceSettings {
     maxAngularAcceleration=1000
     maxAngularVelocity=3
     maxLinearAcceleration=300
     maxLinearVelocity=700
     velocityScaleZ=0
     externalLinearDragV=<0 0.5 0>
    }}
   onGround_Standing {simSpaceSettings {
     maxAngularAcceleration=1000
     maxAngularVelocity=3
     maxLinearAcceleration=300
     maxLinearVelocity=700
     velocityScaleZ=0
     externalLinearDragV=<0 0.5 0>
    }}
   rail_Grinding {simSpaceSettings {
     maxAngularAcceleration=1000
     maxAngularVelocity=3
     maxLinearAcceleration=300
     maxLinearVelocity=700
     velocityScaleZ=0
     worldAlpha=0.1
     externalLinearDragV=<0 0.5 0>
    }}
   riding_Moving {simSpaceSettings {
     maxAngularAcceleration=1000
     maxAngularVelocity=3
     maxLinearAcceleration=300
     maxLinearVelocity=700
     velocityScaleZ=0
     externalLinearDragV=<0 0.5 0>
    }}
   riding_Moving_Targeting {simSpaceSettings {
     maxAngularAcceleration=1000
     maxAngularVelocity=3
     maxLinearAcceleration=300
     maxLinearVelocity=700
     velocityScaleZ=0
     externalLinearDragV=<0 0.5 0>
    }}
   riding_Standing {simSpaceSettings {
     maxAngularAcceleration=1000
     maxAngularVelocity=3
     maxLinearAcceleration=300
     maxLinearVelocity=700
     velocityScaleZ=0
     externalLinearDragV=<0 0.5 0>
    }}
   skydive_DivingDown {simSpaceSettings {
     maxAngularAcceleration=1000
     maxAngularVelocity=3
     maxLinearAcceleration=300
     maxLinearVelocity=700
     velocityScaleZ=0
     externalLinearDragV=<0 0.5 0>
    }}
   skydive_FloatingDown {
    componentLinearAccScale=<0 0.5 0>
    componentLinearVelScale=<0 0.5 0>
    simSpaceSettings {
     maxAngularAcceleration=1000
     maxAngularVelocity=3
     maxLinearAcceleration=300
     maxLinearVelocity=700
     velocityScaleZ=0
     externalLinearDragV=<0 0.5 0>
    }
   }
   skydive_GoingUp {simSpaceSettings {
     maxAngularAcceleration=1000
     maxAngularVelocity=3
     maxLinearAcceleration=300
     maxLinearVelocity=700
     velocityScaleZ=0
     externalLinearDragV=<0 0.5 0>
    }}
   skydive_Parachuting {
    componentLinearAccScale=<0 0.1 0.3>
    componentLinearVelScale=<0 0.1 0.3>
    simSpaceSettings {
     maxAngularAcceleration=1000
     maxAngularVelocity=3
     maxLinearAcceleration=300
     maxLinearVelocity=700
     velocityScaleZ=0
     externalLinearDragV=<0 0.5 0>
    }
   }
   skydive_Surfing {simSpaceSettings {
     maxAngularAcceleration=1000
     maxAngularVelocity=3
     maxLinearAcceleration=300
     maxLinearVelocity=700
     velocityScaleZ=0
     externalLinearDragV=<0 0.5 0>
    }}
   swimming {simSpaceSettings {
     maxAngularAcceleration=1000
     maxAngularVelocity=3
     maxLinearAcceleration=300
     maxLinearVelocity=700
     velocityScaleZ=0
     externalLinearDragV=<0 0.5 0>
    }}
   tactical_Sprint {simSpaceSettings {
     maxAngularAcceleration=1000
     maxAngularVelocity=3
     maxLinearAcceleration=300
     maxLinearVelocity=700
     velocityScaleZ=0
     externalLinearDragV=<0 0.5 0>
    }}
   ziplining {simSpaceSettings {
     maxAngularAcceleration=1000
     maxAngularVelocity=3
     maxLinearAcceleration=300
     maxLinearVelocity=700
     velocityScaleZ=0
     externalLinearDragV=<0 0.5 0>
    }}
  }
 }
 RBAN Settings Out {type=FortRigidBodyAnimNodeParameters}
 RBAN Settings Out Deprecated {
  type=RBANSettingStruct
  default {
   componentLinearAccScale=<0 0 0>
   componentLinearVelScale=<0 0 0>
   simSpaceSettings {
    dampingAlpha=1
    maxAngularAcceleration=1000
    maxAngularVelocity=3
    maxLinearAcceleration=300
    maxLinearVelocity=700
    velocityScaleZ=0
    worldAlpha=1
    externalAngularVelocity=<0 0 0>
    externalLinearDragV=<0 0.5 0>
    externalLinearVelocity=<0 0 0>
   }
  }
 }
 Selected_Pickaxe {type=boolean}
 SkelMesh {type=SkeletalMeshComponent}
 Test_DetectedRotAvocadoEmote {type=boolean}
 upperarm_L_rot {type=number}
 upperarm_R_rot {type=number}
]
