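# Control Rig for the FlossYawn face accessory mesh that corrects both ponytail base bones as the upper body twists during melee.
# Hair Fix calls Delta Rotation, which swing-twist decomposes spine_05 relative to a pelvis-parented null and remaps the twist to a clamped weight.
# Opposite twist ranges drive left and right rotation offsets, each scaled by PlayingMelee.
# A debug branch, switched off, can drive spine_05 from a control.
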ControlRigBlueprint /BRCosmetics/Characters/Player/Female/Medium/Bodies/F_MED_FlossYawn/Meshes/Parts/F_MED_FlossYawn_FaceAcc_CtrlRig
graphs [
 New Function_ContainedGraph {nodes [
   n1 = RigVMFunctionEntryNode()
   Clamp = &RigVMFunction_MathIntClamp(Value=@n1.primAxis)
   RigVMFunction_MathBoolToInteger = &RigVMFunction_MathBoolToInteger(Value=@n1.onlyTwist)
   RigVMFunction_MathQuaternionSwingTwist = &RigVMFunction_MathQuaternionSwingTwist(
    Input=&RigUnit_GetRelativeTransformForItem(
     Child=(Type=Bone,Name="None")
     Parent=(Type=Bone,Name="None")
     bChildInitial=False
     bParentInitial=False
     Parent.Type=@n1.ParentType
     Child.Name=@n1.child
     Parent.Name=@n1.parent
    ).RelativeTransform.Rotation
    TwistAxis=@n1.twistAxis
   )
   RigVMFunction_MathQuaternionToEuler = &RigVMFunction_MathQuaternionToEuler(
    Value=RigVMDispatchNode(
     Values=((),())
     Index=@RigVMFunction_MathBoolToInteger.Result
     Values.0=@RigVMFunction_MathQuaternionSwingTwist.Swing
     Values.1=@RigVMFunction_MathQuaternionSwingTwist.Twist
    ) {template="DISPATCH_RigVMDispatch_SelectInt32(in Index,in Values,out Result)"}.Result
    RotationOrder=@n1.rotationOrder
   )
   DISPATCH_RigVMDispatch_SelectInt32 = RigVMDispatchNode(
    Values=(0,0,0)
    Index=@Clamp.Result
    Values.0=@RigVMFunction_MathQuaternionToEuler.Result.X
    Values.1=@RigVMFunction_MathQuaternionToEuler.Result.Y
    Values.2=@RigVMFunction_MathQuaternionToEuler.Result.Z
   ) {template="DISPATCH_RigVMDispatch_SelectInt32(in Index,in Values,out Result)"}
   Remap = &RigVMFunction_MathDoubleRemap(
    TargetMaximum=1
    TargetMinimum=0
    bClamp=False
    Value=@DISPATCH_RigVMDispatch_SelectInt32.Result
    SourceMinimum=@n1.min
    SourceMaximum=@n1.max
   )
   > RigVMFunctionReturnNode(ouputWeight=&RigVMFunction_MathDoubleClamp(Value=@Remap.Result).Result)
  ]}
 New Function_ContainedGraph {nodes [
   RigVMFunctionEntryNode()
   > Delta Rotation = RigVMFunctionReferenceNode(
    ParentType=Null
    child=spine_05
    debugRotation=false
    max=-50
    onlyTwist=true
    parent=spine_05_null
    primAxis=0
    rotationOrder=XYZ
    twistAxis=(X=1,Y=0,Z=0)
   )
   Multiply = &RigVMFunction_MathDoubleMul(A=@"Delta Rotation".ouputWeight, B=RigVMVariableNode() {variable=PlayingMelee}.Value)
   > &RigUnit_ModifyTransforms(
    ItemToModify=((Item=(Type=Bone,Name="base_ponytail_l"),Transform=(Rotation=(X=0.1238,Y=0.3618,Z=-0.3748,W=0.8446),Translation=(X=0,Y=0,Z=0),Scale3D=(X=1,Y=1,Z=1))))
    WeightMinimum=0
    Weight=@Multiply.Result
   )
   > Delta Rotation_1 = RigVMFunctionReferenceNode(
    ParentType=Null
    child=spine_05
    debugRotation=false
    max=40
    min=10
    onlyTwist=true
    parent=spine_05_null
    primAxis=0
    rotationOrder=XYZ
    twistAxis=(X=1,Y=0,Z=0)
   )
   > &RigUnit_ModifyTransforms(
    ItemToModify=((Item=(Type=Bone,Name="base_ponytail_r"),Transform=(Rotation=(X=0.0868,Y=0.1504,Z=-0.4924,W=0.8529),Translation=(X=0,Y=0,Z=0),Scale3D=(X=1,Y=1,Z=1))))
    WeightMinimum=0
    Weight=&RigVMFunction_MathDoubleMul(A=@"Delta Rotation_1".ouputWeight, B=RigVMVariableNode() {variable=PlayingMelee}.Value).Result
   )
   > RigVMFunctionReturnNode()
  ]}
 New Function_ContainedGraph {nodes [
   n1 = RigVMFunctionEntryNode()
   RigVMFunction_ControlFlowBranch = &RigVMFunction_ControlFlowBranch(Condition=@n1.DebugControls, ExecuteContext=@n1.ExecuteContext)
   RigVMFunctionReturnNode(ExecuteContext=@RigVMFunction_ControlFlowBranch.Completed)
   RigUnit_GetTransform = &RigUnit_GetTransform(Item=(Type=Control,Name="spine_05_ctrl"), bInitial=False)
   RigUnit_SetTransform = &RigUnit_SetTransform(
    Item=(Type=Bone,Name="spine_05")
    Weight=1
    bInitial=False
    Value=@RigUnit_GetTransform.Transform
    ExecutePin=@RigVMFunction_ControlFlowBranch.True
   )
  ]}
 RigVMFunctionLibrary {nodes [
   Controls = RigVMCollapseNode(DebugControls=False, ExecuteContext=())
   Delta Rotation = RigVMCollapseNode(ExecuteContext=(), debugRotation=False, onlyTwist=False, twistAxis=(X=0,Y=0,Z=0))
   Hair Fix = RigVMCollapseNode(ExecuteContext=())
  ]}
 RigVMModel {nodes [
   RigUnit_BeginExecution = &RigUnit_BeginExecution()
   Controls = RigVMFunctionReferenceNode(DebugControls=false, ExecuteContext=@RigUnit_BeginExecution.ExecutePin)
   > Hair Fix = RigVMFunctionReferenceNode()
  ]}
]
hierarchy [
 root {loc=<0 0 0>, rot=<0 0 0 1>}
 pelvis {parent=root, loc=<0 -0.5301 98.0571>, rot=<0.0127 0.707 -0.0127 -0.707>}
 spine_01 {parent=pelvis, loc=<6.1208 -0.9328 0>, rot=<0 0 0.034 -0.9994>}
 spine_02 {parent=spine_01, loc=<6.5997 -0.2095 0>, rot=<0 0 0.0238 -0.9997>}
 spine_03 {parent=spine_02, loc=<6.4914 -0.1237 0>, rot=<0 0 0.0149 -0.9999>}
 spine_04 {parent=spine_03, loc=<6.9598 0.0794 0>, rot=<0 0 -0.0132 -0.9999>}
 spine_05 {parent=spine_04, loc=<6.9987 0 0>, rot=<0 0 -0.0015 -1>}
 neck_01 {parent=spine_05, loc=<11.3138 0.1529 0>, rot=<0 0 -0.2003 -0.9797>}
 neck_02 {parent=neck_01, loc=<4.5627 0 0>, rot=<0 0 0.0504 -0.9987>}
 head {parent=neck_02, loc=<5.3914 0 0>, rot=<0 0 0.115 -0.9934>}
 base_ponytail_l {parent=head, loc=<15.0768 -10.3102 -10.5716>, rot=<0.0041 -1 0 0>}
 dyn_ponytail_01_l {parent=base_ponytail_l, loc=<0 0 0>, rot=<0 0 0 1>}
 dyn_ponytail_02_l {parent=dyn_ponytail_01_l, loc=<13.0575 -5.8894 5.1035>, rot=<0 0 0 1>}
 dyn_ponytail_03_l {parent=dyn_ponytail_02_l, loc=<22.8758 -2.0748 2.6083>, rot=<0 0 0 1>}
 dyn_ponytail_04_l {parent=dyn_ponytail_03_l, loc=<15.3042 -2.0748 2.6083>, rot=<0 0 0 1>}
 dyn_ponytail_05_l {parent=dyn_ponytail_04_l, loc=<15.3042 -2.0748 2.6083>, rot=<0 0 0 1>}
 dyn_ponytail_06_l {parent=dyn_ponytail_05_l, loc=<15.3042 -2.0748 2.6083>, rot=<0.9769 -0.0461 -0.1985 -0.0642>}
 dyn_ponytail_strand_b_l {parent=dyn_ponytail_06_l, loc=<15.555 0 0>, rot=<0 0 0 1>}
 dyn_ponytail_strand_a_l {parent=dyn_ponytail_06_l, loc=<10.2114 -1.2188 1.7308>, rot=<0 0 0 1>}
 base_ponytail_r {parent=head, loc=<15.0765 -10.3102 10.5716>, rot=<1 0.0041 0 0>}
 dyn_ponytail_01_r {parent=base_ponytail_r, loc=<0 0 0>, rot=<0 0 0 -1>}
 dyn_ponytail_02_r {parent=dyn_ponytail_01_r, loc=<-13.057 5.8894 -5.1035>, rot=<0 0 0 -1>}
 dyn_ponytail_03_r {parent=dyn_ponytail_02_r, loc=<-22.876 2.0747 -2.6083>, rot=<0 0 0 -1>}
 dyn_ponytail_04_r {parent=dyn_ponytail_03_r, loc=<-15.304 2.0748 -2.6083>, rot=<0 0 0 -1>}
 dyn_ponytail_05_r {parent=dyn_ponytail_04_r, loc=<-15.304 2.0747 -2.6083>, rot=<0 0 0 -1>}
 dyn_ponytail_06_r {parent=dyn_ponytail_05_r, loc=<-15.3045 2.0748 -2.6083>, rot=<0 0 0 -1>}
 dyn_ponytail_strand_b_r {parent=dyn_ponytail_06_r, loc=<-1.4301 4.6607 0.3978>, rot=<0.655 0.2665 -0.655 -0.2665>}
 dyn_ponytail_strand_a_r {parent=dyn_ponytail_06_r, loc=<-0.861 0.4434 4.2351>, rot=<0.655 0.2665 -0.655 -0.2665>}
 base_hair_sd_a_l {parent=head, loc=<11.0224 11.0712 -3.2186>, rot=<0.1058 -0.87 -0.4081 -0.2557>}
 dyn_hair_sd_a_01_l {parent=base_hair_sd_a_l, loc=<0 0 0>, rot=<0 0 0 1>}
 dyn_hair_sd_a_02_l {parent=dyn_hair_sd_a_01_l, loc=<7.7981 0 0>, rot=<0 -0.2162 0 -0.9764>}
 base_hair_sd_c_l {parent=head, loc=<3.23 5.1016 -9.8626>, rot=<0.0549 -0.9939 -0.0749 0.0588>}
 dyn_hair_sd_c_01_l {parent=base_hair_sd_c_l, loc=<0 0 0>, rot=<0 0 0 1>}
 base_hair_fr_a_c {parent=head, loc=<12.1169 11.2754 1.8468>, rot=<-0.1079 -0.8981 0.424 -0.0445>}
 dyn_hair_fr_a_01_c {parent=base_hair_fr_a_c, loc=<0 0 0>, rot=<0 0 0 1>}
 dyn_hair_fr_a_02_c {parent=dyn_hair_fr_a_01_c, loc=<3.6872 0 0>, rot=<0 0 0 1>}
 dyn_hair_fr_a_03_c {parent=dyn_hair_fr_a_02_c, loc=<3.2991 0 0>, rot=<0 0 0 1>}
 base_hair_sd_b_l {parent=head, loc=<4.6998 7.7249 -9.1055>, rot=<-0.0221 -0.9952 -0.0701 0.0644>}
 dyn_hair_sd_b_01_l {parent=base_hair_sd_b_l, loc=<0 0 0>, rot=<0 0 0 1>}
 dyn_hair_sd_b_02_l {parent=dyn_hair_sd_b_01_l, loc=<3.7485 0 0>, rot=<0 -0.1063 0 -0.9943>}
 dyn_hair_sd_b_03_l {parent=dyn_hair_sd_b_02_l, loc=<3.3812 0 0>, rot=<0 0 0 1>}
 base_hair_sd_b_r {parent=head, loc=<11.8522 6.2279 9.0315>, rot=<0.0133 -0.9569 -0.0345 0.2881>}
 dyn_hair_sd_b_01_r {parent=base_hair_sd_b_r, loc=<0 0 -0.5108>, rot=<0 0 0 1>}
 dyn_hair_sd_b_02_r {parent=dyn_hair_sd_b_01_r, loc=<3.5803 0 0>, rot=<0 0.1071 0 -0.9943>}
 dyn_hair_sd_b_03_r {parent=dyn_hair_sd_b_02_r, loc=<3.4652 0 0.4705>, rot=<0 0 0 1>}
 base_hair_sd_a_r {parent=head, loc=<4.9667 7.1642 9.2394>, rot=<0.0041 -1 0 0>}
 dyn_hair_sd_a_r {parent=base_hair_sd_a_r, loc=<0 0 0>, rot=<0 0 0 1>}
 dyn_hair_sd_a_01_r {parent=dyn_hair_sd_a_r, loc=<5.3338 2.15 0.9263>, rot=<0 0.1136 0 -0.9935>}
 dyn_hair_sd_a_02_r {parent=dyn_hair_sd_a_r, loc=<5.3214 -1.7411 -0.2785>, rot=<0 0.1136 0 -0.9935>}
 base_hair_fr_b_l {parent=head, loc=<6.3008 10.4545 6.1057>, rot=<0.0034 -0.8347 0.5507 -0.0023>}
 dyn_hair_fr_b_l {parent=base_hair_fr_b_l, loc=<0 0 0>, rot=<0 0 0 1>}
 base_middle_collision {parent=head, loc=<-29.9526 -5.9061 0>, rot=<0 0 0.036 -0.9994>}
 base_backHair_l {parent=head, loc=<-1.7018 -5.3809 -1.8485>, rot=<-0.3963 0.3167 0.8483 -0.1516>}
 base_backHair_r {parent=head, loc=<-1.7663 -5.3194 1.9838>, rot=<0.1647 -0.8335 -0.3204 0.4189>}
 ponytail_ASW_01_l {parent=neck_01, loc=<-16.7814 -10.9132 -18.2024>, rot=<-0.1137 -0.6979 0.1137 -0.6979>}
 ponytail_ASW_01_r {parent=neck_01, loc=<-16.7814 -10.9132 18.2024>, rot=<0.6979 -0.1137 -0.6979 -0.1137>}
 clavicle_l {parent=spine_05, loc=<5.4831 -0.1053 -3.2946>, rot=<0.0465 -0.7289 -0.009 -0.6829>}
 upperarm_l {parent=clavicle_l, loc=<9.8658 -1.2136 -0.4105>, rot=<0.0164 -0.3245 -0.0269 -0.9454>}
 lowerarm_l {parent=upperarm_l, loc=<26.465 -0.8345 0.259>, rot=<0.0036 0.0161 -0.2298 -0.9731>}
 hand_l {parent=lowerarm_l, loc=<22.7355 -0.3203 -0.6105>, rot=<-0.6784 -0.0223 0.0735 -0.7307>}
 thumb_01_l {parent=hand_l, loc=<1.3118 -0.3191 -2.459>, rot=<0.6546 -0.4042 0.0664 -0.6354>}
 thumb_02_l {parent=thumb_01_l, loc=<5.9336 -0.0448 0.0401>, rot=<-0.0037 0.0213 0.298 -0.9543>}
 thumb_03_l {parent=thumb_02_l, loc=<2.2223 0.3682 -0.0907>, rot=<0.0118 -0.013 0.0066 -0.9998>}
 index_metacarpal_l {parent=hand_l, loc=<1.6535 0.0183 -2.96>, rot=<0.0158 -0.0124 -0.0173 -0.9996>}
 index_01_l {parent=index_metacarpal_l, loc=<8.3008 0.0357 -0.0524>, rot=<-0.0044 0.0485 0.0924 -0.9945>}
 index_02_l {parent=index_01_l, loc=<3.3587 -0.3004 0.0452>, rot=<-0.002 0.0027 0.3225 -0.9466>}
 index_03_l {parent=index_02_l, loc=<2.5578 -0.123 -0.0252>, rot=<-0.002 0.0067 0.2558 -0.9667>}
 middle_metacarpal_l {parent=hand_l, loc=<2.1856 -0.1725 -0.5984>, rot=<0.006 0.0168 -0.0329 -0.9993>}
 middle_01_l {parent=middle_metacarpal_l, loc=<7.3275 -0.2447 0.0129>, rot=<-0.0208 0.0042 0.0858 -0.9961>}
 middle_02_l {parent=middle_01_l, loc=<4.0123 0.1401 0.1175>, rot=<-0.0218 0.0258 0.3512 -0.9357>}
 middle_03_l {parent=middle_02_l, loc=<2.6077 0.0534 0.1373>, rot=<0.0082 0.0188 0.1708 -0.9851>}
 ring_metacarpal_l {parent=hand_l, loc=<2.2074 -0.3625 1.3638>, rot=<-0.0284 0.0722 -0.0458 -0.9959>}
 ring_01_l {parent=ring_metacarpal_l, loc=<7.0295 -0.6197 -0.1446>, rot=<-0.0489 -0.0252 0.1892 -0.9804>}
 ring_02_l {parent=ring_01_l, loc=<3.6174 0.4647 -0.1027>, rot=<0 0 0.2806 -0.9598>}
 ring_03_l {parent=ring_02_l, loc=<2.4152 0.114 0.0199>, rot=<0 0 0.1189 -0.9929>}
 pinky_metacarpal_l {parent=hand_l, loc=<1.7616 -0.9753 3.3814>, rot=<-0.0397 0.0814 -0.069 -0.9935>}
 pinky_01_l {parent=pinky_metacarpal_l, loc=<7.1709 -0.4831 -0.1094>, rot=<-0.0654 -0.0035 0.2242 -0.9723>}
 pinky_02_l {parent=pinky_01_l, loc=<2.8274 0.0893 -0.1043>, rot=<0.0037 0.0011 0.2854 -0.9584>}
 pinky_03_l {parent=pinky_02_l, loc=<2.2046 0.0307 -0.0078>, rot=<-0.0002 0.0098 0.1445 -0.9895>}
 lowerarm_twist_02_l {parent=lowerarm_l, loc=<10.6707 0 0>, rot=<0 0 0 1>}
 lowerarm_twist_01_l {parent=lowerarm_l, loc=<18.316 0 0>, rot=<0 0 0 1>}
 upperarm_twist_01_l {parent=upperarm_l, loc=<7.2135 -0.1134 0>, rot=<0 0 0 1>}
 upperarm_twist_02_l {parent=upperarm_l, loc=<19.7998 -0.1134 0>, rot=<0 0 0 1>}
 clavicle_r {parent=spine_05, loc=<5.4831 -0.1053 3.2946>, rot=<0.7289 0.0465 -0.6829 0.009>}
 upperarm_r {parent=clavicle_r, loc=<-9.8658 1.2136 0.4105>, rot=<0.0164 -0.3245 -0.0269 -0.9454>}
 lowerarm_r {parent=upperarm_r, loc=<-26.465 0.8345 -0.259>, rot=<0.0036 0.0161 -0.2298 -0.9731>}
 hand_r {parent=lowerarm_r, loc=<-22.7355 0.3203 0.6106>, rot=<-0.6784 -0.0223 0.0735 -0.7307>}
 thumb_01_r {parent=hand_r, loc=<-1.3118 0.3191 2.459>, rot=<0.6546 -0.4042 0.0664 -0.6354>}
 thumb_02_r {parent=thumb_01_r, loc=<-5.9336 0.0448 -0.0401>, rot=<-0.0037 0.0213 0.298 -0.9543>}
 thumb_03_r {parent=thumb_02_r, loc=<-2.2223 -0.3682 0.0907>, rot=<0.0118 -0.013 0.0066 -0.9998>}
 index_metacarpal_r {parent=hand_r, loc=<-1.6535 -0.0183 2.96>, rot=<0.0158 -0.0124 -0.0173 -0.9996>}
 index_01_r {parent=index_metacarpal_r, loc=<-8.3008 -0.0357 0.0524>, rot=<-0.0044 0.0485 0.0924 -0.9945>}
 index_02_r {parent=index_01_r, loc=<-3.3587 0.3004 -0.0452>, rot=<-0.002 0.0027 0.3225 -0.9466>}
 index_03_r {parent=index_02_r, loc=<-2.5578 0.123 0.0252>, rot=<-0.002 0.0067 0.2558 -0.9667>}
 middle_metacarpal_r {parent=hand_r, loc=<-2.1856 0.1725 0.5984>, rot=<0.006 0.0168 -0.0329 -0.9993>}
 middle_01_r {parent=middle_metacarpal_r, loc=<-7.3275 0.2447 -0.0129>, rot=<-0.0208 0.0042 0.0858 -0.9961>}
 middle_02_r {parent=middle_01_r, loc=<-4.0123 -0.1401 -0.1175>, rot=<-0.0218 0.0258 0.3512 -0.9357>}
 middle_03_r {parent=middle_02_r, loc=<-2.6077 -0.0534 -0.1373>, rot=<0.0082 0.0188 0.1708 -0.9851>}
 ring_metacarpal_r {parent=hand_r, loc=<-2.2074 0.3625 -1.3638>, rot=<-0.0284 0.0722 -0.0458 -0.9959>}
 ring_01_r {parent=ring_metacarpal_r, loc=<-7.0295 0.6197 0.1446>, rot=<-0.0489 -0.0252 0.1892 -0.9804>}
 ring_02_r {parent=ring_01_r, loc=<-3.6174 -0.4647 0.1027>, rot=<0 0 0.2806 -0.9598>}
 ring_03_r {parent=ring_02_r, loc=<-2.4152 -0.114 -0.0199>, rot=<0 0 0.1189 -0.9929>}
 pinky_metacarpal_r {parent=hand_r, loc=<-1.7616 0.9753 -3.3814>, rot=<-0.0397 0.0814 -0.069 -0.9935>}
 pinky_01_r {parent=pinky_metacarpal_r, loc=<-7.1709 0.4831 0.1094>, rot=<-0.0654 -0.0035 0.2242 -0.9723>}
 pinky_02_r {parent=pinky_01_r, loc=<-2.8274 -0.0893 0.1043>, rot=<0.0037 0.0011 0.2854 -0.9584>}
 pinky_03_r {parent=pinky_02_r, loc=<-2.2046 -0.0307 0.0078>, rot=<-0.0002 0.0098 0.1445 -0.9895>}
 lowerarm_twist_02_r {parent=lowerarm_r, loc=<-10.6707 0 0>, rot=<0 0 0 -1>}
 lowerarm_twist_01_r {parent=lowerarm_r, loc=<-18.316 0 0>, rot=<0 0 0 -1>}
 upperarm_twist_01_r {parent=upperarm_r, loc=<-7.2135 0.1134 0>, rot=<0 0 0 -1>}
 upperarm_twist_02_r {parent=upperarm_r, loc=<-19.7998 0.1134 0>, rot=<0 0 0 -1>}
 spine_05_null {parent=pelvis, type=Null, loc=<32.9717 -4.2232 0>, rot=<0 0 0.058 -0.9983>}
 thigh_l {parent=pelvis, loc=<-1.0855 -0.2109 -8.7053>, rot=<0.0009 0.0355 0.0155 -0.9993>}
 calf_l {parent=thigh_l, loc=<-42.5284 -0.0019 -0.7297>, rot=<0.0006 -0.0161 -0.0391 -0.9991>}
 ponytail_ASW_02_l {parent=calf_l, loc=<58.842 -26.1033 -10.8819>, rot=<-0.0309 -0.7201 0.0276 -0.6927>}
 thigh_twist_01_l {parent=thigh_l, loc=<-27.952 0 0>, rot=<0.0013 0 0 -1>}
 thigh_r {parent=pelvis, loc=<-1.0855 -0.2109 8.7053>, rot=<0.0355 -0.0009 0.9993 0.0155>}
 calf_r {parent=thigh_r, loc=<42.5284 0.0019 0.7297>, rot=<0.0006 -0.0161 -0.0391 -0.9991>}
 ponytail_ASW_02_r {parent=calf_r, loc=<-58.8421 26.1033 10.8819>, rot=<-0.0309 -0.7201 0.0276 -0.6927>}
 thigh_twist_01_r {parent=thigh_r, loc=<27.952 0 0>, rot=<0.0013 0 0 -1>}
 spine_05_ctrl {controlType=EULER_TRANSFORM, type=Control, loc=<0 0 0>, rot=<0 0 0 1>}
]
variables [PlayingMelee {type=double, default=0}]
