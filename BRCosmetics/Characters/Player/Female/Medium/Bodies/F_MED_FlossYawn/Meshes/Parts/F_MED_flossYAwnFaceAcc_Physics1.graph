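# Physics asset for the FlossYawn female medium cosmetic's face accessory mesh, setting up physics for left and right ponytails with body collision shapes.
# Capsules cover the head, upper spine, pelvis, arms and calves, linked by free-rotation constraints.
# Each ponytail chains from a head-attached base through six dynamic segments, forking into two end strands.
# Base and first-segment joints rotate freely, while later joints carry narrow twist limits and wider swing limits.
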PhysicsAsset /BRCosmetics/Characters/Player/Female/Medium/Bodies/F_MED_FlossYawn/Meshes/Parts/F_MED_flossYAwnFaceAcc_Physics1
bodies {
 base_ponytail_l [{shapeType=Capsule, length=6, radius=3, rotation=<rot 90 0 0>}]
 base_ponytail_r [{shapeType=Capsule, length=6, radius=3, rotation=<rot 90 0 0>}]
 calf_l [{shapeType=Capsule, length=26, radius=12, shapeName=calf_l_capsule, center=<-14.7061 -0.5088 -0.1658>, rotation=<rot 90 80.7734 84.0177>}]
 calf_r [{shapeType=Capsule, length=26, radius=12, shapeName=calf_l_capsule, center=<14.7061 0.5088 -0.1658>, rotation=<rot 90 80.7734 -95.9823>}]
 dyn_ponytail_01_l [{shapeType=Capsule, length=6, radius=6, center=<5 0 0>, rotation=<rot 90 0 0>}]
 dyn_ponytail_01_r [{shapeType=Capsule, length=6, radius=6, center=<-5 0 0>, rotation=<rot 90 0 0>}]
 dyn_ponytail_02_l [{shapeType=Capsule, length=6, radius=6, center=<5 0 0>, rotation=<rot 90 0 0>}]
 dyn_ponytail_02_r [{shapeType=Capsule, length=6, radius=6, center=<-5 0 0>, rotation=<rot 90 0 0>}]
 dyn_ponytail_03_l [{shapeType=Capsule, length=6, radius=6, center=<5 0 0>, rotation=<rot 90 0 0>}]
 dyn_ponytail_03_r [{shapeType=Capsule, length=6, radius=6, center=<-5 0 0>, rotation=<rot 90 0 0>}]
 dyn_ponytail_04_l [{shapeType=Capsule, length=6, radius=6, center=<5 0 0>, rotation=<rot 90 0 0>}]
 dyn_ponytail_04_r [{shapeType=Capsule, length=6, radius=6, center=<-5 0 0>, rotation=<rot 90 0 0>}]
 dyn_ponytail_05_l [{shapeType=Capsule, length=6, radius=6, center=<5 0 0>, rotation=<rot 90 0 0>}]
 dyn_ponytail_05_r [{shapeType=Capsule, length=6, radius=6, center=<-5 0 0>, rotation=<rot 90 0 0>}]
 dyn_ponytail_06_l [{shapeType=Capsule, length=6, radius=6, center=<5 0 0>, rotation=<rot 90 0 0>}]
 dyn_ponytail_06_r [{shapeType=Capsule, length=6, radius=6, center=<-5 0 0>, rotation=<rot 90 0 0>}]
 dyn_ponytail_strand_a_l [{shapeType=Capsule, length=6, radius=6, center=<5 0 0>, rotation=<rot 90 0 0>}]
 dyn_ponytail_strand_a_r [{shapeType=Capsule, length=6, radius=6, center=<-5 0 0>, rotation=<rot 90 0 0>}]
 dyn_ponytail_strand_b_l [{shapeType=Capsule, length=6, radius=6, center=<5 0 0>, rotation=<rot 90 0 0>}]
 dyn_ponytail_strand_b_r [{shapeType=Capsule, length=6, radius=6, center=<-5 0 0>, rotation=<rot 90 0 0>}]
 head [{shapeType=Capsule, length=2, radius=2, shapeName=head_capsule}]
 lowerarm_l [{shapeType=Capsule, length=35.51, radius=6, shapeName=lowerarm_l_capsule, center=<18.2404 0 0>, rotation=<rot 90 0 0>}]
 lowerarm_r [{shapeType=Capsule, length=35.51, radius=6, shapeName=lowerarm_l_capsule, center=<-18.2404 0 0>, rotation=<rot 90 0 180>}]
 pelvis [{shapeType=Capsule, length=36.3326, radius=17.0968, shapeName=pelvis_capsule, center=<0 5.0491 0>}]
 spine_04 [{shapeType=Capsule, length=15, radius=18.5, shapeName=spine_04_capsule, center=<0 8.7048 0>}]
 spine_05 [{shapeType=Capsule, length=20.2077, radius=7.8576, shapeName=spine_05_capsule, center=<-4.6403 -8.7054 0>, rotation=<rot 90 4.6802 0>}]
 upperarm_l [{shapeType=Capsule, length=19.005, radius=6, shapeName=upperarm_l_capsule, center=<12.4116 0 0>, rotation=<rot 90 0 0>}]
 upperarm_r [{shapeType=Capsule, length=19.005, radius=6, shapeName=upperarm_l_capsule, center=<-12.4116 0 0>, rotation=<rot 90 0 180>}]
}
constraints [
 {
  bone1Name=head
  bone2Name=spine_05
  swing1LimitDegrees=45
  swing1Motion=Free
  swing2LimitDegrees=45
  swing2Motion=Free
  twistLimitDegrees=45
  twistMotion=Free
 }
 {
  bone1Name=spine_04
  bone2Name=pelvis
  swing1LimitDegrees=45
  swing1Motion=Free
  swing2LimitDegrees=45
  swing2Motion=Free
  twistLimitDegrees=45
  twistMotion=Free
 }
 {
  bone1Name=upperarm_l
  bone2Name=spine_05
  swing1LimitDegrees=45
  swing1Motion=Free
  swing2LimitDegrees=45
  swing2Motion=Free
  twistLimitDegrees=45
  twistMotion=Free
 }
 {
  bone1Name=upperarm_r
  bone2Name=spine_05
  swing1LimitDegrees=45
  swing1Motion=Free
  swing2LimitDegrees=45
  swing2Motion=Free
  twistLimitDegrees=45
  twistMotion=Free
 }
 {
  bone1Name=lowerarm_r
  bone2Name=upperarm_r
  swing1LimitDegrees=45
  swing1Motion=Free
  swing2LimitDegrees=45
  swing2Motion=Free
  twistLimitDegrees=45
  twistMotion=Free
 }
 {
  bone1Name=lowerarm_l
  bone2Name=upperarm_l
  swing1LimitDegrees=45
  swing1Motion=Free
  swing2LimitDegrees=45
  swing2Motion=Free
  twistLimitDegrees=45
  twistMotion=Free
 }
 {
  bone1Name=calf_l
  bone2Name=pelvis
  swing1LimitDegrees=45
  swing1Motion=Free
  swing2LimitDegrees=45
  swing2Motion=Free
  twistLimitDegrees=45
  twistMotion=Free
 }
 {
  bone1Name=calf_r
  bone2Name=pelvis
  swing1LimitDegrees=45
  swing1Motion=Free
  swing2LimitDegrees=45
  swing2Motion=Free
  twistLimitDegrees=45
  twistMotion=Free
 }
 {
  bone1Name=spine_05
  bone2Name=spine_04
  swing1LimitDegrees=45
  swing1Motion=Free
  swing2LimitDegrees=45
  swing2Motion=Free
  twistLimitDegrees=45
  twistMotion=Free
 }
 {
  bone1Name=base_ponytail_l
  bone2Name=head
  swing1LimitDegrees=45
  swing1Motion=Free
  swing2LimitDegrees=45
  swing2Motion=Free
  twistLimitDegrees=45
  twistMotion=Free
 }
 {
  bone1Name=dyn_ponytail_01_l
  bone2Name=base_ponytail_l
  swing1LimitDegrees=45
  swing1Motion=Free
  swing2LimitDegrees=45
  swing2Motion=Free
  twistLimitDegrees=45
  twistMotion=Free
 }
 {bone1Name=dyn_ponytail_02_l, bone2Name=dyn_ponytail_01_l, swing1LimitDegrees=45, swing2LimitDegrees=45, twistLimitDegrees=5}
 {bone1Name=dyn_ponytail_03_l, bone2Name=dyn_ponytail_02_l, swing1LimitDegrees=45, swing2LimitDegrees=45, twistLimitDegrees=5}
 {bone1Name=dyn_ponytail_04_l, bone2Name=dyn_ponytail_03_l, swing1LimitDegrees=45, swing2LimitDegrees=45, twistLimitDegrees=5}
 {bone1Name=dyn_ponytail_05_l, bone2Name=dyn_ponytail_04_l, swing1LimitDegrees=45, swing2LimitDegrees=45, twistLimitDegrees=5}
 {bone1Name=dyn_ponytail_06_l, bone2Name=dyn_ponytail_05_l, swing1LimitDegrees=45, swing2LimitDegrees=45, twistLimitDegrees=5}
 {bone1Name=dyn_ponytail_strand_a_l, bone2Name=dyn_ponytail_06_l, swing1LimitDegrees=45, swing2LimitDegrees=45, twistLimitDegrees=5}
 {bone1Name=dyn_ponytail_strand_b_l, bone2Name=dyn_ponytail_06_l, swing1LimitDegrees=45, swing2LimitDegrees=45, twistLimitDegrees=5}
 {
  bone1Name=base_ponytail_r
  bone2Name=head
  swing1LimitDegrees=45
  swing1Motion=Free
  swing2LimitDegrees=45
  swing2Motion=Free
  twistLimitDegrees=45
  twistMotion=Free
 }
 {
  bone1Name=dyn_ponytail_01_r
  bone2Name=base_ponytail_r
  swing1LimitDegrees=45
  swing1Motion=Free
  swing2LimitDegrees=45
  swing2Motion=Free
  twistLimitDegrees=45
  twistMotion=Free
 }
 {bone1Name=dyn_ponytail_02_r, bone2Name=dyn_ponytail_01_r, swing1LimitDegrees=45, swing2LimitDegrees=45, twistLimitDegrees=5}
 {bone1Name=dyn_ponytail_03_r, bone2Name=dyn_ponytail_02_r, swing1LimitDegrees=45, swing2LimitDegrees=45, twistLimitDegrees=5}
 {bone1Name=dyn_ponytail_04_r, bone2Name=dyn_ponytail_03_r, swing1LimitDegrees=45, swing2LimitDegrees=45, twistLimitDegrees=5}
 {bone1Name=dyn_ponytail_05_r, bone2Name=dyn_ponytail_04_r, swing1LimitDegrees=45, swing2LimitDegrees=45, twistLimitDegrees=5}
 {bone1Name=dyn_ponytail_06_r, bone2Name=dyn_ponytail_05_r, swing1LimitDegrees=45, swing2LimitDegrees=45, twistLimitDegrees=5}
 {bone1Name=dyn_ponytail_strand_b_r, bone2Name=dyn_ponytail_06_r, swing1LimitDegrees=45, swing2LimitDegrees=45, twistLimitDegrees=5}
 {bone1Name=dyn_ponytail_strand_a_r, bone2Name=dyn_ponytail_06_r, swing1LimitDegrees=45, swing2LimitDegrees=45, twistLimitDegrees=5}
]
